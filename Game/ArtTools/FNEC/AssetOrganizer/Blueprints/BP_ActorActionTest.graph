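# Editor utility that removes selected Fortnite building actors whose texture data duplicates another selected actor's.
# RemoveDuplicates walks the selection in reverse, skipping actors whose class differs from the first's, and reads TextureData through casts to wall, prop wall, prop, floor or roof.
# It compares four texture slots against a growing list of unique sets, destroying full matches and recording new sets, with verbose print logging.
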EditorUtilityBlueprint /Game/ArtTools/FNEC/AssetOrganizer/Blueprints/BP_ActorActionTest
graphs [
 EventGraph {links [], nodes []}
 RemoveDuplicates {nodes [
   RemoveDuplicates() {is=title K2Node_FunctionEntry}
   > n1 = GetSelectionSet!(execute) -> (then, ReturnValue:Array of Actor Object References)
   > n2 = Set TDList(TDList:Array of Actor Object References=@n1) -> (Output_Get:Array of Actor Object References)
   > n3 = Set NumberOfItems(NumberOfItems:int=Length(TargetArray=@n2)) -> (Output_Get:int=0)
   > n4 = Compare Int!(Exec:Exec, Input:int=@n3, CompareWith:int=2) -> (">":Exec, "==":Exec, "<":Exec) {is=title K2Node_MacroInstance}
   PrintString(execute=@n4."<", InString=Too few items selected. Select 2 or more)
   Set ItemClass(
    execute=[@n4."==", @n4.">"]
    ItemClass:Actor Class Reference=Get Class(Object="Get (a copy)"(Array:Array of Actor Object References=@n2) -> (Output:Actor Object Reference))
   ) -> (Output_Get:Actor Class Reference)
   > Clear(TargetArray:Array of Str FNEC Asset Organizer TDArray Structures=Get ListOfTDs() -> (ListOfTDs:Array of Str FNEC Asset Organizer TDArray Structures))
   > n5 = Reverse for Each Loop!(Exec:Exec, Array:Array of Actor Object References=Get TDList() -> (TDList:Array of Actor Object References)) -> (LoopBody:Exec, ArrayIndex:int, ArrayElement:Actor Object Reference, Completed:Exec) {is=title K2Node_MacroInstance}
   > n6 = Set CurrentItem(CurrentItem:Actor Object Reference=@n5.ArrayElement) -> (Output_Get:Actor Object Reference)
   n7 = Get Class(Object=@n5.ArrayElement)
   > n8 = Branch(Condition="Equal (Class)"(A=Get ItemClass() -> (ItemClass:Actor Class Reference), B=@n7))
   > PrintString(InString=Append(A=">>> Analizing Object named : ", B=GetDisplayName(Object=@n6)))
   > n9 = Sequence()
   > n10 = Sequence!(execute) -> (then_0, then_1, then_2:Exec, then_3:Exec, then_4:Exec)
   n11 = Get CurrentItem() -> (CurrentItem:Actor Object Reference)
   > n12 = Cast To BuildingPropWall!(execute, Object=@n11) -> (then, CastFailed, AsBuilding Prop Wall:Building Prop Wall Object Reference) {is=title K2Node_DynamicCast}
   > PrintString(InString=CAST PROP WALL)
   > n13 = GetEditorProperty!(execute:Exec, Object:Object Reference=@n12, PropertyName:Name=TextureData) -> (then:Exec, PropertyValue:Array of Building Texture Data Object References, ReturnValue:bool)
   > Branch(Condition=@n13.ReturnValue)
   > Set CurrentTD(CurrentTD:Array of Building Texture Data Object References=@n13) -> (Output_Get:Array of Building Texture Data Object References)
   n14 = Cast To BuildingWall!(execute=@n10.then_1, Object=@n11) -> (then, CastFailed, AsBuilding Wall:Building Wall Object Reference) {is=title K2Node_DynamicCast}
   > PrintString(InString=CAST WALL)
   > n15 = GetEditorProperty(Object=@n14, PropertyName=TextureData)
   > Branch(Condition=@n15.ReturnValue)
   > Set CurrentTD(CurrentTD=@n15)
   n16 = Cast To BuildingProp!(execute=@n10.then_2, Object=@n11) -> (then, CastFailed, AsBuilding Prop:Building Prop Object Reference) {is=title K2Node_DynamicCast}
   > PrintString(InString=CAST PROP)
   > n17 = GetEditorProperty(Object=@n16, PropertyName=TextureData)
   > Branch(Condition=@n17.ReturnValue)
   > Set CurrentTD(CurrentTD=@n17)
   n18 = Cast To BuildingFloor!(execute=@n10.then_3, Object=@n11) -> (then, CastFailed, AsBuilding Floor:Building Floor Object Reference) {is=title K2Node_DynamicCast}
   > PrintString(InString=CAST FLOOR)
   > n19 = GetEditorProperty(Object=@n18, PropertyName=TextureData)
   > Branch(Condition=@n19.ReturnValue)
   > Set CurrentTD(CurrentTD=@n19)
   n20 = Cast To BuildingRoof!(execute=@n10.then_4, Object=@n11) -> (then, CastFailed, AsBuilding Roof:Building Roof Object Reference) {is=title K2Node_DynamicCast}
   > PrintString(InString=CAST ROOF)
   > n21 = GetEditorProperty(Object=@n20, PropertyName=TextureData)
   > Branch(Condition=@n21.ReturnValue)
   > Set CurrentTD(CurrentTD=@n21)
   n22 = Sequence!(execute=@n9.then_1) -> (then_0, then_1)
   n23 = "Get (a copy)"(
    Array:Array of Building Texture Data Object References=Get CurrentTD() -> (CurrentTD:Array of Building Texture Data Object References)
    Dimension 1=2
   ) -> (Output:Building Texture Data Object Reference)
   n24 = "Get (a copy)"(Array=Get CurrentTD(), Dimension 1=3)
   n25 = "Get (a copy)"(Array=Get CurrentTD())
   n26 = "Get (a copy)"(Array=Get CurrentTD(), Dimension 1=1)
   > PrintString(InString=Append!(
     A="Texture Data : 1= "
     B=GetDisplayName(Object=@n25)
     C:String=" 2= "
     D:String=GetDisplayName(Object=@n26)
     E:String=" 3= "
     F:String=GetDisplayName(Object=@n23)
     G:String=" 4= "
     H:String=GetDisplayName(Object=@n24)
    ) -> (ReturnValue))
   n27 = Get ListOfTDs()
   n28 = Branch(execute=@n22.then_1, Condition="integer > integer"(A=Length(TargetArray:Array of Str FNEC Asset Organizer TDArray Structures=@n27)))
   > Set MarkForDelete(MarkForDelete:bool=false) -> (Output_Get:bool)
   > n29 = For Each Loop(Array:Array of Str FNEC Asset Organizer TDArray Structures=Get ListOfTDs()) -> (Array Element:Str FNEC Asset Organizer TDArray Structure)
   n30 = Branch(execute=@n29.Completed, Condition=Get MarkForDelete() -> (MarkForDelete:bool))
   n31 = Get CurrentItem()
   > Remove Item!(execute, TargetArray:Array of Actor Object References=Get TDList(), Item:"Actor Object Reference (by ref)"=@n31) -> (then, ReturnValue:bool)
   > Destroy Actor(self=@n31)
   > PrintString(InString="► Item is a DUPLICATE. Deleting")
   Add!(
    execute=@n30.else
    TargetArray:Array of Str FNEC Asset Organizer TDArray Structures=Get ListOfTDs()
    NewItem_TDArray_3_#1:Array of Building Texture Data Object References=Get CurrentTD()
   ) -> (then, ReturnValue) {is=title K2Node_CallArrayFunction}
   > PrintString(InString="► Item is not a duplicate. Adding to the list of unique TDs")
   n32 = Branch(execute=@n29.LoopBody, Condition=Get MarkForDelete())
   Set DifferenceFound(execute=@n32.else, DifferenceFound:bool=false) -> (Output_Get:bool)
   > PrintString(InString=Append!(A="Comparing against unique TD set ##", B="To String (Integer)"(InInt=@n29.Array Index)) -> (ReturnValue))
   > n33 = For Loop(LastIndex=3)
   n34 = Branch(execute=@n33.Completed, Condition=Get DifferenceFound() -> (DifferenceFound:bool))
   Set MarkForDelete(execute=@n34.else, MarkForDelete=true)
   n35 = "Get (a copy)"(
    Array=Break Str FNEC Asset Organizer TDArray!(Str_FNEC_AssetOrganizer_TDArray:"Str FNEC Asset Organizer TDArray Structure (by ref)"=@n29) -> (TDArray_3_#1:Array of Building Texture Data Object References) {is=title K2Node_BreakStruct}
    Dimension 1=@n33
   )
   n36 = "Get (a copy)"(Array=Get CurrentTD(), Dimension 1=@n33)
   n37 = "Equal (Object)"(A=@n35, B=@n36)
   PrintString(
    execute=@n33.LoopBody
    InString=Append!(
     A="Comparing "
     B=GetDisplayName(Object=@n35)
     C:String=" to  "
     D:String=GetDisplayName(Object=@n36)
     E:String=SelectString(A="    -> They are the same", B="    -> They are DIFFERENT", bPickA=@n37)
    ) -> (ReturnValue)
   )
   > Set DifferenceFound(DifferenceFound=Select(Option 0:bool=true, Option 1:bool=Get DifferenceFound(), Index=@n37) -> (ReturnValue:bool))
   Add(execute=@n28.else, TargetArray=Get ListOfTDs(), NewItem_TDArray_3_#1=Get CurrentTD())
   > PrintString(InString="It is the first element. Adding to the list of unique TDs ##0")
   PrintString(
    execute=@n8.else
    InString=Append(
     A="[ERROR] Skipping actor "
     B=GetDisplayName(Object=@n5.ArrayElement)
     C=". It is of a diferent class ("
     D=Get Class Display Name(Class=@n7)
     E=")"
    )
   )
   PrintString(InString=Different TD)
   PrintString(InString=Same TD)
   Reverse for Each Loop(Array:Array of Wildcards) -> (ArrayElement:Wildcard)
  ]}
]
variables [
 ClassPath {default {
   &Class'&BuildingFloor'=3
   &Class'&BuildingProp'=2
   &Class'/Script/FortniteGame.BuildingPropWall'=0
   &Class'&BuildingRoof'=4
   &Class'&BuildingWall'=1
  }}
 TDList {type="array<Actor>"}
]
