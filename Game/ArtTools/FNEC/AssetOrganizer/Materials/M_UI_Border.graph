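# M_UI_Border is a translucent UI material in the FNEC AssetOrganizer art tools, apparently meant for a border.
# A centered SphereMask on the texture coordinates forms a radial mask, with hardness set by the inverted "Line Width" parameter.
# That mask feeds both a Ceil and a second SphereMask, which remaps its values around the midpoint.
# The graph shows neither branch connected to a material output.
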Material /Game/ArtTools/FNEC/AssetOrganizer/Materials/M_UI_Border
graph {
 properties {blendMode=BLEND_Translucent, materialDomain=MD_UI, pixelDepthOffsetMode=PDOM_Legacy, usedShadingModels=Default Lit}
 nodes [
  n1 = Constant() {r=0.5}
  n2 = SphereMask(A=TextureCoordinate(), B=@n1, Radius=@n1, Hardness=OneMinus(None=ScalarParameter() {defaultValue=0.152, parameterName=Line Width})) {hardnessPercent=0}
  Ceil(None=@n2)
  n3 = Constant() {r=0.5}
  SphereMask(A=@n2, B=@n3, Radius=@n3) {hardnessPercent=0}
 ]
}
