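# Editor utility widget for the FNEC Asset Organizer that cleans up, sorts and arranges level actors.
# Sorts actors into Content, Disallowed, Gallery and Flagged folders using data asset class lists, with buttons to select by folder and delete Disallowed ones.
# Removes same-class duplicates by comparing four TextureData slots, and resets scale, rotation and mirroring.
# Arranges selections into per-class grids through the parent tools widget, wrapping edits in undo transactions.
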EditorUtilityWidgetBlueprint /Game/ArtTools/FNEC/AssetOrganizer/Widgets/EUW_AssetOrganizer_ActorCleanup
graphs [
 ArrangeActorsByClass {nodes [
   n1 = ArrangeActorsByClass!() -> (then, Assets:Array of Actor Object References) {is=title K2Node_FunctionEntry}
   > n2 = SortActorsByClass!(execute, self, Assets:Array of Actor Object References=@n1) -> (then, SortedActors:Map of Actor Class References to Str FNEC Asset Organizer Actor Array Structures)
   > Set SortedActorsList(SortedActorsList:Map of Actor Class References to Str FNEC Asset Organizer Actor Array Structures=@n2) -> (Output_Get:Map of Actor Class References to Str FNEC Asset Organizer Actor Array Structures)
   > n3 = Sequence!(execute) -> (then_0, then_1, then_2:Exec)
   BeginTransaction(execute=@n3.then_1, Context=Asset Organizer, Description="LOCTEXT(\"Sort Actors by Class\")")
   > Set ArrangementOffset(ArrangementOffset:double=0) -> (Output_Get:double)
   > n4 = "For Each Loop (Map)"!(
    execute:Exec
    MapPin:Map of Actor Class References to Str FNEC Asset Organizer Actor Array Structures=Get SortedActorsList() -> (SortedActorsList:Map of Actor Class References to Str FNEC Asset Organizer Actor Array Structures)
    BreakPin:Exec
   ) -> (then:Exec, KeyPin:Actor Class Reference, ValuePin:Str FNEC Asset Organizer Actor Array Structure, CompletedPin:Exec)
   n5 = Break Str FNEC Asset Organizer Actor Array!(Str_FNEC_AssetOrganizer_ActorArray:"Str FNEC Asset Organizer Actor Array Structure (by ref)"=@n4.ValuePin) -> (ActorArray_3_#1:Array of Actor Object References) {is=title K2Node_BreakStruct}
   n6 = "GetText (Editable Text)"!(self:Editable Text Object Reference=Get T_SortByClass_Rows() -> (T_SortByClass_Rows:Editor Utility Editable Text Object Reference)) -> (ReturnValue:Text)
   n7 = Truncate(A=String To Float(InString="To String (Text)"(InText=@n6)))
   > Arrange Objects with Origin Options!(
    execute
    self:EUW Asset Organizer Tools Object Reference=Get OwnerWidget() -> (OwnerWidget:EUW Asset Organizer Tools Object Reference)
    Selected Actors:Array of Actor Object References=@n5
    OriginMode:E_FNEC_AssetOrganizer_GridOrigin Enum=Get ArrayOriginMode() -> (ArrayOriginMode:E_FNEC_AssetOrganizer_GridOrigin Enum=NewEnumerator0)
    Columns:int=@n7
    Columns in X Direction:bool=true
    Distance X:double=640
    Distance Y:double=640
    By Size:bool=true
    Reset Rotation:bool=true
    Reset Scale:bool=true
    Origin Offset:Vector=MakeVector(X=Get ArrangementOffset() -> (ArrangementOffset:double), Y=0)
    Explicit Origin:Vector=Get FloorOrigin() -> (FloorOrigin:Vector)
    Skip Transaction Begin and End:bool=true
   ) -> (then)
   n8 = int + int(A="Min (Integer)"(A=Length(TargetArray=@n5), B=@n7) -> (ReturnValue=0))
   > Set ArrangementOffset(ArrangementOffset=float + float(A=Get ArrangementOffset(), B=float * float(A:int=@n8, B=-640) {is=title K2Node_PromotableOperator}))
   EndTransaction(execute=@n4.CompletedPin)
   n9 = GetSelectedIndex!(self:"Combo Box (String) Object Reference"=Get Combo_Arrange_OriginMode() -> (Combo_Arrange_OriginMode:Editor Utility Combo Box String Object Reference)) -> (ReturnValue:int=0)
   Set ArrayOriginMode(
    execute=@n3.then_0
    ArrayOriginMode:E_FNEC_AssetOrganizer_GridOrigin Enum=Select!(
     Option 0:E_FNEC_AssetOrganizer_GridOrigin Enum=NewEnumerator5
     Option 1:E_FNEC_AssetOrganizer_GridOrigin Enum=NewEnumerator6
     Option 2:E_FNEC_AssetOrganizer_GridOrigin Enum=NewEnumerator7
     Option 3:E_FNEC_AssetOrganizer_GridOrigin Enum=NewEnumerator7
     Option 4:E_FNEC_AssetOrganizer_GridOrigin Enum=NewEnumerator7
     Index:int=@n9
    ) -> (ReturnValue:E_FNEC_AssetOrganizer_GridOrigin Enum)
   ) -> (Output_Get:E_FNEC_AssetOrganizer_GridOrigin Enum=NewEnumerator0)
   > n10 = Switch on Int!(execute:Exec, Selection:int=@n9) -> (Default:Exec, 0:Exec, 1:Exec, 2:Exec, 3:Exec, 4:Exec)
   n11 = GetAllActorsOfClass!(
    execute=[@n10.3, @n10.4]
    WorldContextObject:Object Reference
    ActorClass:Actor Class Reference=/Game/Playgrounds/Items/IconRenders/PlaysetIcon_Floor.PlaysetIcon_Floor_C
   ) -> (then, OutActors:Array of Playset Icon Floor Object References)
   n12 = "Get (a copy)"(Array:Array of Playset Icon Floor Object References=@n11) -> (Output:Playset Icon Floor Object Reference)
   > n13 = Branch(Condition=IsValid(Object=@n12))
   n14 = Get Combo_Arrange_OriginMode()
   n15 = BreakVector(InVec=vector * vector(A=Get Actor Bounds(self=@n12, bOnlyCollidingComponents=false).BoxExtent, B:Vector="2,2,2") -> (ReturnValue=~))
   n16 = float * float(A:double=@n15, B=-1)
   > Set FloorOrigin(FloorOrigin:Vector=Select(
     Option 0:Vector="0, 0, 0"
     Option 1:Vector="0, 0, 0"
     Option 2:Vector="0, 0, 0"
     Option 3:Vector=MakeVector(Y=@n15.Y)
     Option 4:Vector=MakeVector(X=@n16, Y=0)
     Index=GetSelectedIndex(self=@n14) -> (ReturnValue=0)
    ) -> (ReturnValue:Vector)) -> (Output_Get:Vector)
   Show Message Dialog(
    execute=@n13.else
    Title=MakeLiteralText(Value="LOCTEXT(\"Asset Organizer Tool - Arrange by Class\")")
    Message=MakeLiteralText(Value="LOCTEXT(\"PlaysetIcon Floor asset not found in scene.\\r\\n\\r\\nAborting Operation\")")
    MessageType=Ok
    DefaultValue=No
    MessageCategory=Warning
   )
   > Return Node()
   Set FloorOrigin(execute=@n10.2, FloorOrigin="0, 0, 0")
   PrintString(execute=@n3.then_2, InString=Found the following Classes and Actors)
   n17 = Length!(TargetMap:Map of Actor Class References to Str FNEC Asset Organizer Actor Array Structures=Get SortedActorsList()) -> (ReturnValue:int=0) {is=title K2Node_CallFunction}
   > PrintString(InString=Append(A="Number of Unique Classes : ", B="To String (Integer)"(InInt=@n17)))
   > n18 = "For Each Loop (Map)"(MapPin=Get SortedActorsList())
   n19 = Break Str FNEC Asset Organizer Actor Array(Str_FNEC_AssetOrganizer_ActorArray=@n18.ValuePin)
   n20 = Length(TargetArray=@n19) {is=title K2Node_CallArrayFunction}
   > PrintString(InString=Append!(A="► Class :", B=Get Class Display Name(Class=@n18), C:String="  with number of assets = ", D:String="To String (Integer)"(InInt=@n20)) -> (ReturnValue))
   > n21 = For Each Loop(Array:Array of Actor Object References=@n19) -> (Array Element:Actor Object Reference)
   > PrintString(InString=Append!(A="              Object : ", B=GetDisplayName(Object=@n21)) -> (ReturnValue))
   Return Node(execute=@n18.CompletedPin)
  ]}
 DeleteDisallowedActors {nodes [
   DeleteDisallowedActors() {is=title K2Node_FunctionEntry}
   > n1 = Show Message Dialog(
    Title=MakeLiteralText(Value="LOCTEXT(\"Asset Organizer Tool - Delete Disallowed Actors\")")
    Message=MakeLiteralText(Value="LOCTEXT(\"You are about to delete every actor in the \\\"Disallowed\\\" folder. \\r\\n\\r\\nDo you want to continue?\")")
    MessageType=YesNo
    DefaultValue=No
    MessageCategory=Warning
   )
   > n2 = Switch on EAppReturnType(Selection=@n1)
   BeginTransaction(execute=@n2.Yes, Context=Asset Organizer, Description="LOCTEXT(\"Sort into Folders\")")
   > n3 = GetEditorWorld(self=Get UnrealEditorSubsystem())
   > SelectAll!(execute, self:Editor Actor Subsystem Object Reference=Get EditorActorSubsystem(), InWorld:World Object Reference=@n3) -> (then)
   > n4 = GetSelectionSet!(execute) -> (then, ReturnValue:Array of Actor Object References)
   > n5 = For Each Loop(Array=@n4)
   ClearActorSelectionSet(execute=@n5.Completed, self:Editor Actor Subsystem Object Reference=Get EditorActorSubsystem())
   > EndTransaction()
   > n6 = Branch(Condition="integer > integer"(A=Get Delete Counter() -> (Delete Counter:int=0)))
   > Show Message Dialog(
    Title=MakeLiteralText(Value="LOCTEXT(\"Asset Organizer Tool - Delete Disallowed Actors\")")
    Message=Format Text!(Format="LOCTEXT(\"Found and deleted {text1} disallowed asset(s)\")", text1:int=Get Delete Counter() -> (Delete Counter=0)) -> (Result)
    MessageType=Ok
    DefaultValue=Ok
    MessageCategory=Success
   )
   Show Message Dialog(
    execute=@n6.else
    Title=MakeLiteralText(Value="LOCTEXT(\"Asset Organizer Tool - Delete Disallowed Actors\")")
    Message=Format Text!(Format="LOCTEXT(\"Disallowed folder was empty. No assets removed\")") -> (Result)
    MessageType=Ok
    DefaultValue=Ok
    MessageCategory=Info
   )
   TransactObject(execute=@n5.LoopBody, Object=@n5)
   n7 = GetFolderPath!(self:Actor Object Reference=@n5) -> (ReturnValue:Name=None)
   > n8 = Compare Int!(
    Exec:Exec
    Input:int=CompareStrings!(LeftStr:String="To String (Name)"(InName=@n7), RightStr:String=Disallowed) -> (ReturnValue:int=0)
    CompareWith:int
   ) -> (">":Exec, "==":Exec, "<":Exec) {is=title K2Node_MacroInstance}
   Destroy Actor(execute=@n8."==", self=@n5)
   > Increment Int(Value=Get Delete Counter() -> (Delete Counter=0))
  ]}
 DeleteDuplicates {nodes [
   n1 = DeleteDuplicates!() -> (then, Move:bool=~, Actors:Array of Actor Object References) {is=title K2Node_FunctionEntry}
   > Clear(TargetArray:Array of Actor Object References=Get Unique Actors() -> (Unique Actors:Array of Actor Object References))
   > n2 = Set Selected Actors(Selected Actors:Array of Actor Object References=@n1.Actors) -> (Output_Get:Array of Actor Object References)
   > n3 = Set NumberOfItems(NumberOfItems:int=Length(TargetArray=@n2)) -> (Output_Get:int=0)
   > n4 = Compare Int(Input=@n3, CompareWith=2)
   PrintString(execute=@n4."<", InString="[Eliminate Duplicates] ERROR : Too few items selected. Select 2 or more", TextColor=(R=1,G=0.4177,B=0,A=1))
   Set SelectedItemsClass(
    execute=[@n4."==", @n4.">"]
    SelectedItemsClass:Actor Class Reference=Get Class(Object="Get (a copy)"(Array:Array of Actor Object References=@n2) -> (Output:Actor Object Reference))
   ) -> (Output_Get:Actor Class Reference)
   > Clear(TargetArray:Array of Str FNEC Asset Organizer TDArray Structures=Get ListOfUniqueTDs() -> (ListOfUniqueTDs:Array of Str FNEC Asset Organizer TDArray Structures))
   > Set DupCount(DupCount:int=0) -> (Output_Get:int=0)
   > n5 = Reverse for Each Loop!(Exec:Exec, Array:Array of Actor Object References=Get Selected Actors() -> (Selected Actors:Array of Actor Object References)) -> (LoopBody:Exec, ArrayIndex:int, ArrayElement:Actor Object Reference, Completed:Exec) {is=title K2Node_MacroInstance}
   > n6 = Set CurrentItem(CurrentItem:Actor Object Reference=@n5.ArrayElement) -> (Output_Get:Actor Object Reference)
   n7 = Get Class(Object=@n5.ArrayElement)
   > n8 = Branch(Condition="Equal (Class)"(A=Get SelectedItemsClass() -> (SelectedItemsClass:Actor Class Reference), B=@n7))
   > PrintString(InString=Append(A=">>> Analizing Object named : ", B=GetDisplayName(Object=@n6)), bPrintToScreen=false)
   > n9 = Sequence!(execute) -> (then_0, then_1)
   > n10 = Sequence!(execute) -> (then_0, then_1, then_2:Exec, then_3:Exec, then_4:Exec, then_5:Exec)
   n11 = Get CurrentItem() -> (CurrentItem:Actor Object Reference)
   > n12 = Cast To BuildingPropWall!(execute, Object=@n11) -> (then, CastFailed, AsBuilding Prop Wall:Building Prop Wall Object Reference) {is=title K2Node_DynamicCast}
   > PrintString(InString=CAST PROP WALL, bPrintToScreen=false)
   > n13 = GetEditorProperty!(execute:Exec, Object:Object Reference=@n12, PropertyName:Name=TextureData) -> (then:Exec, PropertyValue:Array of Building Texture Data Object References, ReturnValue:bool)
   > Branch(Condition=@n13.ReturnValue)
   > Set CurrentTDs(CurrentTDs:Array of Building Texture Data Object References=@n13) -> (Output_Get:Array of Building Texture Data Object References)
   > Set ClassName(ClassName:String=Prop Wall) -> (Output_Get:String)
   n14 = Cast To BuildingWall!(execute=@n10.then_1, Object=@n11) -> (then, CastFailed, AsBuilding Wall:Building Wall Object Reference) {is=title K2Node_DynamicCast}
   > PrintString(InString=CAST WALL, bPrintToScreen=false)
   > n15 = GetEditorProperty(Object=@n14, PropertyName=TextureData)
   > Branch(Condition=@n15.ReturnValue)
   > Set CurrentTDs(CurrentTDs=@n15)
   > Set ClassName(ClassName=Wall)
   n16 = Cast To BuildingProp!(execute=@n10.then_2, Object=@n11) -> (then, CastFailed, AsBuilding Prop:Building Prop Object Reference) {is=title K2Node_DynamicCast}
   > PrintString(InString=CAST PROP, bPrintToScreen=false)
   > n17 = GetEditorProperty(Object=@n16, PropertyName=TextureData)
   > Branch(Condition=@n17.ReturnValue)
   > Set CurrentTDs(CurrentTDs=@n17)
   > Set ClassName(ClassName=Prop)
   n18 = Cast To BuildingFloor!(execute=@n10.then_3, Object=@n11) -> (then, CastFailed, AsBuilding Floor:Building Floor Object Reference) {is=title K2Node_DynamicCast}
   > PrintString(InString=CAST FLOOR, bPrintToScreen=false)
   > n19 = GetEditorProperty(Object=@n18, PropertyName=TextureData)
   > Branch(Condition=@n19.ReturnValue)
   > Set CurrentTDs(CurrentTDs=@n19)
   > Set ClassName(ClassName=Floor)
   n20 = Cast To BuildingRoof!(execute=@n10.then_4, Object=@n11) -> (then, CastFailed, AsBuilding Roof:Building Roof Object Reference) {is=title K2Node_DynamicCast}
   > PrintString(InString=CAST ROOF, bPrintToScreen=false)
   > n21 = GetEditorProperty(Object=@n20, PropertyName=TextureData)
   > Branch(Condition=@n21.ReturnValue)
   > Set CurrentTDs(CurrentTDs=@n21)
   > Set ClassName(ClassName=Roof)
   n22 = "Get (a copy)"(
    Array:Array of Building Texture Data Object References=Get CurrentTDs() -> (CurrentTDs:Array of Building Texture Data Object References)
    Dimension 1=2
   ) -> (Output:Building Texture Data Object Reference)
   n23 = "Get (a copy)"(Array=Get CurrentTDs(), Dimension 1=3)
   n24 = "Get (a copy)"(Array=Get CurrentTDs())
   n25 = "Get (a copy)"(Array=Get CurrentTDs(), Dimension 1=1)
   PrintString(
    execute=@n10.then_5
    InString=Append!(
     A="Texture Data : 1= "
     B=GetDisplayName(Object=@n24)
     C:String=" 2= "
     D:String=GetDisplayName(Object=@n25)
     E:String=" 3= "
     F:String=GetDisplayName(Object=@n22)
     G:String=" 4= "
     H:String=GetDisplayName(Object=@n23)
    ) -> (ReturnValue)
    bPrintToScreen=false
   )
   n26 = Get ListOfUniqueTDs()
   n27 = Branch(execute=@n9.then_1, Condition="integer > integer"(A=Length(TargetArray:Array of Str FNEC Asset Organizer TDArray Structures=@n26)))
   > Set MarkForDelete(MarkForDelete:bool=false) -> (Output_Get:bool)
   > n28 = For Each Loop(Array:Array of Str FNEC Asset Organizer TDArray Structures=Get ListOfUniqueTDs()) -> (Array Element:Str FNEC Asset Organizer TDArray Structure)
   n29 = Branch(execute=@n28.Completed, Condition=Get MarkForDelete() -> (MarkForDelete:bool))
   n30 = Get CurrentItem()
   > Remove Item!(execute, TargetArray:Array of Actor Object References=Get Selected Actors(), Item:"Actor Object Reference (by ref)"=@n30) -> (then, ReturnValue:bool)
   > Destroy Actor(self=@n30)
   > PrintString(InString="► Item is a DUPLICATE. Deleting", bPrintToScreen=false)
   > Increment Int(Value=Get DupCount() -> (DupCount:int=0))
   Add!(
    execute=@n29.else
    TargetArray:Array of Str FNEC Asset Organizer TDArray Structures=Get ListOfUniqueTDs()
    NewItem_TDArray_3_#2:Array of Building Texture Data Object References=Get CurrentTDs()
   ) -> (then, ReturnValue) {is=title K2Node_CallArrayFunction}
   > n31 = PrintString(InString="► Item is not a duplicate. Adding to the list of unique TDs", bPrintToScreen=false)
   Add!(execute=[@n31.then, @n38.then], TargetArray:Array of Actor Object References=Get Unique Actors(), NewItem:"Actor Object Reference (by ref)") -> (then, ReturnValue)
   > Branch(Condition=Get Move() -> (Move:bool))
   > Set Actor Location!(execute, self:Actor Object Reference=Get CurrentItem(), NewLocation:Vector="0, 0, 0", bSweep:bool=false, bTeleport:bool=false) -> (then, SweepHitResult:Hit Result Structure, ReturnValue:bool) {is=title K2Node_CallFunction}
   n32 = Branch(execute=@n28.LoopBody, Condition=Get MarkForDelete())
   Set DifferenceFound(execute=@n32.else, DifferenceFound:bool=false) -> (Output_Get:bool)
   > PrintString(
    InString=Append!(A="Comparing against unique TD set ##", B="To String (Integer)"(InInt=@n28.Array Index)) -> (ReturnValue)
    bPrintToScreen=false
   )
   > n33 = For Loop(LastIndex=3)
   n34 = "Get (a copy)"(
    Array=Break Str FNEC Asset Organizer TDArray!(Str_FNEC_AssetOrganizer_TDArray:"Str FNEC Asset Organizer TDArray Structure (by ref)"=@n28) -> (TDArray_3_#2:Array of Building Texture Data Object References) {is=title K2Node_BreakStruct}
    Dimension 1=@n33
   )
   n35 = "Get (a copy)"(Array=Get CurrentTDs(), Dimension 1=@n33)
   n36 = "Equal (Object)"(A=@n34, B=@n35)
   > PrintString(
    InString=Append!(
     A="Comparing "
     B=GetDisplayName(Object=@n34)
     C:String=" to  "
     D:String=GetDisplayName(Object=@n35)
     E:String=SelectString(A="    -> They are the same", B="    -> They are DIFFERENT", bPickA=@n36)
    ) -> (ReturnValue)
    bPrintToScreen=false
   )
   > Set DifferenceFound(DifferenceFound=Select!(Option 0:bool=true, Option 1:bool=Get DifferenceFound() -> (DifferenceFound:bool), Index:bool=@n36) -> (ReturnValue:bool))
   n37 = Branch(execute=@n33.Completed, Condition=Get DifferenceFound())
   Set MarkForDelete(execute=@n37.else, MarkForDelete=true)
   Add!(
    execute=@n27.else
    TargetArray:Array of Str FNEC Asset Organizer TDArray Structures=Get ListOfUniqueTDs()
    NewItem_TDArray_3_#2:Array of Building Texture Data Object References=Get CurrentTDs()
   ) -> (then, ReturnValue)
   > n38 = PrintString(InString="It is the first element. Adding to the list of unique TDs ##0", bPrintToScreen=false)
   PrintString(
    execute=@n8.else
    InString=Append(
     A="[Eliminate Duplicates] Warning - Skipping actor "
     B=GetDisplayName(Object=@n5.ArrayElement)
     C=". It is of a diferent class ("
     D=Get Class Display Name(Class=@n7)
     E=")"
    )
    TextColor=(R=1,G=0.1808,B=0,A=1)
   )
   n39 = Get NumberOfItems() -> (NumberOfItems:int=0)
   n40 = "To String (Integer)"(InInt=Get DupCount() -> (DupCount=0))
   n41 = "To String (Integer)"(InInt=Length(TargetArray=Get ListOfUniqueTDs()))
   PrintString(
    execute=@n5.Completed
    InString=Append!(
     A="[Eliminate Duplicates] Job Done! Analized [ "
     B="To String (Integer)"(InInt=@n39)
     C=" ] Objects of class \""
     D=Get ClassName() -> (ClassName:String)
     E="\"."
     F:String="         "
     G:String=Append!(A="Found [  ", B=@n41, C=" ] unique objects.     Deleted [ ", D=@n40, E=" ] duplicate objects.") -> (ReturnValue)
    ) -> (ReturnValue)
    TextColor=(R=0.0028,G=1,B=0.1274,A=1)
    Duration=5
   )
   > Return Node!(execute, UniqueActors:Array of Actor Object References=Get Unique Actors())
  ]}
 EventGraph {nodes [
   "On Clicked (B_SortActorsByClass)"() {is=title K2Node_ComponentBoundEvent}
   > n1 = GetSelectionSet()
   > ArrangeActorsByClass!(execute, self, Assets:Array of Actor Object References=@n1) -> (then) {is=title K2Node_CallFunction}
   "On Clicked (B_SortAssetsIntoFolders)"() {is=title K2Node_ComponentBoundEvent}
   > SortActorsIntoFolders()
   "On Clicked (B_EliminateDuplicates)"() {is=title K2Node_ComponentBoundEvent}
   > RemoveDoublesOnSelection()
   "On Clicked (B_DeleteDisallowedAssets)"() {is=title K2Node_ComponentBoundEvent}
   > DeleteDisallowedActors() {is=title K2Node_CallFunction}
   "On Clicked (B_SelectFolder_1)"() {is=title K2Node_ComponentBoundEvent}
   > FindActorsByFolder!(execute, self, Folder:String=Disallowed, Select:bool=true) -> (then, Actors:Array of Actor Object References)
   "On Clicked (B_SelectFolder_2)"() {is=title K2Node_ComponentBoundEvent}
   > FindActorsByFolder(Folder=Content, Select=true)
   "On Clicked (B_SelectFolder_3)"() {is=title K2Node_ComponentBoundEvent}
   > FindActorsByFolder(Folder=Creative Gallery Actors, Select=true)
   "On Clicked (B_SelectFolder_4)"() {is=title K2Node_ComponentBoundEvent}
   > FindActorsByFolder(Folder="To String (Name)"(InName=Get FolderFlaggedActor() -> (FolderFlaggedActor:Name=None)), Select=true)
   "On Clicked (B_ResetProperties)"() {is=title K2Node_ComponentBoundEvent}
   > ResetActorProperties!(
    execute
    self
    Scale:bool=IsChecked(self=Get CB_Toggle_ResetScale() -> (CB_Toggle_ResetScale:Editor Utility Check Box Object Reference))
    Rotation:bool=IsChecked(self=Get CB_Toggle_ResetRotation() -> (CB_Toggle_ResetRotation:Editor Utility Check Box Object Reference))
    Mirror:bool=IsChecked(self=Get CB_Toggle_ResetMirror() -> (CB_Toggle_ResetMirror:Editor Utility Check Box Object Reference))
   ) -> (then)
   Event Construct()
   n2 = GetParent!(self:Widget Object Reference) -> (ReturnValue:Panel Widget Object Reference)
   > n3 = Cast To EUW_AssetOrganizerTools!(execute, Object=GetOuterObject!(Object:Object Reference=GetOuterObject(Object=@n2)) -> (ReturnValue:Object Reference)) -> (then, CastFailed, AsEUW Asset Organizer Tools:EUW Asset Organizer Tools Object Reference) {is=title K2Node_DynamicCast}
   > n4 = Set OwnerWidget(OwnerWidget:EUW Asset Organizer Tools Object Reference=@n3) -> (Output_Get:EUW Asset Organizer Tools Object Reference)
   > Set AssetOrganizerDataAsset(AssetOrganizerDataAsset:DA FNEC Asset Organizer Data Asset Class Object Reference=Get DataAsset!(self:EUW Asset Organizer Tools Object Reference=@n4) -> (DataAsset:DA FNEC Asset Organizer Data Asset Class Object Reference)) -> (Output_Get:DA FNEC Asset Organizer Data Asset Class Object Reference)
  ]}
 FindActorsByFolder {nodes [
   n1 = FindActorsByFolder!() -> (then, Folder:String, Select:bool=~) {is=title K2Node_FunctionEntry}
   > Clear(TargetArray:Array of Actor Object References=Get FoundActors() -> (FoundActors:Array of Actor Object References))
   > n2 = GetEditorWorld(self=Get UnrealEditorSubsystem())
   > SelectAll(self=Get EditorActorSubsystem(), InWorld=@n2)
   > n3 = GetSelectionSet()
   > n4 = For Each Loop(Array:Array of Actor Object References=@n3) -> (Array Element:Actor Object Reference)
   n5 = Compare Int(Exec=@n4.Completed, Input=Length(TargetArray:Array of Actor Object References=Get FoundActors()))
   n6 = Show Message Dialog(
    execute=@n5."=="
    Title=MakeLiteralText(Value="LOCTEXT(\"Asset Organizer Tool - Select by Folder\")")
    Message=Format Text!(Format="LOCTEXT(\"No actors found in folder \\\"{1}\\\" or folder does no exist in scene.\")", 1:String=Get Folder() -> (Folder:String)) -> (Result)
    MessageType=Ok
    DefaultValue=Ok
    MessageCategory=Warning
   )
   Return Node!(execute=[@n9.then, @n6.then, @n7.else], Actors:Array of Actor Object References=Get FoundActors())
   n7 = Branch(execute=@n5.">", Condition=Get Select() -> (Select:bool))
   n8 = Get EditorActorSubsystem()
   > ClearActorSelectionSet(self=@n8)
   > n9 = SetSelectedLevelActors(self=@n8, ActorsToSelect=Get FoundActors())
   n10 = GetFolderPath(self=@n4) -> (ReturnValue=None)
   n11 = Compare Int(Exec=@n4.LoopBody, Input=CompareStrings(LeftStr="To String (Name)"(InName=@n10), RightStr=@n1) -> (ReturnValue=0))
   Add!(execute=@n11."==", TargetArray:Array of Actor Object References=Get FoundActors(), NewItem:"Actor Object Reference (by ref)"=@n4) -> (then, ReturnValue)
  ]}
 RemoveDoublesOnSelection {nodes [
   RemoveDoublesOnSelection() {is=title K2Node_FunctionEntry}
   > BeginTransaction(Context=Asset Organizer, Description="LOCTEXT(\"Eliminate Duplicates\")")
   > n1 = Sequence!(execute) -> (then_0, then_1)
   > n2 = GetSelectionSet()
   > n3 = SortActorsByClass(Assets=@n2)
   n4 = Length(TargetMap=@n3) -> (ReturnValue=0) {is=title K2Node_CallFunction}
   > n5 = Compare Int(Input=@n4, CompareWith=1)
   n6 = "For Each Loop (Map)"(execute=[@n10.Yes, @n5."=="], MapPin=@n3)
   > DeleteDuplicates!(
    execute
    self
    Move:bool=false
    Actors:Array of Actor Object References=Break Str FNEC Asset Organizer Actor Array(Str_FNEC_AssetOrganizer_ActorArray=@n6.ValuePin)
   ) -> (then, UniqueActors:Array of Actor Object References) {is=title K2Node_CallFunction}
   n7 = Get Key Value By Index!(TargetMap:Map of Actor Class References to Str FNEC Asset Organizer Actor Array Structures=@n3, Index:int=0) -> (Key:Actor Class Reference, Value:Str FNEC Asset Organizer Actor Array Structure) {is=title K2Node_CallFunction}
   n8 = Get Class Display Name(Class=@n7)
   n9 = Show Message Dialog(
    execute=@n5.">"
    Title=MakeLiteralText(Value="LOCTEXT(\"Asset Organizer Tool - Delete Duplicates\")")
    Message=Format Text!(
     Format="LOCTEXT(\"More than one class of actors have been selected to remove duplicates. ( {text2} classes selected )\\r\\n\\r\\nProceed with entire selection?\\r\\n\\r\\nYES - Operate over entire selection\\r\\nNO -  Only operates over the first class ( {text1} )\\r\\nCancel - Cancels the operation\")"
     text2:int=@n4
     text1:String=MakeLiteralString(Value=@n8)
    ) -> (Result)
    MessageType=YesNoCancel
    DefaultValue=Yes
    MessageCategory=Warning
   )
   > n10 = Switch on EAppReturnType(Selection=@n9)
   > n11 = DeleteDuplicates(Move=false, Actors=Break Str FNEC Asset Organizer Actor Array(Str_FNEC_AssetOrganizer_ActorArray=@n7.Value))
   n12 = Get EditorActorSubsystem()
   > ClearActorSelectionSet(self=@n12)
   > SetSelectedLevelActors(self=@n12, ActorsToSelect=@n11)
   EndTransaction(execute=@n1.then_1)
  ]}
 ResetActorProperties {nodes [
   ResetActorProperties!() -> (then, Scale:bool=~, Rotation:bool=~, Mirror:bool=~) {is=title K2Node_FunctionEntry}
   > BeginTransaction(Context=Asset Organizer, Description="LOCTEXT(\"Reset Properties\")")
   > n1 = GetSelectionSet()
   > n2 = For Each Loop(Array=@n1)
   > n3 = Set CurrentActor(CurrentActor:Actor Object Reference=@n2) -> (Output_Get:Actor Object Reference)
   > TransactObject(Object=@n3)
   > n4 = Sequence!(execute) -> (then_0, then_1, then_2:Exec, then_3:Exec)
   > Branch(Condition=Get Scale() -> (Scale:bool))
   > SetActorScale3D!(execute, self:Actor Object Reference=Get CurrentActor() -> (CurrentActor:Actor Object Reference), NewScale3D:Vector="1,1,1") -> (then)
   Branch(execute=@n4.then_1, Condition=Get Rotation() -> (Rotation:bool))
   > Set Actor Rotation!(execute, self:Actor Object Reference=Get CurrentActor(), NewRotation:Rotator="0, 0, 0", bTeleportPhysics:bool=false) -> (then, ReturnValue:bool) {is=title K2Node_CallFunction}
   Branch(execute=@n4.then_2, Condition=Get Mirror() -> (Mirror:bool))
   > n5 = Sequence!(execute) -> (then_0, then_1, then_2, then_3, then_4:Exec, then_5:Exec)
   n6 = Get CurrentActor()
   > n7 = Cast To BuildingPropWall(Object=@n6)
   n8 = Cast To BuildingWall(execute=@n5.then_1, Object=@n6)
   n9 = Cast To BuildingFloor(execute=@n5.then_3, Object=@n6)
   n10 = Cast To BuildingRoof(execute=@n5.then_4, Object=@n6)
   n11 = Cast To BuildingProp(execute=@n5.then_2, Object=@n6)
   SetMirrored!(
    execute=[@n8.then, @n9.then, @n10.then, @n11.then, @n7.then]
    self:Building Static Mesh Object Reference=[@n8, @n9, @n10, @n11, @n7]
    bIsMirrored:bool=false
   ) -> (then)
   EndTransaction(execute=@n4.then_3)
  ]}
 SortActorsByClass {nodes [
   n1 = SortActorsByClass!() -> (then, Assets:Array of Actor Object References) {is=title K2Node_FunctionEntry}
   > n2 = Set Selected Assets(Selected Assets:Array of Actor Object References=@n1) -> (Output_Get:Array of Actor Object References)
   > n3 = For Each Loop(Array=@n2)
   Return Node!(execute=@n3.Completed, SortedActors:Map of Actor Class References to Str FNEC Asset Organizer Actor Array Structures=Get SortedActorsList())
   n4 = Get Class(Object=@n3)
   > n5 = Branch(Condition=Contains Item(
     TargetArray:Array of Actor Class References=Get Class List() -> (Class List:Array of Actor Class References)
     ItemToFind:"Actor Class Reference (by ref)"=@n4
    ))
   n6 = Get Class(Object=@n3)
   n7 = Find(
    TargetMap:Map of Actor Class References to Str FNEC Asset Organizer Actor Array Structures=Get SortedActorsList()
    Key:"Actor Class Reference (by ref)"=@n6
   ) -> (Value:Str FNEC Asset Organizer Actor Array Structure)
   > Branch(Condition=@n7.ReturnValue)
   > n8 = Set TransferArray(TransferArray:Array of Actor Object References=Break Str FNEC Asset Organizer Actor Array(Str_FNEC_AssetOrganizer_ActorArray=@n7)) -> (Output_Get:Array of Actor Object References)
   > Add(TargetArray=@n8, NewItem=@n3)
   > Add!(
    execute
    TargetMap:Map of Actor Class References to Str FNEC Asset Organizer Actor Array Structures=Get SortedActorsList()
    Key:"Actor Class Reference (by ref)"=@n6
    Value:"Str FNEC Asset Organizer Actor Array Structure (by ref)"=Make Str FNEC Asset Organizer Actor Array!(ActorArray_3_#1:Array of Actor Object References=@n8) -> (Str_FNEC_AssetOrganizer_ActorArray:Str FNEC Asset Organizer Actor Array Structure) {is=title K2Node_MakeStruct}
   ) -> (then) {is=title K2Node_CallFunction}
   Add(execute=@n5.else, TargetArray:Array of Actor Class References=Get Class List(), NewItem:"Actor Class Reference (by ref)"=@n4) {is=title K2Node_CallArrayFunction}
   > Add(
    TargetMap=Get SortedActorsList()
    Key=@n4
    Value=Make Str FNEC Asset Organizer Actor Array(ActorArray_3_#1=Make Array("[0]":Actor Object Reference=@n3) -> (Array:Array of Actor Object References))
   ) {is=title K2Node_CallFunction}
  ]}
 SortActorsIntoFolders {nodes [
   SortActorsIntoFolders() {is=title K2Node_FunctionEntry}
   > BeginTransaction(Context=Asset Organizer, Description="LOCTEXT(\"Sort into Folders\")")
   > n1 = GetEditorWorld(self=Get UnrealEditorSubsystem())
   > SelectAll(self=Get EditorActorSubsystem(), InWorld=@n1)
   > Clear(TargetArray=Get FoundDisallowedActor() -> (FoundDisallowedActor:Array of Actor Object References))
   > Clear(TargetArray=Get FoundFlaggedActor() -> (FoundFlaggedActor:Array of Actor Object References))
   > n2 = Sequence!(execute) -> (then_0, then_1)
   > n3 = GetSelectionSet()
   > n4 = For Each Loop(Array=@n3)
   > n5 = Set CurrentActor(CurrentActor=@n4)
   > TransactObject(Object=@n5)
   n6 = Get CurrentActor()
   > Branch(Condition=NOT Boolean(A=ActorHasTag!(self:Actor Object Reference=@n6, Tag:Name=Marker) -> (ReturnValue:bool)))
   n7 = Get Class(Object=Get CurrentActor())
   n8 = Get AssetOrganizerDataAsset() -> (AssetOrganizerDataAsset:DA FNEC Asset Organizer Data Asset Class Object Reference)
   > n9 = Branch(Condition=Contains Item(
     TargetArray=Get DisallowedActorClasses!(self:DA FNEC Asset Organizer Data Asset Class Object Reference=@n8) -> (DisallowedActorClasses:Array of Actor Class References)
     ItemToFind=@n7
    ))
   > Add(TargetArray:Array of Actor Object References=Get FoundDisallowedActor(), NewItem:"Actor Object Reference (by ref)"=Get CurrentActor()) {is=title K2Node_CallArrayFunction}
   > SetFolderPath!(
    execute
    self:Actor Object Reference=Get CurrentActor()
    NewFolderPath:"Name (by ref)"=MakeLiteralName(Value=Get FolderDisallowed() -> (FolderDisallowed:Name=None))
   ) -> (then)
   n10 = Branch(
    execute=@n9.else
    Condition=Contains Item(
     TargetArray=Get GalleryActorClasses!(self:DA FNEC Asset Organizer Data Asset Class Object Reference=@n8) -> (GalleryActorClasses:Array of Actor Class References)
     ItemToFind=@n7
    )
   )
   > SetFolderPath(self=Get CurrentActor(), NewFolderPath=MakeLiteralName(Value=Get FolderCreativeActor() -> (FolderCreativeActor:Name=None)))
   Set IsContentClass(execute=@n10.else, IsContentClass:bool=false) -> (Output_Get:bool)
   > n11 = For Each Loop with Break!(
    Exec:Exec
    Array:Array of Actor Class References=Get ContentClasses!(self:DA FNEC Asset Organizer Data Asset Class Object Reference=Get AssetOrganizerDataAsset()) -> (ContentClasses:Array of Actor Class References)
    Break:Exec
   ) -> (LoopBody:Exec, Array Element:Actor Class Reference, Array Index:int, Completed:Exec)
   n12 = Get Class(Object=Get CurrentActor())
   > Set IsContentClass(IsContentClass=Select(Option 0=Get IsContentClass() -> (IsContentClass:bool), Option 1=true, Index=ClassIsChildOf(TestClass=@n12, ParentClass=@n11)))
   n13 = Branch(execute=@n11.Completed, Condition=Get IsContentClass())
   > n14 = Cast To RandomColor_Prop!(execute, Object=Get CurrentActor()) -> (then, CastFailed, AsRandom Color Prop:Random Color Prop Object Reference) {is=title K2Node_DynamicCast}
   > Add(TargetArray=Get FoundFlaggedActor(), NewItem=Get CurrentActor())
   > SetFolderPath(self=Get CurrentActor(), NewFolderPath=MakeLiteralName(Value=Get FolderFlaggedActor() -> (FolderFlaggedActor=None)))
   SetFolderPath(execute=@n14.CastFailed, self=Get CurrentActor(), NewFolderPath=MakeLiteralName(Value=Get FolderContent() -> (FolderContent:Name=None)))
   Add(execute=@n13.else, TargetArray=Get FoundFlaggedActor(), NewItem=Get CurrentActor())
   > SetFolderPath(self=Get CurrentActor(), NewFolderPath=MakeLiteralName(Value=Get FolderFlaggedActor() -> (FolderFlaggedActor=None)))
   n15 = Sequence(execute=@n2.then_1)
   > ClearActorSelectionSet(self=Get EditorActorSubsystem())
   n16 = Get FoundFlaggedActor()
   n17 = Get FoundDisallowedActor()
   > Show Message Dialog(
    Title=MakeLiteralText(Value="LOCTEXT(\"Asset Organizer Tool - Sort Actors into Folders\")")
    Message=Format Text!(
     Format="LOCTEXT(\"Sorted actors into folders.\\r\\n\\r\\nFound {1} Disallowed actors and {2} Flagged Actors.\")"
     1:int=Length(TargetArray=@n17) {is=title K2Node_CallArrayFunction}
     2:int=Length(TargetArray=@n16)
    ) -> (Result)
    MessageType=Ok
    DefaultValue=Ok
    MessageCategory=Info
   )
   EndTransaction(execute=@n15.then_1)
   n18 = Switch on EAppReturnType(Selection=No)
   n19 = Get EditorActorSubsystem()
   > ClearActorSelectionSet(self=@n19)
   ClearActorSelectionSet(execute=@n18.Yes, self=@n19)
   > SetSelectedLevelActors(self=@n19, ActorsToSelect=Get FoundDisallowedActor())
  ]}
]
variables [
 AssetOrganizerDataAsset {type=DA_FNEC_AssetOrganizer_DataAssetClass_C, default=null}
 FolderContent {type=string, default=Content}
 FolderCreativeActor {type=string, default=Creative Gallery Actors}
 FolderDisallowed {type=string, default=Disallowed}
 FolderFlaggedActor {type=string, default=Flagged Actors}
 OwnerWidget {type=EUW_AssetOrganizerTools_C, default=null}
]
widgets {
 info {rootWidgetClass=&StackBox, widgetCount=57}
 widgets [
  StackBox_0 {widgetClassPath=&StackBox, widget=StackBox_0}
  VerticalBox_4 {
   widgetClassPath=&VerticalBox
   parent=StackBox_0
   widget=VerticalBox_4
   slot {class=StackBoxSlot, properties {size {sizeRule=Fill, value=1}}}
  }
  HorizontalBox-Step1 {
   widgetClassPath=&HorizontalBox
   parent=VerticalBox_4
   widget=HorizontalBox-Step1
   slot {class=VerticalBoxSlot, properties {padding=<ltrb 5 5 5 5>}}
  }
  SizeBox_3 {
   widgetClassPath=&SizeBox
   parent=HorizontalBox-Step1
   widget=SizeBox_3
   properties {bOverride_MinDesiredHeight=true, minDesiredHeight=60}
   slot {class=HorizontalBoxSlot, properties {size {sizeRule=Fill, value=1}}}
  }
  B_SortAssetsIntoFolders {
   widgetClassPath=&EditorUtilityButton
   parent=SizeBox_3
   widget=B_SortAssetsIntoFolders
   properties {
    toolTipText="Goes over all assets in the scene and sorts them into folders\r\n\r\n\"Content\" - Props to be processed and made into galleries\r\n\"Disallowed\" - Gameplay assets that should not be added to galleries\r\n\"Creative Gallery Assets\" - Assets related to gallery work\r\n\"Flagged Actors\" - assets that need to be analized"
    widgetStyle {
     clickedSlateSound {resourceObject=null}
     disabled {
      bIsDynamicallyLoaded=false
      drawAs=RoundedBox
      imageType=NoImage
      mirroring=NoMirror
      resourceName=None
      resourceObject=null
      tiling=NoTile
      imageSize=<0 0>
      margin=<ltrb 0 0 0 0>
      outlineSettings {
       bUseBrushTransparency=false
       roundingType=FixedRadius
       width=1
       color {colorUseRule=UseColor_ColorTable, specifiedColor=<rgba 0 0 0 0>}
       cornerRadii=<4 4 4 4>
      }
      tintColor {colorUseRule=UseColor_Specified, specifiedColor=<rgba 0.0395 0.0395 0.0395 1>}
      uVRegion {bIsValid=false, max=<0 0>, min=<0 0>}
     }
     disabledForeground {colorUseRule=UseColor_ColorTable, specifiedColor=<rgba 0 0 0 0>}
     hovered {
      bIsDynamicallyLoaded=false
      drawAs=RoundedBox
      imageType=NoImage
      mirroring=NoMirror
      resourceName=None
      resourceObject=null
      tiling=NoTile
      imageSize=<32 32>
      margin=<ltrb 0 0 0 0>
      outlineSettings {
       bUseBrushTransparency=false
       roundingType=FixedRadius
       width=1
       color {colorUseRule=UseColor_ColorTable, specifiedColor=<rgba 0 0 0 0>}
       cornerRadii=<4 4 4 4>
      }
      tintColor {colorUseRule=UseColor_Specified, specifiedColor=<rgba 0.0953 0.0953 0.0953 1>}
      uVRegion {bIsValid=false, max=<0 0>, min=<0 0>}
     }
     hoveredForeground {colorUseRule=UseColor_ColorTable, specifiedColor=<rgba 0 0 0 0>}
     hoveredSlateSound {resourceObject=null}
     normal {
      bIsDynamicallyLoaded=false
      drawAs=RoundedBox
      imageType=NoImage
      mirroring=NoMirror
      resourceName=None
      resourceObject=null
      tiling=NoTile
      imageSize=<32 32>
      margin=<ltrb 0 0 0 0>
      outlineSettings {
       bUseBrushTransparency=false
       roundingType=FixedRadius
       width=1
       color {colorUseRule=UseColor_ColorTable, specifiedColor=<rgba 0 0 0 0>}
       cornerRadii=<4 4 4 4>
      }
      tintColor {colorUseRule=UseColor_Specified, specifiedColor=<rgba 0.0395 0.0395 0.0395 1>}
      uVRegion {bIsValid=false, max=<0 0>, min=<0 0>}
     }
     normalForeground {colorUseRule=UseColor_ColorTable, specifiedColor=<rgba 0 0 0 0>}
     normalPadding=<ltrb 12 1.5 12 1.5>
     pressed {
      bIsDynamicallyLoaded=false
      drawAs=RoundedBox
      imageType=NoImage
      mirroring=NoMirror
      resourceName=None
      resourceObject=null
      tiling=NoTile
      imageSize=<32 32>
      margin=<ltrb 0 0 0 0>
      outlineSettings {
       bUseBrushTransparency=false
       roundingType=FixedRadius
       width=1
       color {colorUseRule=UseColor_ColorTable, specifiedColor=<rgba 0 0 0 0>}
       cornerRadii=<4 4 4 4>
      }
      tintColor {colorUseRule=UseColor_Specified, specifiedColor=<rgba 0.0284 0.0284 0.0284 1>}
      uVRegion {bIsValid=false, max=<0 0>, min=<0 0>}
     }
     pressedForeground {colorUseRule=UseColor_ColorTable, specifiedColor=<rgba 0 0 0 0>}
     pressedPadding=<ltrb 12 2.5 12 0.5>
     pressedSlateSound {resourceObject=null}
    }
   }
   slot {class=SizeBoxSlot}
  }
  TextBlock_7 {
   widgetClassPath=&TextBlock
   parent=B_SortAssetsIntoFolders
   widget=TextBlock_7
   properties {
    text=Sort Assets into Folders
    font {
     bForceMonospaced=false
     bMaterialIsStencil=false
     fontMaterial=null
     fontObject=&Roboto
     letterSpacing=0
     monospacedWidth=1
     size=9
     skewAmount=0
     typefaceFontName=Regular
     outlineSettings {
      bApplyOutlineToDropShadows=false
      bMiteredCorners=false
      bSeparateFillAlpha=false
      outlineMaterial=null
      outlineSize=0
      outlineColor=<rgba 0 0 0 1>
     }
    }
   }
   slot {class=ButtonSlot}
  }
  SizeBox_116 {
   widgetClassPath=&SizeBox
   parent=HorizontalBox-Step1
   widget=SizeBox_116
   properties {bOverride_MaxDesiredWidth=true, bOverride_MinDesiredWidth=true, maxDesiredWidth=270, minDesiredWidth=270}
   slot {class=HorizontalBoxSlot, properties {padding=<ltrb 10 0 0 0>}}
  }
  VerticalBox_0 {widgetClassPath=&VerticalBox, parent=SizeBox_116, widget=VerticalBox_0, slot {class=SizeBoxSlot}}
  TextBlock_11 {
   widgetClassPath=&TextBlock
   parent=VerticalBox_0
   widget=TextBlock_11
   properties {
    text=Select Actors by Folder
    font {
     bForceMonospaced=false
     bMaterialIsStencil=false
     fontMaterial=null
     fontObject=&Roboto
     letterSpacing=0
     monospacedWidth=1
     size=9
     skewAmount=0
     typefaceFontName=Light
     outlineSettings {
      bApplyOutlineToDropShadows=false
      bMiteredCorners=false
      bSeparateFillAlpha=false
      outlineMaterial=null
      outlineSize=0
      outlineColor=<rgba 0 0 0 1>
     }
    }
   }
   slot {class=VerticalBoxSlot, properties {horizontalAlignment=HAlign_Center, verticalAlignment=VAlign_Center, size {sizeRule=Fill, value=1}}}
  }
  UniformGridPanel_99 {widgetClassPath=/Script/UMG.UniformGridPanel, parent=VerticalBox_0, widget=UniformGridPanel_99, slot {class=VerticalBoxSlot}}
  SizeBox_7 {
   widgetClassPath=&SizeBox
   parent=UniformGridPanel_99
   widget=SizeBox_7
   slot {class=UniformGridSlot, properties {horizontalAlignment=HAlign_Fill, verticalAlignment=VAlign_Fill}}
  }
  B_SelectFolder_1 {
   widgetClassPath=&EditorUtilityButton
   parent=SizeBox_7
   widget=B_SelectFolder_1
   properties {
    toolTipText="Goes over all assets in the scene and sorts them into folders\r\n\r\n\"Content\" - Props to be processed and made into galleries\r\n\"Disallowed\" - Gameplay assets that should not be added to galleries\r\n\"Creative Gallery Assets\" - Assets related to gallery work"
    widgetStyle {
     clickedSlateSound {resourceObject=null}
     disabled {
      bIsDynamicallyLoaded=false
      drawAs=RoundedBox
      imageType=NoImage
      mirroring=NoMirror
      resourceName=None
      resourceObject=null
      tiling=NoTile
      imageSize=<0 0>
      margin=<ltrb 0 0 0 0>
      outlineSettings {
       bUseBrushTransparency=false
       roundingType=FixedRadius
       width=1
       color {colorUseRule=UseColor_ColorTable, specifiedColor=<rgba 0 0 0 0>}
       cornerRadii=<4 4 4 4>
      }
      tintColor {colorUseRule=UseColor_Specified, specifiedColor=<rgba 0.0395 0.0395 0.0395 1>}
      uVRegion {bIsValid=false, max=<0 0>, min=<0 0>}
     }
     disabledForeground {colorUseRule=UseColor_ColorTable, specifiedColor=<rgba 0 0 0 0>}
     hovered {
      bIsDynamicallyLoaded=false
      drawAs=RoundedBox
      imageType=NoImage
      mirroring=NoMirror
      resourceName=None
      resourceObject=null
      tiling=NoTile
      imageSize=<32 32>
      margin=<ltrb 0 0 0 0>
      outlineSettings {
       bUseBrushTransparency=false
       roundingType=FixedRadius
       width=1
       color {colorUseRule=UseColor_ColorTable, specifiedColor=<rgba 0 0 0 0>}
       cornerRadii=<4 4 4 4>
      }
      tintColor {colorUseRule=UseColor_Specified, specifiedColor=<rgba 0.0953 0.0953 0.0953 1>}
      uVRegion {bIsValid=false, max=<0 0>, min=<0 0>}
     }
     hoveredForeground {colorUseRule=UseColor_ColorTable, specifiedColor=<rgba 0 0 0 0>}
     hoveredSlateSound {resourceObject=null}
     normal {
      bIsDynamicallyLoaded=false
      drawAs=RoundedBox
      imageType=NoImage
      mirroring=NoMirror
      resourceName=None
      resourceObject=null
      tiling=NoTile
      imageSize=<32 32>
      margin=<ltrb 0 0 0 0>
      outlineSettings {
       bUseBrushTransparency=false
       roundingType=FixedRadius
       width=1
       color {colorUseRule=UseColor_ColorTable, specifiedColor=<rgba 0 0 0 0>}
       cornerRadii=<4 4 4 4>
      }
      tintColor {colorUseRule=UseColor_Specified, specifiedColor=<rgba 0.0395 0.0395 0.0395 1>}
      uVRegion {bIsValid=false, max=<0 0>, min=<0 0>}
     }
     normalForeground {colorUseRule=UseColor_ColorTable, specifiedColor=<rgba 0 0 0 0>}
     normalPadding=<ltrb 0 0 0 0>
     pressed {
      bIsDynamicallyLoaded=false
      drawAs=RoundedBox
      imageType=NoImage
      mirroring=NoMirror
      resourceName=None
      resourceObject=null
      tiling=NoTile
      imageSize=<32 32>
      margin=<ltrb 0 0 0 0>
      outlineSettings {
       bUseBrushTransparency=false
       roundingType=FixedRadius
       width=1
       color {colorUseRule=UseColor_ColorTable, specifiedColor=<rgba 0 0 0 0>}
       cornerRadii=<4 4 4 4>
      }
      tintColor {colorUseRule=UseColor_Specified, specifiedColor=<rgba 0.0284 0.0284 0.0284 1>}
      uVRegion {bIsValid=false, max=<0 0>, min=<0 0>}
     }
     pressedForeground {colorUseRule=UseColor_ColorTable, specifiedColor=<rgba 0 0 0 0>}
     pressedPadding=<ltrb 0 0 0 0>
     pressedSlateSound {resourceObject=null}
    }
   }
   slot {class=SizeBoxSlot, properties {padding=<ltrb 2 2 2 2>}}
  }
  TextBlock_1 {
   widgetClassPath=&TextBlock
   parent=B_SelectFolder_1
   widget=TextBlock_1
   properties {
    text=Disallowed
    font {
     bForceMonospaced=false
     bMaterialIsStencil=false
     fontMaterial=null
     fontObject=&Roboto
     letterSpacing=0
     monospacedWidth=1
     size=9
     skewAmount=0
     typefaceFontName=Regular
     outlineSettings {
      bApplyOutlineToDropShadows=false
      bMiteredCorners=false
      bSeparateFillAlpha=false
      outlineMaterial=null
      outlineSize=0
      outlineColor=<rgba 0 0 0 1>
     }
    }
   }
   slot {class=ButtonSlot}
  }
  SizeBox_8 {
   widgetClassPath=&SizeBox
   parent=UniformGridPanel_99
   widget=SizeBox_8
   slot {class=UniformGridSlot, properties {column=1, horizontalAlignment=HAlign_Fill, verticalAlignment=VAlign_Fill}}
  }
  B_SelectFolder_2 {
   widgetClassPath=&EditorUtilityButton
   parent=SizeBox_8
   widget=B_SelectFolder_2
   properties {
    toolTipText="Goes over all assets in the scene and sorts them into folders\r\n\r\n\"Content\" - Props to be processed and made into galleries\r\n\"Disallowed\" - Gameplay assets that should not be added to galleries\r\n\"Creative Gallery Assets\" - Assets related to gallery work"
    widgetStyle {
     clickedSlateSound {resourceObject=null}
     disabled {
      bIsDynamicallyLoaded=false
      drawAs=RoundedBox
      imageType=NoImage
      mirroring=NoMirror
      resourceName=None
      resourceObject=null
      tiling=NoTile
      imageSize=<0 0>
      margin=<ltrb 0 0 0 0>
      outlineSettings {
       bUseBrushTransparency=false
       roundingType=FixedRadius
       width=1
       color {colorUseRule=UseColor_ColorTable, specifiedColor=<rgba 0 0 0 0>}
       cornerRadii=<4 4 4 4>
      }
      tintColor {colorUseRule=UseColor_Specified, specifiedColor=<rgba 0.0395 0.0395 0.0395 1>}
      uVRegion {bIsValid=false, max=<0 0>, min=<0 0>}
     }
     disabledForeground {colorUseRule=UseColor_ColorTable, specifiedColor=<rgba 0 0 0 0>}
     hovered {
      bIsDynamicallyLoaded=false
      drawAs=RoundedBox
      imageType=NoImage
      mirroring=NoMirror
      resourceName=None
      resourceObject=null
      tiling=NoTile
      imageSize=<32 32>
      margin=<ltrb 0 0 0 0>
      outlineSettings {
       bUseBrushTransparency=false
       roundingType=FixedRadius
       width=1
       color {colorUseRule=UseColor_ColorTable, specifiedColor=<rgba 0 0 0 0>}
       cornerRadii=<4 4 4 4>
      }
      tintColor {colorUseRule=UseColor_Specified, specifiedColor=<rgba 0.0953 0.0953 0.0953 1>}
      uVRegion {bIsValid=false, max=<0 0>, min=<0 0>}
     }
     hoveredForeground {colorUseRule=UseColor_ColorTable, specifiedColor=<rgba 0 0 0 0>}
     hoveredSlateSound {resourceObject=null}
     normal {
      bIsDynamicallyLoaded=false
      drawAs=RoundedBox
      imageType=NoImage
      mirroring=NoMirror
      resourceName=None
      resourceObject=null
      tiling=NoTile
      imageSize=<32 32>
      margin=<ltrb 0 0 0 0>
      outlineSettings {
       bUseBrushTransparency=false
       roundingType=FixedRadius
       width=1
       color {colorUseRule=UseColor_ColorTable, specifiedColor=<rgba 0 0 0 0>}
       cornerRadii=<4 4 4 4>
      }
      tintColor {colorUseRule=UseColor_Specified, specifiedColor=<rgba 0.0395 0.0395 0.0395 1>}
      uVRegion {bIsValid=false, max=<0 0>, min=<0 0>}
     }
     normalForeground {colorUseRule=UseColor_ColorTable, specifiedColor=<rgba 0 0 0 0>}
     normalPadding=<ltrb 0 0 0 0>
     pressed {
      bIsDynamicallyLoaded=false
      drawAs=RoundedBox
      imageType=NoImage
      mirroring=NoMirror
      resourceName=None
      resourceObject=null
      tiling=NoTile
      imageSize=<32 32>
      margin=<ltrb 0 0 0 0>
      outlineSettings {
       bUseBrushTransparency=false
       roundingType=FixedRadius
       width=1
       color {colorUseRule=UseColor_ColorTable, specifiedColor=<rgba 0 0 0 0>}
       cornerRadii=<4 4 4 4>
      }
      tintColor {colorUseRule=UseColor_Specified, specifiedColor=<rgba 0.0284 0.0284 0.0284 1>}
      uVRegion {bIsValid=false, max=<0 0>, min=<0 0>}
     }
     pressedForeground {colorUseRule=UseColor_ColorTable, specifiedColor=<rgba 0 0 0 0>}
     pressedPadding=<ltrb 0 0 0 0>
     pressedSlateSound {resourceObject=null}
    }
   }
   slot {class=SizeBoxSlot, properties {padding=<ltrb 2 2 2 2>}}
  }
  TextBlock_6 {
   widgetClassPath=&TextBlock
   parent=B_SelectFolder_2
   widget=TextBlock_6
   properties {
    text=Content
    font {
     bForceMonospaced=false
     bMaterialIsStencil=false
     fontMaterial=null
     fontObject=&Roboto
     letterSpacing=0
     monospacedWidth=1
     size=9
     skewAmount=0
     typefaceFontName=Regular
     outlineSettings {
      bApplyOutlineToDropShadows=false
      bMiteredCorners=false
      bSeparateFillAlpha=false
      outlineMaterial=null
      outlineSize=0
      outlineColor=<rgba 0 0 0 1>
     }
    }
   }
   slot {class=ButtonSlot}
  }
  SizeBox_9 {
   widgetClassPath=&SizeBox
   parent=UniformGridPanel_99
   widget=SizeBox_9
   slot {class=UniformGridSlot, properties {horizontalAlignment=HAlign_Fill, row=1, verticalAlignment=VAlign_Fill}}
  }
  B_SelectFolder_3 {
   widgetClassPath=&EditorUtilityButton
   parent=SizeBox_9
   widget=B_SelectFolder_3
   properties {
    toolTipText="Goes over all assets in the scene and sorts them into folders\r\n\r\n\"Content\" - Props to be processed and made into galleries\r\n\"Disallowed\" - Gameplay assets that should not be added to galleries\r\n\"Creative Gallery Assets\" - Assets related to gallery work"
    widgetStyle {
     clickedSlateSound {resourceObject=null}
     disabled {
      bIsDynamicallyLoaded=false
      drawAs=RoundedBox
      imageType=NoImage
      mirroring=NoMirror
      resourceName=None
      resourceObject=null
      tiling=NoTile
      imageSize=<0 0>
      margin=<ltrb 0 0 0 0>
      outlineSettings {
       bUseBrushTransparency=false
       roundingType=FixedRadius
       width=1
       color {colorUseRule=UseColor_ColorTable, specifiedColor=<rgba 0 0 0 0>}
       cornerRadii=<4 4 4 4>
      }
      tintColor {colorUseRule=UseColor_Specified, specifiedColor=<rgba 0.0395 0.0395 0.0395 1>}
      uVRegion {bIsValid=false, max=<0 0>, min=<0 0>}
     }
     disabledForeground {colorUseRule=UseColor_ColorTable, specifiedColor=<rgba 0 0 0 0>}
     hovered {
      bIsDynamicallyLoaded=false
      drawAs=RoundedBox
      imageType=NoImage
      mirroring=NoMirror
      resourceName=None
      resourceObject=null
      tiling=NoTile
      imageSize=<32 32>
      margin=<ltrb 0 0 0 0>
      outlineSettings {
       bUseBrushTransparency=false
       roundingType=FixedRadius
       width=1
       color {colorUseRule=UseColor_ColorTable, specifiedColor=<rgba 0 0 0 0>}
       cornerRadii=<4 4 4 4>
      }
      tintColor {colorUseRule=UseColor_Specified, specifiedColor=<rgba 0.0953 0.0953 0.0953 1>}
      uVRegion {bIsValid=false, max=<0 0>, min=<0 0>}
     }
     hoveredForeground {colorUseRule=UseColor_ColorTable, specifiedColor=<rgba 0 0 0 0>}
     hoveredSlateSound {resourceObject=null}
     normal {
      bIsDynamicallyLoaded=false
      drawAs=RoundedBox
      imageType=NoImage
      mirroring=NoMirror
      resourceName=None
      resourceObject=null
      tiling=NoTile
      imageSize=<32 32>
      margin=<ltrb 0 0 0 0>
      outlineSettings {
       bUseBrushTransparency=false
       roundingType=FixedRadius
       width=1
       color {colorUseRule=UseColor_ColorTable, specifiedColor=<rgba 0 0 0 0>}
       cornerRadii=<4 4 4 4>
      }
      tintColor {colorUseRule=UseColor_Specified, specifiedColor=<rgba 0.0395 0.0395 0.0395 1>}
      uVRegion {bIsValid=false, max=<0 0>, min=<0 0>}
     }
     normalForeground {colorUseRule=UseColor_ColorTable, specifiedColor=<rgba 0 0 0 0>}
     normalPadding=<ltrb 0 0 0 0>
     pressed {
      bIsDynamicallyLoaded=false
      drawAs=RoundedBox
      imageType=NoImage
      mirroring=NoMirror
      resourceName=None
      resourceObject=null
      tiling=NoTile
      imageSize=<32 32>
      margin=<ltrb 0 0 0 0>
      outlineSettings {
       bUseBrushTransparency=false
       roundingType=FixedRadius
       width=1
       color {colorUseRule=UseColor_ColorTable, specifiedColor=<rgba 0 0 0 0>}
       cornerRadii=<4 4 4 4>
      }
      tintColor {colorUseRule=UseColor_Specified, specifiedColor=<rgba 0.0284 0.0284 0.0284 1>}
      uVRegion {bIsValid=false, max=<0 0>, min=<0 0>}
     }
     pressedForeground {colorUseRule=UseColor_ColorTable, specifiedColor=<rgba 0 0 0 0>}
     pressedPadding=<ltrb 0 0 0 0>
     pressedSlateSound {resourceObject=null}
    }
   }
   slot {class=SizeBoxSlot, properties {padding=<ltrb 2 2 2 2>}}
  }
  TextBlock_8 {
   widgetClassPath=&TextBlock
   parent=B_SelectFolder_3
   widget=TextBlock_8
   properties {
    text=Creative Gallery Actors
    font {
     bForceMonospaced=false
     bMaterialIsStencil=false
     fontMaterial=null
     fontObject=&Roboto
     letterSpacing=0
     monospacedWidth=1
     size=9
     skewAmount=0
     typefaceFontName=Regular
     outlineSettings {
      bApplyOutlineToDropShadows=false
      bMiteredCorners=false
      bSeparateFillAlpha=false
      outlineMaterial=null
      outlineSize=0
      outlineColor=<rgba 0 0 0 1>
     }
    }
   }
   slot {class=ButtonSlot}
  }
  SizeBox_10 {
   widgetClassPath=&SizeBox
   parent=UniformGridPanel_99
   widget=SizeBox_10
   slot {class=UniformGridSlot, properties {column=1, horizontalAlignment=HAlign_Fill, row=1, verticalAlignment=VAlign_Fill}}
  }
  B_SelectFolder_4 {
   widgetClassPath=&EditorUtilityButton
   parent=SizeBox_10
   widget=B_SelectFolder_4
   properties {
    toolTipText="Goes over all assets in the scene and sorts them into folders\r\n\r\n\"Content\" - Props to be processed and made into galleries\r\n\"Disallowed\" - Gameplay assets that should not be added to galleries\r\n\"Creative Gallery Assets\" - Assets related to gallery work"
    widgetStyle {
     clickedSlateSound {resourceObject=null}
     disabled {
      bIsDynamicallyLoaded=false
      drawAs=RoundedBox
      imageType=NoImage
      mirroring=NoMirror
      resourceName=None
      resourceObject=null
      tiling=NoTile
      imageSize=<0 0>
      margin=<ltrb 0 0 0 0>
      outlineSettings {
       bUseBrushTransparency=false
       roundingType=FixedRadius
       width=1
       color {colorUseRule=UseColor_ColorTable, specifiedColor=<rgba 0 0 0 0>}
       cornerRadii=<4 4 4 4>
      }
      tintColor {colorUseRule=UseColor_Specified, specifiedColor=<rgba 0.0395 0.0395 0.0395 1>}
      uVRegion {bIsValid=false, max=<0 0>, min=<0 0>}
     }
     disabledForeground {colorUseRule=UseColor_ColorTable, specifiedColor=<rgba 0 0 0 0>}
     hovered {
      bIsDynamicallyLoaded=false
      drawAs=RoundedBox
      imageType=NoImage
      mirroring=NoMirror
      resourceName=None
      resourceObject=null
      tiling=NoTile
      imageSize=<32 32>
      margin=<ltrb 0 0 0 0>
      outlineSettings {
       bUseBrushTransparency=false
       roundingType=FixedRadius
       width=1
       color {colorUseRule=UseColor_ColorTable, specifiedColor=<rgba 0 0 0 0>}
       cornerRadii=<4 4 4 4>
      }
      tintColor {colorUseRule=UseColor_Specified, specifiedColor=<rgba 0.0953 0.0953 0.0953 1>}
      uVRegion {bIsValid=false, max=<0 0>, min=<0 0>}
     }
     hoveredForeground {colorUseRule=UseColor_ColorTable, specifiedColor=<rgba 0 0 0 0>}
     hoveredSlateSound {resourceObject=null}
     normal {
      bIsDynamicallyLoaded=false
      drawAs=RoundedBox
      imageType=NoImage
      mirroring=NoMirror
      resourceName=None
      resourceObject=null
      tiling=NoTile
      imageSize=<32 32>
      margin=<ltrb 0 0 0 0>
      outlineSettings {
       bUseBrushTransparency=false
       roundingType=FixedRadius
       width=1
       color {colorUseRule=UseColor_ColorTable, specifiedColor=<rgba 0 0 0 0>}
       cornerRadii=<4 4 4 4>
      }
      tintColor {colorUseRule=UseColor_Specified, specifiedColor=<rgba 0.0395 0.0395 0.0395 1>}
      uVRegion {bIsValid=false, max=<0 0>, min=<0 0>}
     }
     normalForeground {colorUseRule=UseColor_ColorTable, specifiedColor=<rgba 0 0 0 0>}
     normalPadding=<ltrb 0 0 0 0>
     pressed {
      bIsDynamicallyLoaded=false
      drawAs=RoundedBox
      imageType=NoImage
      mirroring=NoMirror
      resourceName=None
      resourceObject=null
      tiling=NoTile
      imageSize=<32 32>
      margin=<ltrb 0 0 0 0>
      outlineSettings {
       bUseBrushTransparency=false
       roundingType=FixedRadius
       width=1
       color {colorUseRule=UseColor_ColorTable, specifiedColor=<rgba 0 0 0 0>}
       cornerRadii=<4 4 4 4>
      }
      tintColor {colorUseRule=UseColor_Specified, specifiedColor=<rgba 0.0284 0.0284 0.0284 1>}
      uVRegion {bIsValid=false, max=<0 0>, min=<0 0>}
     }
     pressedForeground {colorUseRule=UseColor_ColorTable, specifiedColor=<rgba 0 0 0 0>}
     pressedPadding=<ltrb 0 0 0 0>
     pressedSlateSound {resourceObject=null}
    }
   }
   slot {class=SizeBoxSlot, properties {padding=<ltrb 2 2 2 2>}}
  }
  TextBlock_9 {
   widgetClassPath=&TextBlock
   parent=B_SelectFolder_4
   widget=TextBlock_9
   properties {
    text=Flagged
    font {
     bForceMonospaced=false
     bMaterialIsStencil=false
     fontMaterial=null
     fontObject=&Roboto
     letterSpacing=0
     monospacedWidth=1
     size=9
     skewAmount=0
     typefaceFontName=Regular
     outlineSettings {
      bApplyOutlineToDropShadows=false
      bMiteredCorners=false
      bSeparateFillAlpha=false
      outlineMaterial=null
      outlineSize=0
      outlineColor=<rgba 0 0 0 1>
     }
    }
   }
   slot {class=ButtonSlot}
  }
  HorizontalBox-Step2 {
   widgetClassPath=&HorizontalBox
   parent=VerticalBox_4
   widget=HorizontalBox-Step2
   slot {class=VerticalBoxSlot, properties {padding=<ltrb 5 15 5 5>}}
  }
  SizeBox_5 {
   widgetClassPath=&SizeBox
   parent=HorizontalBox-Step2
   widget=SizeBox_5
   properties {bOverride_MinDesiredHeight=true, minDesiredHeight=60}
   slot {class=HorizontalBoxSlot, properties {size {sizeRule=Fill, value=1}}}
  }
  B_SortActorsByClass {
   widgetClassPath=&EditorUtilityButton
   parent=SizeBox_5
   widget=B_SortActorsByClass
   properties {
    toolTipText="Separates selected actos by blueprint class and arranges them into grids, for easy visualization"
    widgetStyle {
     clickedSlateSound {resourceObject=null}
     disabled {
      bIsDynamicallyLoaded=false
      drawAs=RoundedBox
      imageType=NoImage
      mirroring=NoMirror
      resourceName=None
      resourceObject=null
      tiling=NoTile
      imageSize=<0 0>
      margin=<ltrb 0 0 0 0>
      outlineSettings {
       bUseBrushTransparency=false
       roundingType=FixedRadius
       width=1
       color {colorUseRule=UseColor_ColorTable, specifiedColor=<rgba 0 0 0 0>}
       cornerRadii=<4 4 4 4>
      }
      tintColor {colorUseRule=UseColor_Specified, specifiedColor=<rgba 0.0395 0.0395 0.0395 1>}
      uVRegion {bIsValid=false, max=<0 0>, min=<0 0>}
     }
     disabledForeground {colorUseRule=UseColor_ColorTable, specifiedColor=<rgba 0 0 0 0>}
     hovered {
      bIsDynamicallyLoaded=false
      drawAs=RoundedBox
      imageType=NoImage
      mirroring=NoMirror
      resourceName=None
      resourceObject=null
      tiling=NoTile
      imageSize=<32 32>
      margin=<ltrb 0 0 0 0>
      outlineSettings {
       bUseBrushTransparency=false
       roundingType=FixedRadius
       width=1
       color {colorUseRule=UseColor_ColorTable, specifiedColor=<rgba 0 0 0 0>}
       cornerRadii=<4 4 4 4>
      }
      tintColor {colorUseRule=UseColor_Specified, specifiedColor=<rgba 0.0953 0.0953 0.0953 1>}
      uVRegion {bIsValid=false, max=<0 0>, min=<0 0>}
     }
     hoveredForeground {colorUseRule=UseColor_ColorTable, specifiedColor=<rgba 0 0 0 0>}
     hoveredSlateSound {resourceObject=null}
     normal {
      bIsDynamicallyLoaded=false
      drawAs=RoundedBox
      imageType=NoImage
      mirroring=NoMirror
      resourceName=None
      resourceObject=null
      tiling=NoTile
      imageSize=<32 32>
      margin=<ltrb 0 0 0 0>
      outlineSettings {
       bUseBrushTransparency=false
       roundingType=FixedRadius
       width=1
       color {colorUseRule=UseColor_ColorTable, specifiedColor=<rgba 0 0 0 0>}
       cornerRadii=<4 4 4 4>
      }
      tintColor {colorUseRule=UseColor_Specified, specifiedColor=<rgba 0.0395 0.0395 0.0395 1>}
      uVRegion {bIsValid=false, max=<0 0>, min=<0 0>}
     }
     normalForeground {colorUseRule=UseColor_ColorTable, specifiedColor=<rgba 0 0 0 0>}
     normalPadding=<ltrb 12 1.5 12 1.5>
     pressed {
      bIsDynamicallyLoaded=false
      drawAs=RoundedBox
      imageType=NoImage
      mirroring=NoMirror
      resourceName=None
      resourceObject=null
      tiling=NoTile
      imageSize=<32 32>
      margin=<ltrb 0 0 0 0>
      outlineSettings {
       bUseBrushTransparency=false
       roundingType=FixedRadius
       width=1
       color {colorUseRule=UseColor_ColorTable, specifiedColor=<rgba 0 0 0 0>}
       cornerRadii=<4 4 4 4>
      }
      tintColor {colorUseRule=UseColor_Specified, specifiedColor=<rgba 0.0284 0.0284 0.0284 1>}
      uVRegion {bIsValid=false, max=<0 0>, min=<0 0>}
     }
     pressedForeground {colorUseRule=UseColor_ColorTable, specifiedColor=<rgba 0 0 0 0>}
     pressedPadding=<ltrb 12 2.5 12 0.5>
     pressedSlateSound {resourceObject=null}
    }
   }
   slot {class=SizeBoxSlot}
  }
  TextBlock_2 {
   widgetClassPath=&TextBlock
   parent=B_SortActorsByClass
   widget=TextBlock_2
   properties {
    text=Organize Actors by Class
    font {
     bForceMonospaced=false
     bMaterialIsStencil=false
     fontMaterial=null
     fontObject=&Roboto
     letterSpacing=0
     monospacedWidth=1
     size=9
     skewAmount=0
     typefaceFontName=Regular
     outlineSettings {
      bApplyOutlineToDropShadows=false
      bMiteredCorners=false
      bSeparateFillAlpha=false
      outlineMaterial=null
      outlineSize=0
      outlineColor=<rgba 0 0 0 1>
     }
    }
   }
   slot {class=ButtonSlot}
  }
  SizeBox_117 {
   widgetClassPath=&SizeBox
   parent=HorizontalBox-Step2
   widget=SizeBox_117
   properties {bOverride_MaxDesiredWidth=true, bOverride_MinDesiredWidth=true, maxDesiredWidth=270, minDesiredWidth=270}
   slot {class=HorizontalBoxSlot, properties {padding=<ltrb 10 0 0 0>}}
  }
  VerticalBox_62 {widgetClassPath=&VerticalBox, parent=SizeBox_117, widget=VerticalBox_62, slot {class=SizeBoxSlot}}
  HorizontalBox_3 {
   widgetClassPath=&HorizontalBox
   parent=VerticalBox_62
   widget=HorizontalBox_3
   slot {class=VerticalBoxSlot, properties {size {sizeRule=Fill, value=1}}}
  }
  TextBlock_3 {
   widgetClassPath=&TextBlock
   parent=HorizontalBox_3
   widget=TextBlock_3
   properties {
    text="Rows per Class : "
    font {
     bForceMonospaced=false
     bMaterialIsStencil=false
     fontMaterial=null
     fontObject=&Roboto
     letterSpacing=0
     monospacedWidth=1
     size=9
     skewAmount=0
     typefaceFontName=Light
     outlineSettings {
      bApplyOutlineToDropShadows=false
      bMiteredCorners=false
      bSeparateFillAlpha=false
      outlineMaterial=null
      outlineSize=0
      outlineColor=<rgba 0 0 0 1>
     }
    }
   }
   slot {class=HorizontalBoxSlot, properties {verticalAlignment=VAlign_Center}}
  }
  Border-textfield {
   widgetClassPath=&Border
   parent=HorizontalBox_3
   widget=Border-textfield
   properties {
    toolTipText=Number of Rows in the X direction to use when arranging actors per class
    verticalAlignment=VAlign_Center
    background {
     bIsDynamicallyLoaded=false
     drawAs=RoundedBox
     imageType=NoImage
     mirroring=NoMirror
     resourceName=None
     resourceObject=null
     tiling=NoTile
     imageSize=<32 32>
     margin=<ltrb 0 0 0 0>
     outlineSettings {
      bUseBrushTransparency=false
      roundingType=FixedRadius
      width=0
      color {colorUseRule=UseColor_Specified, specifiedColor=<rgba 0 0 0 0>}
      cornerRadii=<5 5 5 5>
     }
     tintColor {colorUseRule=UseColor_Specified, specifiedColor=<rgba 0.01 0.01 0.01 1>}
     uVRegion {bIsValid=false, max=<0 0>, min=<0 0>}
    }
   }
   slot {class=HorizontalBoxSlot, properties {verticalAlignment=VAlign_Center}}
  }
  T_SortByClass_Rows {
   widgetClassPath=/Script/Blutility.EditorUtilityEditableText
   bIsVariable=true
   parent=Border-textfield
   widget=T_SortByClass_Rows
   properties {
    text=1
    toolTipText=Number of rows to arrange selected assets on.
    widgetStyle {
     backgroundImageComposing {
      bIsDynamicallyLoaded=false
      drawAs=Border
      imageType=FullColor
      mirroring=NoMirror
      resourceName=../../../Engine/Content/Slate/Old/HyperlinkDotted.png
      resourceObject=null
      tiling=Both
      imageSize=<0 0>
      margin=<ltrb 0 0 0 0.1875>
      outlineSettings {
       bUseBrushTransparency=false
       roundingType=HalfHeightRadius
       width=0
       color {colorUseRule=UseColor_Specified, specifiedColor=<rgba 0 0 0 0>}
       cornerRadii=<0 0 0 0>
      }
      tintColor {colorUseRule=UseColor_Specified, specifiedColor=<rgba 1 1 1 1>}
      uVRegion {bIsValid=false, max=<0 0>, min=<0 0>}
     }
     backgroundImageSelected {
      bIsDynamicallyLoaded=false
      drawAs=Image
      imageType=NoImage
      mirroring=NoMirror
      resourceName=None
      resourceObject=null
      tiling=NoTile
      imageSize=<0 0>
      margin=<ltrb 0 0 0 0>
      outlineSettings {
       bUseBrushTransparency=false
       roundingType=HalfHeightRadius
       width=0
       color {colorUseRule=UseColor_Specified, specifiedColor=<rgba 0 0 0 0>}
       cornerRadii=<0 0 0 0>
      }
      tintColor {colorUseRule=UseColor_Specified, specifiedColor=<rgba 0 0.162 0.7454 1>}
      uVRegion {bIsValid=false, max=<0 0>, min=<0 0>}
     }
     caretImage {
      bIsDynamicallyLoaded=false
      drawAs=Image
      imageType=NoImage
      mirroring=NoMirror
      resourceName=None
      resourceObject=null
      tiling=NoTile
      imageSize=<0 0>
      margin=<ltrb 0 0 0 0>
      outlineSettings {
       bUseBrushTransparency=false
       roundingType=HalfHeightRadius
       width=0
       color {colorUseRule=UseColor_Specified, specifiedColor=<rgba 0 0 0 0>}
       cornerRadii=<0 0 0 0>
      }
      tintColor {colorUseRule=UseColor_Specified, specifiedColor=<rgba 1 1 1 1>}
      uVRegion {bIsValid=false, max=<0 0>, min=<0 0>}
     }
     colorAndOpacity {colorUseRule=UseColor_Foreground, specifiedColor=<rgba 1 0 1 1>}
     font {
      bForceMonospaced=false
      bMaterialIsStencil=false
      fontMaterial=null
      fontObject=&Roboto
      letterSpacing=0
      monospacedWidth=1
      size=10.5
      skewAmount=0
      typefaceFontName=Bold
      outlineSettings {
       bApplyOutlineToDropShadows=false
       bMiteredCorners=false
       bSeparateFillAlpha=false
       outlineMaterial=null
       outlineSize=0
       outlineColor=<rgba 0 0 0 1>
      }
     }
     hintColorAndOpacity {colorUseRule=UseColor_Foreground, specifiedColor=<rgba 1 0 1 1>}
     selectedBackgroundColor {colorUseRule=UseColor_Foreground, specifiedColor=<rgba 1 0 1 1>}
    }
   }
   slot {class=BorderSlot, properties {verticalAlignment=VAlign_Center}}
  }
  HorizontalBox_9 {
   widgetClassPath=&HorizontalBox
   parent=VerticalBox_62
   widget=HorizontalBox_9
   slot {class=VerticalBoxSlot, properties {horizontalAlignment=HAlign_Left, size {sizeRule=Fill, value=1}}}
  }
  TextBlock_10 {
   widgetClassPath=&TextBlock
   parent=HorizontalBox_9
   widget=TextBlock_10
   properties {
    text="Move to : "
    font {
     bForceMonospaced=false
     bMaterialIsStencil=false
     fontMaterial=null
     fontObject=&Roboto
     letterSpacing=0
     monospacedWidth=1
     size=9
     skewAmount=0
     typefaceFontName=Light
     outlineSettings {
      bApplyOutlineToDropShadows=false
      bMiteredCorners=false
      bSeparateFillAlpha=false
      outlineMaterial=null
      outlineSize=0
      outlineColor=<rgba 0 0 0 1>
     }
    }
   }
   slot {class=HorizontalBoxSlot, properties {verticalAlignment=VAlign_Center}}
  }
  Combo_Arrange_OriginMode {
   widgetClassPath=/Script/Blutility.EditorUtilityComboBoxString
   bIsVariable=true
   parent=HorizontalBox_9
   widget=Combo_Arrange_OriginMode
   properties {
    selectedOption=Stored Position
    toolTipText="When arranging actors by class, move them to the following location.\r\n\r\nStored Position - Align the grid origin with the Stored Position.\r\n\r\nMarker Object Position - Align the grid origin with the Marker Object position.\r\n\r\nStored Position - Align the grid origin with 0/0/0\r\n\r\nAbove Playset Floor - Align the grid above the gallery gray floor.\r\n\r\nBesides Playset Floor - Align the grid besides the gallery gray floor."
    defaultOptions [Stored Position, Marker Object Position, World Origin, Above Playset Floor, Besides Playset Floor]
    font {
     bForceMonospaced=false
     bMaterialIsStencil=false
     fontMaterial=null
     fontObject=&Roboto
     letterSpacing=0
     monospacedWidth=1
     size=9
     skewAmount=0
     typefaceFontName=Regular
     outlineSettings {
      bApplyOutlineToDropShadows=false
      bMiteredCorners=false
      bSeparateFillAlpha=false
      outlineMaterial=null
      outlineSize=0
      outlineColor=<rgba 0 0 0 1>
     }
    }
    foregroundColor {colorUseRule=UseColor_Specified, specifiedColor=<rgba 0.5271 0.5271 0.5271 1>}
   }
   slot {class=HorizontalBoxSlot}
  }
  HorizontalBox-Step3_1 {
   widgetClassPath=&HorizontalBox
   parent=VerticalBox_4
   widget=HorizontalBox-Step3_1
   slot {class=VerticalBoxSlot, properties {padding=<ltrb 5 15 5 5>}}
  }
  SizeBox_4 {
   widgetClassPath=&SizeBox
   parent=HorizontalBox-Step3_1
   widget=SizeBox_4
   properties {bOverride_MinDesiredHeight=true, bOverride_MinDesiredWidth=true, minDesiredHeight=60}
   slot {class=HorizontalBoxSlot, properties {size {sizeRule=Fill, value=1}}}
  }
  B_DeleteDisallowedAssets {
   widgetClassPath=&EditorUtilityButton
   parent=SizeBox_4
   widget=B_DeleteDisallowedAssets
   properties {
    toolTipText="Deletes every actor in the \"Disallowed\" folder. \r\n\r\nMake sure to review the contents of this folder before using this function."
    backgroundColor=<rgba 1 0 0.1194 1>
    widgetStyle {
     clickedSlateSound {resourceObject=null}
     disabled {
      bIsDynamicallyLoaded=false
      drawAs=RoundedBox
      imageType=NoImage
      mirroring=NoMirror
      resourceName=None
      resourceObject=null
      tiling=NoTile
      imageSize=<0 0>
      margin=<ltrb 0 0 0 0>
      outlineSettings {
       bUseBrushTransparency=false
       roundingType=FixedRadius
       width=1
       color {colorUseRule=UseColor_ColorTable, specifiedColor=<rgba 0 0 0 0>}
       cornerRadii=<4 4 4 4>
      }
      tintColor {colorUseRule=UseColor_Specified, specifiedColor=<rgba 0.0395 0.0395 0.0395 1>}
      uVRegion {bIsValid=false, max=<0 0>, min=<0 0>}
     }
     disabledForeground {colorUseRule=UseColor_ColorTable, specifiedColor=<rgba 0 0 0 0>}
     hovered {
      bIsDynamicallyLoaded=false
      drawAs=RoundedBox
      imageType=NoImage
      mirroring=NoMirror
      resourceName=None
      resourceObject=null
      tiling=NoTile
      imageSize=<32 32>
      margin=<ltrb 0 0 0 0>
      outlineSettings {
       bUseBrushTransparency=false
       roundingType=FixedRadius
       width=1
       color {colorUseRule=UseColor_ColorTable, specifiedColor=<rgba 0 0 0 0>}
       cornerRadii=<4 4 4 4>
      }
      tintColor {colorUseRule=UseColor_Specified, specifiedColor=<rgba 0.0953 0.0953 0.0953 1>}
      uVRegion {bIsValid=false, max=<0 0>, min=<0 0>}
     }
     hoveredForeground {colorUseRule=UseColor_ColorTable, specifiedColor=<rgba 0 0 0 0>}
     hoveredSlateSound {resourceObject=null}
     normal {
      bIsDynamicallyLoaded=false
      drawAs=RoundedBox
      imageType=NoImage
      mirroring=NoMirror
      resourceName=None
      resourceObject=null
      tiling=NoTile
      imageSize=<32 32>
      margin=<ltrb 0 0 0 0>
      outlineSettings {
       bUseBrushTransparency=false
       roundingType=FixedRadius
       width=1
       color {colorUseRule=UseColor_ColorTable, specifiedColor=<rgba 0 0 0 0>}
       cornerRadii=<4 4 4 4>
      }
      tintColor {colorUseRule=UseColor_Specified, specifiedColor=<rgba 0.0395 0.0395 0.0395 1>}
      uVRegion {bIsValid=false, max=<0 0>, min=<0 0>}
     }
     normalForeground {colorUseRule=UseColor_ColorTable, specifiedColor=<rgba 0 0 0 0>}
     normalPadding=<ltrb 12 1.5 12 1.5>
     pressed {
      bIsDynamicallyLoaded=false
      drawAs=RoundedBox
      imageType=NoImage
      mirroring=NoMirror
      resourceName=None
      resourceObject=null
      tiling=NoTile
      imageSize=<32 32>
      margin=<ltrb 0 0 0 0>
      outlineSettings {
       bUseBrushTransparency=false
       roundingType=FixedRadius
       width=1
       color {colorUseRule=UseColor_ColorTable, specifiedColor=<rgba 0 0 0 0>}
       cornerRadii=<4 4 4 4>
      }
      tintColor {colorUseRule=UseColor_Specified, specifiedColor=<rgba 0.0284 0.0284 0.0284 1>}
      uVRegion {bIsValid=false, max=<0 0>, min=<0 0>}
     }
     pressedForeground {colorUseRule=UseColor_ColorTable, specifiedColor=<rgba 0 0 0 0>}
     pressedPadding=<ltrb 12 2.5 12 0.5>
     pressedSlateSound {resourceObject=null}
    }
   }
   slot {class=SizeBoxSlot}
  }
  EliminatedDuplicates_1 {
   widgetClassPath=&TextBlock
   parent=B_DeleteDisallowedAssets
   widget=EliminatedDuplicates_1
   properties {
    text=Delete Disallowed Actors
    font {
     bForceMonospaced=false
     bMaterialIsStencil=false
     fontMaterial=null
     fontObject=&Roboto
     letterSpacing=0
     monospacedWidth=1
     size=9
     skewAmount=0
     typefaceFontName=Bold
     outlineSettings {
      bApplyOutlineToDropShadows=false
      bMiteredCorners=false
      bSeparateFillAlpha=false
      outlineMaterial=null
      outlineSize=0
      outlineColor=<rgba 0 0 0 1>
     }
    }
   }
   slot {class=ButtonSlot}
  }
  SizeBox_188 {
   widgetClassPath=&SizeBox
   parent=HorizontalBox-Step3_1
   widget=SizeBox_188
   properties {bOverride_MaxDesiredWidth=true, bOverride_MinDesiredWidth=true, maxDesiredWidth=270, minDesiredWidth=270}
   slot {class=HorizontalBoxSlot, properties {padding=<ltrb 10 0 0 0>}}
  }
  HorizontalBox-Step4 {
   widgetClassPath=&HorizontalBox
   parent=VerticalBox_4
   widget=HorizontalBox-Step4
   slot {class=VerticalBoxSlot, properties {padding=<ltrb 5 15 5 5>}}
  }
  SizeBox_6 {
   widgetClassPath=&SizeBox
   parent=HorizontalBox-Step4
   widget=SizeBox_6
   properties {bOverride_MinDesiredHeight=true, minDesiredHeight=60}
   slot {class=HorizontalBoxSlot, properties {size {sizeRule=Fill, value=1}}}
  }
  B_EliminateDuplicates {
   widgetClassPath=&EditorUtilityButton
   parent=SizeBox_6
   widget=B_EliminateDuplicates
   properties {
    toolTipText="Goes over all selected actors to find unique combinations of texture data, then deletes the duplicates.\r\n\r\nMake sure to review the assets before deleting them. Make sure to identify world position actors."
    backgroundColor=<rgba 1 0 0.1194 1>
    widgetStyle {
     clickedSlateSound {resourceObject=null}
     disabled {
      bIsDynamicallyLoaded=false
      drawAs=RoundedBox
      imageType=NoImage
      mirroring=NoMirror
      resourceName=None
      resourceObject=null
      tiling=NoTile
      imageSize=<0 0>
      margin=<ltrb 0 0 0 0>
      outlineSettings {
       bUseBrushTransparency=false
       roundingType=FixedRadius
       width=1
       color {colorUseRule=UseColor_ColorTable, specifiedColor=<rgba 0 0 0 0>}
       cornerRadii=<4 4 4 4>
      }
      tintColor {colorUseRule=UseColor_Specified, specifiedColor=<rgba 0.0395 0.0395 0.0395 1>}
      uVRegion {bIsValid=false, max=<0 0>, min=<0 0>}
     }
     disabledForeground {colorUseRule=UseColor_ColorTable, specifiedColor=<rgba 0 0 0 0>}
     hovered {
      bIsDynamicallyLoaded=false
      drawAs=RoundedBox
      imageType=NoImage
      mirroring=NoMirror
      resourceName=None
      resourceObject=null
      tiling=NoTile
      imageSize=<32 32>
      margin=<ltrb 0 0 0 0>
      outlineSettings {
       bUseBrushTransparency=false
       roundingType=FixedRadius
       width=1
       color {colorUseRule=UseColor_ColorTable, specifiedColor=<rgba 0 0 0 0>}
       cornerRadii=<4 4 4 4>
      }
      tintColor {colorUseRule=UseColor_Specified, specifiedColor=<rgba 0.0953 0.0953 0.0953 1>}
      uVRegion {bIsValid=false, max=<0 0>, min=<0 0>}
     }
     hoveredForeground {colorUseRule=UseColor_ColorTable, specifiedColor=<rgba 0 0 0 0>}
     hoveredSlateSound {resourceObject=null}
     normal {
      bIsDynamicallyLoaded=false
      drawAs=RoundedBox
      imageType=NoImage
      mirroring=NoMirror
      resourceName=None
      resourceObject=null
      tiling=NoTile
      imageSize=<32 32>
      margin=<ltrb 0 0 0 0>
      outlineSettings {
       bUseBrushTransparency=false
       roundingType=FixedRadius
       width=1
       color {colorUseRule=UseColor_ColorTable, specifiedColor=<rgba 0 0 0 0>}
       cornerRadii=<4 4 4 4>
      }
      tintColor {colorUseRule=UseColor_Specified, specifiedColor=<rgba 0.0395 0.0395 0.0395 1>}
      uVRegion {bIsValid=false, max=<0 0>, min=<0 0>}
     }
     normalForeground {colorUseRule=UseColor_ColorTable, specifiedColor=<rgba 0 0 0 0>}
     normalPadding=<ltrb 12 1.5 12 1.5>
     pressed {
      bIsDynamicallyLoaded=false
      drawAs=RoundedBox
      imageType=NoImage
      mirroring=NoMirror
      resourceName=None
      resourceObject=null
      tiling=NoTile
      imageSize=<32 32>
      margin=<ltrb 0 0 0 0>
      outlineSettings {
       bUseBrushTransparency=false
       roundingType=FixedRadius
       width=1
       color {colorUseRule=UseColor_ColorTable, specifiedColor=<rgba 0 0 0 0>}
       cornerRadii=<4 4 4 4>
      }
      tintColor {colorUseRule=UseColor_Specified, specifiedColor=<rgba 0.0284 0.0284 0.0284 1>}
      uVRegion {bIsValid=false, max=<0 0>, min=<0 0>}
     }
     pressedForeground {colorUseRule=UseColor_ColorTable, specifiedColor=<rgba 0 0 0 0>}
     pressedPadding=<ltrb 12 2.5 12 0.5>
     pressedSlateSound {resourceObject=null}
    }
   }
   slot {class=SizeBoxSlot}
  }
  EliminatedDuplicates {
   widgetClassPath=&TextBlock
   parent=B_EliminateDuplicates
   widget=EliminatedDuplicates
   properties {
    text=Delete Duplicates
    font {
     bForceMonospaced=false
     bMaterialIsStencil=false
     fontMaterial=null
     fontObject=&Roboto
     letterSpacing=0
     monospacedWidth=1
     size=9
     skewAmount=0
     typefaceFontName=Bold
     outlineSettings {
      bApplyOutlineToDropShadows=false
      bMiteredCorners=false
      bSeparateFillAlpha=false
      outlineMaterial=null
      outlineSize=0
      outlineColor=<rgba 0 0 0 1>
     }
    }
   }
   slot {class=ButtonSlot}
  }
  SizeBox_1 {
   widgetClassPath=&SizeBox
   parent=HorizontalBox-Step4
   widget=SizeBox_1
   properties {bOverride_MaxDesiredWidth=true, bOverride_MinDesiredWidth=true, maxDesiredWidth=270, minDesiredWidth=270}
   slot {class=HorizontalBoxSlot, properties {padding=<ltrb 10 0 0 0>}}
  }
  HorizontalBox-Step5 {
   widgetClassPath=&HorizontalBox
   parent=VerticalBox_4
   widget=HorizontalBox-Step5
   slot {class=VerticalBoxSlot, properties {padding=<ltrb 5 15 5 5>}}
  }
  SizeBox {
   widgetClassPath=&SizeBox
   parent=HorizontalBox-Step5
   widget=SizeBox
   properties {bOverride_MinDesiredHeight=true, minDesiredHeight=60}
   slot {class=HorizontalBoxSlot, properties {size {sizeRule=Fill, value=1}}}
  }
  B_ResetProperties {
   widgetClassPath=&EditorUtilityButton
   parent=SizeBox
   widget=B_ResetProperties
   properties {
    toolTipText="Reset actor's properties. Currently Scale, Rotation and Mirrored"
    widgetStyle {
     clickedSlateSound {resourceObject=null}
     disabled {
      bIsDynamicallyLoaded=false
      drawAs=RoundedBox
      imageType=NoImage
      mirroring=NoMirror
      resourceName=None
      resourceObject=null
      tiling=NoTile
      imageSize=<0 0>
      margin=<ltrb 0 0 0 0>
      outlineSettings {
       bUseBrushTransparency=false
       roundingType=FixedRadius
       width=1
       color {colorUseRule=UseColor_ColorTable, specifiedColor=<rgba 0 0 0 0>}
       cornerRadii=<4 4 4 4>
      }
      tintColor {colorUseRule=UseColor_Specified, specifiedColor=<rgba 0.0395 0.0395 0.0395 1>}
      uVRegion {bIsValid=false, max=<0 0>, min=<0 0>}
     }
     disabledForeground {colorUseRule=UseColor_ColorTable, specifiedColor=<rgba 0 0 0 0>}
     hovered {
      bIsDynamicallyLoaded=false
      drawAs=RoundedBox
      imageType=NoImage
      mirroring=NoMirror
      resourceName=None
      resourceObject=null
      tiling=NoTile
      imageSize=<32 32>
      margin=<ltrb 0 0 0 0>
      outlineSettings {
       bUseBrushTransparency=false
       roundingType=FixedRadius
       width=1
       color {colorUseRule=UseColor_ColorTable, specifiedColor=<rgba 0 0 0 0>}
       cornerRadii=<4 4 4 4>
      }
      tintColor {colorUseRule=UseColor_Specified, specifiedColor=<rgba 0.0953 0.0953 0.0953 1>}
      uVRegion {bIsValid=false, max=<0 0>, min=<0 0>}
     }
     hoveredForeground {colorUseRule=UseColor_ColorTable, specifiedColor=<rgba 0 0 0 0>}
     hoveredSlateSound {resourceObject=null}
     normal {
      bIsDynamicallyLoaded=false
      drawAs=RoundedBox
      imageType=NoImage
      mirroring=NoMirror
      resourceName=None
      resourceObject=null
      tiling=NoTile
      imageSize=<32 32>
      margin=<ltrb 0 0 0 0>
      outlineSettings {
       bUseBrushTransparency=false
       roundingType=FixedRadius
       width=1
       color {colorUseRule=UseColor_ColorTable, specifiedColor=<rgba 0 0 0 0>}
       cornerRadii=<4 4 4 4>
      }
      tintColor {colorUseRule=UseColor_Specified, specifiedColor=<rgba 0.0395 0.0395 0.0395 1>}
      uVRegion {bIsValid=false, max=<0 0>, min=<0 0>}
     }
     normalForeground {colorUseRule=UseColor_ColorTable, specifiedColor=<rgba 0 0 0 0>}
     normalPadding=<ltrb 12 1.5 12 1.5>
     pressed {
      bIsDynamicallyLoaded=false
      drawAs=RoundedBox
      imageType=NoImage
      mirroring=NoMirror
      resourceName=None
      resourceObject=null
      tiling=NoTile
      imageSize=<32 32>
      margin=<ltrb 0 0 0 0>
      outlineSettings {
       bUseBrushTransparency=false
       roundingType=FixedRadius
       width=1
       color {colorUseRule=UseColor_ColorTable, specifiedColor=<rgba 0 0 0 0>}
       cornerRadii=<4 4 4 4>
      }
      tintColor {colorUseRule=UseColor_Specified, specifiedColor=<rgba 0.0284 0.0284 0.0284 1>}
      uVRegion {bIsValid=false, max=<0 0>, min=<0 0>}
     }
     pressedForeground {colorUseRule=UseColor_ColorTable, specifiedColor=<rgba 0 0 0 0>}
     pressedPadding=<ltrb 12 2.5 12 0.5>
     pressedSlateSound {resourceObject=null}
    }
   }
   slot {class=SizeBoxSlot}
  }
  TextBlock_12 {
   widgetClassPath=&TextBlock
   parent=B_ResetProperties
   widget=TextBlock_12
   properties {
    text=Reset Content Properties
    font {
     bForceMonospaced=false
     bMaterialIsStencil=false
     fontMaterial=null
     fontObject=&Roboto
     letterSpacing=0
     monospacedWidth=1
     size=9
     skewAmount=0
     typefaceFontName=Regular
     outlineSettings {
      bApplyOutlineToDropShadows=false
      bMiteredCorners=false
      bSeparateFillAlpha=false
      outlineMaterial=null
      outlineSize=0
      outlineColor=<rgba 0 0 0 1>
     }
    }
   }
   slot {class=ButtonSlot}
  }
  SizeBox_295 {
   widgetClassPath=&SizeBox
   parent=HorizontalBox-Step5
   widget=SizeBox_295
   properties {bOverride_MaxDesiredWidth=true, bOverride_MinDesiredWidth=true, maxDesiredWidth=270, minDesiredWidth=270}
   slot {class=HorizontalBoxSlot, properties {padding=<ltrb 10 0 0 0>}}
  }
  HorizontalBox_2 {widgetClassPath=&HorizontalBox, parent=SizeBox_295, widget=HorizontalBox_2, slot {class=SizeBoxSlot}}
  CB_Toggle_ResetScale {
   widgetClassPath=&EditorUtilityCheckBox
   bIsVariable=true
   parent=HorizontalBox_2
   widget=CB_Toggle_ResetScale
   properties {
    checkedState=Checked
    toolTipText=Toggle visibility of the Stored Position widget.
    widgetStyle {
     checkBoxType=CheckBox
     backgroundHoveredImage {
      bIsDynamicallyLoaded=false
      drawAs=RoundedBox
      imageType=NoImage
      mirroring=NoMirror
      resourceName=None
      resourceObject=null
      tiling=NoTile
      imageSize=<18 18>
      margin=<ltrb 0 0 0 0>
      outlineSettings {
       bUseBrushTransparency=false
       roundingType=FixedRadius
       width=1
       color {colorUseRule=UseColor_ColorTable, specifiedColor=<rgba 0 0 0 0>}
       cornerRadii=<3 3 3 3>
      }
      tintColor {colorUseRule=UseColor_Specified, specifiedColor=<rgba 0.0048 0.0048 0.0048 1>}
      uVRegion {bIsValid=false, max=<0 0>, min=<0 0>}
     }
     backgroundImage {
      bIsDynamicallyLoaded=false
      drawAs=RoundedBox
      imageType=NoImage
      mirroring=NoMirror
      resourceName=None
      resourceObject=null
      tiling=NoTile
      imageSize=<18 18>
      margin=<ltrb 0 0 0 0>
      outlineSettings {
       bUseBrushTransparency=false
       roundingType=FixedRadius
       width=1
       color {colorUseRule=UseColor_ColorTable, specifiedColor=<rgba 0 0 0 0>}
       cornerRadii=<3 3 3 3>
      }
      tintColor {colorUseRule=UseColor_Specified, specifiedColor=<rgba 0.0048 0.0048 0.0048 1>}
      uVRegion {bIsValid=false, max=<0 0>, min=<0 0>}
     }
     backgroundPressedImage {
      bIsDynamicallyLoaded=false
      drawAs=RoundedBox
      imageType=NoImage
      mirroring=NoMirror
      resourceName=None
      resourceObject=null
      tiling=NoTile
      imageSize=<18 18>
      margin=<ltrb 0 0 0 0>
      outlineSettings {
       bUseBrushTransparency=false
       roundingType=FixedRadius
       width=1
       color {colorUseRule=UseColor_ColorTable, specifiedColor=<rgba 0 0 0 0>}
       cornerRadii=<3 3 3 3>
      }
      tintColor {colorUseRule=UseColor_Specified, specifiedColor=<rgba 0.0048 0.0048 0.0048 1>}
      uVRegion {bIsValid=false, max=<0 0>, min=<0 0>}
     }
     borderBackgroundColor {colorUseRule=UseColor_Specified, specifiedColor=<rgba 1 1 1 1>}
     checkedForeground {colorUseRule=UseColor_Specified, specifiedColor=<rgba 1 1 1 1>}
     checkedHoveredForeground {colorUseRule=UseColor_Specified, specifiedColor=<rgba 1 1 1 1>}
     checkedHoveredImage {
      bIsDynamicallyLoaded=false
      drawAs=Image
      imageType=Vector
      mirroring=NoMirror
      resourceName=../../../Engine/Content/Slate/Starship/CoreWidgets/CheckBox/check.svg
      resourceObject=null
      tiling=NoTile
      imageSize=<16 16>
      margin=<ltrb 0 0 0 0>
      outlineSettings {
       bUseBrushTransparency=false
       roundingType=HalfHeightRadius
       width=0
       color {colorUseRule=UseColor_Specified, specifiedColor=<rgba 0 0 0 0>}
       cornerRadii=<0 0 0 0>
      }
      tintColor {colorUseRule=UseColor_Specified, specifiedColor=<rgba 0.0044 0.2384 1 1>}
      uVRegion {bIsValid=false, max=<0 0>, min=<0 0>}
     }
     checkedImage {
      bIsDynamicallyLoaded=false
      drawAs=Image
      imageType=Vector
      mirroring=NoMirror
      resourceName=../../../Engine/Content/Slate/Starship/CoreWidgets/CheckBox/check.svg
      resourceObject=null
      tiling=NoTile
      imageSize=<16 16>
      margin=<ltrb 0 0 0 0>
      outlineSettings {
       bUseBrushTransparency=false
       roundingType=HalfHeightRadius
       width=0
       color {colorUseRule=UseColor_Specified, specifiedColor=<rgba 0 0 0 0>}
       cornerRadii=<0 0 0 0>
      }
      tintColor {colorUseRule=UseColor_Specified, specifiedColor=<rgba 0 0.162 0.7454 1>}
      uVRegion {bIsValid=false, max=<0 0>, min=<0 0>}
     }
     checkedPressedForeground {colorUseRule=UseColor_Specified, specifiedColor=<rgba 1 1 1 1>}
     checkedPressedImage {
      bIsDynamicallyLoaded=false
      drawAs=Image
      imageType=Vector
      mirroring=NoMirror
      resourceName=../../../Engine/Content/Slate/Starship/CoreWidgets/CheckBox/check.svg
      resourceObject=null
      tiling=NoTile
      imageSize=<16 16>
      margin=<ltrb 0 0 0 0>
      outlineSettings {
       bUseBrushTransparency=false
       roundingType=HalfHeightRadius
       width=0
       color {colorUseRule=UseColor_Specified, specifiedColor=<rgba 0 0 0 0>}
       cornerRadii=<0 0 0 0>
      }
      tintColor {colorUseRule=UseColor_Specified, specifiedColor=<rgba 0 0.0802 0.3515 1>}
      uVRegion {bIsValid=false, max=<0 0>, min=<0 0>}
     }
     checkedSlateSound {resourceObject=null}
     foregroundColor {colorUseRule=UseColor_Specified, specifiedColor=<rgba 1 1 1 1>}
     hoveredForeground {colorUseRule=UseColor_Specified, specifiedColor=<rgba 1 1 1 1>}
     hoveredSlateSound {resourceObject=null}
     padding=<ltrb 2 0 0 0>
     pressedForeground {colorUseRule=UseColor_Specified, specifiedColor=<rgba 1 1 1 1>}
     uncheckedHoveredImage {
      bIsDynamicallyLoaded=false
      drawAs=NoDrawType
      imageType=NoImage
      mirroring=NoMirror
      resourceName=None
      resourceObject=null
      tiling=NoTile
      imageSize=<0 0>
      margin=<ltrb 0 0 0 0>
      outlineSettings {
       bUseBrushTransparency=false
       roundingType=HalfHeightRadius
       width=0
       color {colorUseRule=UseColor_Specified, specifiedColor=<rgba 0 0 0 0>}
       cornerRadii=<0 0 0 0>
      }
      tintColor {colorUseRule=UseColor_Specified, specifiedColor=<rgba 1 1 1 1>}
      uVRegion {bIsValid=false, max=<0 0>, min=<0 0>}
     }
     uncheckedImage {
      bIsDynamicallyLoaded=false
      drawAs=NoDrawType
      imageType=NoImage
      mirroring=NoMirror
      resourceName=None
      resourceObject=null
      tiling=NoTile
      imageSize=<0 0>
      margin=<ltrb 0 0 0 0>
      outlineSettings {
       bUseBrushTransparency=false
       roundingType=HalfHeightRadius
       width=0
       color {colorUseRule=UseColor_Specified, specifiedColor=<rgba 0 0 0 0>}
       cornerRadii=<0 0 0 0>
      }
      tintColor {colorUseRule=UseColor_Specified, specifiedColor=<rgba 1 1 1 1>}
      uVRegion {bIsValid=false, max=<0 0>, min=<0 0>}
     }
     uncheckedPressedImage {
      bIsDynamicallyLoaded=false
      drawAs=NoDrawType
      imageType=NoImage
      mirroring=NoMirror
      resourceName=None
      resourceObject=null
      tiling=NoTile
      imageSize=<0 0>
      margin=<ltrb 0 0 0 0>
      outlineSettings {
       bUseBrushTransparency=false
       roundingType=HalfHeightRadius
       width=0
       color {colorUseRule=UseColor_Specified, specifiedColor=<rgba 0 0 0 0>}
       cornerRadii=<0 0 0 0>
      }
      tintColor {colorUseRule=UseColor_Specified, specifiedColor=<rgba 1 1 1 1>}
      uVRegion {bIsValid=false, max=<0 0>, min=<0 0>}
     }
     uncheckedSlateSound {resourceObject=null}
     undeterminedForeground {colorUseRule=UseColor_Specified, specifiedColor=<rgba 1 1 1 1>}
     undeterminedHoveredImage {
      bIsDynamicallyLoaded=false
      drawAs=Image
      imageType=Vector
      mirroring=NoMirror
      resourceName=../../../Engine/Content/Slate/Starship/CoreWidgets/CheckBox/indeterminate.svg
      resourceObject=null
      tiling=NoTile
      imageSize=<16 16>
      margin=<ltrb 0 0 0 0>
      outlineSettings {
       bUseBrushTransparency=false
       roundingType=HalfHeightRadius
       width=0
       color {colorUseRule=UseColor_Specified, specifiedColor=<rgba 0 0 0 0>}
       cornerRadii=<0 0 0 0>
      }
      tintColor {colorUseRule=UseColor_Specified, specifiedColor=<rgba 0.0044 0.2384 1 1>}
      uVRegion {bIsValid=false, max=<0 0>, min=<0 0>}
     }
     undeterminedImage {
      bIsDynamicallyLoaded=false
      drawAs=Image
      imageType=Vector
      mirroring=NoMirror
      resourceName=../../../Engine/Content/Slate/Starship/CoreWidgets/CheckBox/indeterminate.svg
      resourceObject=null
      tiling=NoTile
      imageSize=<16 16>
      margin=<ltrb 0 0 0 0>
      outlineSettings {
       bUseBrushTransparency=false
       roundingType=HalfHeightRadius
       width=0
       color {colorUseRule=UseColor_Specified, specifiedColor=<rgba 0 0 0 0>}
       cornerRadii=<0 0 0 0>
      }
      tintColor {colorUseRule=UseColor_Specified, specifiedColor=<rgba 0 0.162 0.7454 1>}
      uVRegion {bIsValid=false, max=<0 0>, min=<0 0>}
     }
     undeterminedPressedImage {
      bIsDynamicallyLoaded=false
      drawAs=Image
      imageType=Vector
      mirroring=NoMirror
      resourceName=../../../Engine/Content/Slate/Starship/CoreWidgets/CheckBox/indeterminate.svg
      resourceObject=null
      tiling=NoTile
      imageSize=<16 16>
      margin=<ltrb 0 0 0 0>
      outlineSettings {
       bUseBrushTransparency=false
       roundingType=HalfHeightRadius
       width=0
       color {colorUseRule=UseColor_Specified, specifiedColor=<rgba 0 0 0 0>}
       cornerRadii=<0 0 0 0>
      }
      tintColor {colorUseRule=UseColor_Specified, specifiedColor=<rgba 0 0.0802 0.3515 1>}
      uVRegion {bIsValid=false, max=<0 0>, min=<0 0>}
     }
    }
   }
   slot {class=HorizontalBoxSlot, properties {horizontalAlignment=HAlign_Center, size {sizeRule=Fill, value=1}}}
  }
  T_TogglePositionVisibility {
   widgetClassPath=&TextBlock
   parent=CB_Toggle_ResetScale
   widget=T_TogglePositionVisibility
   properties {
    justification=Center
    text=Scale
    font {
     bForceMonospaced=false
     bMaterialIsStencil=false
     fontMaterial=null
     fontObject=&Roboto
     letterSpacing=0
     monospacedWidth=1
     size=9
     skewAmount=0
     typefaceFontName=Regular
     outlineSettings {
      bApplyOutlineToDropShadows=false
      bMiteredCorners=false
      bSeparateFillAlpha=false
      outlineMaterial=null
      outlineSize=0
      outlineColor=<rgba 0 0 0 1>
     }
    }
   }
   slot {class=PanelSlot}
  }
  CB_Toggle_ResetRotation {
   widgetClassPath=&EditorUtilityCheckBox
   bIsVariable=true
   parent=HorizontalBox_2
   widget=CB_Toggle_ResetRotation
   properties {
    checkedState=Checked
    toolTipText=Toggle visibility of the Stored Position widget.
    widgetStyle {
     checkBoxType=CheckBox
     backgroundHoveredImage {
      bIsDynamicallyLoaded=false
      drawAs=RoundedBox
      imageType=NoImage
      mirroring=NoMirror
      resourceName=None
      resourceObject=null
      tiling=NoTile
      imageSize=<18 18>
      margin=<ltrb 0 0 0 0>
      outlineSettings {
       bUseBrushTransparency=false
       roundingType=FixedRadius
       width=1
       color {colorUseRule=UseColor_ColorTable, specifiedColor=<rgba 0 0 0 0>}
       cornerRadii=<3 3 3 3>
      }
      tintColor {colorUseRule=UseColor_Specified, specifiedColor=<rgba 0.0048 0.0048 0.0048 1>}
      uVRegion {bIsValid=false, max=<0 0>, min=<0 0>}
     }
     backgroundImage {
      bIsDynamicallyLoaded=false
      drawAs=RoundedBox
      imageType=NoImage
      mirroring=NoMirror
      resourceName=None
      resourceObject=null
      tiling=NoTile
      imageSize=<18 18>
      margin=<ltrb 0 0 0 0>
      outlineSettings {
       bUseBrushTransparency=false
       roundingType=FixedRadius
       width=1
       color {colorUseRule=UseColor_ColorTable, specifiedColor=<rgba 0 0 0 0>}
       cornerRadii=<3 3 3 3>
      }
      tintColor {colorUseRule=UseColor_Specified, specifiedColor=<rgba 0.0048 0.0048 0.0048 1>}
      uVRegion {bIsValid=false, max=<0 0>, min=<0 0>}
     }
     backgroundPressedImage {
      bIsDynamicallyLoaded=false
      drawAs=RoundedBox
      imageType=NoImage
      mirroring=NoMirror
      resourceName=None
      resourceObject=null
      tiling=NoTile
      imageSize=<18 18>
      margin=<ltrb 0 0 0 0>
      outlineSettings {
       bUseBrushTransparency=false
       roundingType=FixedRadius
       width=1
       color {colorUseRule=UseColor_ColorTable, specifiedColor=<rgba 0 0 0 0>}
       cornerRadii=<3 3 3 3>
      }
      tintColor {colorUseRule=UseColor_Specified, specifiedColor=<rgba 0.0048 0.0048 0.0048 1>}
      uVRegion {bIsValid=false, max=<0 0>, min=<0 0>}
     }
     borderBackgroundColor {colorUseRule=UseColor_Specified, specifiedColor=<rgba 1 1 1 1>}
     checkedForeground {colorUseRule=UseColor_Specified, specifiedColor=<rgba 1 1 1 1>}
     checkedHoveredForeground {colorUseRule=UseColor_Specified, specifiedColor=<rgba 1 1 1 1>}
     checkedHoveredImage {
      bIsDynamicallyLoaded=false
      drawAs=Image
      imageType=Vector
      mirroring=NoMirror
      resourceName=../../../Engine/Content/Slate/Starship/CoreWidgets/CheckBox/check.svg
      resourceObject=null
      tiling=NoTile
      imageSize=<16 16>
      margin=<ltrb 0 0 0 0>
      outlineSettings {
       bUseBrushTransparency=false
       roundingType=HalfHeightRadius
       width=0
       color {colorUseRule=UseColor_Specified, specifiedColor=<rgba 0 0 0 0>}
       cornerRadii=<0 0 0 0>
      }
      tintColor {colorUseRule=UseColor_Specified, specifiedColor=<rgba 0.0044 0.2384 1 1>}
      uVRegion {bIsValid=false, max=<0 0>, min=<0 0>}
     }
     checkedImage {
      bIsDynamicallyLoaded=false
      drawAs=Image
      imageType=Vector
      mirroring=NoMirror
      resourceName=../../../Engine/Content/Slate/Starship/CoreWidgets/CheckBox/check.svg
      resourceObject=null
      tiling=NoTile
      imageSize=<16 16>
      margin=<ltrb 0 0 0 0>
      outlineSettings {
       bUseBrushTransparency=false
       roundingType=HalfHeightRadius
       width=0
       color {colorUseRule=UseColor_Specified, specifiedColor=<rgba 0 0 0 0>}
       cornerRadii=<0 0 0 0>
      }
      tintColor {colorUseRule=UseColor_Specified, specifiedColor=<rgba 0 0.162 0.7454 1>}
      uVRegion {bIsValid=false, max=<0 0>, min=<0 0>}
     }
     checkedPressedForeground {colorUseRule=UseColor_Specified, specifiedColor=<rgba 1 1 1 1>}
     checkedPressedImage {
      bIsDynamicallyLoaded=false
      drawAs=Image
      imageType=Vector
      mirroring=NoMirror
      resourceName=../../../Engine/Content/Slate/Starship/CoreWidgets/CheckBox/check.svg
      resourceObject=null
      tiling=NoTile
      imageSize=<16 16>
      margin=<ltrb 0 0 0 0>
      outlineSettings {
       bUseBrushTransparency=false
       roundingType=HalfHeightRadius
       width=0
       color {colorUseRule=UseColor_Specified, specifiedColor=<rgba 0 0 0 0>}
       cornerRadii=<0 0 0 0>
      }
      tintColor {colorUseRule=UseColor_Specified, specifiedColor=<rgba 0 0.0802 0.3515 1>}
      uVRegion {bIsValid=false, max=<0 0>, min=<0 0>}
     }
     checkedSlateSound {resourceObject=null}
     foregroundColor {colorUseRule=UseColor_Specified, specifiedColor=<rgba 1 1 1 1>}
     hoveredForeground {colorUseRule=UseColor_Specified, specifiedColor=<rgba 1 1 1 1>}
     hoveredSlateSound {resourceObject=null}
     padding=<ltrb 2 0 0 0>
     pressedForeground {colorUseRule=UseColor_Specified, specifiedColor=<rgba 1 1 1 1>}
     uncheckedHoveredImage {
      bIsDynamicallyLoaded=false
      drawAs=NoDrawType
      imageType=NoImage
      mirroring=NoMirror
      resourceName=None
      resourceObject=null
      tiling=NoTile
      imageSize=<0 0>
      margin=<ltrb 0 0 0 0>
      outlineSettings {
       bUseBrushTransparency=false
       roundingType=HalfHeightRadius
       width=0
       color {colorUseRule=UseColor_Specified, specifiedColor=<rgba 0 0 0 0>}
       cornerRadii=<0 0 0 0>
      }
      tintColor {colorUseRule=UseColor_Specified, specifiedColor=<rgba 1 1 1 1>}
      uVRegion {bIsValid=false, max=<0 0>, min=<0 0>}
     }
     uncheckedImage {
      bIsDynamicallyLoaded=false
      drawAs=NoDrawType
      imageType=NoImage
      mirroring=NoMirror
      resourceName=None
      resourceObject=null
      tiling=NoTile
      imageSize=<0 0>
      margin=<ltrb 0 0 0 0>
      outlineSettings {
       bUseBrushTransparency=false
       roundingType=HalfHeightRadius
       width=0
       color {colorUseRule=UseColor_Specified, specifiedColor=<rgba 0 0 0 0>}
       cornerRadii=<0 0 0 0>
      }
      tintColor {colorUseRule=UseColor_Specified, specifiedColor=<rgba 1 1 1 1>}
      uVRegion {bIsValid=false, max=<0 0>, min=<0 0>}
     }
     uncheckedPressedImage {
      bIsDynamicallyLoaded=false
      drawAs=NoDrawType
      imageType=NoImage
      mirroring=NoMirror
      resourceName=None
      resourceObject=null
      tiling=NoTile
      imageSize=<0 0>
      margin=<ltrb 0 0 0 0>
      outlineSettings {
       bUseBrushTransparency=false
       roundingType=HalfHeightRadius
       width=0
       color {colorUseRule=UseColor_Specified, specifiedColor=<rgba 0 0 0 0>}
       cornerRadii=<0 0 0 0>
      }
      tintColor {colorUseRule=UseColor_Specified, specifiedColor=<rgba 1 1 1 1>}
      uVRegion {bIsValid=false, max=<0 0>, min=<0 0>}
     }
     uncheckedSlateSound {resourceObject=null}
     undeterminedForeground {colorUseRule=UseColor_Specified, specifiedColor=<rgba 1 1 1 1>}
     undeterminedHoveredImage {
      bIsDynamicallyLoaded=false
      drawAs=Image
      imageType=Vector
      mirroring=NoMirror
      resourceName=../../../Engine/Content/Slate/Starship/CoreWidgets/CheckBox/indeterminate.svg
      resourceObject=null
      tiling=NoTile
      imageSize=<16 16>
      margin=<ltrb 0 0 0 0>
      outlineSettings {
       bUseBrushTransparency=false
       roundingType=HalfHeightRadius
       width=0
       color {colorUseRule=UseColor_Specified, specifiedColor=<rgba 0 0 0 0>}
       cornerRadii=<0 0 0 0>
      }
      tintColor {colorUseRule=UseColor_Specified, specifiedColor=<rgba 0.0044 0.2384 1 1>}
      uVRegion {bIsValid=false, max=<0 0>, min=<0 0>}
     }
     undeterminedImage {
      bIsDynamicallyLoaded=false
      drawAs=Image
      imageType=Vector
      mirroring=NoMirror
      resourceName=../../../Engine/Content/Slate/Starship/CoreWidgets/CheckBox/indeterminate.svg
      resourceObject=null
      tiling=NoTile
      imageSize=<16 16>
      margin=<ltrb 0 0 0 0>
      outlineSettings {
       bUseBrushTransparency=false
       roundingType=HalfHeightRadius
       width=0
       color {colorUseRule=UseColor_Specified, specifiedColor=<rgba 0 0 0 0>}
       cornerRadii=<0 0 0 0>
      }
      tintColor {colorUseRule=UseColor_Specified, specifiedColor=<rgba 0 0.162 0.7454 1>}
      uVRegion {bIsValid=false, max=<0 0>, min=<0 0>}
     }
     undeterminedPressedImage {
      bIsDynamicallyLoaded=false
      drawAs=Image
      imageType=Vector
      mirroring=NoMirror
      resourceName=../../../Engine/Content/Slate/Starship/CoreWidgets/CheckBox/indeterminate.svg
      resourceObject=null
      tiling=NoTile
      imageSize=<16 16>
      margin=<ltrb 0 0 0 0>
      outlineSettings {
       bUseBrushTransparency=false
       roundingType=HalfHeightRadius
       width=0
       color {colorUseRule=UseColor_Specified, specifiedColor=<rgba 0 0 0 0>}
       cornerRadii=<0 0 0 0>
      }
      tintColor {colorUseRule=UseColor_Specified, specifiedColor=<rgba 0 0.0802 0.3515 1>}
      uVRegion {bIsValid=false, max=<0 0>, min=<0 0>}
     }
    }
   }
   slot {class=HorizontalBoxSlot, properties {horizontalAlignment=HAlign_Center, size {sizeRule=Fill, value=1}}}
  }
  T_TogglePositionVisibility_1 {
   widgetClassPath=&TextBlock
   parent=CB_Toggle_ResetRotation
   widget=T_TogglePositionVisibility_1
   properties {
    justification=Center
    text=Rotation
    font {
     bForceMonospaced=false
     bMaterialIsStencil=false
     fontMaterial=null
     fontObject=&Roboto
     letterSpacing=0
     monospacedWidth=1
     size=9
     skewAmount=0
     typefaceFontName=Regular
     outlineSettings {
      bApplyOutlineToDropShadows=false
      bMiteredCorners=false
      bSeparateFillAlpha=false
      outlineMaterial=null
      outlineSize=0
      outlineColor=<rgba 0 0 0 1>
     }
    }
   }
   slot {class=PanelSlot}
  }
  CB_Toggle_ResetMirror {
   widgetClassPath=&EditorUtilityCheckBox
   bIsVariable=true
   parent=HorizontalBox_2
   widget=CB_Toggle_ResetMirror
   properties {
    checkedState=Checked
    toolTipText=Toggle visibility of the Stored Position widget.
    widgetStyle {
     checkBoxType=CheckBox
     backgroundHoveredImage {
      bIsDynamicallyLoaded=false
      drawAs=RoundedBox
      imageType=NoImage
      mirroring=NoMirror
      resourceName=None
      resourceObject=null
      tiling=NoTile
      imageSize=<18 18>
      margin=<ltrb 0 0 0 0>
      outlineSettings {
       bUseBrushTransparency=false
       roundingType=FixedRadius
       width=1
       color {colorUseRule=UseColor_ColorTable, specifiedColor=<rgba 0 0 0 0>}
       cornerRadii=<3 3 3 3>
      }
      tintColor {colorUseRule=UseColor_Specified, specifiedColor=<rgba 0.0048 0.0048 0.0048 1>}
      uVRegion {bIsValid=false, max=<0 0>, min=<0 0>}
     }
     backgroundImage {
      bIsDynamicallyLoaded=false
      drawAs=RoundedBox
      imageType=NoImage
      mirroring=NoMirror
      resourceName=None
      resourceObject=null
      tiling=NoTile
      imageSize=<18 18>
      margin=<ltrb 0 0 0 0>
      outlineSettings {
       bUseBrushTransparency=false
       roundingType=FixedRadius
       width=1
       color {colorUseRule=UseColor_ColorTable, specifiedColor=<rgba 0 0 0 0>}
       cornerRadii=<3 3 3 3>
      }
      tintColor {colorUseRule=UseColor_Specified, specifiedColor=<rgba 0.0048 0.0048 0.0048 1>}
      uVRegion {bIsValid=false, max=<0 0>, min=<0 0>}
     }
     backgroundPressedImage {
      bIsDynamicallyLoaded=false
      drawAs=RoundedBox
      imageType=NoImage
      mirroring=NoMirror
      resourceName=None
      resourceObject=null
      tiling=NoTile
      imageSize=<18 18>
      margin=<ltrb 0 0 0 0>
      outlineSettings {
       bUseBrushTransparency=false
       roundingType=FixedRadius
       width=1
       color {colorUseRule=UseColor_ColorTable, specifiedColor=<rgba 0 0 0 0>}
       cornerRadii=<3 3 3 3>
      }
      tintColor {colorUseRule=UseColor_Specified, specifiedColor=<rgba 0.0048 0.0048 0.0048 1>}
      uVRegion {bIsValid=false, max=<0 0>, min=<0 0>}
     }
     borderBackgroundColor {colorUseRule=UseColor_Specified, specifiedColor=<rgba 1 1 1 1>}
     checkedForeground {colorUseRule=UseColor_Specified, specifiedColor=<rgba 1 1 1 1>}
     checkedHoveredForeground {colorUseRule=UseColor_Specified, specifiedColor=<rgba 1 1 1 1>}
     checkedHoveredImage {
      bIsDynamicallyLoaded=false
      drawAs=Image
      imageType=Vector
      mirroring=NoMirror
      resourceName=../../../Engine/Content/Slate/Starship/CoreWidgets/CheckBox/check.svg
      resourceObject=null
      tiling=NoTile
      imageSize=<16 16>
      margin=<ltrb 0 0 0 0>
      outlineSettings {
       bUseBrushTransparency=false
       roundingType=HalfHeightRadius
       width=0
       color {colorUseRule=UseColor_Specified, specifiedColor=<rgba 0 0 0 0>}
       cornerRadii=<0 0 0 0>
      }
      tintColor {colorUseRule=UseColor_Specified, specifiedColor=<rgba 0.0044 0.2384 1 1>}
      uVRegion {bIsValid=false, max=<0 0>, min=<0 0>}
     }
     checkedImage {
      bIsDynamicallyLoaded=false
      drawAs=Image
      imageType=Vector
      mirroring=NoMirror
      resourceName=../../../Engine/Content/Slate/Starship/CoreWidgets/CheckBox/check.svg
      resourceObject=null
      tiling=NoTile
      imageSize=<16 16>
      margin=<ltrb 0 0 0 0>
      outlineSettings {
       bUseBrushTransparency=false
       roundingType=HalfHeightRadius
       width=0
       color {colorUseRule=UseColor_Specified, specifiedColor=<rgba 0 0 0 0>}
       cornerRadii=<0 0 0 0>
      }
      tintColor {colorUseRule=UseColor_Specified, specifiedColor=<rgba 0 0.162 0.7454 1>}
      uVRegion {bIsValid=false, max=<0 0>, min=<0 0>}
     }
     checkedPressedForeground {colorUseRule=UseColor_Specified, specifiedColor=<rgba 1 1 1 1>}
     checkedPressedImage {
      bIsDynamicallyLoaded=false
      drawAs=Image
      imageType=Vector
      mirroring=NoMirror
      resourceName=../../../Engine/Content/Slate/Starship/CoreWidgets/CheckBox/check.svg
      resourceObject=null
      tiling=NoTile
      imageSize=<16 16>
      margin=<ltrb 0 0 0 0>
      outlineSettings {
       bUseBrushTransparency=false
       roundingType=HalfHeightRadius
       width=0
       color {colorUseRule=UseColor_Specified, specifiedColor=<rgba 0 0 0 0>}
       cornerRadii=<0 0 0 0>
      }
      tintColor {colorUseRule=UseColor_Specified, specifiedColor=<rgba 0 0.0802 0.3515 1>}
      uVRegion {bIsValid=false, max=<0 0>, min=<0 0>}
     }
     checkedSlateSound {resourceObject=null}
     foregroundColor {colorUseRule=UseColor_Specified, specifiedColor=<rgba 1 1 1 1>}
     hoveredForeground {colorUseRule=UseColor_Specified, specifiedColor=<rgba 1 1 1 1>}
     hoveredSlateSound {resourceObject=null}
     padding=<ltrb 2 0 0 0>
     pressedForeground {colorUseRule=UseColor_Specified, specifiedColor=<rgba 1 1 1 1>}
     uncheckedHoveredImage {
      bIsDynamicallyLoaded=false
      drawAs=NoDrawType
      imageType=NoImage
      mirroring=NoMirror
      resourceName=None
      resourceObject=null
      tiling=NoTile
      imageSize=<0 0>
      margin=<ltrb 0 0 0 0>
      outlineSettings {
       bUseBrushTransparency=false
       roundingType=HalfHeightRadius
       width=0
       color {colorUseRule=UseColor_Specified, specifiedColor=<rgba 0 0 0 0>}
       cornerRadii=<0 0 0 0>
      }
      tintColor {colorUseRule=UseColor_Specified, specifiedColor=<rgba 1 1 1 1>}
      uVRegion {bIsValid=false, max=<0 0>, min=<0 0>}
     }
     uncheckedImage {
      bIsDynamicallyLoaded=false
      drawAs=NoDrawType
      imageType=NoImage
      mirroring=NoMirror
      resourceName=None
      resourceObject=null
      tiling=NoTile
      imageSize=<0 0>
      margin=<ltrb 0 0 0 0>
      outlineSettings {
       bUseBrushTransparency=false
       roundingType=HalfHeightRadius
       width=0
       color {colorUseRule=UseColor_Specified, specifiedColor=<rgba 0 0 0 0>}
       cornerRadii=<0 0 0 0>
      }
      tintColor {colorUseRule=UseColor_Specified, specifiedColor=<rgba 1 1 1 1>}
      uVRegion {bIsValid=false, max=<0 0>, min=<0 0>}
     }
     uncheckedPressedImage {
      bIsDynamicallyLoaded=false
      drawAs=NoDrawType
      imageType=NoImage
      mirroring=NoMirror
      resourceName=None
      resourceObject=null
      tiling=NoTile
      imageSize=<0 0>
      margin=<ltrb 0 0 0 0>
      outlineSettings {
       bUseBrushTransparency=false
       roundingType=HalfHeightRadius
       width=0
       color {colorUseRule=UseColor_Specified, specifiedColor=<rgba 0 0 0 0>}
       cornerRadii=<0 0 0 0>
      }
      tintColor {colorUseRule=UseColor_Specified, specifiedColor=<rgba 1 1 1 1>}
      uVRegion {bIsValid=false, max=<0 0>, min=<0 0>}
     }
     uncheckedSlateSound {resourceObject=null}
     undeterminedForeground {colorUseRule=UseColor_Specified, specifiedColor=<rgba 1 1 1 1>}
     undeterminedHoveredImage {
      bIsDynamicallyLoaded=false
      drawAs=Image
      imageType=Vector
      mirroring=NoMirror
      resourceName=../../../Engine/Content/Slate/Starship/CoreWidgets/CheckBox/indeterminate.svg
      resourceObject=null
      tiling=NoTile
      imageSize=<16 16>
      margin=<ltrb 0 0 0 0>
      outlineSettings {
       bUseBrushTransparency=false
       roundingType=HalfHeightRadius
       width=0
       color {colorUseRule=UseColor_Specified, specifiedColor=<rgba 0 0 0 0>}
       cornerRadii=<0 0 0 0>
      }
      tintColor {colorUseRule=UseColor_Specified, specifiedColor=<rgba 0.0044 0.2384 1 1>}
      uVRegion {bIsValid=false, max=<0 0>, min=<0 0>}
     }
     undeterminedImage {
      bIsDynamicallyLoaded=false
      drawAs=Image
      imageType=Vector
      mirroring=NoMirror
      resourceName=../../../Engine/Content/Slate/Starship/CoreWidgets/CheckBox/indeterminate.svg
      resourceObject=null
      tiling=NoTile
      imageSize=<16 16>
      margin=<ltrb 0 0 0 0>
      outlineSettings {
       bUseBrushTransparency=false
       roundingType=HalfHeightRadius
       width=0
       color {colorUseRule=UseColor_Specified, specifiedColor=<rgba 0 0 0 0>}
       cornerRadii=<0 0 0 0>
      }
      tintColor {colorUseRule=UseColor_Specified, specifiedColor=<rgba 0 0.162 0.7454 1>}
      uVRegion {bIsValid=false, max=<0 0>, min=<0 0>}
     }
     undeterminedPressedImage {
      bIsDynamicallyLoaded=false
      drawAs=Image
      imageType=Vector
      mirroring=NoMirror
      resourceName=../../../Engine/Content/Slate/Starship/CoreWidgets/CheckBox/indeterminate.svg
      resourceObject=null
      tiling=NoTile
      imageSize=<16 16>
      margin=<ltrb 0 0 0 0>
      outlineSettings {
       bUseBrushTransparency=false
       roundingType=HalfHeightRadius
       width=0
       color {colorUseRule=UseColor_Specified, specifiedColor=<rgba 0 0 0 0>}
       cornerRadii=<0 0 0 0>
      }
      tintColor {colorUseRule=UseColor_Specified, specifiedColor=<rgba 0 0.0802 0.3515 1>}
      uVRegion {bIsValid=false, max=<0 0>, min=<0 0>}
     }
    }
   }
   slot {class=HorizontalBoxSlot, properties {horizontalAlignment=HAlign_Center, size {sizeRule=Fill, value=1}}}
  }
  T_TogglePositionVisibility_2 {
   widgetClassPath=&TextBlock
   parent=CB_Toggle_ResetMirror
   widget=T_TogglePositionVisibility_2
   properties {
    justification=Center
    text=Mirrored
    font {
     bForceMonospaced=false
     bMaterialIsStencil=false
     fontMaterial=null
     fontObject=&Roboto
     letterSpacing=0
     monospacedWidth=1
     size=9
     skewAmount=0
     typefaceFontName=Regular
     outlineSettings {
      bApplyOutlineToDropShadows=false
      bMiteredCorners=false
      bSeparateFillAlpha=false
      outlineMaterial=null
      outlineSize=0
      outlineColor=<rgba 0 0 0 1>
     }
    }
   }
   slot {class=PanelSlot}
  }
 ]
}
