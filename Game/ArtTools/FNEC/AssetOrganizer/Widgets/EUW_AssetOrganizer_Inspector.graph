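# Editor Utility Widget that inspects the selected actor for the FNEC Asset Organizer, showing whether it belongs in a creative gallery.
# Every tick it polls the selection, fills in name, class and parent class, and warns if the OwnerWidget reference is invalid.
# The owner tool's data asset lookup classifies the parent class, driving color-coded explanations for disallowed, gallery, content or flagged actors.
# Marker-tagged actors and unlisted classes get their own warnings.
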EditorUtilityWidgetBlueprint /Game/ArtTools/FNEC/AssetOrganizer/Widgets/EUW_AssetOrganizer_Inspector
graphs [
 EventGraph {nodes [
   Event Tick()
   > GetActorInfo()
   n1 = Get OwnerWidget() -> (OwnerWidget:EUW Asset Organizer Tools Object Reference)
   > Set Text(
    Text=SelectText(B="LOCTEXT(\"ERROR : Invalid Widget Reference. Restart Inspector Widget.\")", bPickA=IsValid(Object=@n1))
    self=Get T_Console() -> (T_Console:Text Object Reference)
   )
  ]}
 GetActorInfo {nodes [
   GetActorInfo() {is=title K2Node_FunctionEntry}
   > n1 = GetSelectionSet!(execute) -> (then, ReturnValue:Array of Actor Object References)
   n2 = "Get (a copy)"(Array:Array of Actor Object References=@n1) -> (Output:Actor Object Reference)
   > n3 = Is Valid(InputObject=@n2)
   Set Text(execute=@n3.Is Not Valid, Text="LOCTEXT(\"-\")", self=Get T_SelectedObject() -> (T_SelectedObject:Text Object Reference))
   > Set Text(Text="LOCTEXT(\"-\")", self=Get T_SelectedClass() -> (T_SelectedClass:Text Object Reference))
   > Set Text(Text="LOCTEXT(\"-\")", self=Get T_SelectedParentClass() -> (T_SelectedParentClass:Text Object Reference))
   > Set Text(self=Get T_SelectedClasification() -> (T_SelectedClasification:Text Object Reference))
   > Set Text(self=Get T_Reason() -> (T_Reason:Text Object Reference))
   n4 = Set CurrentActor(execute=@n3.Is Valid, CurrentActor:Actor Object Reference=@n2) -> (Output_Get:Actor Object Reference)
   > Set CurrentClass(CurrentClass:Actor Class Reference=Get Class(Object=@n4)) -> (Output_Get:Actor Class Reference)
   n5 = Get CurrentActor() -> (CurrentActor:Actor Object Reference)
   > Set Text(Text="To Text (String)"(InString=GetDisplayName(Object=@n5)), self=Get T_SelectedObject())
   n6 = Get CurrentClass() -> (CurrentClass:Actor Class Reference)
   > Set Text(Text="To Text (String)"(InString=Get Class Display Name(Class=@n6)), self=Get T_SelectedClass())
   > n7 = FindParentClassInDataAsset!(execute, self:EUW Asset Organizer Tools Object Reference=Get OwnerWidget(), ActorClass:Actor Class Reference=Get Class(Object=@n2)) -> (then, ParentClass:Actor Class Reference, Found:bool, Map Value:E_FNEC_AssetOrganizer_ClassType Enum=NewEnumerator0)
   > Set CurrentParentClass(CurrentParentClass:Actor Class Reference=@n7) -> (Output_Get:Actor Class Reference)
   > Set CurrentAssetClass(CurrentAssetClass:E_FNEC_AssetOrganizer_ClassType Enum=@n7.Map Value) -> (Output_Get:E_FNEC_AssetOrganizer_ClassType Enum=NewEnumerator0)
   > n8 = Branch(Condition=@n7.Found)
   > n9 = Sequence!(execute) -> (then_0, then_1, then_2:Exec)
   n10 = Enum to String!(Enumerator:E_FNEC_AssetOrganizer_ClassType Enum=Get CurrentAssetClass() -> (CurrentAssetClass:E_FNEC_AssetOrganizer_ClassType Enum=NewEnumerator0)) -> (ReturnValue:String)
   n11 = ActorHasTag!(self:Actor Object Reference=Get CurrentActor(), Tag:Name=Marker) -> (ReturnValue:bool)
   Set Text(
    execute=@n9.then_1
    Text=SelectText(A="To Text (String)"(InString=@n10), B="LOCTEXT(\"Marker Object\")", bPickA=NOT Boolean(A=@n11))
    self=Get T_SelectedClasification()
   )
   n12 = ActorHasTag(self=Get CurrentActor(), Tag=Marker)
   n13 = Get GalleryActorClasses!(self:DA FNEC Asset Organizer Data Asset Class Object Reference=Get DataAsset!(self:EUW Asset Organizer Tools Object Reference=Get OwnerWidget()) -> (DataAsset:DA FNEC Asset Organizer Data Asset Class Object Reference)) -> (GalleryActorClasses:Array of Actor Class References)
   n14 = Get CurrentParentClass() -> (CurrentParentClass:Actor Class Reference)
   n15 = Get DataAsset(self=Get OwnerWidget())
   n16 = SelectText(
    A="LOCTEXT(\"This is a gameplay actor.\\r\\nIt is not allowed in creative galleries and must be deleted.\\r\\n\\r\\n(Actor found in Disallowed Actor list in Data Asset)\")"
    B=SelectText(
     A="LOCTEXT(\"This is a Creative Gallery Actor. It is allowed in the scene and wont show up in-game. Make sure this asset is not duplicated, is meant to be in the scene and set up correctly.\")"
     B="LOCTEXT(\"This actor is of an approved Content Class. \\r\\nThis does not mean the actor is automatically cleared for gallery use, however. Keep an eye out for potential issues this actor might have dispite being a valid content class, such as the presence of RVTs, random meshes and swatches, etc.\")"
     bPickA=Contains Item(TargetArray:Array of Actor Class References=@n13, ItemToFind:"Actor Class Reference (by ref)"=@n14)
    )
    bPickA=Contains Item(
     TargetArray=Get DisallowedActorClasses!(self:DA FNEC Asset Organizer Data Asset Class Object Reference=@n15) -> (DisallowedActorClasses:Array of Actor Class References)
     ItemToFind=@n14
    )
   )
   n17 = Get OwnerWidget()
   n18 = Find(
    TargetMap:Map of Actor Class References to Strings=Get ClassesWithDescription!(self:DA FNEC Asset Organizer Data Asset Class Object Reference=Get DataAsset(self=@n17)) -> (ClassesWithDescription:Map of Actor Class References to Strings)
    Key:"Actor Class Reference (by ref)"=@n14
   )
   n19 = Format Text!(Format="LOCTEXT(\"{1}\\r\\n\\r\\n(Actor found in Flagged Content list in Data Asset)\")", 1:String=@n18) -> (Result)
   Set Text(
    execute=@n9.then_2
    Text=SelectText(
     A=SelectText(A=@n19, B=@n16, bPickA=@n18.ReturnValue)
     B="LOCTEXT(\"This is the Marker Object. \\r\\nIt is not a valid gallery actor. Make sure this actor is not saved in the scene. Delete before publishing the gallery.\")"
     bPickA=NOT Boolean(A=@n12)
    )
    self=Get T_Reason()
   )
   n20 = Get CurrentParentClass()
   Set Text(execute=@n9.then_0, Text="To Text (String)"(InString=Get Class Display Name(Class=@n20)), self=Get T_SelectedParentClass())
   n21 = Get CurrentAssetClass() -> (CurrentAssetClass=NewEnumerator0)
   n22 = Select(
    Option 0:Slate Color Structure=Make Slate Color(SpecifiedColor=Select!(
      NewEnumerator0:Linear Color Structure=(R=1,G=0.3353,B=0.8338,A=1)
      NewEnumerator1:Linear Color Structure=(R=1,G=0.049,B=0.058,A=1)
      NewEnumerator2:Linear Color Structure=(R=0.2881,G=1,B=0.4502,A=1)
      NewEnumerator4:Linear Color Structure=(R=1,G=0.3663,B=0.0467,A=1)
      NewEnumerator3:Linear Color Structure=(R=1,G=0.3669,B=0.0466,A=1)
      Index:E_FNEC_AssetOrganizer_ClassType Enum=@n21
     ) -> (ReturnValue:Linear Color Structure))
    Option 1:Slate Color Structure=Make Slate Color(SpecifiedColor=(R=0.3502,G=0.8407,B=1,A=1))
    Index=ActorHasTag(self=Get CurrentActor(), Tag=Marker)
   ) -> (ReturnValue:Slate Color Structure)
   > Set ColorAndOpacity!(execute, ColorAndOpacity:Slate Color Structure=@n22, self:Text Object Reference=Get T_SelectedClasification()) -> (then, Output_Get:Slate Color Structure)
   > Set ColorAndOpacity(ColorAndOpacity=@n22, self=Get T_Reason())
   Set Text(execute=@n8.else, Text="LOCTEXT(\"Flagged\")", self=Get T_SelectedClasification())
   > Set Text(Text="LOCTEXT(\"Non Content Class\")", self=Get T_SelectedParentClass())
   > Set Text(
    Text="LOCTEXT(\"This actor is of a class not found in any of the Tool\\'s class lists. It may be a valid content class that is currently unaccounted for or it could be a disallowed actor. \\r\\nAnalize this actor. If it is a disallowed or flagged actor please add this actor\\'s class to the corresponding list in the Tool\\'s Data Asset\")"
    self=Get T_Reason()
   )
   n23 = Make Slate Color(SpecifiedColor=(R=1,G=0.3663,B=0.0467,A=1))
   > Set ColorAndOpacity(ColorAndOpacity=@n23, self=Get T_SelectedClasification())
   > Set ColorAndOpacity(ColorAndOpacity=@n23, self=Get T_Reason())
   "To Text (String)"(InString=@n18)
  ]}
]
variables [OwnerWidget {type=EUW_AssetOrganizerTools_C, default=null}]
widgets {
 info {rootWidgetClass=&StackBox, widgetCount=26}
 widgets [
  StackBox_0 {widgetClassPath=&StackBox, widget=StackBox_0}
  Border-PositionUtilities {
   widgetClassPath=&Border
   parent=StackBox_0
   widget=Border-PositionUtilities
   properties {
    horizontalAlignment=HAlign_Left
    verticalAlignment=VAlign_Top
    background {
     bIsDynamicallyLoaded=false
     drawAs=Box
     imageType=FullColor
     mirroring=NoMirror
     resourceName=None
     resourceObject=/Game/ArtTools/FNEC/AssetOrganizer/Materials/M_UI_Border
     tiling=NoTile
     imageSize=<40 40>
     margin=<ltrb 0.5 0.5 0.5 0.5>
     outlineSettings {
      bUseBrushTransparency=false
      roundingType=HalfHeightRadius
      width=0
      color {colorUseRule=UseColor_Specified, specifiedColor=<rgba 0 0 0 0>}
      cornerRadii=<0 0 0 0>
     }
     tintColor {colorUseRule=UseColor_Specified, specifiedColor=<rgba 1 0.3875 0 0>}
     uVRegion {bIsValid=false, max=<0 0>, min=<0 0>}
    }
    padding=<ltrb 0 0 0 0>
   }
   slot {class=StackBoxSlot, properties {padding=<ltrb 0 10 10 0>, size {sizeRule=Fill, value=1}}}
  }
  SizeBox_1 {
   widgetClassPath=&SizeBox
   parent=Border-PositionUtilities
   widget=SizeBox_1
   properties {bOverride_MaxDesiredWidth=true, bOverride_MinDesiredWidth=true, maxDesiredWidth=500, minDesiredWidth=300}
   slot {class=BorderSlot, properties {horizontalAlignment=HAlign_Left, verticalAlignment=VAlign_Top, padding=<ltrb 0 0 0 0>}}
  }
  VerticalBox_9 {widgetClassPath=&VerticalBox, parent=SizeBox_1, widget=VerticalBox_9, slot {class=SizeBoxSlot}}
  T_Console {
   widgetClassPath=&TextBlock
   bIsVariable=true
   parent=VerticalBox_9
   widget=T_Console
   properties {
    text=-Console-
    colorAndOpacity {colorUseRule=UseColor_Specified, specifiedColor=<rgba 1 0.0906 0.0448 1>}
    font {
     bForceMonospaced=false
     bMaterialIsStencil=false
     fontMaterial=null
     fontObject=&Roboto
     letterSpacing=0
     monospacedWidth=1
     size=7.5
     skewAmount=0
     typefaceFontName=Italic
     outlineSettings {
      bApplyOutlineToDropShadows=false
      bMiteredCorners=false
      bSeparateFillAlpha=false
      outlineMaterial=null
      outlineSize=0
      outlineColor=<rgba 0 0 0 1>
     }
    }
   }
   slot {class=VerticalBoxSlot, properties {padding=<ltrb 10 0 0 0>}}
  }
  HorizontalBox_142 {
   widgetClassPath=&HorizontalBox
   parent=VerticalBox_9
   widget=HorizontalBox_142
   slot {class=VerticalBoxSlot, properties {horizontalAlignment=HAlign_Left}}
  }
  Spacer_233 {
   widgetClassPath=&Spacer
   parent=HorizontalBox_142
   widget=Spacer_233
   properties {size=<12 1>}
   slot {class=HorizontalBoxSlot, properties {horizontalAlignment=HAlign_Left}}
  }
  T_ToggleButton_ActorPosUtil {
   widgetClassPath=&TextBlock
   bIsVariable=true
   parent=HorizontalBox_142
   widget=T_ToggleButton_ActorPosUtil
   properties {
    text=Asset Organizer Inspector
    font {
     bForceMonospaced=false
     bMaterialIsStencil=false
     fontMaterial=null
     fontObject=&Roboto
     letterSpacing=0
     monospacedWidth=1
     size=9
     skewAmount=0
     typefaceFontName=Bold
     outlineSettings {
      bApplyOutlineToDropShadows=false
      bMiteredCorners=false
      bSeparateFillAlpha=false
      outlineMaterial=null
      outlineSize=0
      outlineColor=<rgba 0 0 0 1>
     }
    }
   }
   slot {class=HorizontalBoxSlot}
  }
  Border-Selection {
   widgetClassPath=&Border
   parent=VerticalBox_9
   widget=Border-Selection
   properties {
    background {
     bIsDynamicallyLoaded=false
     drawAs=RoundedBox
     imageType=FullColor
     mirroring=NoMirror
     resourceName=None
     resourceObject=null
     tiling=NoTile
     imageSize=<32 32>
     margin=<ltrb 0.5 0.5 0.5 0.5>
     outlineSettings {
      bUseBrushTransparency=false
      roundingType=FixedRadius
      width=1.2219
      color {colorUseRule=UseColor_Specified, specifiedColor=<rgba 0 0 0 1>}
      cornerRadii=<5 5 5 5>
     }
     tintColor {colorUseRule=UseColor_Specified, specifiedColor=<rgba 0 0 0 0>}
     uVRegion {bIsValid=false, max=<0 0>, min=<0 0>}
    }
    padding=<ltrb 5 5 5 5>
   }
   slot {class=VerticalBoxSlot, properties {padding=<ltrb 10 10 0 0>, size {sizeRule=Fill, value=1}}}
  }
  VerticalBox_0 {
   widgetClassPath=&VerticalBox
   parent=Border-Selection
   widget=VerticalBox_0
   slot {class=BorderSlot, properties {padding=<ltrb 5 5 5 5>}}
  }
  HorizontalBox_0 {
   widgetClassPath=&HorizontalBox
   parent=VerticalBox_0
   widget=HorizontalBox_0
   slot {class=VerticalBoxSlot, properties {horizontalAlignment=HAlign_Left, padding=<ltrb 0 0 0 3>}}
  }
  TextBlock_88 {
   widgetClassPath=&TextBlock
   parent=HorizontalBox_0
   widget=TextBlock_88
   properties {
    text=Actor Name :
    font {
     bForceMonospaced=false
     bMaterialIsStencil=false
     fontMaterial=null
     fontObject=&Roboto
     letterSpacing=0
     monospacedWidth=1
     size=9
     skewAmount=0
     typefaceFontName=Regular
     outlineSettings {
      bApplyOutlineToDropShadows=false
      bMiteredCorners=false
      bSeparateFillAlpha=false
      outlineMaterial=null
      outlineSize=0
      outlineColor=<rgba 0 0 0 1>
     }
    }
   }
   slot {class=HorizontalBoxSlot}
  }
  T_SelectedObject {
   widgetClassPath=&TextBlock
   bIsVariable=true
   parent=HorizontalBox_0
   widget=T_SelectedObject
   properties {
    text=XXX
    font {
     bForceMonospaced=false
     bMaterialIsStencil=false
     fontMaterial=null
     fontObject=&Roboto
     letterSpacing=0
     monospacedWidth=1
     size=9
     skewAmount=0
     typefaceFontName=Bold
     outlineSettings {
      bApplyOutlineToDropShadows=false
      bMiteredCorners=false
      bSeparateFillAlpha=false
      outlineMaterial=null
      outlineSize=0
      outlineColor=<rgba 0 0 0 1>
     }
    }
   }
   slot {class=HorizontalBoxSlot}
  }
  HorizontalBox_1 {
   widgetClassPath=&HorizontalBox
   parent=VerticalBox_0
   widget=HorizontalBox_1
   slot {class=VerticalBoxSlot, properties {horizontalAlignment=HAlign_Left, padding=<ltrb 0 0 0 3>}}
  }
  TextBlock {
   widgetClassPath=&TextBlock
   parent=HorizontalBox_1
   widget=TextBlock
   properties {
    text=Class :
    font {
     bForceMonospaced=false
     bMaterialIsStencil=false
     fontMaterial=null
     fontObject=&Roboto
     letterSpacing=0
     monospacedWidth=1
     size=9
     skewAmount=0
     typefaceFontName=Regular
     outlineSettings {
      bApplyOutlineToDropShadows=false
      bMiteredCorners=false
      bSeparateFillAlpha=false
      outlineMaterial=null
      outlineSize=0
      outlineColor=<rgba 0 0 0 1>
     }
    }
   }
   slot {class=HorizontalBoxSlot}
  }
  T_SelectedClass {
   widgetClassPath=&TextBlock
   bIsVariable=true
   parent=HorizontalBox_1
   widget=T_SelectedClass
   properties {
    text=XXX
    font {
     bForceMonospaced=false
     bMaterialIsStencil=false
     fontMaterial=null
     fontObject=&Roboto
     letterSpacing=0
     monospacedWidth=1
     size=9
     skewAmount=0
     typefaceFontName=Bold
     outlineSettings {
      bApplyOutlineToDropShadows=false
      bMiteredCorners=false
      bSeparateFillAlpha=false
      outlineMaterial=null
      outlineSize=0
      outlineColor=<rgba 0 0 0 1>
     }
    }
   }
   slot {class=HorizontalBoxSlot}
  }
  HorizontalBox {
   widgetClassPath=&HorizontalBox
   parent=VerticalBox_0
   widget=HorizontalBox
   slot {class=VerticalBoxSlot, properties {horizontalAlignment=HAlign_Left, padding=<ltrb 0 0 0 3>}}
  }
  TextBlock_1 {
   widgetClassPath=&TextBlock
   parent=HorizontalBox
   widget=TextBlock_1
   properties {
    text=Parent Class :
    font {
     bForceMonospaced=false
     bMaterialIsStencil=false
     fontMaterial=null
     fontObject=&Roboto
     letterSpacing=0
     monospacedWidth=1
     size=9
     skewAmount=0
     typefaceFontName=Regular
     outlineSettings {
      bApplyOutlineToDropShadows=false
      bMiteredCorners=false
      bSeparateFillAlpha=false
      outlineMaterial=null
      outlineSize=0
      outlineColor=<rgba 0 0 0 1>
     }
    }
   }
   slot {class=HorizontalBoxSlot}
  }
  T_SelectedParentClass {
   widgetClassPath=&TextBlock
   bIsVariable=true
   parent=HorizontalBox
   widget=T_SelectedParentClass
   properties {
    text=XXX
    font {
     bForceMonospaced=false
     bMaterialIsStencil=false
     fontMaterial=null
     fontObject=&Roboto
     letterSpacing=0
     monospacedWidth=1
     size=9
     skewAmount=0
     typefaceFontName=Bold
     outlineSettings {
      bApplyOutlineToDropShadows=false
      bMiteredCorners=false
      bSeparateFillAlpha=false
      outlineMaterial=null
      outlineSize=0
      outlineColor=<rgba 0 0 0 1>
     }
    }
   }
   slot {class=HorizontalBoxSlot}
  }
  HorizontalBox_3 {
   widgetClassPath=&HorizontalBox
   parent=VerticalBox_0
   widget=HorizontalBox_3
   properties {bIsEnabled=false, visibility=Collapsed}
   slot {class=VerticalBoxSlot, properties {horizontalAlignment=HAlign_Left, padding=<ltrb 0 0 0 3>}}
  }
  TextBlock_2 {
   widgetClassPath=&TextBlock
   parent=HorizontalBox_3
   widget=TextBlock_2
   properties {
    text=Properties :
    font {
     bForceMonospaced=false
     bMaterialIsStencil=false
     fontMaterial=null
     fontObject=&Roboto
     letterSpacing=0
     monospacedWidth=1
     size=9
     skewAmount=0
     typefaceFontName=Regular
     outlineSettings {
      bApplyOutlineToDropShadows=false
      bMiteredCorners=false
      bSeparateFillAlpha=false
      outlineMaterial=null
      outlineSize=0
      outlineColor=<rgba 0 0 0 1>
     }
    }
   }
   slot {class=HorizontalBoxSlot}
  }
  T_SelectedPropoerties {
   widgetClassPath=&TextBlock
   bIsVariable=true
   parent=HorizontalBox_3
   widget=T_SelectedPropoerties
   properties {
    text=XXX
    font {
     bForceMonospaced=false
     bMaterialIsStencil=false
     fontMaterial=null
     fontObject=&Roboto
     letterSpacing=0
     monospacedWidth=1
     size=9
     skewAmount=0
     typefaceFontName=Bold
     outlineSettings {
      bApplyOutlineToDropShadows=false
      bMiteredCorners=false
      bSeparateFillAlpha=false
      outlineMaterial=null
      outlineSize=0
      outlineColor=<rgba 0 0 0 1>
     }
    }
   }
   slot {class=HorizontalBoxSlot}
  }
  HorizontalBox_2 {
   widgetClassPath=&HorizontalBox
   parent=VerticalBox_0
   widget=HorizontalBox_2
   slot {class=VerticalBoxSlot, properties {padding=<ltrb 0 0 0 3>}}
  }
  VerticalBox_2 {
   widgetClassPath=&VerticalBox
   parent=HorizontalBox_2
   widget=VerticalBox_2
   slot {class=HorizontalBoxSlot, properties {padding=<ltrb 0 10 0 0>, size {sizeRule=Fill, value=1}}}
  }
  T_SelectedClasification {
   widgetClassPath=&TextBlock
   bIsVariable=true
   parent=VerticalBox_2
   widget=T_SelectedClasification
   properties {
    text=XXX
    font {
     bForceMonospaced=false
     bMaterialIsStencil=false
     fontMaterial=null
     fontObject=&Roboto
     letterSpacing=0
     monospacedWidth=1
     size=12
     skewAmount=0
     typefaceFontName=Bold
     outlineSettings {
      bApplyOutlineToDropShadows=false
      bMiteredCorners=false
      bSeparateFillAlpha=false
      outlineMaterial=null
      outlineSize=0
      outlineColor=<rgba 0 0 0 1>
     }
    }
   }
   slot {class=VerticalBoxSlot}
  }
  T_Reason {
   widgetClassPath=&TextBlock
   bIsVariable=true
   parent=VerticalBox_2
   widget=T_Reason
   properties {
    autoWrapText=true
    text=XXX
    font {
     bForceMonospaced=false
     bMaterialIsStencil=false
     fontMaterial=null
     fontObject=&Roboto
     letterSpacing=0
     monospacedWidth=1
     size=9
     skewAmount=0
     typefaceFontName=Regular
     outlineSettings {
      bApplyOutlineToDropShadows=false
      bMiteredCorners=false
      bSeparateFillAlpha=false
      outlineMaterial=null
      outlineSize=0
      outlineColor=<rgba 0 0 0 1>
     }
    }
   }
   slot {class=VerticalBoxSlot}
  }
 ]
}
