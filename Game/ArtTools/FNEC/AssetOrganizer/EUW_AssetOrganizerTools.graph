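# EUW_AssetOrganizerTools is an editor utility widget for organizing selected level actors, with collapsible grid arrangement, position, transform, cleanup and settings sections.
# Grid arrangement resolves an origin mode, optionally sorts by size, lays out columns, then repositions the group within undo transactions.
# A stored position or spawned gnomon marker supplies target points, drawn as debug axes on Tick.
# A data asset supplies class lists for categorizing, flagging and parent-class selection.
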EditorUtilityWidgetBlueprint /Game/ArtTools/FNEC/AssetOrganizer/EUW_AssetOrganizerTools
graphs [
 Arrange In Grid {nodes [
   n1 = Arrange In Grid!() -> (
    then
    Columns:int=5
    Columns in X Direction:bool=true
    Distance X:double=-640
    Distance Y:double=400
    List of Actors:Array of Actor Object References
    Reset Rotation:bool=~
    Reset Scale:bool=~
   ) {is=title K2Node_FunctionEntry}
   > Set CurrentOrigin(CurrentOrigin:Vector="-999999,999999,999999") -> (Output_Get:Vector)
   > n2 = Sequence()
   n3 = For Each Loop(Exec=@n2.then_1, Array:Array of Actor Object References=Get List of Actors() -> (List of Actors:Array of Actor Object References)) -> (Array Element:Actor Object Reference)
   Return Node!(execute=@n3.Completed, Arranged Actors:Array of Actor Object References=@n1.List of Actors)
   n4 = Get Columns() -> (Columns:int=0)
   n5 = "To Float (Integer)"(InInt=@n4)
   n6 = "To Float (Integer)"(InInt=@n3.Array Index)
   n7 = float * float(A:int=Floor(A=float / float(A=@n6, B=@n5)) {is=title K2Node_CallFunction}, B=Get Distance X() -> (Distance X:double)) {is=title K2Node_PromotableOperator}
   n8 = float * float(A="% (Integer)"(A=@n3.Array Index, B=@n4) -> (ReturnValue=0), B=Get Distance Y() -> (Distance Y:double))
   n9 = Get CurrentOrigin() -> (CurrentOrigin:Vector)
   n10 = Get Columns() -> (Columns=0)
   n11 = float * float(A="% (Integer)"(A=@n3.Array Index, B=@n10) -> (ReturnValue=0), B=Get Distance X())
   n12 = "To Float (Integer)"(InInt=@n3.Array Index)
   n13 = "To Float (Integer)"(InInt=@n10)
   n14 = float * float(A=Floor(A=float / float(A=@n12, B=@n13)), B=Get Distance Y())
   n15 = Get CurrentOrigin()
   > Set Actor Location!(
    execute
    self:Actor Object Reference=@n3
    NewLocation:Vector=SelectVector(
     A=vector + vector!(A=@n15, B_X:double=@n11, B_Y:double=@n14, B_Z:double=0) -> (ReturnValue)
     B=vector + vector(A=@n9, B_X=@n7, B_Y=@n8, B_Z=0)
     bPickA=Get Columns in X Direction() -> (Columns in X Direction:bool)
    )
    bSweep:bool=false
    bTeleport:bool=false
   ) -> (then, SweepHitResult:Hit Result Structure, ReturnValue:bool) {is=title K2Node_CallFunction}
   n16 = For Each Loop(Exec=@n2.then_0, Array=Get List of Actors())
   > TransactObject(Object=@n16)
   > n17 = Sequence!(execute) -> (then_0, then_1, then_2:Exec)
   n18 = Get Actor Location!(self=@n16) -> (ReturnValue_X:double, ReturnValue_Y:double, ReturnValue_Z:double)
   n19 = Get!() -> (CurrentOrigin_X:double, CurrentOrigin_Y:double, CurrentOrigin_Z:double)
   > Set!(
    execute
    CurrentOrigin_X:double="Max (Float)"(A=@n18, B=@n19)
    CurrentOrigin_Y:double="Min (Float)"(A:double=@n18.ReturnValue_Y, B:double=@n19.CurrentOrigin_Y) -> (ReturnValue:double) {is=title K2Node_CommutativeAssociativeBinaryOperator}
    CurrentOrigin_Z:double="Min (Float)"(A=@n18.ReturnValue_Z, B=@n19.CurrentOrigin_Z)
   ) -> (then, Output_Get:Vector)
   Branch(execute=@n17.then_1, Condition=Get Reset Scale() -> (Reset Scale:bool))
   > SetActorScale3D!(execute, self:Actor Object Reference=@n16, NewScale3D:Vector="1,1,1") -> (then)
   Branch(execute=@n17.then_2, Condition=Get Reset Rotation() -> (Reset Rotation:bool))
   > Set Actor Rotation!(execute, self:Actor Object Reference=@n16, NewRotation:Rotator="0, 0, 0", bTeleportPhysics:bool=false) -> (then, ReturnValue:bool) {is=title K2Node_CallFunction}
  ]}
 Arrange Objects with Origin Options {nodes [
   n1 = Arrange Objects with Origin Options!() -> (
    then
    Selected Actors:Array of Actor Object References
    OriginMode:E_FNEC_AssetOrganizer_GridOrigin Enum=NewEnumerator0
    Columns:int
    Columns in X Direction:bool=~
    Distance X:double=~
    Distance Y:double=~
    By Size:bool=~
    Reset Rotation:bool=~
    Reset Scale:bool=~
    Origin Offset:Vector=~
    Explicit Origin:Vector=~
    Skip Transaction Begin and End:bool=~
   ) {is=title K2Node_FunctionEntry}
   > n2 = Branch(Condition=@n1.Skip Transaction Begin and End)
   Set Actors To Process(execute=[@n31.then, @n2.then], Actors To Process:Array of Actor Object References=@n1) -> (Output_Get:Array of Actor Object References)
   > n3 = Sequence()
   > n4 = Switch on E_FNEC_AssetOrganizer_GridOrigin!(
    execute
    Selection:E_FNEC_AssetOrganizer_GridOrigin Enum=Get OriginMode() -> (OriginMode:E_FNEC_AssetOrganizer_GridOrigin Enum=NewEnumerator0)
   ) -> (
    NewEnumerator0:Exec
    NewEnumerator1:Exec
    NewEnumerator2:Exec
    NewEnumerator3:Exec
    NewEnumerator4:Exec
    NewEnumerator5:Exec
    NewEnumerator6:Exec
    NewEnumerator7:Exec
    NewEnumerator10:Exec
    NewEnumerator11:Exec
   ) {is=title K2Node_SwitchEnum}
   n5 = Get Actors To Process() -> (Actors To Process:Array of Actor Object References)
   n6 = int - int(A=Length(TargetArray=@n5))
   n7 = Set Target Origin(
    execute=@n4.NewEnumerator2
    Target Origin:Vector=Get Actor Location!(self="Get (a copy)"(Array:Array of Actor Object References=@n5, Dimension 1=@n6) -> (Output:Actor Object Reference)) -> (ReturnValue:Vector)
   ) -> (Output_Get:Vector)
   Set Reposition at end of process(execute=[@n11.then, @n20.then, @n14.then, @n9.then, @n7.then, @n12.then, @n13.then, @n10.then], Reposition at end of process:bool=true) -> (Output_Get:bool)
   > Set Target Origin(Target Origin=vector + vector!(A=Get Target Origin() -> (Target Origin:Vector), B:Vector=Get Origin Offset() -> (Origin Offset:Vector)) -> (ReturnValue))
   n8 = Get Actors To Process()
   n9 = Set Target Origin(execute=@n4.NewEnumerator1, Target Origin=Get Actor Location(self="Get (a copy)"(Array=@n8)))
   Set Reposition at end of process(execute=@n4.NewEnumerator0, Reposition at end of process=false)
   n10 = Set Target Origin(execute=@n4.NewEnumerator5, Target Origin=GetStoredPositions!(From Pos Object:bool=false) -> (Position:Vector))
   n11 = Set Target Origin(execute=@n4.NewEnumerator6, Target Origin=GetStoredPositions(From Pos Object=true))
   n12 = Set Target Origin(execute=@n4.NewEnumerator7, Target Origin=Get Explicit Origin() -> (Explicit Origin:Vector))
   n13 = Set Target Origin(execute=@n4.NewEnumerator10, Target Origin="0, 0, 0")
   n14 = Set Target Origin(
    execute=@n4.NewEnumerator11
    Target Origin=Get Default Gallery Corner!(self:DA FNEC Asset Organizer Data Asset Class Object Reference=Get DataAsset() -> (DataAsset:DA FNEC Asset Organizer Data Asset Class Object Reference)) -> (Default Gallery Corner:Vector)
   )
   n15 = Get Actors To Process()
   n16 = "Get (a copy)"(Array=@n15, Dimension 1=int - int(A=Length(TargetArray=@n15)))
   n17 = Get Distance X()
   n18 = Set Target Origin(execute=@n4.NewEnumerator4, Target Origin=vector + vector(A=Get Actor Location(self=@n16), B=MakeVector(X=@n17, Y=0)))
   n19 = Get Actors To Process()
   n20 = Remove Index!(execute=[@n24.then, @n18.then], TargetArray:Array of Actor Object References=@n19, IndexToRemove:int=int - int(A=Length(TargetArray=@n19))) -> (then)
   n21 = Get Actors To Process()
   n22 = "Get (a copy)"(Array=@n21, Dimension 1=int - int(A=Length(TargetArray=@n21)))
   n23 = Get Distance Y()
   n24 = Set Target Origin(execute=@n4.NewEnumerator3, Target Origin=vector + vector(A=Get Actor Location(self=@n22), B=MakeVector(Y=@n23)))
   n25 = Get Actors To Process()
   n26 = Sort By Size!(execute=@n3.then_1, self, List of Actors:Array of Actor Object References=@n25) -> (then, OrderedActors:Array of Actor Object References)
   > n27 = Arrange In Grid!(
    execute
    self
    Columns:int=Get Columns() -> (Columns=0)
    Columns in X Direction:bool=Get Columns in X Direction()
    Distance X:double=Get Distance X()
    Distance Y:double=Get Distance Y()
    List of Actors:Array of Actor Object References=Select(Option 0:Array of Actor Object References=@n25, Option 1:Array of Actor Object References=@n26, Index=Get By Size() -> (By Size:bool)) -> (ReturnValue:Array of Actor Object References)
    Reset Rotation:bool=Get Reset Rotation()
    Reset Scale:bool=Get Reset Scale()
   ) -> (then, Arranged Actors:Array of Actor Object References) {is=title K2Node_CallFunction}
   > n28 = Branch(Condition=Get By Size())
   > n29 = XMirror!(execute, self, List of Actors:Array of Actor Object References=@n27) -> (then)
   Branch(execute=[@n29.then, @n28.else], Condition=Get Reposition at end of process() -> (Reposition at end of process:bool))
   > Reposition Group to XYZ!(execute, self, Target:Vector=Get Target Origin(), List of Actors:Array of Actor Object References=@n27, Write to Undo History:bool=false) -> (then)
   n30 = Branch(execute=@n3.then_2, Condition=Get Skip Transaction Begin and End() -> (Skip Transaction Begin and End:bool))
   EndTransaction(execute=@n30.else)
   n31 = BeginTransaction(execute=@n2.else, Context=Asset Organizer, Description="LOCTEXT(\"Arrange in Grid\")")
  ]}
 CollapseObjectsToPosition {nodes [
   n1 = CollapseObjectsToPosition!() -> (then, Position:Vector=~, List of Actors:Array of Actor Object References, Write to Undo History:bool=~) {is=title K2Node_FunctionEntry}
   > n2 = Branch(Condition=@n1.Write to Undo History)
   > n3 = BeginTransaction(Context=Asset Organizer, Description="LOCTEXT(\"Move Actors to Position\")")
   n4 = For Each Loop(Exec=[@n3.then, @n2.else], Array=@n1.List of Actors)
   Branch(execute=@n4.Completed, Condition=Get Write to Undo History() -> (Write to Undo History:bool))
   > EndTransaction()
   n5 = Branch(execute=@n4.LoopBody, Condition=Get Write to Undo History())
   > n6 = TransactObject(Object=@n4)
   Set Actor Location(execute=[@n6.then, @n5.else], self=@n4, NewLocation=@n1, bSweep=false, bTeleport=false)
  ]}
 DetectFlaggedContent {nodes [
   n1 = DetectFlaggedContent!() -> (then, ActorToAnalize:Actor Object Reference) {is=title K2Node_FunctionEntry}
   > Set IsFlaggedClass(IsFlaggedClass:bool=false) -> (Output_Get:bool)
   > n2 = Sequence!(execute) -> (then_0, then_1)
   > n3 = For Each Loop with Break!(
    Exec:Exec
    Array:Array of Actor Class References=Get FlaggedContentClasses!(self:DA FNEC Asset Organizer Data Asset Class Object Reference=Get DataAsset()) -> (FlaggedContentClasses:Array of Actor Class References)
    Break:Exec
   ) -> (LoopBody:Exec, Array Element:Actor Class Reference, Array Index:int, Completed:Exec)
   > Set FlaggedClass(FlaggedClass:Actor Class Reference=@n3) -> (Output_Get:Actor Class Reference)
   n4 = Get Class(Object=@n1)
   > Set IsFlaggedClass(IsFlaggedClass=Select(Option 0:bool=Get IsFlaggedClass() -> (IsFlaggedClass:bool), Option 1:bool=true, Index=ClassIsChildOf(TestClass=@n4, ParentClass=@n3)) -> (ReturnValue:bool))
   n5 = Branch(execute=@n2.then_1, Condition=Get IsFlaggedClass())
   > Return Node!(
    execute
    isFlagged:bool=Get IsFlaggedClass()
    Class:Actor Class Reference=Get FlaggedClass() -> (FlaggedClass:Actor Class Reference)
    Special:bool=false
   )
   n6 = GetEditorProperty!(
    execute:Exec=@n5.else
    Object:Object Reference=Get ActorToAnalize() -> (ActorToAnalize:Actor Object Reference)
    PropertyName:Name=SpawnFoliage
   ) -> (then:Exec, PropertyValue:bool, ReturnValue:bool)
   > Set RetrievedValue(RetrievedValue:bool=@n6.ReturnValue) -> (Output_Get:bool)
   > n7 = Branch(Condition=@n6)
   > Return Node(isFlagged=true, Special=true)
   Return Node(execute=@n7.else, isFlagged=false, Special=false)
  ]}
 Displace Objects as a Group {nodes [
   n1 = Displace Objects as a Group!() -> (then, Offset:Vector=~, Offset Multiplier:double=1, List of Actors:Array of Actor Object References, Write to Undo History:bool=~) {is=title K2Node_FunctionEntry}
   > n2 = Branch(Condition=@n1.Write to Undo History)
   > n3 = BeginTransaction(Context=Asset Organizer, Description="LOCTEXT(\"Displace Group of Actors\")")
   n4 = For Each Loop(Exec=[@n3.then, @n2.else], Array=@n1.List of Actors)
   Branch(execute=@n4.Completed, Condition=Get Write to Undo History())
   > EndTransaction()
   n5 = Branch(execute=@n4.LoopBody, Condition=Get Write to Undo History())
   > n6 = TransactObject(Object=@n4)
   Set Actor Location(
    execute=[@n6.then, @n5.else]
    self=@n4
    NewLocation=vector + vector(A=Get Actor Location(self=@n4), B=vector * vector(A=@n1, B=@n1.Offset Multiplier) -> (ReturnValue=~))
    bSweep=false
    bTeleport=false
   )
  ]}
 DrawDebugCoordinates {nodes [
   n1 = DrawDebugCoordinates!() -> (then, Draw Position:Vector=~, Delta T:double=~, Center Color:Linear Color Structure=(R=1,G=0,B=1,A=1)) {is=title K2Node_FunctionEntry}
   n2 = float + float(A=@n1.Delta T, B=0.1)
   DrawDebugBox!(
    execute
    WorldContextObject:Object Reference
    Center:Vector=Get Draw Position() -> (Draw Position:Vector)
    Extent:Vector="50,50,50"
    LineColor:Linear Color Structure=@n1.Center Color
    Rotation:Rotator
    Duration:float=@n2
    Thickness:float=10
    DepthPriority:EDrawDebugSceneDepthPriorityGroup Enum=World
   ) -> (then)
   n3 = Get Draw Position()
   n4 = Get LineThickness() -> (LineThickness:float)
   > DrawDebugLine!(
    execute
    WorldContextObject:Object Reference
    LineStart:Vector=@n3
    LineEnd:Vector=vector + vector(A=@n3, B="512,0,0")
    LineColor:Linear Color Structure=(R=1,G=0,B=0,A=1)
    Duration:float=@n2
    Thickness:float=@n4
    DepthPriority:EDrawDebugSceneDepthPriorityGroup Enum=World
   ) -> (then)
   n5 = Get Draw Position()
   > DrawDebugLine(LineStart=@n5, LineEnd=vector + vector(A=@n5, B="0,512,0"), LineColor=(R=0,G=1,B=0,A=1), Duration=@n2, Thickness=@n4, DepthPriority=World)
   n6 = Get Draw Position()
   > DrawDebugLine(LineStart=@n6, LineEnd=vector + vector(A=@n6, B="0,0,512"), LineColor=(R=0,G=0,B=1,A=1), Duration=@n2, Thickness=@n4, DepthPriority=World)
   n7 = Get Draw Position()
   n8 = Get LineThickness()
   n9 = float + float(A=@n1.Delta T, B=0.1)
   DrawDebugLine(
    execute=@n12.then
    LineStart=vector + vector(A=@n7, B="0,0,256")
    LineEnd=vector + vector(A=@n7, B="0,0,-256")
    LineColor=Get Center Color() -> (Center Color:Linear Color Structure)
    Duration=@n9
    Thickness=@n8
    DepthPriority=World
   )
   DrawDebugBox(execute=@n1.then, Center=Get Draw Position(), Extent="50,50,50", LineColor=@n1.Center Color, Duration=@n9, Thickness=10, DepthPriority=World)
   n10 = Get Draw Position()
   > DrawDebugLine(
    LineStart=vector + vector(A=@n10, B="256,0,0")
    LineEnd=vector + vector(A=@n10, B="-256,0,0")
    LineColor=Get Center Color()
    Duration=@n9
    Thickness=@n8
    DepthPriority=World
   )
   n11 = Get Draw Position()
   > n12 = DrawDebugLine(
    LineStart=vector + vector(A=@n11, B="0,256,0")
    LineEnd=vector + vector(A=@n11, B="0,-256,0")
    LineColor=Get Center Color()
    Duration=@n9
    Thickness=@n8
    DepthPriority=World
   )
  ]}
 EventGraph {nodes [
   n1 = GetSelectionSet!(execute) -> (then, ReturnValue:Array of Actor Object References)
   n2 = "Get (a copy)"(Array=@n1)
   > n3 = GetEditorProperty(Object=@n2, PropertyName=bMirrored) -> (PropertyValue:Wildcard)
   > Branch(Condition=@n3.ReturnValue)
   n4 = Get EditorActorSubsystem()
   ClearActorSelectionSet(self:Editor Actor Subsystem Object Reference=@n4)
   > SetSelectedLevelActors(self=@n4)
   "On Clicked (B_Arrange)"() {is=title K2Node_ComponentBoundEvent}
   > n5 = GetSelectionSet()
   n6 = GetSelectedIndex!(self:"Combo Box (String) Object Reference"=Get Combo_RowAlignment() -> (Combo_RowAlignment:Editor Utility Combo Box String Object Reference)) -> (ReturnValue:int=0)
   n7 = "To String (Text)"(InText="GetText (Editable Text)"!(self:Editable Text Object Reference=Get T_Arrange_Num() -> (T_Arrange_Num:Editor Utility Editable Text Object Reference)) -> (ReturnValue:Text))
   n8 = GetSelectedIndex(self=Get Combo_Arrange_OriginMode() -> (Combo_Arrange_OriginMode:Editor Utility Combo Box String Object Reference)) -> (ReturnValue=0)
   n9 = "GetText (Editable Text)"(self=Get T_Arrange_Y() -> (T_Arrange_Y:Editor Utility Editable Text Object Reference))
   n10 = "GetText (Editable Text)"(self=Get T_Arrange_X() -> (T_Arrange_X:Editor Utility Editable Text Object Reference))
   > Arrange Objects with Origin Options!(
    execute
    self
    Selected Actors:Array of Actor Object References=@n5
    OriginMode:E_FNEC_AssetOrganizer_GridOrigin Enum=Byte to Enum E_FNEC_AssetOrganizer_GridOrigin!(Byte:Byte="To Byte (Integer)"!(InInt:int=@n8) -> (ReturnValue:Byte=0)) -> (ReturnValue:E_FNEC_AssetOrganizer_GridOrigin Enum) {is=title K2Node_CastByteToEnum}
    Columns:int=Truncate(A=String To Float(InString=@n7))
    Columns in X Direction:bool=NOT Boolean(A="To Boolean (Integer)"!(InInt:int=@n6) -> (ReturnValue:bool))
    Distance X:double=String To Float(InString="To String (Text)"(InText=@n10))
    Distance Y:double=String To Float(InString="To String (Text)"(InText=@n9))
    By Size:bool=IsChecked(self=Get CB_Arrange_BySize() -> (CB_Arrange_BySize:Editor Utility Check Box Object Reference))
    Reset Rotation:bool=false
    Reset Scale:bool=false
    Origin Offset:Vector="0, 0, 0"
    Explicit Origin:Vector="0, 0, 0"
    Skip Transaction Begin and End:bool=false
   ) -> (then) {is=title K2Node_CallFunction}
   "On Clicked (B_Arrange_Distances_SmallProps)"() {is=title K2Node_ComponentBoundEvent}
   > SetGridSize!(execute, self, X:double=-128, Y:double=128) -> (then)
   n11 = Get Style_GrayButton_Toggled() -> (Style_GrayButton_Toggled:Button Style Structure)
   > SetStyle!(
    execute
    self:Button Object Reference=Get B_Arrange_Distances_SmallProps() -> (B_Arrange_Distances_SmallProps:Editor Utility Button Object Reference)
    InStyle:"Button Style Structure (by ref)"=@n11
   ) -> (then)
   n12 = Get Style_GrayButton_Default() -> (Style_GrayButton_Default:Button Style Structure)
   > n13 = SetStyle(
    self=[
     Get B_Arrange_Distances_Walls() -> (B_Arrange_Distances_Walls:Editor Utility Button Object Reference)
     Get B_Arrange_Distances_Props() -> (B_Arrange_Distances_Props:Editor Utility Button Object Reference)
     Get B_Arrange_Distances_Floors() -> (B_Arrange_Distances_Floors:Editor Utility Button Object Reference)
     Get B_Arrange_Distances_Custom() -> (B_Arrange_Distances_Custom:Editor Utility Button Object Reference)
    ]
    InStyle=@n12
   )
   SetIsEnabled(execute=[@n52.then, @n59.then, @n58.then, @n13.then], self=[Get T_Arrange_Y(), Get T_Arrange_X()], bInIsEnabled=false)
   > SetBrushColor(
    self=[Get Border_XDistance() -> (Border_XDistance:Border Object Reference), Get Border_YDistance() -> (Border_YDistance:Border Object Reference)]
    InBrushColor=Get Color_DisabledText() -> (Color_DisabledText:Linear Color Structure)
   )
   n14 = "On Check State Changed (CB_Toggle_SpawnMarkerObject)"!() -> (OutputDelegate, then, bIsChecked:bool) {is=title K2Node_ComponentBoundEvent}
   > n15 = Branch(Condition=@n14.bIsChecked)
   n16 = Is Valid(exec=[@n61.then, @n15.then], InputObject=Get MarkerObject() -> (MarkerObject:Actor Object Reference))
   n17 = SpawnActorFromClass!(
    execute=@n16.Is Not Valid
    ActorClass:Actor Class Reference=&StaticMeshActor
    Location:Vector=GetStoredPositions(From Pos Object=false)
    Rotation:Rotator
    bTransient:bool
   ) -> (then, ReturnValue:Static Mesh Actor Object Reference)
   > SetActorLabel(self=@n17, NewActorLabel=Marker Object, bMarkDirty=false)
   > AddActorTag!(execute, Actor:Actor Object Reference=@n17, Tag:"Name (by ref)"=MakeLiteralName(Value=Marker), __WorldContext:Object Reference) -> (then)
   > SetActorScale3D(self=@n17, NewScale3D="50,50,50")
   > SetStaticMesh(self=Get StaticMeshComponent(self=@n17), NewMesh=/Niagara/DefaultAssets/S_Gnomon)
   > n18 = Set MarkerObject(MarkerObject:Actor Object Reference=@n17) -> (Output_Get:Actor Object Reference)
   n19 = Get EditorActorSubsystem()
   > ClearActorSelectionSet(self=@n19)
   > SetSelectedLevelActors(self=@n19, ActorsToSelect=Make Array("[0]":Actor Object Reference=@n18) -> (Array:Array of Actor Object References))
   > Set Text(Text="LOCTEXT(\"Despawn Marker Object\")", self=Get T_Toggle_SpawnMarkerObject() -> (T_Toggle_SpawnMarkerObject:Text Object Reference))
   n20 = Get MarkerObject()
   Is Valid(exec=[@n65.then, @n15.else], InputObject=@n20)
   > Destroy Actor(self=@n20)
   > Set Text(Text="LOCTEXT(\"Spawn Marker Object\")", self=Get T_Toggle_SpawnMarkerObject())
   n21 = "On Check State Changed (ToggleButton_Collapse_ActorPosUtil)"!() -> (OutputDelegate, then, bIsChecked:bool) {is=title K2Node_ComponentBoundEvent}
   > n22 = Branch(Condition=@n21.bIsChecked)
   n23 = Get T_ToggleButton_ActorPosUtil() -> (T_ToggleButton_ActorPosUtil:Text Object Reference)
   Set Text(execute=[@n22.then, @n63.then], Text="LOCTEXT(\"Position Utilities ▲\")", self=@n23)
   n24 = Get Socket_ActorPosUtil() -> (Socket_ActorPosUtil:Vertical Box Object Reference)
   > SetIsEnabled(self=@n24, bInIsEnabled=true)
   > SetVisibility(self=@n24, InVisibility=Visible)
   Set Text(execute=[@n22.else, @n63.else], Text="LOCTEXT(\"Position Utilities ▼\")", self=@n23)
   > SetIsEnabled(self=@n24, bInIsEnabled=false)
   > SetVisibility(self=@n24, InVisibility=Collapsed)
   "On Clicked (B_SpawnInspector)"() {is=title K2Node_ComponentBoundEvent}
   > n25 = SpawnAndRegisterTab(self=Get EditorUtilitySubsystem(), InBlueprint=/Game/ArtTools/FNEC/AssetOrganizer/Widgets/EUW_AssetOrganizer_Inspector)
   > n26 = Cast To EUW_AssetOrganizer_Inspector!(execute, Object=@n25) -> (then, CastFailed, AsEUW Asset Organizer Inspector:EUW Asset Organizer Inspector Object Reference) {is=title K2Node_DynamicCast}
   > n27 = Set InspectorWidget(InspectorWidget:EUW Asset Organizer Inspector Object Reference=@n26) -> (Output_Get:EUW Asset Organizer Inspector Object Reference)
   > Set OwnerWidget!(execute, OwnerWidget:EUW Asset Organizer Tools Object Reference=Self-Reference(), self:EUW Asset Organizer Inspector Object Reference=@n27) -> (then, Output_Get:EUW Asset Organizer Tools Object Reference)
   "On Clicked (B_Select_Marker)"() {is=title K2Node_ComponentBoundEvent}
   > Select Marker Object()
   "On Clicked (B_Arrange_Distances_Custom)"() {is=title K2Node_ComponentBoundEvent}
   > SetGridSize(X=-640, Y=640)
   > SetStyle(self=Get B_Arrange_Distances_Custom(), InStyle=@n11)
   > SetStyle(
    self=[Get B_Arrange_Distances_Walls(), Get B_Arrange_Distances_SmallProps(), Get B_Arrange_Distances_Props(), Get B_Arrange_Distances_Floors()]
    InStyle=@n12
   )
   > SetIsEnabled(self=[Get T_Arrange_Y(), Get T_Arrange_X()], bInIsEnabled=true)
   > SetBrushColor(self=[Get Border_XDistance(), Get Border_YDistance()], InBrushColor=Get Color_EnabledText() -> (Color_EnabledText:Linear Color Structure))
   n28 = "On Check State Changed (ToggleButton_Collapse_ActorCleanup)"!() -> (OutputDelegate, then, bIsChecked:bool) {is=title K2Node_ComponentBoundEvent}
   > n29 = Branch(Condition=@n28.bIsChecked)
   n30 = Get T_ToggleButton_ActorCleanup() -> (T_ToggleButton_ActorCleanup:Text Object Reference)
   Set Text(execute=[@n29.then, @n66.then], Text="LOCTEXT(\"Actor Cleanup Utilities ▲\")", self=@n30)
   n31 = Get Socket_ActorCleanupTools() -> (Socket_ActorCleanupTools:Vertical Box Object Reference)
   > SetIsEnabled(self=@n31, bInIsEnabled=true)
   > SetVisibility(self=@n31, InVisibility=Visible)
   Set Text(execute=[@n29.else, @n66.else], Text="LOCTEXT(\"Actor Cleanup Utilities ▼\")", self=@n30)
   > SetIsEnabled(self=@n31, bInIsEnabled=false)
   > SetVisibility(self=@n31, InVisibility=Collapsed)
   "On Clicked (B_Arrange_Square)"() {is=title K2Node_ComponentBoundEvent}
   > n32 = GetSelectionSet()
   n33 = Length(TargetArray=@n32)
   n34 = Round(A=Power!(Base:double="To Float (Integer)"(InInt=@n33), Exp:double=0.5) -> (ReturnValue:double) {is=title K2Node_CallFunction}) {is=title K2Node_CallFunction}
   > Set Text(
    Text="To Text (Float)"(
     Value="To Float (Integer)"(InInt=@n34)
     MinimumIntegralDigits=1
     MaximumIntegralDigits=324
     MinimumFractionalDigits=0
     MaximumFractionalDigits=3
    )
    self:Editable Text Object Reference=Get T_Arrange_Num()
   )
   "On Clicked (B_Select_ByParentClass)"() {is=title K2Node_ComponentBoundEvent}
   > n35 = GetSelectionSet()
   > n36 = For Each Loop(Array:Array of Actor Class References=Get ContentClasses!(self:DA FNEC Asset Organizer Data Asset Class Object Reference=Get DataAsset()) -> (ContentClasses:Array of Actor Class References)) -> (Array Element:Actor Class Reference)
   n37 = "Get (a copy)"(Array=@n35)
   > Branch(Condition=ClassIsChildOf(TestClass=Get Class(Object=@n37), ParentClass=@n36))
   > n38 = GetAllActorsOfClass!(execute, WorldContextObject:Object Reference, ActorClass:Actor Class Reference=@n36) -> (then, OutActors:Array of Actor Object References)
   n39 = Get EditorActorSubsystem()
   > ClearActorSelectionSet(self=@n39)
   > SetSelectedLevelActors(self=@n39, ActorsToSelect=@n38)
   n40 = "On Check State Changed (ToggleButton_Collapse_ArrangeInGrid)"!() -> (OutputDelegate, then, bIsChecked:bool) {is=title K2Node_ComponentBoundEvent}
   > n41 = Branch(Condition=@n40.bIsChecked)
   n42 = Get T_ToggleButton_ArrageInGrid() -> (T_ToggleButton_ArrageInGrid:Text Object Reference)
   Set Text(execute=[@n41.then, @n72.then], Text="LOCTEXT(\"Grid Arrangement Utilities ▲\")", self:Text Object Reference=@n42)
   n43 = Get Socket_ArrangeInGrid() -> (Socket_ArrangeInGrid:Vertical Box Object Reference)
   > SetIsEnabled(self=@n43, bInIsEnabled=true)
   > SetVisibility(self=@n43, InVisibility=Visible)
   Set Text(execute=[@n41.else, @n72.else], Text="LOCTEXT(\"Grid Arrangement Utilities ▼\")", self=@n42)
   > SetIsEnabled(self=@n43, bInIsEnabled=false)
   > SetVisibility(self=@n43, InVisibility=Collapsed)
   "On Clicked (B_Select_ByClass)"() {is=title K2Node_ComponentBoundEvent}
   > SelectAllActorsOfClass()
   n44 = "On Check State Changed (ToggleButton_Collapse_Settings)"!() -> (OutputDelegate, then, bIsChecked:bool) {is=title K2Node_ComponentBoundEvent}
   > n45 = Branch(Condition=@n44.bIsChecked)
   n46 = Get T_ToggleButton_Settings() -> (T_ToggleButton_Settings:Text Object Reference)
   Set Text(execute=[@n45.then, @n62.then], Text="LOCTEXT(\"Settings ▲\")", self=@n46)
   n47 = Get Socket_Settings() -> (Socket_Settings:Vertical Box Object Reference)
   > SetIsEnabled(self=@n47, bInIsEnabled=true)
   > SetVisibility(self=@n47, InVisibility=Visible)
   Set Text(execute=[@n45.else, @n62.else], Text="LOCTEXT(\"Settings ▼\")", self=@n46)
   > SetIsEnabled(self=@n47, bInIsEnabled=false)
   > SetVisibility(self=@n47, InVisibility=Collapsed)
   "On Clicked (B_Info)"() {is=title K2Node_ComponentBoundEvent}
   > ExecutePythonCommand(PythonCommand="To String (Text)"(InText=Format Text(Format="LOCTEXT(\"import webbrowser\\r\\n\\r\\nwebbrowser.open(\\'https://confluence-epicgames.atlassian.net/wiki/spaces/VAL/pages/119100389/FNEC+Asset+Organizer+Tool\\',new=2)\")")))
   "On Clicked (B_StorePosition)"() {is=title K2Node_ComponentBoundEvent}
   > n48 = GetSelectionSet()
   n49 = "Get (a copy)"(Array=@n48)
   > Set StoredPosition(StoredPosition:Vector=SelectVector(A=Get Actor Location(self=@n49), B=Get StoredPosition() -> (StoredPosition:Vector), bPickA=IsValid(Object=@n49))) -> (Output_Get:Vector)
   n50 = "On Check State Changed (CB_Toggle_GridSnapping)"!() -> (OutputDelegate, then, bIsChecked:bool) {is=title K2Node_ComponentBoundEvent}
   > n51 = Branch(Condition=@n50.bIsChecked)
   > ExecuteConsoleCommand(Command=BuildingActor.SuppressSnapToGrid false)
   ExecuteConsoleCommand(execute=@n51.else, Command=BuildingActor.SuppressSnapToGrid true)
   "On Clicked (B_Arrange_Distances_Walls)"() {is=title K2Node_ComponentBoundEvent}
   > SetGridSize(X=-640, Y=400)
   > SetStyle(self=Get B_Arrange_Distances_Walls(), InStyle=@n11)
   > n52 = SetStyle(
    self=[Get B_Arrange_Distances_SmallProps(), Get B_Arrange_Distances_Props(), Get B_Arrange_Distances_Floors(), Get B_Arrange_Distances_Custom()]
    InStyle=@n12
   )
   n53 = "On Check State Changed (ToggleButton_Collapse_ActorTransforms)"!() -> (OutputDelegate, then, bIsChecked:bool) {is=title K2Node_ComponentBoundEvent}
   > n54 = Branch(Condition=@n53.bIsChecked)
   n55 = Get T_ToggleButton_ActorTransforms() -> (T_ToggleButton_ActorTransforms:Text Object Reference)
   Set Text(execute=[@n54.then, @n60.then], Text="LOCTEXT(\"Actor Transform Utilities ▲\")", self=@n55)
   n56 = Get Socket_ActorTransforms() -> (Socket_ActorTransforms:Vertical Box Object Reference)
   > SetIsEnabled(self=@n56, bInIsEnabled=true)
   > SetVisibility(self=@n56, InVisibility=Visible)
   Set Text(execute=[@n54.else, @n60.else], Text="LOCTEXT(\"Actor Transform Utilities ▼\")", self=@n55)
   > SetIsEnabled(self=@n56, bInIsEnabled=false)
   > SetVisibility(self=@n56, InVisibility=Collapsed)
   n57 = "On Clicked (B_Arrange_Distances_Floors)"() {is=title K2Node_ComponentBoundEvent}
   SetGridSize(execute=[@n57.then, @n64.then], X=-640, Y=640)
   > SetStyle(self=Get B_Arrange_Distances_Floors(), InStyle=@n11)
   > n58 = SetStyle(
    self=[Get B_Arrange_Distances_Walls(), Get B_Arrange_Distances_SmallProps(), Get B_Arrange_Distances_Props(), Get B_Arrange_Distances_Custom()]
    InStyle=@n12
   )
   "On Clicked (B_Arrange_Distances_Props)"() {is=title K2Node_ComponentBoundEvent}
   > SetGridSize(X=-256, Y=256)
   > SetStyle(self=Get B_Arrange_Distances_Props(), InStyle=@n11)
   > n59 = SetStyle(
    self=[Get B_Arrange_Distances_Walls(), Get B_Arrange_Distances_SmallProps(), Get B_Arrange_Distances_Floors(), Get B_Arrange_Distances_Custom()]
    InStyle=@n12
   )
   InitializeLayout_ActorTransformsModule() {is=title K2Node_CustomEvent}
   > n60 = Branch(Condition=IsChecked(self=Get ToggleButton_Collapse_ActorTransforms() -> (ToggleButton_Collapse_ActorTransforms:Editor Utility Check Box Object Reference)))
   Create Position Actor() {is=title K2Node_CustomEvent}
   > n61 = SetCheckedState!(
    execute
    self:Check Box Object Reference=Get CB_Toggle_SpawnMarkerObject() -> (CB_Toggle_SpawnMarkerObject:Editor Utility Check Box Object Reference)
    InCheckedState:ECheckBoxState Enum=Checked
   ) -> (then)
   InitializeLayout_SettingsModule() {is=title K2Node_CustomEvent}
   > n62 = Branch(Condition=IsChecked(self=Get ToggleButton_Collapse_Settings() -> (ToggleButton_Collapse_Settings:Editor Utility Check Box Object Reference)))
   InitializeLayout_ActorPositionUtilitiesModule() {is=title K2Node_CustomEvent}
   > n63 = Branch(Condition=IsChecked(self=Get ToggleButton_Collapse_ActorPosUtil() -> (ToggleButton_Collapse_ActorPosUtil:Editor Utility Check Box Object Reference)))
   n64 = Initialize_Arranger() {is=title K2Node_CustomEvent}
   n65 = Deconstruct Position Actor() {is=title K2Node_CustomEvent}
   InitializeLayout_ActorCleanupModule() {is=title K2Node_CustomEvent}
   > n66 = Branch(Condition=IsChecked(self=Get ToggleButton_Collapse_ActorCleanup() -> (ToggleButton_Collapse_ActorCleanup:Editor Utility Check Box Object Reference)))
   Select Marker Object() {is=title K2Node_CustomEvent}
   n67 = Get MarkerObject()
   > n68 = Is Valid(InputObject=@n67)
   > SetSelectedLevelActors(self=Get EditorActorSubsystem(), ActorsToSelect=Make Array("[0]"=@n67))
   n69 = Show Message Dialog(
    execute=@n68.Is Not Valid
    Title=MakeLiteralText(Value="LOCTEXT(\"Asset Organizer Tool - Select Marker Object\")")
    Message=MakeLiteralText(Value="LOCTEXT(\"No marker object in the scene to select. Create one?\")")
    MessageType=OkCancel
    DefaultValue=Cancel
    MessageCategory=Warning
   )
   > n70 = Switch on EAppReturnType(Selection=@n69)
   Create Position Actor(execute=@n70.Ok) {is=title K2Node_CallFunction}
   Update Marker Object state() {is=title K2Node_CustomEvent}
   > n71 = Is Valid(InputObject=Get MarkerObject())
   Branch(execute=@n71.Is Not Valid, Condition=IsChecked(self=Get CB_Toggle_SpawnMarkerObject()))
   > SetCheckedState(self=Get CB_Toggle_SpawnMarkerObject(), InCheckedState=Unchecked)
   > Set Text(Text="LOCTEXT(\"Spawn Marker Object\")", self=Get T_Toggle_SpawnMarkerObject())
   InitializeLayout_ArrangeInGridModule() {is=title K2Node_CustomEvent}
   > n72 = Branch(Condition=IsChecked(self=Get ToggleButton_Collapse_ArrangeInGrid() -> (ToggleButton_Collapse_ArrangeInGrid:Editor Utility Check Box Object Reference)))
   n73 = Event Tick()
   > Branch(Condition=IsChecked(self=Get CB_Toggle_PositionVisibility() -> (CB_Toggle_PositionVisibility:Editor Utility Check Box Object Reference)))
   > DrawDebugCoordinates!(
    execute
    self
    Draw Position:Vector=GetStoredPositions(From Pos Object=false)
    Delta T:double=@n73.InDeltaTime
    Center Color:Linear Color Structure=(R=1,G=0,B=1,A=1)
   ) -> (then) {is=title K2Node_CallFunction}
   > Update Marker Object state() {is=title K2Node_CallFunction}
   Event Construct()
   > n74 = Sequence!(execute) -> (then_0, then_1, then_2:Exec)
   InitializeLayout_ActorPositionUtilitiesModule(execute=@n74.then_1) {is=title K2Node_CallFunction}
   > InitializeLayout_ArrangeInGridModule() {is=title K2Node_CallFunction}
   > InitializeLayout_SettingsModule() {is=title K2Node_CallFunction}
   > InitializeLayout_ActorTransformsModule() {is=title K2Node_CallFunction}
   > InitializeLayout_ActorCleanupModule() {is=title K2Node_CallFunction}
   Initialize_Arranger(execute=@n74.then_2) {is=title K2Node_CallFunction}
   Clear!(
    execute=@n74.then_0
    TargetMap:Map of Actor Class References to E_FNEC_AssetOrganizer_ClassType Enums=Get DataClasses() -> (DataClasses:Map of Actor Class References to E_FNEC_AssetOrganizer_ClassType Enums)
   ) -> (then) {is=title K2Node_CallFunction}
   > n75 = Sequence!(execute) -> (then_0, then_1, then_2, then_3:Exec)
   n76 = For Each Loop(Exec=@n75.then_3, Array=Get ContentClasses(self=Get DataAsset()))
   > Add!(
    execute
    TargetMap:Map of Actor Class References to E_FNEC_AssetOrganizer_ClassType Enums=Get DataClasses()
    Key:"Actor Class Reference (by ref)"=@n76
    Value:"E_FNEC_AssetOrganizer_ClassType Enum (by ref)"=NewEnumerator2
   ) -> (then) {is=title K2Node_CallFunction}
   n77 = For Each Loop(
    Exec=@n75.then_2
    Array=Get DisallowedActorClasses!(self:DA FNEC Asset Organizer Data Asset Class Object Reference=Get DataAsset()) -> (DisallowedActorClasses:Array of Actor Class References)
   )
   > Add(TargetMap=Get DataClasses(), Key=@n77, Value=NewEnumerator1)
   n78 = For Each Loop(Exec=@n75.then_1, Array=Get FlaggedContentClasses(self=Get DataAsset()))
   > Add(TargetMap=Get DataClasses(), Key=@n78, Value=NewEnumerator4)
   n79 = For Each Loop(
    Exec=@n75.then_0
    Array=Get GalleryActorClasses!(self:DA FNEC Asset Organizer Data Asset Class Object Reference=Get DataAsset()) -> (GalleryActorClasses:Array of Actor Class References)
   )
   > Add(TargetMap=Get DataClasses(), Key=@n79, Value=NewEnumerator0)
   Event Destruct()
   > Deconstruct Position Actor() {is=title K2Node_CallFunction}
   n80 = Get Socket_ActorCleanupTools()
   n81 = Construct Object from Class(Class=/Game/ArtTools/FNEC/AssetOrganizer/Widgets/EUW_AssetOrganizer_ActorCleanup.EUW_AssetOrganizer_ActorCleanup_C, self=@n80) -> (ReturnValue:EUW Asset Organizer Actor Cleanup Object Reference)
   > AddChild(self=@n80, Content=@n81)
   Reroute Node()
   > RemoveChildAt!(
    execute
    self:Panel Widget Object Reference=Get VerticalBox-MoveToPosition() -> (VerticalBox-MoveToPosition:Vertical Box Object Reference)
    Index:int=2
   ) -> (then, ReturnValue:bool)
   > n82 = Construct Object from Class(Class=/Game/ArtTools/FNEC/AssetOrganizer/Widgets/EUW_AssetOrganizer_MoveSelected_ToXYZ.EUW_AssetOrganizer_MoveSelected_ToXYZ_C) -> (ReturnValue:EUW Asset Organizer Move Selected To XYZ Object Reference)
   > n83 = AddChild(self=Get VerticalBox-MoveToPosition(), Content=@n82)
   > SetPadding!(
    execute
    self:Vertical Box Slot Object Reference=Cast To VerticalBoxSlot!(execute, Object=@n83) -> (then, CastFailed, AsVertical Box Slot:Vertical Box Slot Object Reference) {is=title K2Node_DynamicCast}
    InPadding_Left:float=0
    InPadding_Top:float=10
    InPadding_Right:float=0
    InPadding_Bottom:float=0
   ) -> (then)
   GetWorld!(self:Layers Subsystem Object Reference=Get LayersSubsystem() -> (ReturnValue:Layers Subsystem Object Reference) {is=title K2Node_GetEditorSubsystem}) -> (ReturnValue:World Object Reference)
   GetFolderPath!(self:Actor Object Reference) -> (ReturnValue:Name=None)
   SetFolderPath!(execute, self:Actor Object Reference, NewFolderPath:"Name (by ref)"=MakeLiteralName(Value=Processed)) -> (then)
   SelectAll!(execute, self:Editor Actor Subsystem Object Reference, InWorld:World Object Reference=GetEditorWorld(self=Get UnrealEditorSubsystem())) -> (then)
   SetStyle()
   ExecutePythonCommand(PythonCommand="To String (Text)"(InText=Format Text(Format="LOCTEXT(\"import webbrowser\\r\\n\\r\\nwebbrowser.open(\\'https://ww.google.com\\',new=2)\")")))
   BeginTransaction(Context=Asset Organizer, Description="LOCTEXT(\"Stuff\")")
   EndTransaction()
   TransactObject()
   "On Clicked (B_Test)"() {is=title K2Node_ComponentBoundEvent}
   Reroute Node(InputPin:Wildcard) -> (OutputPin:Wildcard)
   DeselectElement!(
    execute:Exec
    self:Object Reference=Get GameViewportSubsystem() -> (ReturnValue:Game Viewport Subsystem Object Reference) {is=title K2Node_GetEngineSubsystem}
    InElementHandle:"Script Typed Element Handle Structure (by ref)"
    InSelectionSet:Typed Element List Structure
    InSelectionOptions:"Typed Element Selection Options Structure (by ref)"
   ) -> (then:Exec, ReturnValue:bool) {is=title K2Node_Message}
   Get bMirrored!(self:Building Static Mesh Object Reference=Cast To BuildingProp!(execute, Object=@n2) -> (then, CastFailed, AsBuilding Prop:Building Prop Object Reference) {is=title K2Node_DynamicCast}) -> (bMirrored:bool)
   Get Style_GrayButton_Toggled()
   Get Style_GrayButton_Default()
   Get B_Arrange_Distances_Custom()
   Get B_Arrange_Distances_Walls()
   Get B_Arrange_Distances_SmallProps()
   Get B_Arrange_Distances_Props()
   Get B_Arrange_Distances_Floors()
  ]}
 FindParentClass {nodes [
   n1 = FindParentClass!() -> (then, ActorClass:Actor Class Reference, ParentClasses:Array of Actor Class References) {is=title K2Node_FunctionEntry}
   > Set ParentFound(ParentFound:bool=false) -> (Output_Get:bool)
   > n2 = For Each Loop with Break(Array=@n1.ParentClasses, Break=@n4.then)
   > n3 = Set ParentFound(ParentFound=ClassIsChildOf(TestClass=Get ActorClass() -> (ActorClass:Actor Class Reference), ParentClass=@n2))
   > Branch(Condition=@n3)
   > n4 = Set FoundClass(FoundClass:Actor Class Reference=@n2) -> (Output_Get:Actor Class Reference)
   Return Node!(
    execute=@n2.Completed
    ParentClass:Actor Class Reference=Get FoundClass() -> (FoundClass:Actor Class Reference)
    Found:bool=Get ParentFound() -> (ParentFound:bool)
   )
  ]}
 FindParentClassInDataAsset {nodes [
   FindParentClassInDataAsset!() -> (then, ActorClass:Actor Class Reference) {is=title K2Node_FunctionEntry}
   > Set ParentFound(ParentFound=false)
   > n1 = "For Each Loop (Map)"!(execute:Exec, MapPin:Map of Actor Class References to E_FNEC_AssetOrganizer_ClassType Enums=Get DataClasses(), BreakPin:Exec) -> (then:Exec, KeyPin:Actor Class Reference, ValuePin:E_FNEC_AssetOrganizer_ClassType Enum, CompletedPin:Exec)
   > n2 = Branch(Condition=Get ParentFound())
   Branch(execute=@n2.else, Condition=ClassIsChildOf(TestClass=Get ActorClass(), ParentClass=@n1))
   > Set ParentFound(ParentFound=true)
   > Set FoundClass(FoundClass=@n1)
   > Set FoundClassCategory(FoundClassCategory:E_FNEC_AssetOrganizer_ClassType Enum=@n1.ValuePin) -> (Output_Get:E_FNEC_AssetOrganizer_ClassType Enum=NewEnumerator0)
   Return Node!(
    execute=@n1.CompletedPin
    ParentClass=Get FoundClass()
    Found=Get ParentFound()
    Map Value:E_FNEC_AssetOrganizer_ClassType Enum=Get FoundClassCategory() -> (FoundClassCategory:E_FNEC_AssetOrganizer_ClassType Enum=NewEnumerator0)
   )
  ]}
 GetStoredPositions {nodes [
   n1 = GetStoredPositions!() -> (then, From Pos Object:bool=~) {is=title K2Node_FunctionEntry}
   > n2 = Branch(Condition=@n1)
   > n3 = Branch(Condition=IsChecked(self=Get CB_Toggle_SpawnMarkerObject()))
   > n4 = Is Valid(InputObject=Get MarkerObject())
   > Return Node!(execute, Position:Vector=Get Actor Location(self=Get MarkerObject()))
   SetCheckedState(execute=@n4.Is Not Valid, self=Get CB_Toggle_SpawnMarkerObject(), InCheckedState=Unchecked)
   > Return Node(Position=Get StoredPosition())
   Return Node(execute=[@n3.else, @n2.else], Position=Get StoredPosition())
  ]}
 Get_T_StoredPostionDisplay_Text {nodes [
   Get_T_StoredPostionDisplay_Text() {is=title K2Node_FunctionEntry}
   n1 = Get StoredPosition()
   > Return Node!(execute, ReturnValue:Text="To Text (String)"(InString="To String (Vector)"!(InVec:Vector=@n1) -> (ReturnValue:String)))
  ]}
 LoadAllPositions {nodes [
   LoadAllPositions() {is=title K2Node_FunctionEntry}
   > n1 = Show Message Dialog(
    Title=MakeLiteralText(Value="LOCTEXT(\"Asset Organizer Tool - Load Positions\")")
    Message=MakeLiteralText(Value="LOCTEXT(\"You are about to load the position of selected actors. This operation cannot be undone. Are you sure?\")")
    MessageType=YesNo
    DefaultValue=No
    MessageCategory=Warning
   )
   > Branch(Condition="Equal (Enum)"(A:EAppReturnType Enum=@n1, B:EAppReturnType Enum=Yes))
   > n2 = GetSelectionSet()
   > n3 = For Each Loop(Array:Array of Actor Object References=@n2) -> (Array Element:Actor Object Reference)
   n4 = Get ActorPositions() -> (ActorPositions:Map of Actor Object References to Vectors)
   > Branch(Condition=Contains!(TargetMap:Map of Actor Object References to Vectors=@n4, Key:"Actor Object Reference (by ref)"=@n3) -> (ReturnValue))
   > Set Actor Location(
    self=@n3
    NewLocation=Find(TargetMap:Map of Actor Object References to Vectors=@n4, Key:"Actor Object Reference (by ref)"=@n3) -> (Value:Vector)
    bSweep=false
    bTeleport=false
   )
  ]}
 PrintStringWithToggle {nodes [
   n1 = PrintStringWithToggle!() -> (then, Print:bool=~, InString:String) {is=title K2Node_FunctionEntry}
   > Branch(Condition=@n1)
   > PrintString(InString=@n1.InString)
  ]}
 Reposition Group to XYZ {nodes [
   Reposition Group to XYZ!() -> (then, Target:Vector=~, List of Actors:Array of Actor Object References, Write to Undo History:bool) {is=title K2Node_FunctionEntry}
   > n1 = Branch(Condition=Get Write to Undo History())
   > n2 = BeginTransaction(Context=Asset Organizer, Description="LOCTEXT(\"Reposition Group of Actors\")")
   Set Corner(execute=[@n2.then, @n1.else], Corner:Vector="-999999,999999,999999") -> (Output_Get:Vector)
   > n3 = Sequence!(execute) -> (then_0, then_1, then_2)
   n4 = Get List of Actors()
   > n5 = For Each Loop(Array=@n4)
   n6 = Get Actor Location!(self=@n5) -> (ReturnValue_X:double, ReturnValue_Y:double, ReturnValue_Z:double)
   n7 = BreakVector(InVec=Get Corner() -> (Corner:Vector))
   > n8 = Set Corner(Corner=MakeVector(X="Max (Float)"(A=@n6, B=@n7), Y="Min (Float)"(A=@n6.ReturnValue_Y, B=@n7.Y), Z="Min (Float)"(A=@n6.ReturnValue_Z, B=@n7.Z)))
   > Set DistanceToMove(DistanceToMove:Vector=vector - vector(A=Get Target() -> (Target:Vector), B=@n8)) -> (Output_Get:Vector)
   n9 = For Each Loop(Exec=@n3.then_1, Array=@n4)
   > n10 = Branch(Condition=Get Write to Undo History())
   > n11 = TransactObject(Object=@n9)
   Set Actor Location(
    execute=[@n11.then, @n10.else]
    self=@n9
    NewLocation=vector + vector(A=Get Actor Location!(self=@n9) -> (ReturnValue:Vector), B=Get DistanceToMove() -> (DistanceToMove:Vector))
    bSweep=false
    bTeleport=false
   )
   Branch(execute=@n3.then_2, Condition=Get Write to Undo History())
   > EndTransaction()
  ]}
 ResetRotationOfActorArray {nodes [
   ResetRotationOfActorArray!() -> (then, List of Actors:Array of Actor Object References, Write to Undo History:bool=~) {is=title K2Node_FunctionEntry}
   > n1 = Branch(Condition=Get Write to Undo History())
   > n2 = BeginTransaction(Context=Asset Organizer, Description="LOCTEXT(\"Reset Rotation of Actors\")")
   n3 = For Each Loop(Exec=[@n2.then, @n1.else], Array=Get List of Actors())
   > n4 = Branch(Condition=Get Write to Undo History())
   > n5 = TransactObject(Object=@n3)
   Set Actor Rotation(execute=[@n5.then, @n4.else], self=@n3, NewRotation="0, 0, 0", bTeleportPhysics=false)
   Branch(execute=@n3.Completed, Condition=Get Write to Undo History())
   > EndTransaction()
  ]}
 ResetScaleOfActorArray {nodes [
   ResetScaleOfActorArray!() -> (then, List of Actors:Array of Actor Object References, Write to Undo History:bool=~) {is=title K2Node_FunctionEntry}
   > n1 = Branch(Condition=Get Write to Undo History())
   > n2 = BeginTransaction(Context=Asset Organizer, Description="LOCTEXT(\"Reset Scale of Actors\")")
   n3 = For Each Loop(Exec=[@n2.then, @n1.else], Array=Get List of Actors())
   > n4 = Branch(Condition=Get Write to Undo History())
   > n5 = TransactObject(Object=@n3)
   SetActorScale3D(execute=[@n5.then, @n4.else], self=@n3, NewScale3D="1,1,1")
   Branch(execute=@n3.Completed, Condition=Get Write to Undo History())
   > EndTransaction()
  ]}
 SaveAllPositions {nodes [
   SaveAllPositions() {is=title K2Node_FunctionEntry}
   > n1 = Show Message Dialog(
    Title=MakeLiteralText(Value="LOCTEXT(\"Asset Organizer Tool - Save Positions\")")
    Message=MakeLiteralText(Value="LOCTEXT(\"Save positions of selected actors?\")")
    MessageType=YesNo
    DefaultValue=No
    MessageCategory=Warning
   )
   > Branch(Condition="Equal (Enum)"(A=@n1, B=Yes))
   > n2 = GetSelectionSet()
   > n3 = For Each Loop(Array=@n2)
   > Add(
    TargetMap:Map of Actor Object References to Vectors=Get ActorPositions()
    Key:"Actor Object Reference (by ref)"=@n3
    Value:"Vector (by ref)"=Get Actor Location(self=@n3)
   )
  ]}
 SelectAllActorsOfClass {nodes [
   SelectAllActorsOfClass() {is=title K2Node_FunctionEntry}
   > n1 = GetSelectionSet()
   n2 = Get Class(Object="Get (a copy)"(Array=@n1))
   > n3 = GetAllActorsOfClass(ActorClass=@n2)
   > n4 = For Each Loop(Array=@n3)
   > Branch(Condition="Equal (Class)"(A=Get Class(Object=@n4), B=@n2))
   > Add(
    TargetArray:Array of Actor Object References=Get SelectedActors() -> (SelectedActors:Array of Actor Object References)
    NewItem:"Actor Object Reference (by ref)"=@n4
   ) {is=title K2Node_CallArrayFunction}
   n5 = Get EditorActorSubsystem()
   ClearActorSelectionSet(execute=@n4.Completed, self=@n5)
   > SetSelectedLevelActors(self=@n5, ActorsToSelect=Get SelectedActors())
  ]}
 SetGridSize {nodes [
   n1 = SetGridSize!() -> (then, X:double=~, Y:double=~) {is=title K2Node_FunctionEntry}
   > Set Text(
    Text="To Text (Float)"(Value=@n1, MinimumIntegralDigits=1, MaximumIntegralDigits=324, MinimumFractionalDigits=0, MaximumFractionalDigits=3)
    self:Editable Text Object Reference=Get T_Arrange_X()
   )
   > Set Text(
    Text="To Text (Float)"(Value=@n1.Y, MinimumIntegralDigits=1, MaximumIntegralDigits=324, MinimumFractionalDigits=0, MaximumFractionalDigits=3)
    self=Get T_Arrange_Y()
   )
  ]}
 Sort By Size {nodes [
   n1 = For Each Loop with Break(Exec=@n10.then, Array:Array of Actor Object References=Get ActorInOrder() -> (ActorInOrder:Array of Actor Object References), Break=@n6.then) -> (Array Element:Actor Object Reference)
   n2 = Get CurrentObject() -> (CurrentObject:Actor Object Reference)
   n3 = PrintString(InString=Append!(A="comparing ", B=GetDisplayName(Object=@n2), C:String=" to ", D:String=GetDisplayName(Object=@n1)) -> (ReturnValue))
   n4 = Get CurrentObject()
   n5 = Compare Float!(
    Exec:Exec=[@n3.then, @n1.LoopBody]
    Input:double=Vector Length(A=Get Actor Bounds(self=@n4, bOnlyCollidingComponents=false).BoxExtent)
    CompareWith:double=Vector Length(A=Get Actor Bounds(self=@n1, bOnlyCollidingComponents=false).BoxExtent)
   ) -> (">":Exec, "==":Exec, "<":Exec) {is=title K2Node_MacroInstance}
   Set InsertIndex(execute=@n5."<", InsertIndex:int=@n1.Array Index) -> (Output_Get:int=0)
   > n6 = Set FoundSpot(FoundSpot:bool=true) -> (Output_Get:bool)
   n7 = Sort By Size!() -> (then, List of Actors:Array of Actor Object References) {is=title K2Node_FunctionEntry}
   > n8 = For Each Loop(Array=@n7)
   Return Node!(execute=@n8.Completed, OrderedActors:Array of Actor Object References=Get ActorInOrder())
   > Set CurrentObject(CurrentObject:Actor Object Reference=@n8) -> (Output_Get:Actor Object Reference)
   > n9 = Branch(Condition="Equal (Integer)"(A=@n8.Array Index))
   > Add(TargetArray=Get ActorInOrder(), NewItem=Get CurrentObject())
   n10 = Set FoundSpot(execute=@n9.else, FoundSpot=false)
   n11 = For Each Loop()
   > PrintString(InString=GetDisplayName(Object=@n11))
   Add(execute=@n12.else, TargetArray=Get ActorInOrder(), NewItem=Get CurrentObject())
   Insert!(
    execute=@n12.then
    TargetArray:Array of Actor Object References=Get ActorInOrder()
    NewItem:"Actor Object Reference (by ref)"=Get CurrentObject()
    Index:int=Get InsertIndex() -> (InsertIndex:int=0)
   ) -> (then) {is=title K2Node_CallArrayFunction}
   PrintString(InString="It's smaller!")
   n12 = Branch(execute=@n1.Completed, Condition=Get FoundSpot() -> (FoundSpot:bool))
  ]}
 Swap {nodes [
   Swap() {is=title K2Node_FunctionEntry}
   > BeginTransaction(Context=Asset Organizer, Description="LOCTEXT(\"Swap Actor locations\")")
   > n1 = Sequence!(execute) -> (then_0, then_1)
   > n2 = GetSelectionSet()
   n3 = Length(TargetArray=@n2)
   > Set Actor Count(Actor Count:int=@n3) -> (Output_Get:int=0)
   > n4 = Branch(Condition="Equal (Integer)"(A="% (Integer)"(A=@n3, B=2) -> (ReturnValue=0)))
   n5 = int / int(A=Get Actor Count() -> (Actor Count:int=0), B=2)
   > n6 = For Loop(LastIndex=int - int(A=@n5))
   > n7 = GetSelectionSet()
   n8 = "Get (a copy)"(Array=@n7, Dimension 1=@n6)
   n9 = "Get (a copy)"(Array=@n7, Dimension 1=int + int(A=@n6, B=@n5))
   > Branch(Condition=AND Boolean(A=IsValid(Object=@n9), B=IsValid(Object=@n8)))
   > Set TempPosition(TempPosition:Vector=Get Actor Location(self=@n9)) -> (Output_Get:Vector)
   > TransactObject(Object=@n9)
   > Set Actor Location(self=@n9, NewLocation=Get Actor Location(self=@n8), bSweep=false, bTeleport=false)
   > TransactObject(Object=@n8)
   > Set Actor Location(self=@n8, NewLocation=Get TempPosition() -> (TempPosition:Vector), bSweep=false, bTeleport=false)
   Show Message Dialog(
    execute=@n4.else
    Title=MakeLiteralText(Value="LOCTEXT(\"Asset Organizer Tool - Swap Actors\")")
    Message=MakeLiteralText(Value="LOCTEXT(\"Odd number of actors selected. Cannot swap.\")")
    MessageType=Ok
    DefaultValue=No
    MessageCategory=Warning
   )
   EndTransaction(execute=@n1.then_1)
  ]}
 XMirror {nodes [
   XMirror!() -> (then, List of Actors:Array of Actor Object References) {is=title K2Node_FunctionEntry}
   > Set Center X(Center X:double=0) -> (Output_Get:double)
   > n1 = Sequence()
   > n2 = Sequence()
   n3 = For Each Loop(Exec=@n2.then_1, Array=Get List of Actors())
   n4 = Get Maximum X() -> (Maximum X:double)
   n5 = Get Minimum X() -> (Minimum X:double)
   n6 = Set Center X(execute=@n3.Completed, Center X=float / float(A=float + float(A=@n4, B=@n5), B=2))
   n7 = Get Maximum X()
   n8 = Get Actor Location!(self=@n3) -> (ReturnValue_X:double, ReturnValue_Y:double, ReturnValue_Z:double)
   Set Maximum X(execute=@n3.LoopBody, Maximum X:double=SelectFloat(A=@n8, B=@n7, bPickA="float > float"(A=@n8, B=@n7))) -> (Output_Get:double)
   n9 = Get Minimum X()
   n10 = Get Actor Location(self=@n3)
   > Set Minimum X(Minimum X:double=SelectFloat(A=@n10, B=@n9, bPickA="float < float"(A=@n10, B=@n9))) -> (Output_Get:double)
   n11 = Get Actor Location(self="Get (a copy)"(Array=Get List of Actors()))
   Set Maximum X(execute=@n2.then_0, Maximum X=@n11)
   > Set Minimum X(Minimum X=@n11)
   n12 = For Each Loop(Exec=@n1.then_1, Array=Get List of Actors())
   n13 = float * float(A:double=float - float(A=Get Center X() -> (Center X:double), B=Get Actor Location(self=@n12)), B=2)
   > Set Actor Location(
    self=@n12
    NewLocation=vector + vector(A=Get Actor Location!(self=@n12) -> (ReturnValue:Vector), B=MakeVector(X=@n13, Y=0))
    bSweep=false
    bTeleport=false
   )
   n14 = Get Minimum X()
   n15 = Get Maximum X()
   PrintString(InString=Append!(
     A="Minimum : "
     B="To String (Float)"(InDouble=@n14)
     C="Maximum : "
     D="To String (Float)"(InDouble=@n15)
     E:String="Center : "
     F:String="To String (Float)"(InDouble=@n6)
    ) -> (ReturnValue))
  ]}
]
variables [
 ActorPositions {default {}}
 ArrangeOrigin {type=E_FNEC_AssetOrganizer_GridOrigin, default=From Corner}
 Color_DisabledText {type=LinearColor, default=<rgba 0.0176 0.0176 0.0176 1>}
 Color_EnabledText {type=LinearColor, default=<rgba 0.01 0.01 0.01 1>}
 DataAsset {type=DA_FNEC_AssetOrganizer_DataAssetClass_C, default=&/DA_AssetOrganizerTools}
 DataClasses {default {}}
 DocsURL {type=string, default=http://www.google.com/ncr}
 InspectorWidget {type=EUW_AssetOrganizer_Inspector_C, default=null}
 MarkerObject {type=Actor, default=null}
 StoredPosition {type=Vector}
 Style_GrayButton_Default {
  type=ButtonStyle
  default {
   clickedSlateSound {resourceObject=null}
   disabled {
    bIsDynamicallyLoaded=false
    drawAs=RoundedBox
    imageType=NoImage
    mirroring=NoMirror
    resourceName=None
    resourceObject=null
    tiling=NoTile
    imageSize=<0 0>
    margin=<ltrb 0 0 0 0>
    outlineSettings {
     bUseBrushTransparency=false
     roundingType=FixedRadius
     width=1
     color {colorUseRule=UseColor_ColorTable, specifiedColor=<rgba 0 0 0 0>}
     cornerRadii=<4 4 4 4>
    }
    tintColor {colorUseRule=UseColor_Specified, specifiedColor=<rgba 0.0156 0.0156 0.0156 1>}
    uVRegion {bIsValid=false, max=<0 0>, min=<0 0>}
   }
   disabledForeground {colorUseRule=UseColor_ColorTable, specifiedColor=<rgba 0 0 0 0>}
   hovered {
    bIsDynamicallyLoaded=false
    drawAs=RoundedBox
    imageType=NoImage
    mirroring=NoMirror
    resourceName=None
    resourceObject=null
    tiling=NoTile
    imageSize=<32 32>
    margin=<ltrb 0 0 0 0>
    outlineSettings {
     bUseBrushTransparency=false
     roundingType=FixedRadius
     width=1
     color {colorUseRule=UseColor_ColorTable, specifiedColor=<rgba 0 0 0 0>}
     cornerRadii=<4 4 4 4>
    }
    tintColor {colorUseRule=UseColor_Specified, specifiedColor=<rgba 0.0953 0.0953 0.0953 1>}
    uVRegion {bIsValid=false, max=<0 0>, min=<0 0>}
   }
   hoveredForeground {colorUseRule=UseColor_ColorTable, specifiedColor=<rgba 0 0 0 0>}
   hoveredSlateSound {resourceObject=null}
   normal {
    bIsDynamicallyLoaded=false
    drawAs=RoundedBox
    imageType=NoImage
    mirroring=NoMirror
    resourceName=None
    resourceObject=null
    tiling=NoTile
    imageSize=<32 32>
    margin=<ltrb 0 0 0 0>
    outlineSettings {
     bUseBrushTransparency=false
     roundingType=FixedRadius
     width=1
     color {colorUseRule=UseColor_ColorTable, specifiedColor=<rgba 0 0 0 0>}
     cornerRadii=<4 4 4 4>
    }
    tintColor {colorUseRule=UseColor_Specified, specifiedColor=<rgba 0.0395 0.0395 0.0395 1>}
    uVRegion {bIsValid=false, max=<0 0>, min=<0 0>}
   }
   normalForeground {colorUseRule=UseColor_ColorTable, specifiedColor=<rgba 0 0 0 0>}
   normalPadding=<ltrb 12 1.5 12 1.5>
   pressed {
    bIsDynamicallyLoaded=false
    drawAs=RoundedBox
    imageType=NoImage
    mirroring=NoMirror
    resourceName=None
    resourceObject=null
    tiling=NoTile
    imageSize=<32 32>
    margin=<ltrb 0 0 0 0>
    outlineSettings {
     bUseBrushTransparency=false
     roundingType=FixedRadius
     width=1
     color {colorUseRule=UseColor_ColorTable, specifiedColor=<rgba 0 0 0 0>}
     cornerRadii=<4 4 4 4>
    }
    tintColor {colorUseRule=UseColor_Specified, specifiedColor=<rgba 0.0284 0.0284 0.0284 1>}
    uVRegion {bIsValid=false, max=<0 0>, min=<0 0>}
   }
   pressedForeground {colorUseRule=UseColor_ColorTable, specifiedColor=<rgba 0 0 0 0>}
   pressedPadding=<ltrb 12 2.5 12 0.5>
   pressedSlateSound {resourceObject=null}
  }
 }
 Style_GrayButton_Toggled {
  type=ButtonStyle
  default {
   clickedSlateSound {resourceObject=null}
   disabled {
    bIsDynamicallyLoaded=false
    drawAs=RoundedBox
    imageType=NoImage
    mirroring=NoMirror
    resourceName=None
    resourceObject=null
    tiling=NoTile
    imageSize=<0 0>
    margin=<ltrb 0 0 0 0>
    outlineSettings {
     bUseBrushTransparency=false
     roundingType=FixedRadius
     width=1
     color {colorUseRule=UseColor_ColorTable, specifiedColor=<rgba 0 0 0 0>}
     cornerRadii=<4 4 4 4>
    }
    tintColor {colorUseRule=UseColor_Specified, specifiedColor=<rgba 0.0156 0.0156 0.0156 1>}
    uVRegion {bIsValid=false, max=<0 0>, min=<0 0>}
   }
   disabledForeground {colorUseRule=UseColor_ColorTable, specifiedColor=<rgba 0 0 0 0>}
   hovered {
    bIsDynamicallyLoaded=false
    drawAs=RoundedBox
    imageType=NoImage
    mirroring=NoMirror
    resourceName=None
    resourceObject=null
    tiling=NoTile
    imageSize=<32 32>
    margin=<ltrb 0 0 0 0>
    outlineSettings {
     bUseBrushTransparency=false
     roundingType=FixedRadius
     width=1
     color {colorUseRule=UseColor_ColorTable, specifiedColor=<rgba 0 0 0 0>}
     cornerRadii=<4 4 4 4>
    }
    tintColor {colorUseRule=UseColor_Specified, specifiedColor=<rgba 0.0477 0.2366 0.9167 1>}
    uVRegion {bIsValid=false, max=<0 0>, min=<0 0>}
   }
   hoveredForeground {colorUseRule=UseColor_ColorTable, specifiedColor=<rgba 0 0 0 0>}
   hoveredSlateSound {resourceObject=null}
   normal {
    bIsDynamicallyLoaded=false
    drawAs=RoundedBox
    imageType=NoImage
    mirroring=NoMirror
    resourceName=None
    resourceObject=null
    tiling=NoTile
    imageSize=<32 32>
    margin=<ltrb 0 0 0 0>
    outlineSettings {
     bUseBrushTransparency=false
     roundingType=FixedRadius
     width=1
     color {colorUseRule=UseColor_ColorTable, specifiedColor=<rgba 0 0 0 0>}
     cornerRadii=<4 4 4 4>
    }
    tintColor {colorUseRule=UseColor_Specified, specifiedColor=<rgba 0 0.162 0.7454 1>}
    uVRegion {bIsValid=false, max=<0 0>, min=<0 0>}
   }
   normalForeground {colorUseRule=UseColor_ColorTable, specifiedColor=<rgba 0 0 0 0>}
   normalPadding=<ltrb 12 1.5 12 1.5>
   pressed {
    bIsDynamicallyLoaded=false
    drawAs=RoundedBox
    imageType=NoImage
    mirroring=NoMirror
    resourceName=None
    resourceObject=null
    tiling=NoTile
    imageSize=<32 32>
    margin=<ltrb 0 0 0 0>
    outlineSettings {
     bUseBrushTransparency=false
     roundingType=FixedRadius
     width=1
     color {colorUseRule=UseColor_ColorTable, specifiedColor=<rgba 0 0 0 0>}
     cornerRadii=<4 4 4 4>
    }
    tintColor {colorUseRule=UseColor_Specified, specifiedColor=<rgba 0 0.0679 0.3125 1>}
    uVRegion {bIsValid=false, max=<0 0>, min=<0 0>}
   }
   pressedForeground {colorUseRule=UseColor_ColorTable, specifiedColor=<rgba 0 0 0 0>}
   pressedPadding=<ltrb 12 2.5 12 0.5>
   pressedSlateSound {resourceObject=null}
  }
 }
]
widgets {
 info {rootWidgetClass=&StackBox, widgetCount=134}
 widgets [
  StackBox_0 {widgetClassPath=&StackBox, widget=StackBox_0}
  Border_1 {
   widgetClassPath=&Border
   parent=StackBox_0
   widget=Border_1
   properties {
    background {
     bIsDynamicallyLoaded=false
     drawAs=Box
     imageType=FullColor
     mirroring=NoMirror
     resourceName=None
     resourceObject=/Game/ArtTools/FNEC/AssetOrganizer/Materials/M_UI_Border
     tiling=NoTile
     imageSize=<32 32>
     margin=<ltrb 0.5 0.5 0.5 0.5>
     outlineSettings {
      bUseBrushTransparency=false
      roundingType=HalfHeightRadius
      width=0
      color {colorUseRule=UseColor_Specified, specifiedColor=<rgba 0 0 0 0>}
      cornerRadii=<0 0 0 0>
     }
     tintColor {colorUseRule=UseColor_Specified, specifiedColor=<rgba 1 0.3875 0 0>}
     uVRegion {bIsValid=false, max=<0 0>, min=<0 0>}
    }
    padding=<ltrb 10 10 10 10>
   }
   slot {class=StackBoxSlot, properties {size {sizeRule=Fill, value=1}}}
  }
  VerticalBox_1 {
   widgetClassPath=&VerticalBox
   parent=Border_1
   widget=VerticalBox_1
   slot {class=BorderSlot, properties {padding=<ltrb 10 10 10 10>}}
  }
  Overlay_0 {
   widgetClassPath=&Overlay
   parent=VerticalBox_1
   widget=Overlay_0
   slot {class=VerticalBoxSlot, properties {verticalAlignment=VAlign_Top, padding=<ltrb 0 0 0 10>}}
  }
  ToolTitle {
   widgetClassPath=&TextBlock
   parent=Overlay_0
   widget=ToolTitle
   properties {
    text=Asset Organizer Tools
    font {
     bForceMonospaced=false
     bMaterialIsStencil=false
     fontMaterial=null
     fontObject=&Roboto
     letterSpacing=0
     monospacedWidth=1
     size=12
     skewAmount=0
     typefaceFontName=Bold
     outlineSettings {
      bApplyOutlineToDropShadows=false
      bMiteredCorners=false
      bSeparateFillAlpha=false
      outlineMaterial=null
      outlineSize=0
      outlineColor=<rgba 0 0 0 1>
     }
    }
   }
   slot {class=OverlaySlot, properties {horizontalAlignment=HAlign_Center}}
  }
  SizeBox_4 {
   widgetClassPath=&SizeBox
   parent=Overlay_0
   widget=SizeBox_4
   properties {
    bOverride_MaxDesiredHeight=true
    bOverride_MaxDesiredWidth=true
    bOverride_MinDesiredHeight=true
    bOverride_MinDesiredWidth=true
    maxDesiredHeight=20
    maxDesiredWidth=20
    minDesiredHeight=20
    minDesiredWidth=20
   }
   slot {class=OverlaySlot, properties {horizontalAlignment=HAlign_Right, verticalAlignment=VAlign_Center}}
  }
  B_Info {
   widgetClassPath=&EditorUtilityButton
   parent=SizeBox_4
   widget=B_Info
   properties {
    toolTipText=Open the web documentation for this tool
    widgetStyle {
     clickedSlateSound {resourceObject=null}
     disabled {
      bIsDynamicallyLoaded=false
      drawAs=RoundedBox
      imageType=NoImage
      mirroring=NoMirror
      resourceName=None
      resourceObject=null
      tiling=NoTile
      imageSize=<0 0>
      margin=<ltrb 0 0 0 0>
      outlineSettings {
       bUseBrushTransparency=false
       roundingType=FixedRadius
       width=1
       color {colorUseRule=UseColor_ColorTable, specifiedColor=<rgba 0 0 0 0>}
       cornerRadii=<4 4 4 4>
      }
      tintColor {colorUseRule=UseColor_Specified, specifiedColor=<rgba 0.0395 0.0395 0.0395 1>}
      uVRegion {bIsValid=false, max=<0 0>, min=<0 0>}
     }
     disabledForeground {colorUseRule=UseColor_ColorTable, specifiedColor=<rgba 0 0 0 0>}
     hovered {
      bIsDynamicallyLoaded=false
      drawAs=RoundedBox
      imageType=NoImage
      mirroring=NoMirror
      resourceName=None
      resourceObject=null
      tiling=NoTile
      imageSize=<32 32>
      margin=<ltrb 0 0 0 0>
      outlineSettings {
       bUseBrushTransparency=false
       roundingType=FixedRadius
       width=1
       color {colorUseRule=UseColor_ColorTable, specifiedColor=<rgba 0 0 0 0>}
       cornerRadii=<4 4 4 4>
      }
      tintColor {colorUseRule=UseColor_Specified, specifiedColor=<rgba 0.0953 0.0953 0.0953 1>}
      uVRegion {bIsValid=false, max=<0 0>, min=<0 0>}
     }
     hoveredForeground {colorUseRule=UseColor_ColorTable, specifiedColor=<rgba 0 0 0 0>}
     hoveredSlateSound {resourceObject=null}
     normal {
      bIsDynamicallyLoaded=false
      drawAs=RoundedBox
      imageType=NoImage
      mirroring=NoMirror
      resourceName=None
      resourceObject=null
      tiling=NoTile
      imageSize=<32 32>
      margin=<ltrb 0 0 0 0>
      outlineSettings {
       bUseBrushTransparency=false
       roundingType=FixedRadius
       width=1
       color {colorUseRule=UseColor_ColorTable, specifiedColor=<rgba 0 0 0 0>}
       cornerRadii=<4 4 4 4>
      }
      tintColor {colorUseRule=UseColor_Specified, specifiedColor=<rgba 0.0395 0.0395 0.0395 1>}
      uVRegion {bIsValid=false, max=<0 0>, min=<0 0>}
     }
     normalForeground {colorUseRule=UseColor_ColorTable, specifiedColor=<rgba 0 0 0 0>}
     normalPadding=<ltrb 0 0 0 0>
     pressed {
      bIsDynamicallyLoaded=false
      drawAs=RoundedBox
      imageType=NoImage
      mirroring=NoMirror
      resourceName=None
      resourceObject=null
      tiling=NoTile
      imageSize=<32 32>
      margin=<ltrb 0 0 0 0>
      outlineSettings {
       bUseBrushTransparency=false
       roundingType=FixedRadius
       width=1
       color {colorUseRule=UseColor_ColorTable, specifiedColor=<rgba 0 0 0 0>}
       cornerRadii=<4 4 4 4>
      }
      tintColor {colorUseRule=UseColor_Specified, specifiedColor=<rgba 0.0284 0.0284 0.0284 1>}
      uVRegion {bIsValid=false, max=<0 0>, min=<0 0>}
     }
     pressedForeground {colorUseRule=UseColor_ColorTable, specifiedColor=<rgba 0 0 0 0>}
     pressedPadding=<ltrb 0 0 0 0>
     pressedSlateSound {resourceObject=null}
    }
   }
   slot {class=SizeBoxSlot}
  }
  TextBlock_86 {
   widgetClassPath=&TextBlock
   parent=B_Info
   widget=TextBlock_86
   properties {
    justification=Center
    text=i
    font {
     bForceMonospaced=false
     bMaterialIsStencil=false
     fontMaterial=null
     fontObject=&Roboto
     letterSpacing=0
     monospacedWidth=1
     size=11.25
     skewAmount=0
     typefaceFontName=Bold
     outlineSettings {
      bApplyOutlineToDropShadows=false
      bMiteredCorners=false
      bSeparateFillAlpha=false
      outlineMaterial=null
      outlineSize=0
      outlineColor=<rgba 0 0 0 1>
     }
    }
   }
   slot {class=ButtonSlot}
  }
  HorizontalBox_8 {
   widgetClassPath=&HorizontalBox
   parent=VerticalBox_1
   widget=HorizontalBox_8
   slot {class=VerticalBoxSlot, properties {padding=<ltrb 10 10 10 10>}}
  }
  Border-Selection {
   widgetClassPath=&Border
   parent=HorizontalBox_8
   widget=Border-Selection
   properties {
    background {
     bIsDynamicallyLoaded=false
     drawAs=RoundedBox
     imageType=FullColor
     mirroring=NoMirror
     resourceName=None
     resourceObject=null
     tiling=NoTile
     imageSize=<32 32>
     margin=<ltrb 0.5 0.5 0.5 0.5>
     outlineSettings {
      bUseBrushTransparency=false
      roundingType=FixedRadius
      width=1.2219
      color {colorUseRule=UseColor_Specified, specifiedColor=<rgba 0 0 0 1>}
      cornerRadii=<5 5 5 5>
     }
     tintColor {colorUseRule=UseColor_Specified, specifiedColor=<rgba 0 0 0 0>}
     uVRegion {bIsValid=false, max=<0 0>, min=<0 0>}
    }
    padding=<ltrb 5 5 5 5>
   }
   slot {class=HorizontalBoxSlot, properties {size {sizeRule=Fill, value=1}}}
  }
  VerticalBox_7 {
   widgetClassPath=&VerticalBox
   parent=Border-Selection
   widget=VerticalBox_7
   slot {class=BorderSlot, properties {padding=<ltrb 5 5 5 5>}}
  }
  TextBlock_22 {
   widgetClassPath=&TextBlock
   parent=VerticalBox_7
   widget=TextBlock_22
   properties {
    text=Select Actors
    font {
     bForceMonospaced=false
     bMaterialIsStencil=false
     fontMaterial=null
     fontObject=&Roboto
     letterSpacing=0
     monospacedWidth=1
     size=9
     skewAmount=0
     typefaceFontName=Bold
     outlineSettings {
      bApplyOutlineToDropShadows=false
      bMiteredCorners=false
      bSeparateFillAlpha=false
      outlineMaterial=null
      outlineSize=0
      outlineColor=<rgba 0 0 0 1>
     }
    }
   }
   slot {class=VerticalBoxSlot, properties {horizontalAlignment=HAlign_Center, verticalAlignment=VAlign_Top}}
  }
  HorizontalBox_2 {
   widgetClassPath=&HorizontalBox
   parent=VerticalBox_7
   widget=HorizontalBox_2
   slot {class=VerticalBoxSlot, properties {horizontalAlignment=HAlign_Center, size {sizeRule=Fill, value=1}}}
  }
  B_Select_ByClass {
   widgetClassPath=&EditorUtilityButton
   parent=HorizontalBox_2
   widget=B_Select_ByClass
   properties {
    toolTipText="Select an actor in the scene then use this button to get all actors of the same class as selected actor. \r\n\r\nSimilar to the editor's 'Shift+E' funcionality, but more strict. \r\n\r\nIf multiple actors have been selected before using this action the class of the first actor will be used.\r\n\r\n(NOTE : It will select child classes as well)"
    widgetStyle {
     clickedSlateSound {resourceObject=null}
     disabled {
      bIsDynamicallyLoaded=false
      drawAs=RoundedBox
      imageType=NoImage
      mirroring=NoMirror
      resourceName=None
      resourceObject=null
      tiling=NoTile
      imageSize=<0 0>
      margin=<ltrb 0 0 0 0>
      outlineSettings {
       bUseBrushTransparency=false
       roundingType=FixedRadius
       width=1
       color {colorUseRule=UseColor_ColorTable, specifiedColor=<rgba 0 0 0 0>}
       cornerRadii=<4 4 4 4>
      }
      tintColor {colorUseRule=UseColor_Specified, specifiedColor=<rgba 0.0395 0.0395 0.0395 1>}
      uVRegion {bIsValid=false, max=<0 0>, min=<0 0>}
     }
     disabledForeground {colorUseRule=UseColor_ColorTable, specifiedColor=<rgba 0 0 0 0>}
     hovered {
      bIsDynamicallyLoaded=false
      drawAs=RoundedBox
      imageType=NoImage
      mirroring=NoMirror
      resourceName=None
      resourceObject=null
      tiling=NoTile
      imageSize=<32 32>
      margin=<ltrb 0 0 0 0>
      outlineSettings {
       bUseBrushTransparency=false
       roundingType=FixedRadius
       width=1
       color {colorUseRule=UseColor_ColorTable, specifiedColor=<rgba 0 0 0 0>}
       cornerRadii=<4 4 4 4>
      }
      tintColor {colorUseRule=UseColor_Specified, specifiedColor=<rgba 0.0953 0.0953 0.0953 1>}
      uVRegion {bIsValid=false, max=<0 0>, min=<0 0>}
     }
     hoveredForeground {colorUseRule=UseColor_ColorTable, specifiedColor=<rgba 0 0 0 0>}
     hoveredSlateSound {resourceObject=null}
     normal {
      bIsDynamicallyLoaded=false
      drawAs=RoundedBox
      imageType=NoImage
      mirroring=NoMirror
      resourceName=None
      resourceObject=null
      tiling=NoTile
      imageSize=<32 32>
      margin=<ltrb 0 0 0 0>
      outlineSettings {
       bUseBrushTransparency=false
       roundingType=FixedRadius
       width=1
       color {colorUseRule=UseColor_ColorTable, specifiedColor=<rgba 0 0 0 0>}
       cornerRadii=<4 4 4 4>
      }
      tintColor {colorUseRule=UseColor_Specified, specifiedColor=<rgba 0.0395 0.0395 0.0395 1>}
      uVRegion {bIsValid=false, max=<0 0>, min=<0 0>}
     }
     normalForeground {colorUseRule=UseColor_ColorTable, specifiedColor=<rgba 0 0 0 0>}
     normalPadding=<ltrb 12 1.5 12 1.5>
     pressed {
      bIsDynamicallyLoaded=false
      drawAs=RoundedBox
      imageType=NoImage
      mirroring=NoMirror
      resourceName=None
      resourceObject=null
      tiling=NoTile
      imageSize=<32 32>
      margin=<ltrb 0 0 0 0>
      outlineSettings {
       bUseBrushTransparency=false
       roundingType=FixedRadius
       width=1
       color {colorUseRule=UseColor_ColorTable, specifiedColor=<rgba 0 0 0 0>}
       cornerRadii=<4 4 4 4>
      }
      tintColor {colorUseRule=UseColor_Specified, specifiedColor=<rgba 0.0284 0.0284 0.0284 1>}
      uVRegion {bIsValid=false, max=<0 0>, min=<0 0>}
     }
     pressedForeground {colorUseRule=UseColor_ColorTable, specifiedColor=<rgba 0 0 0 0>}
     pressedPadding=<ltrb 12 2.5 12 0.5>
     pressedSlateSound {resourceObject=null}
    }
   }
   slot {class=HorizontalBoxSlot, properties {padding=<ltrb 5 0 5 0>}}
  }
  TextBlock_9 {
   widgetClassPath=&TextBlock
   parent=B_Select_ByClass
   widget=TextBlock_9
   properties {
    text=Same Class
    font {
     bForceMonospaced=false
     bMaterialIsStencil=false
     fontMaterial=null
     fontObject=&Roboto
     letterSpacing=0
     monospacedWidth=1
     size=9
     skewAmount=0
     typefaceFontName=Regular
     outlineSettings {
      bApplyOutlineToDropShadows=false
      bMiteredCorners=false
      bSeparateFillAlpha=false
      outlineMaterial=null
      outlineSize=0
      outlineColor=<rgba 0 0 0 1>
     }
    }
   }
   slot {class=ButtonSlot}
  }
  B_Select_ByParentClass {
   widgetClassPath=&EditorUtilityButton
   parent=HorizontalBox_2
   widget=B_Select_ByParentClass
   properties {
    toolTipText="Select an actor in the scene then use this button to get all actors of the same PARENT class as selected actor. \r\n\r\nFor example, if the an actor of the class 'BP_Village_Window_A' is selected, the action will find and select all actors of class 'Building Wall'.\r\n\r\nIf multiple actors have been selected before using this action the class of the first actor will be used."
    widgetStyle {
     clickedSlateSound {resourceObject=null}
     disabled {
      bIsDynamicallyLoaded=false
      drawAs=RoundedBox
      imageType=NoImage
      mirroring=NoMirror
      resourceName=None
      resourceObject=null
      tiling=NoTile
      imageSize=<0 0>
      margin=<ltrb 0 0 0 0>
      outlineSettings {
       bUseBrushTransparency=false
       roundingType=FixedRadius
       width=1
       color {colorUseRule=UseColor_ColorTable, specifiedColor=<rgba 0 0 0 0>}
       cornerRadii=<4 4 4 4>
      }
      tintColor {colorUseRule=UseColor_Specified, specifiedColor=<rgba 0.0395 0.0395 0.0395 1>}
      uVRegion {bIsValid=false, max=<0 0>, min=<0 0>}
     }
     disabledForeground {colorUseRule=UseColor_ColorTable, specifiedColor=<rgba 0 0 0 0>}
     hovered {
      bIsDynamicallyLoaded=false
      drawAs=RoundedBox
      imageType=NoImage
      mirroring=NoMirror
      resourceName=None
      resourceObject=null
      tiling=NoTile
      imageSize=<32 32>
      margin=<ltrb 0 0 0 0>
      outlineSettings {
       bUseBrushTransparency=false
       roundingType=FixedRadius
       width=1
       color {colorUseRule=UseColor_ColorTable, specifiedColor=<rgba 0 0 0 0>}
       cornerRadii=<4 4 4 4>
      }
      tintColor {colorUseRule=UseColor_Specified, specifiedColor=<rgba 0.0953 0.0953 0.0953 1>}
      uVRegion {bIsValid=false, max=<0 0>, min=<0 0>}
     }
     hoveredForeground {colorUseRule=UseColor_ColorTable, specifiedColor=<rgba 0 0 0 0>}
     hoveredSlateSound {resourceObject=null}
     normal {
      bIsDynamicallyLoaded=false
      drawAs=RoundedBox
      imageType=NoImage
      mirroring=NoMirror
      resourceName=None
      resourceObject=null
      tiling=NoTile
      imageSize=<32 32>
      margin=<ltrb 0 0 0 0>
      outlineSettings {
       bUseBrushTransparency=false
       roundingType=FixedRadius
       width=1
       color {colorUseRule=UseColor_ColorTable, specifiedColor=<rgba 0 0 0 0>}
       cornerRadii=<4 4 4 4>
      }
      tintColor {colorUseRule=UseColor_Specified, specifiedColor=<rgba 0.0395 0.0395 0.0395 1>}
      uVRegion {bIsValid=false, max=<0 0>, min=<0 0>}
     }
     normalForeground {colorUseRule=UseColor_ColorTable, specifiedColor=<rgba 0 0 0 0>}
     normalPadding=<ltrb 12 1.5 12 1.5>
     pressed {
      bIsDynamicallyLoaded=false
      drawAs=RoundedBox
      imageType=NoImage
      mirroring=NoMirror
      resourceName=None
      resourceObject=null
      tiling=NoTile
      imageSize=<32 32>
      margin=<ltrb 0 0 0 0>
      outlineSettings {
       bUseBrushTransparency=false
       roundingType=FixedRadius
       width=1
       color {colorUseRule=UseColor_ColorTable, specifiedColor=<rgba 0 0 0 0>}
       cornerRadii=<4 4 4 4>
      }
      tintColor {colorUseRule=UseColor_Specified, specifiedColor=<rgba 0.0284 0.0284 0.0284 1>}
      uVRegion {bIsValid=false, max=<0 0>, min=<0 0>}
     }
     pressedForeground {colorUseRule=UseColor_ColorTable, specifiedColor=<rgba 0 0 0 0>}
     pressedPadding=<ltrb 12 2.5 12 0.5>
     pressedSlateSound {resourceObject=null}
    }
   }
   slot {class=HorizontalBoxSlot, properties {padding=<ltrb 5 0 5 0>}}
  }
  TextBlock_11 {
   widgetClassPath=&TextBlock
   parent=B_Select_ByParentClass
   widget=TextBlock_11
   properties {
    text=Parent Class
    font {
     bForceMonospaced=false
     bMaterialIsStencil=false
     fontMaterial=null
     fontObject=&Roboto
     letterSpacing=0
     monospacedWidth=1
     size=9
     skewAmount=0
     typefaceFontName=Regular
     outlineSettings {
      bApplyOutlineToDropShadows=false
      bMiteredCorners=false
      bSeparateFillAlpha=false
      outlineMaterial=null
      outlineSize=0
      outlineColor=<rgba 0 0 0 1>
     }
    }
   }
   slot {class=ButtonSlot}
  }
  B_Select_Marker {
   widgetClassPath=&EditorUtilityButton
   parent=HorizontalBox_2
   widget=B_Select_Marker
   properties {
    toolTipText=Select the Marker Object if it exists. Otherwise ask if you want to create it.
    widgetStyle {
     clickedSlateSound {resourceObject=null}
     disabled {
      bIsDynamicallyLoaded=false
      drawAs=RoundedBox
      imageType=NoImage
      mirroring=NoMirror
      resourceName=None
      resourceObject=null
      tiling=NoTile
      imageSize=<0 0>
      margin=<ltrb 0 0 0 0>
      outlineSettings {
       bUseBrushTransparency=false
       roundingType=FixedRadius
       width=1
       color {colorUseRule=UseColor_ColorTable, specifiedColor=<rgba 0 0 0 0>}
       cornerRadii=<4 4 4 4>
      }
      tintColor {colorUseRule=UseColor_Specified, specifiedColor=<rgba 0.0395 0.0395 0.0395 1>}
      uVRegion {bIsValid=false, max=<0 0>, min=<0 0>}
     }
     disabledForeground {colorUseRule=UseColor_ColorTable, specifiedColor=<rgba 0 0 0 0>}
     hovered {
      bIsDynamicallyLoaded=false
      drawAs=RoundedBox
      imageType=NoImage
      mirroring=NoMirror
      resourceName=None
      resourceObject=null
      tiling=NoTile
      imageSize=<32 32>
      margin=<ltrb 0 0 0 0>
      outlineSettings {
       bUseBrushTransparency=false
       roundingType=FixedRadius
       width=1
       color {colorUseRule=UseColor_ColorTable, specifiedColor=<rgba 0 0 0 0>}
       cornerRadii=<4 4 4 4>
      }
      tintColor {colorUseRule=UseColor_Specified, specifiedColor=<rgba 0.0953 0.0953 0.0953 1>}
      uVRegion {bIsValid=false, max=<0 0>, min=<0 0>}
     }
     hoveredForeground {colorUseRule=UseColor_ColorTable, specifiedColor=<rgba 0 0 0 0>}
     hoveredSlateSound {resourceObject=null}
     normal {
      bIsDynamicallyLoaded=false
      drawAs=RoundedBox
      imageType=NoImage
      mirroring=NoMirror
      resourceName=None
      resourceObject=null
      tiling=NoTile
      imageSize=<32 32>
      margin=<ltrb 0 0 0 0>
      outlineSettings {
       bUseBrushTransparency=false
       roundingType=FixedRadius
       width=1
       color {colorUseRule=UseColor_ColorTable, specifiedColor=<rgba 0 0 0 0>}
       cornerRadii=<4 4 4 4>
      }
      tintColor {colorUseRule=UseColor_Specified, specifiedColor=<rgba 0.0395 0.0395 0.0395 1>}
      uVRegion {bIsValid=false, max=<0 0>, min=<0 0>}
     }
     normalForeground {colorUseRule=UseColor_ColorTable, specifiedColor=<rgba 0 0 0 0>}
     normalPadding=<ltrb 12 1.5 12 1.5>
     pressed {
      bIsDynamicallyLoaded=false
      drawAs=RoundedBox
      imageType=NoImage
      mirroring=NoMirror
      resourceName=None
      resourceObject=null
      tiling=NoTile
      imageSize=<32 32>
      margin=<ltrb 0 0 0 0>
      outlineSettings {
       bUseBrushTransparency=false
       roundingType=FixedRadius
       width=1
       color {colorUseRule=UseColor_ColorTable, specifiedColor=<rgba 0 0 0 0>}
       cornerRadii=<4 4 4 4>
      }
      tintColor {colorUseRule=UseColor_Specified, specifiedColor=<rgba 0.0284 0.0284 0.0284 1>}
      uVRegion {bIsValid=false, max=<0 0>, min=<0 0>}
     }
     pressedForeground {colorUseRule=UseColor_ColorTable, specifiedColor=<rgba 0 0 0 0>}
     pressedPadding=<ltrb 12 2.5 12 0.5>
     pressedSlateSound {resourceObject=null}
    }
   }
   slot {class=HorizontalBoxSlot, properties {padding=<ltrb 5 0 5 0>}}
  }
  TextBlock_20 {
   widgetClassPath=&TextBlock
   parent=B_Select_Marker
   widget=TextBlock_20
   properties {
    text=Marker Object
    font {
     bForceMonospaced=false
     bMaterialIsStencil=false
     fontMaterial=null
     fontObject=&Roboto
     letterSpacing=0
     monospacedWidth=1
     size=9
     skewAmount=0
     typefaceFontName=Regular
     outlineSettings {
      bApplyOutlineToDropShadows=false
      bMiteredCorners=false
      bSeparateFillAlpha=false
      outlineMaterial=null
      outlineSize=0
      outlineColor=<rgba 0 0 0 1>
     }
    }
   }
   slot {class=ButtonSlot}
  }
  Border-Selection_1 {
   widgetClassPath=&Border
   parent=HorizontalBox_8
   widget=Border-Selection_1
   properties {
    background {
     bIsDynamicallyLoaded=false
     drawAs=RoundedBox
     imageType=FullColor
     mirroring=NoMirror
     resourceName=None
     resourceObject=null
     tiling=NoTile
     imageSize=<32 32>
     margin=<ltrb 0.5 0.5 0.5 0.5>
     outlineSettings {
      bUseBrushTransparency=false
      roundingType=FixedRadius
      width=1.2219
      color {colorUseRule=UseColor_Specified, specifiedColor=<rgba 0 0 0 1>}
      cornerRadii=<5 5 5 5>
     }
     tintColor {colorUseRule=UseColor_Specified, specifiedColor=<rgba 0 0 0 0>}
     uVRegion {bIsValid=false, max=<0 0>, min=<0 0>}
    }
    padding=<ltrb 5 5 5 5>
   }
   slot {class=HorizontalBoxSlot}
  }
  VerticalBox_8 {
   widgetClassPath=&VerticalBox
   parent=Border-Selection_1
   widget=VerticalBox_8
   slot {class=BorderSlot, properties {padding=<ltrb 5 5 5 5>}}
  }
  TextBlock_6 {
   widgetClassPath=&TextBlock
   parent=VerticalBox_8
   widget=TextBlock_6
   properties {
    text=Actor Inspector
    font {
     bForceMonospaced=false
     bMaterialIsStencil=false
     fontMaterial=null
     fontObject=&Roboto
     letterSpacing=0
     monospacedWidth=1
     size=9
     skewAmount=0
     typefaceFontName=Bold
     outlineSettings {
      bApplyOutlineToDropShadows=false
      bMiteredCorners=false
      bSeparateFillAlpha=false
      outlineMaterial=null
      outlineSize=0
      outlineColor=<rgba 0 0 0 1>
     }
    }
   }
   slot {class=VerticalBoxSlot, properties {horizontalAlignment=HAlign_Center, verticalAlignment=VAlign_Top}}
  }
  B_SpawnInspector {
   widgetClassPath=&EditorUtilityButton
   parent=VerticalBox_8
   widget=B_SpawnInspector
   properties {
    toolTipText="Select an actor in the scene then use this button to get all actors of the same class as selected actor. \r\n\r\nSimilar to the editor's 'Shift+E' funcionality, but more strict. \r\n\r\nIf multiple actors have been selected before using this action the class of the first actor will be used."
    widgetStyle {
     clickedSlateSound {resourceObject=null}
     disabled {
      bIsDynamicallyLoaded=false
      drawAs=RoundedBox
      imageType=NoImage
      mirroring=NoMirror
      resourceName=None
      resourceObject=null
      tiling=NoTile
      imageSize=<0 0>
      margin=<ltrb 0 0 0 0>
      outlineSettings {
       bUseBrushTransparency=false
       roundingType=FixedRadius
       width=1
       color {colorUseRule=UseColor_ColorTable, specifiedColor=<rgba 0 0 0 0>}
       cornerRadii=<4 4 4 4>
      }
      tintColor {colorUseRule=UseColor_Specified, specifiedColor=<rgba 0.0395 0.0395 0.0395 1>}
      uVRegion {bIsValid=false, max=<0 0>, min=<0 0>}
     }
     disabledForeground {colorUseRule=UseColor_ColorTable, specifiedColor=<rgba 0 0 0 0>}
     hovered {
      bIsDynamicallyLoaded=false
      drawAs=RoundedBox
      imageType=NoImage
      mirroring=NoMirror
      resourceName=None
      resourceObject=null
      tiling=NoTile
      imageSize=<32 32>
      margin=<ltrb 0 0 0 0>
      outlineSettings {
       bUseBrushTransparency=false
       roundingType=FixedRadius
       width=1
       color {colorUseRule=UseColor_ColorTable, specifiedColor=<rgba 0 0 0 0>}
       cornerRadii=<4 4 4 4>
      }
      tintColor {colorUseRule=UseColor_Specified, specifiedColor=<rgba 0.0953 0.0953 0.0953 1>}
      uVRegion {bIsValid=false, max=<0 0>, min=<0 0>}
     }
     hoveredForeground {colorUseRule=UseColor_ColorTable, specifiedColor=<rgba 0 0 0 0>}
     hoveredSlateSound {resourceObject=null}
     normal {
      bIsDynamicallyLoaded=false
      drawAs=RoundedBox
      imageType=NoImage
      mirroring=NoMirror
      resourceName=None
      resourceObject=null
      tiling=NoTile
      imageSize=<32 32>
      margin=<ltrb 0 0 0 0>
      outlineSettings {
       bUseBrushTransparency=false
       roundingType=FixedRadius
       width=1
       color {colorUseRule=UseColor_ColorTable, specifiedColor=<rgba 0 0 0 0>}
       cornerRadii=<4 4 4 4>
      }
      tintColor {colorUseRule=UseColor_Specified, specifiedColor=<rgba 0.0395 0.0395 0.0395 1>}
      uVRegion {bIsValid=false, max=<0 0>, min=<0 0>}
     }
     normalForeground {colorUseRule=UseColor_ColorTable, specifiedColor=<rgba 0 0 0 0>}
     normalPadding=<ltrb 12 1.5 12 1.5>
     pressed {
      bIsDynamicallyLoaded=false
      drawAs=RoundedBox
      imageType=NoImage
      mirroring=NoMirror
      resourceName=None
      resourceObject=null
      tiling=NoTile
      imageSize=<32 32>
      margin=<ltrb 0 0 0 0>
      outlineSettings {
       bUseBrushTransparency=false
       roundingType=FixedRadius
       width=1
       color {colorUseRule=UseColor_ColorTable, specifiedColor=<rgba 0 0 0 0>}
       cornerRadii=<4 4 4 4>
      }
      tintColor {colorUseRule=UseColor_Specified, specifiedColor=<rgba 0.0284 0.0284 0.0284 1>}
      uVRegion {bIsValid=false, max=<0 0>, min=<0 0>}
     }
     pressedForeground {colorUseRule=UseColor_ColorTable, specifiedColor=<rgba 0 0 0 0>}
     pressedPadding=<ltrb 12 2.5 12 0.5>
     pressedSlateSound {resourceObject=null}
    }
   }
   slot {class=VerticalBoxSlot, properties {padding=<ltrb 5 0 5 0>}}
  }
  TextBlock_21 {
   widgetClassPath=&TextBlock
   parent=B_SpawnInspector
   widget=TextBlock_21
   properties {
    text=Spawn Widget
    font {
     bForceMonospaced=false
     bMaterialIsStencil=false
     fontMaterial=null
     fontObject=&Roboto
     letterSpacing=0
     monospacedWidth=1
     size=9
     skewAmount=0
     typefaceFontName=Regular
     outlineSettings {
      bApplyOutlineToDropShadows=false
      bMiteredCorners=false
      bSeparateFillAlpha=false
      outlineMaterial=null
      outlineSize=0
      outlineColor=<rgba 0 0 0 1>
     }
    }
   }
   slot {class=ButtonSlot}
  }
  EditorUtilityScrollBox_0 {
   widgetClassPath=&EditorUtilityScrollBox
   parent=VerticalBox_1
   widget=EditorUtilityScrollBox_0
   properties {widgetBarStyle {
     thickness=8
     draggedThumbImage {
      bIsDynamicallyLoaded=false
      drawAs=RoundedBox
      imageType=NoImage
      mirroring=NoMirror
      resourceName=None
      resourceObject=null
      tiling=NoTile
      imageSize=<0 0>
      margin=<ltrb 0 0 0 0>
      outlineSettings {
       bUseBrushTransparency=false
       roundingType=FixedRadius
       width=0
       color {colorUseRule=UseColor_Specified, specifiedColor=<rgba 0 0 0 0>}
       cornerRadii=<4 4 4 4>
      }
      tintColor {colorUseRule=UseColor_Specified, specifiedColor=<rgba 0.2159 0.2159 0.2159 1>}
      uVRegion {bIsValid=false, max=<0 0>, min=<0 0>}
     }
     horizontalBackgroundImage {
      bIsDynamicallyLoaded=false
      drawAs=NoDrawType
      imageType=NoImage
      mirroring=NoMirror
      resourceName=None
      resourceObject=null
      tiling=NoTile
      imageSize=<0 0>
      margin=<ltrb 0 0 0 0>
      outlineSettings {
       bUseBrushTransparency=false
       roundingType=HalfHeightRadius
       width=0
       color {colorUseRule=UseColor_Specified, specifiedColor=<rgba 0 0 0 0>}
       cornerRadii=<0 0 0 0>
      }
      tintColor {colorUseRule=UseColor_Specified, specifiedColor=<rgba 1 1 1 1>}
      uVRegion {bIsValid=false, max=<0 0>, min=<0 0>}
     }
     horizontalBottomSlotImage {
      bIsDynamicallyLoaded=false
      drawAs=NoDrawType
      imageType=NoImage
      mirroring=NoMirror
      resourceName=None
      resourceObject=null
      tiling=NoTile
      imageSize=<0 0>
      margin=<ltrb 0 0 0 0>
      outlineSettings {
       bUseBrushTransparency=false
       roundingType=HalfHeightRadius
       width=0
       color {colorUseRule=UseColor_Specified, specifiedColor=<rgba 0 0 0 0>}
       cornerRadii=<0 0 0 0>
      }
      tintColor {colorUseRule=UseColor_Specified, specifiedColor=<rgba 1 1 1 1>}
      uVRegion {bIsValid=false, max=<0 0>, min=<0 0>}
     }
     horizontalTopSlotImage {
      bIsDynamicallyLoaded=false
      drawAs=NoDrawType
      imageType=NoImage
      mirroring=NoMirror
      resourceName=None
      resourceObject=null
      tiling=NoTile
      imageSize=<0 0>
      margin=<ltrb 0 0 0 0>
      outlineSettings {
       bUseBrushTransparency=false
       roundingType=HalfHeightRadius
       width=0
       color {colorUseRule=UseColor_Specified, specifiedColor=<rgba 0 0 0 0>}
       cornerRadii=<0 0 0 0>
      }
      tintColor {colorUseRule=UseColor_Specified, specifiedColor=<rgba 1 1 1 1>}
      uVRegion {bIsValid=false, max=<0 0>, min=<0 0>}
     }
     hoveredThumbImage {
      bIsDynamicallyLoaded=false
      drawAs=RoundedBox
      imageType=NoImage
      mirroring=NoMirror
      resourceName=None
      resourceObject=null
      tiling=NoTile
      imageSize=<0 0>
      margin=<ltrb 0 0 0 0>
      outlineSettings {
       bUseBrushTransparency=false
       roundingType=FixedRadius
       width=0
       color {colorUseRule=UseColor_Specified, specifiedColor=<rgba 0 0 0 0>}
       cornerRadii=<4 4 4 4>
      }
      tintColor {colorUseRule=UseColor_Specified, specifiedColor=<rgba 0.2159 0.2159 0.2159 1>}
      uVRegion {bIsValid=false, max=<0 0>, min=<0 0>}
     }
     normalThumbImage {
      bIsDynamicallyLoaded=false
      drawAs=RoundedBox
      imageType=NoImage
      mirroring=NoMirror
      resourceName=None
      resourceObject=null
      tiling=NoTile
      imageSize=<0 0>
      margin=<ltrb 0 0 0 0>
      outlineSettings {
       bUseBrushTransparency=false
       roundingType=FixedRadius
       width=0
       color {colorUseRule=UseColor_Specified, specifiedColor=<rgba 0 0 0 0>}
       cornerRadii=<4 4 4 4>
      }
      tintColor {colorUseRule=UseColor_Specified, specifiedColor=<rgba 0.0953 0.0953 0.0953 1>}
      uVRegion {bIsValid=false, max=<0 0>, min=<0 0>}
     }
     verticalBackgroundImage {
      bIsDynamicallyLoaded=false
      drawAs=NoDrawType
      imageType=NoImage
      mirroring=NoMirror
      resourceName=None
      resourceObject=null
      tiling=NoTile
      imageSize=<0 0>
      margin=<ltrb 0 0 0 0>
      outlineSettings {
       bUseBrushTransparency=false
       roundingType=HalfHeightRadius
       width=0
       color {colorUseRule=UseColor_Specified, specifiedColor=<rgba 0 0 0 0>}
       cornerRadii=<0 0 0 0>
      }
      tintColor {colorUseRule=UseColor_Specified, specifiedColor=<rgba 1 1 1 1>}
      uVRegion {bIsValid=false, max=<0 0>, min=<0 0>}
     }
     verticalBottomSlotImage {
      bIsDynamicallyLoaded=false
      drawAs=NoDrawType
      imageType=NoImage
      mirroring=NoMirror
      resourceName=None
      resourceObject=null
      tiling=NoTile
      imageSize=<0 0>
      margin=<ltrb 0 0 0 0>
      outlineSettings {
       bUseBrushTransparency=false
       roundingType=HalfHeightRadius
       width=0
       color {colorUseRule=UseColor_Specified, specifiedColor=<rgba 0 0 0 0>}
       cornerRadii=<0 0 0 0>
      }
      tintColor {colorUseRule=UseColor_Specified, specifiedColor=<rgba 1 1 1 1>}
      uVRegion {bIsValid=false, max=<0 0>, min=<0 0>}
     }
     verticalTopSlotImage {
      bIsDynamicallyLoaded=false
      drawAs=NoDrawType
      imageType=NoImage
      mirroring=NoMirror
      resourceName=None
      resourceObject=null
      tiling=NoTile
      imageSize=<0 0>
      margin=<ltrb 0 0 0 0>
      outlineSettings {
       bUseBrushTransparency=false
       roundingType=HalfHeightRadius
       width=0
       color {colorUseRule=UseColor_Specified, specifiedColor=<rgba 0 0 0 0>}
       cornerRadii=<0 0 0 0>
      }
      tintColor {colorUseRule=UseColor_Specified, specifiedColor=<rgba 1 1 1 1>}
      uVRegion {bIsValid=false, max=<0 0>, min=<0 0>}
     }
    }}
   slot {class=VerticalBoxSlot, properties {size {sizeRule=Fill, value=1}}}
  }
  VerticalBox_Master {
   widgetClassPath=&VerticalBox
   bIsVariable=true
   parent=EditorUtilityScrollBox_0
   widget=VerticalBox_Master
   slot {class=ScrollBoxSlot, properties {size {sizeRule=Fill, value=1}}}
  }
  Border-Settings {
   widgetClassPath=&Border
   parent=VerticalBox_Master
   widget=Border-Settings
   properties {
    bIsEnabled=false
    visibility=Collapsed
    background {
     bIsDynamicallyLoaded=false
     drawAs=Box
     imageType=FullColor
     mirroring=NoMirror
     resourceName=None
     resourceObject=&M_UI_Border
     tiling=NoTile
     imageSize=<40 40>
     margin=<ltrb 0.5 0.5 0.5 0.5>
     outlineSettings {
      bUseBrushTransparency=false
      roundingType=HalfHeightRadius
      width=0
      color {colorUseRule=UseColor_Specified, specifiedColor=<rgba 0 0 0 0>}
      cornerRadii=<0 0 0 0>
     }
     tintColor {colorUseRule=UseColor_Specified, specifiedColor=<rgba 1 0.3875 0 0>}
     uVRegion {bIsValid=false, max=<0 0>, min=<0 0>}
    }
    padding=<ltrb 0 0 0 0>
   }
   slot {class=VerticalBoxSlot, properties {padding=<ltrb 0 10 10 0>}}
  }
  VerticalBox_4 {
   widgetClassPath=&VerticalBox
   parent=Border-Settings
   widget=VerticalBox_4
   slot {class=BorderSlot, properties {padding=<ltrb 0 0 0 0>}}
  }
  ToggleButton_Collapse_Settings {
   widgetClassPath=&EditorUtilityCheckBox
   bIsVariable=true
   parent=VerticalBox_4
   widget=ToggleButton_Collapse_Settings
   properties {
    toolTipText=Expands / Contracts the Settings section
    widgetStyle {
     checkBoxType=ToggleButton
     backgroundHoveredImage {
      bIsDynamicallyLoaded=false
      drawAs=RoundedBox
      imageType=NoImage
      mirroring=NoMirror
      resourceName=None
      resourceObject=null
      tiling=NoTile
      imageSize=<18 18>
      margin=<ltrb 0 0 0 0>
      outlineSettings {
       bUseBrushTransparency=false
       roundingType=FixedRadius
       width=1
       color {colorUseRule=UseColor_ColorTable, specifiedColor=<rgba 0 0 0 0>}
       cornerRadii=<3 3 3 3>
      }
      tintColor {colorUseRule=UseColor_Specified, specifiedColor=<rgba 0.0048 0.0048 0.0048 1>}
      uVRegion {bIsValid=false, max=<0 0>, min=<0 0>}
     }
     backgroundImage {
      bIsDynamicallyLoaded=false
      drawAs=RoundedBox
      imageType=NoImage
      mirroring=NoMirror
      resourceName=None
      resourceObject=null
      tiling=NoTile
      imageSize=<18 18>
      margin=<ltrb 0 0 0 0>
      outlineSettings {
       bUseBrushTransparency=false
       roundingType=FixedRadius
       width=1
       color {colorUseRule=UseColor_ColorTable, specifiedColor=<rgba 0 0 0 0>}
       cornerRadii=<3 3 3 3>
      }
      tintColor {colorUseRule=UseColor_Specified, specifiedColor=<rgba 0.0048 0.0048 0.0048 1>}
      uVRegion {bIsValid=false, max=<0 0>, min=<0 0>}
     }
     backgroundPressedImage {
      bIsDynamicallyLoaded=false
      drawAs=RoundedBox
      imageType=NoImage
      mirroring=NoMirror
      resourceName=None
      resourceObject=null
      tiling=NoTile
      imageSize=<18 18>
      margin=<ltrb 0 0 0 0>
      outlineSettings {
       bUseBrushTransparency=false
       roundingType=FixedRadius
       width=1
       color {colorUseRule=UseColor_ColorTable, specifiedColor=<rgba 0 0 0 0>}
       cornerRadii=<3 3 3 3>
      }
      tintColor {colorUseRule=UseColor_Specified, specifiedColor=<rgba 0.0048 0.0048 0.0048 1>}
      uVRegion {bIsValid=false, max=<0 0>, min=<0 0>}
     }
     borderBackgroundColor {colorUseRule=UseColor_Specified, specifiedColor=<rgba 1 1 1 1>}
     checkedForeground {colorUseRule=UseColor_Specified, specifiedColor=<rgba 1 1 1 1>}
     checkedHoveredForeground {colorUseRule=UseColor_Specified, specifiedColor=<rgba 1 1 1 1>}
     checkedHoveredImage {
      bIsDynamicallyLoaded=false
      drawAs=Image
      imageType=NoImage
      mirroring=NoMirror
      resourceName=None
      resourceObject=null
      tiling=NoTile
      imageSize=<32 10.5981>
      margin=<ltrb 0 0 0 0>
      outlineSettings {
       bUseBrushTransparency=false
       roundingType=HalfHeightRadius
       width=0
       color {colorUseRule=UseColor_Specified, specifiedColor=<rgba 0 0 0 0>}
       cornerRadii=<0 0 0 0>
      }
      tintColor {colorUseRule=UseColor_Specified, specifiedColor=<rgba 0.0953 0.0953 0.0953 1>}
      uVRegion {bIsValid=false, max=<0 0>, min=<0 0>}
     }
     checkedImage {
      bIsDynamicallyLoaded=false
      drawAs=Image
      imageType=NoImage
      mirroring=NoMirror
      resourceName=None
      resourceObject=null
      tiling=NoTile
      imageSize=<32 10.5981>
      margin=<ltrb 0 0 0 0>
      outlineSettings {
       bUseBrushTransparency=false
       roundingType=HalfHeightRadius
       width=0
       color {colorUseRule=UseColor_Specified, specifiedColor=<rgba 0 0 0 0>}
       cornerRadii=<0 0 0 0>
      }
      tintColor {colorUseRule=UseColor_Specified, specifiedColor=<rgba 0.0284 0.0284 0.0284 1>}
      uVRegion {bIsValid=false, max=<0 0>, min=<0 0>}
     }
     checkedPressedForeground {colorUseRule=UseColor_Specified, specifiedColor=<rgba 1 1 1 1>}
     checkedPressedImage {
      bIsDynamicallyLoaded=false
      drawAs=Image
      imageType=NoImage
      mirroring=NoMirror
      resourceName=None
      resourceObject=null
      tiling=NoTile
      imageSize=<32 10.5981>
      margin=<ltrb 0 0 0 0>
      outlineSettings {
       bUseBrushTransparency=false
       roundingType=HalfHeightRadius
       width=0
       color {colorUseRule=UseColor_Specified, specifiedColor=<rgba 0 0 0 0>}
       cornerRadii=<0 0 0 0>
      }
      tintColor {colorUseRule=UseColor_Specified, specifiedColor=<rgba 0.0284 0.0284 0.0284 1>}
      uVRegion {bIsValid=false, max=<0 0>, min=<0 0>}
     }
     checkedSlateSound {resourceObject=null}
     foregroundColor {colorUseRule=UseColor_Specified, specifiedColor=<rgba 1 1 1 1>}
     hoveredForeground {colorUseRule=UseColor_Specified, specifiedColor=<rgba 1 1 1 1>}
     hoveredSlateSound {resourceObject=null}
     padding=<ltrb 0 0 0 0>
     pressedForeground {colorUseRule=UseColor_Specified, specifiedColor=<rgba 1 1 1 1>}
     uncheckedHoveredImage {
      bIsDynamicallyLoaded=false
      drawAs=Image
      imageType=NoImage
      mirroring=NoMirror
      resourceName=None
      resourceObject=null
      tiling=NoTile
      imageSize=<32 10.5981>
      margin=<ltrb 0 0 0 0>
      outlineSettings {
       bUseBrushTransparency=false
       roundingType=HalfHeightRadius
       width=0
       color {colorUseRule=UseColor_Specified, specifiedColor=<rgba 0 0 0 0>}
       cornerRadii=<0 0 0 0>
      }
      tintColor {colorUseRule=UseColor_Specified, specifiedColor=<rgba 0.0953 0.0953 0.0953 1>}
      uVRegion {bIsValid=false, max=<0 0>, min=<0 0>}
     }
     uncheckedImage {
      bIsDynamicallyLoaded=false
      drawAs=Image
      imageType=NoImage
      mirroring=NoMirror
      resourceName=None
      resourceObject=null
      tiling=NoTile
      imageSize=<32 10.5981>
      margin=<ltrb 0 0 0 0>
      outlineSettings {
       bUseBrushTransparency=false
       roundingType=HalfHeightRadius
       width=0
       color {colorUseRule=UseColor_Specified, specifiedColor=<rgba 0 0 0 0>}
       cornerRadii=<0 0 0 0>
      }
      tintColor {colorUseRule=UseColor_Specified, specifiedColor=<rgba 0.0284 0.0284 0.0284 1>}
      uVRegion {bIsValid=false, max=<0 0>, min=<0 0>}
     }
     uncheckedPressedImage {
      bIsDynamicallyLoaded=false
      drawAs=Image
      imageType=NoImage
      mirroring=NoMirror
      resourceName=None
      resourceObject=null
      tiling=NoTile
      imageSize=<32 10.5981>
      margin=<ltrb 0 0 0 0>
      outlineSettings {
       bUseBrushTransparency=false
       roundingType=HalfHeightRadius
       width=0
       color {colorUseRule=UseColor_Specified, specifiedColor=<rgba 0 0 0 0>}
       cornerRadii=<0 0 0 0>
      }
      tintColor {colorUseRule=UseColor_Specified, specifiedColor=<rgba 0.0284 0.0284 0.0284 1>}
      uVRegion {bIsValid=false, max=<0 0>, min=<0 0>}
     }
     uncheckedSlateSound {resourceObject=null}
     undeterminedForeground {colorUseRule=UseColor_Specified, specifiedColor=<rgba 1 1 1 1>}
     undeterminedHoveredImage {
      bIsDynamicallyLoaded=false
      drawAs=Image
      imageType=Vector
      mirroring=NoMirror
      resourceName=../../../Engine/Content/Slate/Starship/CoreWidgets/CheckBox/indeterminate.svg
      resourceObject=null
      tiling=NoTile
      imageSize=<16 16>
      margin=<ltrb 0 0 0 0>
      outlineSettings {
       bUseBrushTransparency=false
       roundingType=HalfHeightRadius
       width=0
       color {colorUseRule=UseColor_Specified, specifiedColor=<rgba 0 0 0 0>}
       cornerRadii=<0 0 0 0>
      }
      tintColor {colorUseRule=UseColor_Specified, specifiedColor=<rgba 0.0044 0.2384 1 1>}
      uVRegion {bIsValid=false, max=<0 0>, min=<0 0>}
     }
     undeterminedImage {
      bIsDynamicallyLoaded=false
      drawAs=Image
      imageType=Vector
      mirroring=NoMirror
      resourceName=../../../Engine/Content/Slate/Starship/CoreWidgets/CheckBox/indeterminate.svg
      resourceObject=null
      tiling=NoTile
      imageSize=<16 16>
      margin=<ltrb 0 0 0 0>
      outlineSettings {
       bUseBrushTransparency=false
       roundingType=HalfHeightRadius
       width=0
       color {colorUseRule=UseColor_Specified, specifiedColor=<rgba 0 0 0 0>}
       cornerRadii=<0 0 0 0>
      }
      tintColor {colorUseRule=UseColor_Specified, specifiedColor=<rgba 0 0.162 0.7454 1>}
      uVRegion {bIsValid=false, max=<0 0>, min=<0 0>}
     }
     undeterminedPressedImage {
      bIsDynamicallyLoaded=false
      drawAs=Image
      imageType=Vector
      mirroring=NoMirror
      resourceName=../../../Engine/Content/Slate/Starship/CoreWidgets/CheckBox/indeterminate.svg
      resourceObject=null
      tiling=NoTile
      imageSize=<16 16>
      margin=<ltrb 0 0 0 0>
      outlineSettings {
       bUseBrushTransparency=false
       roundingType=HalfHeightRadius
       width=0
       color {colorUseRule=UseColor_Specified, specifiedColor=<rgba 0 0 0 0>}
       cornerRadii=<0 0 0 0>
      }
      tintColor {colorUseRule=UseColor_Specified, specifiedColor=<rgba 0 0.0802 0.3515 1>}
      uVRegion {bIsValid=false, max=<0 0>, min=<0 0>}
     }
    }
   }
   slot {class=VerticalBoxSlot}
  }
  HorizontalBox_5 {widgetClassPath=&HorizontalBox, parent=ToggleButton_Collapse_Settings, widget=HorizontalBox_5, slot {class=PanelSlot}}
  Spacer_1 {
   widgetClassPath=&Spacer
   parent=HorizontalBox_5
   widget=Spacer_1
   properties {size=<12 1>}
   slot {class=HorizontalBoxSlot, properties {horizontalAlignment=HAlign_Left}}
  }
  T_ToggleButton_Settings {
   widgetClassPath=&TextBlock
   bIsVariable=true
   parent=HorizontalBox_5
   widget=T_ToggleButton_Settings
   properties {
    text="Settings ▲"
    font {
     bForceMonospaced=false
     bMaterialIsStencil=false
     fontMaterial=null
     fontObject=&Roboto
     letterSpacing=0
     monospacedWidth=1
     size=12
     skewAmount=0
     typefaceFontName=Bold
     outlineSettings {
      bApplyOutlineToDropShadows=false
      bMiteredCorners=false
      bSeparateFillAlpha=false
      outlineMaterial=null
      outlineSize=0
      outlineColor=<rgba 0 0 0 1>
     }
    }
   }
   slot {class=HorizontalBoxSlot}
  }
  Socket_Settings {
   widgetClassPath=&VerticalBox
   bIsVariable=true
   parent=VerticalBox_4
   widget=Socket_Settings
   slot {class=VerticalBoxSlot, properties {padding=<ltrb 10 10 0 20>}}
  }
  T_ArrangeTooltip_8 {
   widgetClassPath=&TextBlock
   parent=Socket_Settings
   widget=T_ArrangeTooltip_8
   properties {
    autoWrapText=true
    text="TIP : Mouse over a button or field to get a tooltip explanation.\r\n\r\nNothing else here at the moment ..."
    colorAndOpacity {colorUseRule=UseColor_Specified, specifiedColor=<rgba 0.349 0.349 0.349 1>}
    font {
     bForceMonospaced=false
     bMaterialIsStencil=false
     fontMaterial=null
     fontObject=&Roboto
     letterSpacing=0
     monospacedWidth=1
     size=9
     skewAmount=0
     typefaceFontName=Italic
     outlineSettings {
      bApplyOutlineToDropShadows=false
      bMiteredCorners=false
      bSeparateFillAlpha=false
      outlineMaterial=null
      outlineSize=0
      outlineColor=<rgba 0 0 0 1>
     }
    }
   }
   slot {class=VerticalBoxSlot, properties {padding=<ltrb 0 3 0 0>}}
  }
  Border-PositionUtilities {
   widgetClassPath=&Border
   parent=VerticalBox_Master
   widget=Border-PositionUtilities
   properties {
    background {
     bIsDynamicallyLoaded=false
     drawAs=Box
     imageType=FullColor
     mirroring=NoMirror
     resourceName=None
     resourceObject=&M_UI_Border
     tiling=NoTile
     imageSize=<40 40>
     margin=<ltrb 0.5 0.5 0.5 0.5>
     outlineSettings {
      bUseBrushTransparency=false
      roundingType=HalfHeightRadius
      width=0
      color {colorUseRule=UseColor_Specified, specifiedColor=<rgba 0 0 0 0>}
      cornerRadii=<0 0 0 0>
     }
     tintColor {colorUseRule=UseColor_Specified, specifiedColor=<rgba 1 0.3875 0 0>}
     uVRegion {bIsValid=false, max=<0 0>, min=<0 0>}
    }
    padding=<ltrb 0 0 0 0>
   }
   slot {class=VerticalBoxSlot, properties {padding=<ltrb 0 10 10 0>}}
  }
  VerticalBox_9 {
   widgetClassPath=&VerticalBox
   parent=Border-PositionUtilities
   widget=VerticalBox_9
   slot {class=BorderSlot, properties {padding=<ltrb 0 0 0 0>}}
  }
  ToggleButton_Collapse_ActorPosUtil {
   widgetClassPath=&EditorUtilityCheckBox
   bIsVariable=true
   parent=VerticalBox_9
   widget=ToggleButton_Collapse_ActorPosUtil
   properties {
    toolTipText=Expands / Contracts the Position Utilities module
    widgetStyle {
     checkBoxType=ToggleButton
     backgroundHoveredImage {
      bIsDynamicallyLoaded=false
      drawAs=RoundedBox
      imageType=NoImage
      mirroring=NoMirror
      resourceName=None
      resourceObject=null
      tiling=NoTile
      imageSize=<18 18>
      margin=<ltrb 0 0 0 0>
      outlineSettings {
       bUseBrushTransparency=false
       roundingType=FixedRadius
       width=1
       color {colorUseRule=UseColor_ColorTable, specifiedColor=<rgba 0 0 0 0>}
       cornerRadii=<3 3 3 3>
      }
      tintColor {colorUseRule=UseColor_Specified, specifiedColor=<rgba 0.0048 0.0048 0.0048 1>}
      uVRegion {bIsValid=false, max=<0 0>, min=<0 0>}
     }
     backgroundImage {
      bIsDynamicallyLoaded=false
      drawAs=RoundedBox
      imageType=NoImage
      mirroring=NoMirror
      resourceName=None
      resourceObject=null
      tiling=NoTile
      imageSize=<18 18>
      margin=<ltrb 0 0 0 0>
      outlineSettings {
       bUseBrushTransparency=false
       roundingType=FixedRadius
       width=1
       color {colorUseRule=UseColor_ColorTable, specifiedColor=<rgba 0 0 0 0>}
       cornerRadii=<3 3 3 3>
      }
      tintColor {colorUseRule=UseColor_Specified, specifiedColor=<rgba 0.0048 0.0048 0.0048 1>}
      uVRegion {bIsValid=false, max=<0 0>, min=<0 0>}
     }
     backgroundPressedImage {
      bIsDynamicallyLoaded=false
      drawAs=RoundedBox
      imageType=NoImage
      mirroring=NoMirror
      resourceName=None
      resourceObject=null
      tiling=NoTile
      imageSize=<18 18>
      margin=<ltrb 0 0 0 0>
      outlineSettings {
       bUseBrushTransparency=false
       roundingType=FixedRadius
       width=1
       color {colorUseRule=UseColor_ColorTable, specifiedColor=<rgba 0 0 0 0>}
       cornerRadii=<3 3 3 3>
      }
      tintColor {colorUseRule=UseColor_Specified, specifiedColor=<rgba 0.0048 0.0048 0.0048 1>}
      uVRegion {bIsValid=false, max=<0 0>, min=<0 0>}
     }
     borderBackgroundColor {colorUseRule=UseColor_Specified, specifiedColor=<rgba 1 1 1 1>}
     checkedForeground {colorUseRule=UseColor_Specified, specifiedColor=<rgba 1 1 1 1>}
     checkedHoveredForeground {colorUseRule=UseColor_Specified, specifiedColor=<rgba 1 1 1 1>}
     checkedHoveredImage {
      bIsDynamicallyLoaded=false
      drawAs=Image
      imageType=NoImage
      mirroring=NoMirror
      resourceName=None
      resourceObject=null
      tiling=NoTile
      imageSize=<32 10.5981>
      margin=<ltrb 0 0 0 0>
      outlineSettings {
       bUseBrushTransparency=false
       roundingType=HalfHeightRadius
       width=0
       color {colorUseRule=UseColor_Specified, specifiedColor=<rgba 0 0 0 0>}
       cornerRadii=<0 0 0 0>
      }
      tintColor {colorUseRule=UseColor_Specified, specifiedColor=<rgba 0.0953 0.0953 0.0953 1>}
      uVRegion {bIsValid=false, max=<0 0>, min=<0 0>}
     }
     checkedImage {
      bIsDynamicallyLoaded=false
      drawAs=Image
      imageType=NoImage
      mirroring=NoMirror
      resourceName=None
      resourceObject=null
      tiling=NoTile
      imageSize=<32 10.5981>
      margin=<ltrb 0 0 0 0>
      outlineSettings {
       bUseBrushTransparency=false
       roundingType=HalfHeightRadius
       width=0
       color {colorUseRule=UseColor_Specified, specifiedColor=<rgba 0 0 0 0>}
       cornerRadii=<0 0 0 0>
      }
      tintColor {colorUseRule=UseColor_Specified, specifiedColor=<rgba 0.0284 0.0284 0.0284 1>}
      uVRegion {bIsValid=false, max=<0 0>, min=<0 0>}
     }
     checkedPressedForeground {colorUseRule=UseColor_Specified, specifiedColor=<rgba 1 1 1 1>}
     checkedPressedImage {
      bIsDynamicallyLoaded=false
      drawAs=Image
      imageType=NoImage
      mirroring=NoMirror
      resourceName=None
      resourceObject=null
      tiling=NoTile
      imageSize=<32 10.5981>
      margin=<ltrb 0 0 0 0>
      outlineSettings {
       bUseBrushTransparency=false
       roundingType=HalfHeightRadius
       width=0
       color {colorUseRule=UseColor_Specified, specifiedColor=<rgba 0 0 0 0>}
       cornerRadii=<0 0 0 0>
      }
      tintColor {colorUseRule=UseColor_Specified, specifiedColor=<rgba 0.0284 0.0284 0.0284 1>}
      uVRegion {bIsValid=false, max=<0 0>, min=<0 0>}
     }
     checkedSlateSound {resourceObject=null}
     foregroundColor {colorUseRule=UseColor_Specified, specifiedColor=<rgba 1 1 1 1>}
     hoveredForeground {colorUseRule=UseColor_Specified, specifiedColor=<rgba 1 1 1 1>}
     hoveredSlateSound {resourceObject=null}
     padding=<ltrb 0 0 0 0>
     pressedForeground {colorUseRule=UseColor_Specified, specifiedColor=<rgba 1 1 1 1>}
     uncheckedHoveredImage {
      bIsDynamicallyLoaded=false
      drawAs=Image
      imageType=NoImage
      mirroring=NoMirror
      resourceName=None
      resourceObject=null
      tiling=NoTile
      imageSize=<32 10.5981>
      margin=<ltrb 0 0 0 0>
      outlineSettings {
       bUseBrushTransparency=false
       roundingType=HalfHeightRadius
       width=0
       color {colorUseRule=UseColor_Specified, specifiedColor=<rgba 0 0 0 0>}
       cornerRadii=<0 0 0 0>
      }
      tintColor {colorUseRule=UseColor_Specified, specifiedColor=<rgba 0.0953 0.0953 0.0953 1>}
      uVRegion {bIsValid=false, max=<0 0>, min=<0 0>}
     }
     uncheckedImage {
      bIsDynamicallyLoaded=false
      drawAs=Image
      imageType=NoImage
      mirroring=NoMirror
      resourceName=None
      resourceObject=null
      tiling=NoTile
      imageSize=<32 10.5981>
      margin=<ltrb 0 0 0 0>
      outlineSettings {
       bUseBrushTransparency=false
       roundingType=HalfHeightRadius
       width=0
       color {colorUseRule=UseColor_Specified, specifiedColor=<rgba 0 0 0 0>}
       cornerRadii=<0 0 0 0>
      }
      tintColor {colorUseRule=UseColor_Specified, specifiedColor=<rgba 0.0284 0.0284 0.0284 1>}
      uVRegion {bIsValid=false, max=<0 0>, min=<0 0>}
     }
     uncheckedPressedImage {
      bIsDynamicallyLoaded=false
      drawAs=Image
      imageType=NoImage
      mirroring=NoMirror
      resourceName=None
      resourceObject=null
      tiling=NoTile
      imageSize=<32 10.5981>
      margin=<ltrb 0 0 0 0>
      outlineSettings {
       bUseBrushTransparency=false
       roundingType=HalfHeightRadius
       width=0
       color {colorUseRule=UseColor_Specified, specifiedColor=<rgba 0 0 0 0>}
       cornerRadii=<0 0 0 0>
      }
      tintColor {colorUseRule=UseColor_Specified, specifiedColor=<rgba 0.0284 0.0284 0.0284 1>}
      uVRegion {bIsValid=false, max=<0 0>, min=<0 0>}
     }
     uncheckedSlateSound {resourceObject=null}
     undeterminedForeground {colorUseRule=UseColor_Specified, specifiedColor=<rgba 1 1 1 1>}
     undeterminedHoveredImage {
      bIsDynamicallyLoaded=false
      drawAs=Image
      imageType=Vector
      mirroring=NoMirror
      resourceName=../../../Engine/Content/Slate/Starship/CoreWidgets/CheckBox/indeterminate.svg
      resourceObject=null
      tiling=NoTile
      imageSize=<16 16>
      margin=<ltrb 0 0 0 0>
      outlineSettings {
       bUseBrushTransparency=false
       roundingType=HalfHeightRadius
       width=0
       color {colorUseRule=UseColor_Specified, specifiedColor=<rgba 0 0 0 0>}
       cornerRadii=<0 0 0 0>
      }
      tintColor {colorUseRule=UseColor_Specified, specifiedColor=<rgba 0.0044 0.2384 1 1>}
      uVRegion {bIsValid=false, max=<0 0>, min=<0 0>}
     }
     undeterminedImage {
      bIsDynamicallyLoaded=false
      drawAs=Image
      imageType=Vector
      mirroring=NoMirror
      resourceName=../../../Engine/Content/Slate/Starship/CoreWidgets/CheckBox/indeterminate.svg
      resourceObject=null
      tiling=NoTile
      imageSize=<16 16>
      margin=<ltrb 0 0 0 0>
      outlineSettings {
       bUseBrushTransparency=false
       roundingType=HalfHeightRadius
       width=0
       color {colorUseRule=UseColor_Specified, specifiedColor=<rgba 0 0 0 0>}
       cornerRadii=<0 0 0 0>
      }
      tintColor {colorUseRule=UseColor_Specified, specifiedColor=<rgba 0 0.162 0.7454 1>}
      uVRegion {bIsValid=false, max=<0 0>, min=<0 0>}
     }
     undeterminedPressedImage {
      bIsDynamicallyLoaded=false
      drawAs=Image
      imageType=Vector
      mirroring=NoMirror
      resourceName=../../../Engine/Content/Slate/Starship/CoreWidgets/CheckBox/indeterminate.svg
      resourceObject=null
      tiling=NoTile
      imageSize=<16 16>
      margin=<ltrb 0 0 0 0>
      outlineSettings {
       bUseBrushTransparency=false
       roundingType=HalfHeightRadius
       width=0
       color {colorUseRule=UseColor_Specified, specifiedColor=<rgba 0 0 0 0>}
       cornerRadii=<0 0 0 0>
      }
      tintColor {colorUseRule=UseColor_Specified, specifiedColor=<rgba 0 0.0802 0.3515 1>}
      uVRegion {bIsValid=false, max=<0 0>, min=<0 0>}
     }
    }
   }
   slot {class=VerticalBoxSlot}
  }
  HorizontalBox_142 {widgetClassPath=&HorizontalBox, parent=ToggleButton_Collapse_ActorPosUtil, widget=HorizontalBox_142, slot {class=PanelSlot}}
  Spacer_233 {
   widgetClassPath=&Spacer
   parent=HorizontalBox_142
   widget=Spacer_233
   properties {size=<12 1>}
   slot {class=HorizontalBoxSlot, properties {horizontalAlignment=HAlign_Left}}
  }
  T_ToggleButton_ActorPosUtil {
   widgetClassPath=&TextBlock
   bIsVariable=true
   parent=HorizontalBox_142
   widget=T_ToggleButton_ActorPosUtil
   properties {
    text="Position Utilities ▲"
    font {
     bForceMonospaced=false
     bMaterialIsStencil=false
     fontMaterial=null
     fontObject=&Roboto
     letterSpacing=0
     monospacedWidth=1
     size=12
     skewAmount=0
     typefaceFontName=Bold
     outlineSettings {
      bApplyOutlineToDropShadows=false
      bMiteredCorners=false
      bSeparateFillAlpha=false
      outlineMaterial=null
      outlineSize=0
      outlineColor=<rgba 0 0 0 1>
     }
    }
   }
   slot {class=HorizontalBoxSlot}
  }
  Socket_ActorPosUtil {
   widgetClassPath=&VerticalBox
   bIsVariable=true
   parent=VerticalBox_9
   widget=Socket_ActorPosUtil
   slot {class=VerticalBoxSlot, properties {padding=<ltrb 10 0 0 20>}}
  }
  HorizontalBox-StoredPosition {widgetClassPath=&HorizontalBox, parent=Socket_ActorPosUtil, widget=HorizontalBox-StoredPosition, slot {class=VerticalBoxSlot}}
  VerticalBox_62 {
   widgetClassPath=&VerticalBox
   parent=HorizontalBox-StoredPosition
   widget=VerticalBox_62
   slot {class=HorizontalBoxSlot, properties {size {sizeRule=Fill, value=1}}}
  }
  VerticalBox-APU-StorePos {
   widgetClassPath=&VerticalBox
   parent=VerticalBox_62
   widget=VerticalBox-APU-StorePos
   slot {class=VerticalBoxSlot, properties {padding=<ltrb 10 10 10 10>}}
  }
  SizeBox_2 {
   widgetClassPath=&SizeBox
   parent=VerticalBox-APU-StorePos
   widget=SizeBox_2
   properties {bOverride_MinDesiredHeight=true, minDesiredHeight=30}
   slot {class=VerticalBoxSlot}
  }
  B_StorePosition {
   widgetClassPath=&EditorUtilityButton
   parent=SizeBox_2
   widget=B_StorePosition
   properties {
    toolTipText="Stores the XYZ position of selected actor. This position can be referenced in other modules.\r\n\r\nStored position is indicated in the world with magenta box."
    widgetStyle {
     clickedSlateSound {resourceObject=null}
     disabled {
      bIsDynamicallyLoaded=false
      drawAs=RoundedBox
      imageType=NoImage
      mirroring=NoMirror
      resourceName=None
      resourceObject=null
      tiling=NoTile
      imageSize=<0 0>
      margin=<ltrb 0 0 0 0>
      outlineSettings {
       bUseBrushTransparency=false
       roundingType=FixedRadius
       width=1
       color {colorUseRule=UseColor_ColorTable, specifiedColor=<rgba 0 0 0 0>}
       cornerRadii=<4 4 4 4>
      }
      tintColor {colorUseRule=UseColor_Specified, specifiedColor=<rgba 0.0395 0.0395 0.0395 1>}
      uVRegion {bIsValid=false, max=<0 0>, min=<0 0>}
     }
     disabledForeground {colorUseRule=UseColor_ColorTable, specifiedColor=<rgba 0 0 0 0>}
     hovered {
      bIsDynamicallyLoaded=false
      drawAs=RoundedBox
      imageType=NoImage
      mirroring=NoMirror
      resourceName=None
      resourceObject=null
      tiling=NoTile
      imageSize=<32 32>
      margin=<ltrb 0 0 0 0>
      outlineSettings {
       bUseBrushTransparency=false
       roundingType=FixedRadius
       width=1
       color {colorUseRule=UseColor_ColorTable, specifiedColor=<rgba 0 0 0 0>}
       cornerRadii=<4 4 4 4>
      }
      tintColor {colorUseRule=UseColor_Specified, specifiedColor=<rgba 0.0953 0.0953 0.0953 1>}
      uVRegion {bIsValid=false, max=<0 0>, min=<0 0>}
     }
     hoveredForeground {colorUseRule=UseColor_ColorTable, specifiedColor=<rgba 0 0 0 0>}
     hoveredSlateSound {resourceObject=null}
     normal {
      bIsDynamicallyLoaded=false
      drawAs=RoundedBox
      imageType=NoImage
      mirroring=NoMirror
      resourceName=None
      resourceObject=null
      tiling=NoTile
      imageSize=<32 32>
      margin=<ltrb 0 0 0 0>
      outlineSettings {
       bUseBrushTransparency=false
       roundingType=FixedRadius
       width=1
       color {colorUseRule=UseColor_ColorTable, specifiedColor=<rgba 0 0 0 0>}
       cornerRadii=<4 4 4 4>
      }
      tintColor {colorUseRule=UseColor_Specified, specifiedColor=<rgba 0.0395 0.0395 0.0395 1>}
      uVRegion {bIsValid=false, max=<0 0>, min=<0 0>}
     }
     normalForeground {colorUseRule=UseColor_ColorTable, specifiedColor=<rgba 0 0 0 0>}
     normalPadding=<ltrb 12 1.5 12 1.5>
     pressed {
      bIsDynamicallyLoaded=false
      drawAs=RoundedBox
      imageType=NoImage
      mirroring=NoMirror
      resourceName=None
      resourceObject=null
      tiling=NoTile
      imageSize=<32 32>
      margin=<ltrb 0 0 0 0>
      outlineSettings {
       bUseBrushTransparency=false
       roundingType=FixedRadius
       width=1
       color {colorUseRule=UseColor_ColorTable, specifiedColor=<rgba 0 0 0 0>}
       cornerRadii=<4 4 4 4>
      }
      tintColor {colorUseRule=UseColor_Specified, specifiedColor=<rgba 0.0284 0.0284 0.0284 1>}
      uVRegion {bIsValid=false, max=<0 0>, min=<0 0>}
     }
     pressedForeground {colorUseRule=UseColor_ColorTable, specifiedColor=<rgba 0 0 0 0>}
     pressedPadding=<ltrb 12 2.5 12 0.5>
     pressedSlateSound {resourceObject=null}
    }
   }
   slot {class=SizeBoxSlot}
  }
  TextBlock_8 {
   widgetClassPath=&TextBlock
   parent=B_StorePosition
   widget=TextBlock_8
   properties {
    text=Store Position
    font {
     bForceMonospaced=false
     bMaterialIsStencil=false
     fontMaterial=null
     fontObject=&Roboto
     letterSpacing=0
     monospacedWidth=1
     size=9
     skewAmount=0
     typefaceFontName=Regular
     outlineSettings {
      bApplyOutlineToDropShadows=false
      bMiteredCorners=false
      bSeparateFillAlpha=false
      outlineMaterial=null
      outlineSize=0
      outlineColor=<rgba 0 0 0 1>
     }
    }
   }
   slot {class=ButtonSlot}
  }
  T_StoredPostionDisplay {
   widgetClassPath=&TextBlock
   parent=VerticalBox-APU-StorePos
   widget=T_StoredPostionDisplay
   properties {
    justification=Center
    text=0 / 0 / 0
    colorAndOpacity {colorUseRule=UseColor_Specified, specifiedColor=<rgba 0.3021 0.3021 0.3021 1>}
    font {
     bForceMonospaced=false
     bMaterialIsStencil=false
     fontMaterial=null
     fontObject=&Roboto
     letterSpacing=0
     monospacedWidth=1
     size=10.5
     skewAmount=0
     typefaceFontName=Bold
     outlineSettings {
      bApplyOutlineToDropShadows=false
      bMiteredCorners=false
      bSeparateFillAlpha=false
      outlineMaterial=null
      outlineSize=0
      outlineColor=<rgba 0 0 0 1>
     }
    }
   }
   slot {class=VerticalBoxSlot, properties {verticalAlignment=VAlign_Center, padding=<ltrb 0 3 0 0>}}
  }
  SizeBox_0 {
   widgetClassPath=&SizeBox
   parent=VerticalBox_62
   widget=SizeBox_0
   properties {bOverride_MinDesiredHeight=true, minDesiredHeight=30}
   slot {class=VerticalBoxSlot, properties {padding=<ltrb 10 10 10 10>, size {sizeRule=Fill, value=1}}}
  }
  CB_Toggle_SpawnMarkerObject {
   widgetClassPath=&EditorUtilityCheckBox
   bIsVariable=true
   parent=SizeBox_0
   widget=CB_Toggle_SpawnMarkerObject
   properties {
    toolTipText="Spawns a Marker Object into the scene. \r\nThis object can be used for various reposition utilities in this tool, such as arranging in grid and sorting by class."
    widgetStyle {
     checkBoxType=ToggleButton
     backgroundHoveredImage {
      bIsDynamicallyLoaded=false
      drawAs=RoundedBox
      imageType=NoImage
      mirroring=NoMirror
      resourceName=None
      resourceObject=null
      tiling=NoTile
      imageSize=<18 18>
      margin=<ltrb 0 0 0 0>
      outlineSettings {
       bUseBrushTransparency=false
       roundingType=FixedRadius
       width=1
       color {colorUseRule=UseColor_ColorTable, specifiedColor=<rgba 0 0 0 0>}
       cornerRadii=<3 3 3 3>
      }
      tintColor {colorUseRule=UseColor_Specified, specifiedColor=<rgba 0.0048 0.0048 0.0048 1>}
      uVRegion {bIsValid=false, max=<0 0>, min=<0 0>}
     }
     backgroundImage {
      bIsDynamicallyLoaded=false
      drawAs=RoundedBox
      imageType=NoImage
      mirroring=NoMirror
      resourceName=None
      resourceObject=null
      tiling=NoTile
      imageSize=<18 18>
      margin=<ltrb 0 0 0 0>
      outlineSettings {
       bUseBrushTransparency=false
       roundingType=FixedRadius
       width=1
       color {colorUseRule=UseColor_ColorTable, specifiedColor=<rgba 0 0 0 0>}
       cornerRadii=<3 3 3 3>
      }
      tintColor {colorUseRule=UseColor_Specified, specifiedColor=<rgba 0.0048 0.0048 0.0048 1>}
      uVRegion {bIsValid=false, max=<0 0>, min=<0 0>}
     }
     backgroundPressedImage {
      bIsDynamicallyLoaded=false
      drawAs=RoundedBox
      imageType=NoImage
      mirroring=NoMirror
      resourceName=None
      resourceObject=null
      tiling=NoTile
      imageSize=<18 18>
      margin=<ltrb 0 0 0 0>
      outlineSettings {
       bUseBrushTransparency=false
       roundingType=FixedRadius
       width=1
       color {colorUseRule=UseColor_ColorTable, specifiedColor=<rgba 0 0 0 0>}
       cornerRadii=<3 3 3 3>
      }
      tintColor {colorUseRule=UseColor_Specified, specifiedColor=<rgba 0.0048 0.0048 0.0048 1>}
      uVRegion {bIsValid=false, max=<0 0>, min=<0 0>}
     }
     borderBackgroundColor {colorUseRule=UseColor_Specified, specifiedColor=<rgba 1 1 1 1>}
     checkedForeground {colorUseRule=UseColor_Specified, specifiedColor=<rgba 1 1 1 1>}
     checkedHoveredForeground {colorUseRule=UseColor_Specified, specifiedColor=<rgba 1 1 1 1>}
     checkedHoveredImage {
      bIsDynamicallyLoaded=false
      drawAs=RoundedBox
      imageType=NoImage
      mirroring=NoMirror
      resourceName=None
      resourceObject=null
      tiling=NoTile
      imageSize=<32 32>
      margin=<ltrb 0 0 0 0>
      outlineSettings {
       bUseBrushTransparency=false
       roundingType=FixedRadius
       width=1
       color {colorUseRule=UseColor_ColorTable, specifiedColor=<rgba 0 0 0 0>}
       cornerRadii=<4 4 4 4>
      }
      tintColor {colorUseRule=UseColor_Specified, specifiedColor=<rgba 0.0477 0.2366 0.9167 1>}
      uVRegion {bIsValid=false, max=<0 0>, min=<0 0>}
     }
     checkedImage {
      bIsDynamicallyLoaded=false
      drawAs=RoundedBox
      imageType=NoImage
      mirroring=NoMirror
      resourceName=None
      resourceObject=null
      tiling=NoTile
      imageSize=<32 32>
      margin=<ltrb 0 0 0 0>
      outlineSettings {
       bUseBrushTransparency=false
       roundingType=FixedRadius
       width=1
       color {colorUseRule=UseColor_ColorTable, specifiedColor=<rgba 0 0 0 0>}
       cornerRadii=<4 4 4 4>
      }
      tintColor {colorUseRule=UseColor_Specified, specifiedColor=<rgba 0 0.162 0.7454 1>}
      uVRegion {bIsValid=false, max=<0 0>, min=<0 0>}
     }
     checkedPressedForeground {colorUseRule=UseColor_Specified, specifiedColor=<rgba 1 1 1 1>}
     checkedPressedImage {
      bIsDynamicallyLoaded=false
      drawAs=RoundedBox
      imageType=NoImage
      mirroring=NoMirror
      resourceName=None
      resourceObject=null
      tiling=NoTile
      imageSize=<32 32>
      margin=<ltrb 0 0 0 0>
      outlineSettings {
       bUseBrushTransparency=false
       roundingType=FixedRadius
       width=1
       color {colorUseRule=UseColor_ColorTable, specifiedColor=<rgba 0 0 0 0>}
       cornerRadii=<4 4 4 4>
      }
      tintColor {colorUseRule=UseColor_Specified, specifiedColor=<rgba 0 0.0679 0.3125 1>}
      uVRegion {bIsValid=false, max=<0 0>, min=<0 0>}
     }
     checkedSlateSound {resourceObject=null}
     foregroundColor {colorUseRule=UseColor_Specified, specifiedColor=<rgba 1 1 1 1>}
     hoveredForeground {colorUseRule=UseColor_Specified, specifiedColor=<rgba 1 1 1 1>}
     hoveredSlateSound {resourceObject=null}
     padding=<ltrb 0 6 0 0>
     pressedForeground {colorUseRule=UseColor_Specified, specifiedColor=<rgba 1 1 1 1>}
     uncheckedHoveredImage {
      bIsDynamicallyLoaded=false
      drawAs=RoundedBox
      imageType=NoImage
      mirroring=NoMirror
      resourceName=None
      resourceObject=null
      tiling=NoTile
      imageSize=<32 32>
      margin=<ltrb 0 0 0 0>
      outlineSettings {
       bUseBrushTransparency=false
       roundingType=FixedRadius
       width=1
       color {colorUseRule=UseColor_ColorTable, specifiedColor=<rgba 0 0 0 0>}
       cornerRadii=<4 4 4 4>
      }
      tintColor {colorUseRule=UseColor_Specified, specifiedColor=<rgba 0.0953 0.0953 0.0953 1>}
      uVRegion {bIsValid=false, max=<0 0>, min=<0 0>}
     }
     uncheckedImage {
      bIsDynamicallyLoaded=false
      drawAs=RoundedBox
      imageType=NoImage
      mirroring=NoMirror
      resourceName=None
      resourceObject=null
      tiling=NoTile
      imageSize=<32 32>
      margin=<ltrb 0 0 0 0>
      outlineSettings {
       bUseBrushTransparency=false
       roundingType=FixedRadius
       width=1
       color {colorUseRule=UseColor_ColorTable, specifiedColor=<rgba 0 0 0 0>}
       cornerRadii=<4 4 4 4>
      }
      tintColor {colorUseRule=UseColor_Specified, specifiedColor=<rgba 0.0395 0.0395 0.0395 1>}
      uVRegion {bIsValid=false, max=<0 0>, min=<0 0>}
     }
     uncheckedPressedImage {
      bIsDynamicallyLoaded=false
      drawAs=RoundedBox
      imageType=NoImage
      mirroring=NoMirror
      resourceName=None
      resourceObject=null
      tiling=NoTile
      imageSize=<32 32>
      margin=<ltrb 0 0 0 0>
      outlineSettings {
       bUseBrushTransparency=false
       roundingType=FixedRadius
       width=1
       color {colorUseRule=UseColor_ColorTable, specifiedColor=<rgba 0 0 0 0>}
       cornerRadii=<4 4 4 4>
      }
      tintColor {colorUseRule=UseColor_Specified, specifiedColor=<rgba 0.0284 0.0284 0.0284 1>}
      uVRegion {bIsValid=false, max=<0 0>, min=<0 0>}
     }
     uncheckedSlateSound {resourceObject=null}
     undeterminedForeground {colorUseRule=UseColor_Specified, specifiedColor=<rgba 1 1 1 1>}
     undeterminedHoveredImage {
      bIsDynamicallyLoaded=false
      drawAs=RoundedBox
      imageType=NoImage
      mirroring=NoMirror
      resourceName=None
      resourceObject=null
      tiling=NoTile
      imageSize=<32 32>
      margin=<ltrb 0 0 0 0>
      outlineSettings {
       bUseBrushTransparency=false
       roundingType=FixedRadius
       width=1
       color {colorUseRule=UseColor_ColorTable, specifiedColor=<rgba 0 0 0 0>}
       cornerRadii=<4 4 4 4>
      }
      tintColor {colorUseRule=UseColor_Specified, specifiedColor=<rgba 0.0395 0.0395 0.0395 1>}
      uVRegion {bIsValid=false, max=<0 0>, min=<0 0>}
     }
     undeterminedImage {
      bIsDynamicallyLoaded=false
      drawAs=RoundedBox
      imageType=NoImage
      mirroring=NoMirror
      resourceName=None
      resourceObject=null
      tiling=NoTile
      imageSize=<32 32>
      margin=<ltrb 0 0 0 0>
      outlineSettings {
       bUseBrushTransparency=false
       roundingType=FixedRadius
       width=1
       color {colorUseRule=UseColor_ColorTable, specifiedColor=<rgba 0 0 0 0>}
       cornerRadii=<4 4 4 4>
      }
      tintColor {colorUseRule=UseColor_Specified, specifiedColor=<rgba 0.0395 0.0395 0.0395 1>}
      uVRegion {bIsValid=false, max=<0 0>, min=<0 0>}
     }
     undeterminedPressedImage {
      bIsDynamicallyLoaded=false
      drawAs=RoundedBox
      imageType=NoImage
      mirroring=NoMirror
      resourceName=None
      resourceObject=null
      tiling=NoTile
      imageSize=<32 32>
      margin=<ltrb 0 0 0 0>
      outlineSettings {
       bUseBrushTransparency=false
       roundingType=FixedRadius
       width=1
       color {colorUseRule=UseColor_ColorTable, specifiedColor=<rgba 0 0 0 0>}
       cornerRadii=<4 4 4 4>
      }
      tintColor {colorUseRule=UseColor_Specified, specifiedColor=<rgba 0.0395 0.0395 0.0395 1>}
      uVRegion {bIsValid=false, max=<0 0>, min=<0 0>}
     }
    }
   }
   slot {class=SizeBoxSlot}
  }
  T_Toggle_SpawnMarkerObject {
   widgetClassPath=&TextBlock
   bIsVariable=true
   parent=CB_Toggle_SpawnMarkerObject
   widget=T_Toggle_SpawnMarkerObject
   properties {
    justification=Center
    text=Spawn Marker Object
    toolTipText="Spawns / Despawns a marker actor in the scene. \r\n\r\nThis actor can be used as a position marker to be used in other modules"
    font {
     bForceMonospaced=false
     bMaterialIsStencil=false
     fontMaterial=null
     fontObject=&Roboto
     letterSpacing=0
     monospacedWidth=1
     size=9
     skewAmount=0
     typefaceFontName=Regular
     outlineSettings {
      bApplyOutlineToDropShadows=false
      bMiteredCorners=false
      bSeparateFillAlpha=false
      outlineMaterial=null
      outlineSize=0
      outlineColor=<rgba 0 0 0 1>
     }
    }
   }
   slot {class=PanelSlot}
  }
  VerticalBox_117 {
   widgetClassPath=&VerticalBox
   parent=HorizontalBox-StoredPosition
   widget=VerticalBox_117
   slot {class=HorizontalBoxSlot, properties {size {sizeRule=Fill, value=1}}}
  }
  SizeBox {
   widgetClassPath=&SizeBox
   parent=VerticalBox_117
   widget=SizeBox
   properties {bOverride_MinDesiredHeight=true, minDesiredHeight=30}
   slot {class=VerticalBoxSlot, properties {padding=<ltrb 10 10 10 10>}}
  }
  CB_Toggle_PositionVisibility {
   widgetClassPath=&EditorUtilityCheckBox
   bIsVariable=true
   parent=SizeBox
   widget=CB_Toggle_PositionVisibility
   properties {
    checkedState=Checked
    toolTipText=Toggle visibility of the Stored Position widget.
    widgetStyle {
     checkBoxType=CheckBox
     backgroundHoveredImage {
      bIsDynamicallyLoaded=false
      drawAs=RoundedBox
      imageType=NoImage
      mirroring=NoMirror
      resourceName=None
      resourceObject=null
      tiling=NoTile
      imageSize=<18 18>
      margin=<ltrb 0 0 0 0>
      outlineSettings {
       bUseBrushTransparency=false
       roundingType=FixedRadius
       width=1
       color {colorUseRule=UseColor_ColorTable, specifiedColor=<rgba 0 0 0 0>}
       cornerRadii=<3 3 3 3>
      }
      tintColor {colorUseRule=UseColor_Specified, specifiedColor=<rgba 0.0048 0.0048 0.0048 1>}
      uVRegion {bIsValid=false, max=<0 0>, min=<0 0>}
     }
     backgroundImage {
      bIsDynamicallyLoaded=false
      drawAs=RoundedBox
      imageType=NoImage
      mirroring=NoMirror
      resourceName=None
      resourceObject=null
      tiling=NoTile
      imageSize=<18 18>
      margin=<ltrb 0 0 0 0>
      outlineSettings {
       bUseBrushTransparency=false
       roundingType=FixedRadius
       width=1
       color {colorUseRule=UseColor_ColorTable, specifiedColor=<rgba 0 0 0 0>}
       cornerRadii=<3 3 3 3>
      }
      tintColor {colorUseRule=UseColor_Specified, specifiedColor=<rgba 0.0048 0.0048 0.0048 1>}
      uVRegion {bIsValid=false, max=<0 0>, min=<0 0>}
     }
     backgroundPressedImage {
      bIsDynamicallyLoaded=false
      drawAs=RoundedBox
      imageType=NoImage
      mirroring=NoMirror
      resourceName=None
      resourceObject=null
      tiling=NoTile
      imageSize=<18 18>
      margin=<ltrb 0 0 0 0>
      outlineSettings {
       bUseBrushTransparency=false
       roundingType=FixedRadius
       width=1
       color {colorUseRule=UseColor_ColorTable, specifiedColor=<rgba 0 0 0 0>}
       cornerRadii=<3 3 3 3>
      }
      tintColor {colorUseRule=UseColor_Specified, specifiedColor=<rgba 0.0048 0.0048 0.0048 1>}
      uVRegion {bIsValid=false, max=<0 0>, min=<0 0>}
     }
     borderBackgroundColor {colorUseRule=UseColor_Specified, specifiedColor=<rgba 1 1 1 1>}
     checkedForeground {colorUseRule=UseColor_Specified, specifiedColor=<rgba 1 1 1 1>}
     checkedHoveredForeground {colorUseRule=UseColor_Specified, specifiedColor=<rgba 1 1 1 1>}
     checkedHoveredImage {
      bIsDynamicallyLoaded=false
      drawAs=Image
      imageType=Vector
      mirroring=NoMirror
      resourceName=../../../Engine/Content/Slate/Starship/CoreWidgets/CheckBox/check.svg
      resourceObject=null
      tiling=NoTile
      imageSize=<16 16>
      margin=<ltrb 0 0 0 0>
      outlineSettings {
       bUseBrushTransparency=false
       roundingType=HalfHeightRadius
       width=0
       color {colorUseRule=UseColor_Specified, specifiedColor=<rgba 0 0 0 0>}
       cornerRadii=<0 0 0 0>
      }
      tintColor {colorUseRule=UseColor_Specified, specifiedColor=<rgba 0.0044 0.2384 1 1>}
      uVRegion {bIsValid=false, max=<0 0>, min=<0 0>}
     }
     checkedImage {
      bIsDynamicallyLoaded=false
      drawAs=Image
      imageType=Vector
      mirroring=NoMirror
      resourceName=../../../Engine/Content/Slate/Starship/CoreWidgets/CheckBox/check.svg
      resourceObject=null
      tiling=NoTile
      imageSize=<16 16>
      margin=<ltrb 0 0 0 0>
      outlineSettings {
       bUseBrushTransparency=false
       roundingType=HalfHeightRadius
       width=0
       color {colorUseRule=UseColor_Specified, specifiedColor=<rgba 0 0 0 0>}
       cornerRadii=<0 0 0 0>
      }
      tintColor {colorUseRule=UseColor_Specified, specifiedColor=<rgba 0 0.162 0.7454 1>}
      uVRegion {bIsValid=false, max=<0 0>, min=<0 0>}
     }
     checkedPressedForeground {colorUseRule=UseColor_Specified, specifiedColor=<rgba 1 1 1 1>}
     checkedPressedImage {
      bIsDynamicallyLoaded=false
      drawAs=Image
      imageType=Vector
      mirroring=NoMirror
      resourceName=../../../Engine/Content/Slate/Starship/CoreWidgets/CheckBox/check.svg
      resourceObject=null
      tiling=NoTile
      imageSize=<16 16>
      margin=<ltrb 0 0 0 0>
      outlineSettings {
       bUseBrushTransparency=false
       roundingType=HalfHeightRadius
       width=0
       color {colorUseRule=UseColor_Specified, specifiedColor=<rgba 0 0 0 0>}
       cornerRadii=<0 0 0 0>
      }
      tintColor {colorUseRule=UseColor_Specified, specifiedColor=<rgba 0 0.0802 0.3515 1>}
      uVRegion {bIsValid=false, max=<0 0>, min=<0 0>}
     }
     checkedSlateSound {resourceObject=null}
     foregroundColor {colorUseRule=UseColor_Specified, specifiedColor=<rgba 1 1 1 1>}
     hoveredForeground {colorUseRule=UseColor_Specified, specifiedColor=<rgba 1 1 1 1>}
     hoveredSlateSound {resourceObject=null}
     padding=<ltrb 2 0 0 0>
     pressedForeground {colorUseRule=UseColor_Specified, specifiedColor=<rgba 1 1 1 1>}
     uncheckedHoveredImage {
      bIsDynamicallyLoaded=false
      drawAs=NoDrawType
      imageType=NoImage
      mirroring=NoMirror
      resourceName=None
      resourceObject=null
      tiling=NoTile
      imageSize=<0 0>
      margin=<ltrb 0 0 0 0>
      outlineSettings {
       bUseBrushTransparency=false
       roundingType=HalfHeightRadius
       width=0
       color {colorUseRule=UseColor_Specified, specifiedColor=<rgba 0 0 0 0>}
       cornerRadii=<0 0 0 0>
      }
      tintColor {colorUseRule=UseColor_Specified, specifiedColor=<rgba 1 1 1 1>}
      uVRegion {bIsValid=false, max=<0 0>, min=<0 0>}
     }
     uncheckedImage {
      bIsDynamicallyLoaded=false
      drawAs=NoDrawType
      imageType=NoImage
      mirroring=NoMirror
      resourceName=None
      resourceObject=null
      tiling=NoTile
      imageSize=<0 0>
      margin=<ltrb 0 0 0 0>
      outlineSettings {
       bUseBrushTransparency=false
       roundingType=HalfHeightRadius
       width=0
       color {colorUseRule=UseColor_Specified, specifiedColor=<rgba 0 0 0 0>}
       cornerRadii=<0 0 0 0>
      }
      tintColor {colorUseRule=UseColor_Specified, specifiedColor=<rgba 1 1 1 1>}
      uVRegion {bIsValid=false, max=<0 0>, min=<0 0>}
     }
     uncheckedPressedImage {
      bIsDynamicallyLoaded=false
      drawAs=NoDrawType
      imageType=NoImage
      mirroring=NoMirror
      resourceName=None
      resourceObject=null
      tiling=NoTile
      imageSize=<0 0>
      margin=<ltrb 0 0 0 0>
      outlineSettings {
       bUseBrushTransparency=false
       roundingType=HalfHeightRadius
       width=0
       color {colorUseRule=UseColor_Specified, specifiedColor=<rgba 0 0 0 0>}
       cornerRadii=<0 0 0 0>
      }
      tintColor {colorUseRule=UseColor_Specified, specifiedColor=<rgba 1 1 1 1>}
      uVRegion {bIsValid=false, max=<0 0>, min=<0 0>}
     }
     uncheckedSlateSound {resourceObject=null}
     undeterminedForeground {colorUseRule=UseColor_Specified, specifiedColor=<rgba 1 1 1 1>}
     undeterminedHoveredImage {
      bIsDynamicallyLoaded=false
      drawAs=Image
      imageType=Vector
      mirroring=NoMirror
      resourceName=../../../Engine/Content/Slate/Starship/CoreWidgets/CheckBox/indeterminate.svg
      resourceObject=null
      tiling=NoTile
      imageSize=<16 16>
      margin=<ltrb 0 0 0 0>
      outlineSettings {
       bUseBrushTransparency=false
       roundingType=HalfHeightRadius
       width=0
       color {colorUseRule=UseColor_Specified, specifiedColor=<rgba 0 0 0 0>}
       cornerRadii=<0 0 0 0>
      }
      tintColor {colorUseRule=UseColor_Specified, specifiedColor=<rgba 0.0044 0.2384 1 1>}
      uVRegion {bIsValid=false, max=<0 0>, min=<0 0>}
     }
     undeterminedImage {
      bIsDynamicallyLoaded=false
      drawAs=Image
      imageType=Vector
      mirroring=NoMirror
      resourceName=../../../Engine/Content/Slate/Starship/CoreWidgets/CheckBox/indeterminate.svg
      resourceObject=null
      tiling=NoTile
      imageSize=<16 16>
      margin=<ltrb 0 0 0 0>
      outlineSettings {
       bUseBrushTransparency=false
       roundingType=HalfHeightRadius
       width=0
       color {colorUseRule=UseColor_Specified, specifiedColor=<rgba 0 0 0 0>}
       cornerRadii=<0 0 0 0>
      }
      tintColor {colorUseRule=UseColor_Specified, specifiedColor=<rgba 0 0.162 0.7454 1>}
      uVRegion {bIsValid=false, max=<0 0>, min=<0 0>}
     }
     undeterminedPressedImage {
      bIsDynamicallyLoaded=false
      drawAs=Image
      imageType=Vector
      mirroring=NoMirror
      resourceName=../../../Engine/Content/Slate/Starship/CoreWidgets/CheckBox/indeterminate.svg
      resourceObject=null
      tiling=NoTile
      imageSize=<16 16>
      margin=<ltrb 0 0 0 0>
      outlineSettings {
       bUseBrushTransparency=false
       roundingType=HalfHeightRadius
       width=0
       color {colorUseRule=UseColor_Specified, specifiedColor=<rgba 0 0 0 0>}
       cornerRadii=<0 0 0 0>
      }
      tintColor {colorUseRule=UseColor_Specified, specifiedColor=<rgba 0 0.0802 0.3515 1>}
      uVRegion {bIsValid=false, max=<0 0>, min=<0 0>}
     }
    }
   }
   slot {class=SizeBoxSlot}
  }
  T_TogglePositionVisibility {
   widgetClassPath=&TextBlock
   parent=CB_Toggle_PositionVisibility
   widget=T_TogglePositionVisibility
   properties {
    justification=Center
    text=Show Stored Position
    font {
     bForceMonospaced=false
     bMaterialIsStencil=false
     fontMaterial=null
     fontObject=&Roboto
     letterSpacing=0
     monospacedWidth=1
     size=9
     skewAmount=0
     typefaceFontName=Regular
     outlineSettings {
      bApplyOutlineToDropShadows=false
      bMiteredCorners=false
      bSeparateFillAlpha=false
      outlineMaterial=null
      outlineSize=0
      outlineColor=<rgba 0 0 0 1>
     }
    }
   }
   slot {class=PanelSlot}
  }
  SizeBox_3 {
   widgetClassPath=&SizeBox
   parent=VerticalBox_117
   widget=SizeBox_3
   properties {bOverride_MinDesiredHeight=true, minDesiredHeight=30}
   slot {class=VerticalBoxSlot, properties {padding=<ltrb 10 10 10 10>}}
  }
  CB_Toggle_GridSnapping {
   widgetClassPath=&EditorUtilityCheckBox
   bIsVariable=true
   parent=SizeBox_3
   widget=CB_Toggle_GridSnapping
   properties {
    checkedState=Checked
    toolTipText=EDITOR COMMAND LINE : Toggles automatic snapping of building actors to the grid
    widgetStyle {
     checkBoxType=CheckBox
     backgroundHoveredImage {
      bIsDynamicallyLoaded=false
      drawAs=RoundedBox
      imageType=NoImage
      mirroring=NoMirror
      resourceName=None
      resourceObject=null
      tiling=NoTile
      imageSize=<18 18>
      margin=<ltrb 0 0 0 0>
      outlineSettings {
       bUseBrushTransparency=false
       roundingType=FixedRadius
       width=1
       color {colorUseRule=UseColor_ColorTable, specifiedColor=<rgba 0 0 0 0>}
       cornerRadii=<3 3 3 3>
      }
      tintColor {colorUseRule=UseColor_Specified, specifiedColor=<rgba 0.0048 0.0048 0.0048 1>}
      uVRegion {bIsValid=false, max=<0 0>, min=<0 0>}
     }
     backgroundImage {
      bIsDynamicallyLoaded=false
      drawAs=RoundedBox
      imageType=NoImage
      mirroring=NoMirror
      resourceName=None
      resourceObject=null
      tiling=NoTile
      imageSize=<18 18>
      margin=<ltrb 0 0 0 0>
      outlineSettings {
       bUseBrushTransparency=false
       roundingType=FixedRadius
       width=1
       color {colorUseRule=UseColor_ColorTable, specifiedColor=<rgba 0 0 0 0>}
       cornerRadii=<3 3 3 3>
      }
      tintColor {colorUseRule=UseColor_Specified, specifiedColor=<rgba 0.0048 0.0048 0.0048 1>}
      uVRegion {bIsValid=false, max=<0 0>, min=<0 0>}
     }
     backgroundPressedImage {
      bIsDynamicallyLoaded=false
      drawAs=RoundedBox
      imageType=NoImage
      mirroring=NoMirror
      resourceName=None
      resourceObject=null
      tiling=NoTile
      imageSize=<18 18>
      margin=<ltrb 0 0 0 0>
      outlineSettings {
       bUseBrushTransparency=false
       roundingType=FixedRadius
       width=1
       color {colorUseRule=UseColor_ColorTable, specifiedColor=<rgba 0 0 0 0>}
       cornerRadii=<3 3 3 3>
      }
      tintColor {colorUseRule=UseColor_Specified, specifiedColor=<rgba 0.0048 0.0048 0.0048 1>}
      uVRegion {bIsValid=false, max=<0 0>, min=<0 0>}
     }
     borderBackgroundColor {colorUseRule=UseColor_Specified, specifiedColor=<rgba 1 1 1 1>}
     checkedForeground {colorUseRule=UseColor_Specified, specifiedColor=<rgba 1 1 1 1>}
     checkedHoveredForeground {colorUseRule=UseColor_Specified, specifiedColor=<rgba 1 1 1 1>}
     checkedHoveredImage {
      bIsDynamicallyLoaded=false
      drawAs=Image
      imageType=Vector
      mirroring=NoMirror
      resourceName=../../../Engine/Content/Slate/Starship/CoreWidgets/CheckBox/check.svg
      resourceObject=null
      tiling=NoTile
      imageSize=<16 16>
      margin=<ltrb 0 0 0 0>
      outlineSettings {
       bUseBrushTransparency=false
       roundingType=HalfHeightRadius
       width=0
       color {colorUseRule=UseColor_Specified, specifiedColor=<rgba 0 0 0 0>}
       cornerRadii=<0 0 0 0>
      }
      tintColor {colorUseRule=UseColor_Specified, specifiedColor=<rgba 0.0044 0.2384 1 1>}
      uVRegion {bIsValid=false, max=<0 0>, min=<0 0>}
     }
     checkedImage {
      bIsDynamicallyLoaded=false
      drawAs=Image
      imageType=Vector
      mirroring=NoMirror
      resourceName=../../../Engine/Content/Slate/Starship/CoreWidgets/CheckBox/check.svg
      resourceObject=null
      tiling=NoTile
      imageSize=<16 16>
      margin=<ltrb 0 0 0 0>
      outlineSettings {
       bUseBrushTransparency=false
       roundingType=HalfHeightRadius
       width=0
       color {colorUseRule=UseColor_Specified, specifiedColor=<rgba 0 0 0 0>}
       cornerRadii=<0 0 0 0>
      }
      tintColor {colorUseRule=UseColor_Specified, specifiedColor=<rgba 0 0.162 0.7454 1>}
      uVRegion {bIsValid=false, max=<0 0>, min=<0 0>}
     }
     checkedPressedForeground {colorUseRule=UseColor_Specified, specifiedColor=<rgba 1 1 1 1>}
     checkedPressedImage {
      bIsDynamicallyLoaded=false
      drawAs=Image
      imageType=Vector
      mirroring=NoMirror
      resourceName=../../../Engine/Content/Slate/Starship/CoreWidgets/CheckBox/check.svg
      resourceObject=null
      tiling=NoTile
      imageSize=<16 16>
      margin=<ltrb 0 0 0 0>
      outlineSettings {
       bUseBrushTransparency=false
       roundingType=HalfHeightRadius
       width=0
       color {colorUseRule=UseColor_Specified, specifiedColor=<rgba 0 0 0 0>}
       cornerRadii=<0 0 0 0>
      }
      tintColor {colorUseRule=UseColor_Specified, specifiedColor=<rgba 0 0.0802 0.3515 1>}
      uVRegion {bIsValid=false, max=<0 0>, min=<0 0>}
     }
     checkedSlateSound {resourceObject=null}
     foregroundColor {colorUseRule=UseColor_Specified, specifiedColor=<rgba 1 1 1 1>}
     hoveredForeground {colorUseRule=UseColor_Specified, specifiedColor=<rgba 1 1 1 1>}
     hoveredSlateSound {resourceObject=null}
     padding=<ltrb 2 0 0 0>
     pressedForeground {colorUseRule=UseColor_Specified, specifiedColor=<rgba 1 1 1 1>}
     uncheckedHoveredImage {
      bIsDynamicallyLoaded=false
      drawAs=NoDrawType
      imageType=NoImage
      mirroring=NoMirror
      resourceName=None
      resourceObject=null
      tiling=NoTile
      imageSize=<0 0>
      margin=<ltrb 0 0 0 0>
      outlineSettings {
       bUseBrushTransparency=false
       roundingType=HalfHeightRadius
       width=0
       color {colorUseRule=UseColor_Specified, specifiedColor=<rgba 0 0 0 0>}
       cornerRadii=<0 0 0 0>
      }
      tintColor {colorUseRule=UseColor_Specified, specifiedColor=<rgba 1 1 1 1>}
      uVRegion {bIsValid=false, max=<0 0>, min=<0 0>}
     }
     uncheckedImage {
      bIsDynamicallyLoaded=false
      drawAs=NoDrawType
      imageType=NoImage
      mirroring=NoMirror
      resourceName=None
      resourceObject=null
      tiling=NoTile
      imageSize=<0 0>
      margin=<ltrb 0 0 0 0>
      outlineSettings {
       bUseBrushTransparency=false
       roundingType=HalfHeightRadius
       width=0
       color {colorUseRule=UseColor_Specified, specifiedColor=<rgba 0 0 0 0>}
       cornerRadii=<0 0 0 0>
      }
      tintColor {colorUseRule=UseColor_Specified, specifiedColor=<rgba 1 1 1 1>}
      uVRegion {bIsValid=false, max=<0 0>, min=<0 0>}
     }
     uncheckedPressedImage {
      bIsDynamicallyLoaded=false
      drawAs=NoDrawType
      imageType=NoImage
      mirroring=NoMirror
      resourceName=None
      resourceObject=null
      tiling=NoTile
      imageSize=<0 0>
      margin=<ltrb 0 0 0 0>
      outlineSettings {
       bUseBrushTransparency=false
       roundingType=HalfHeightRadius
       width=0
       color {colorUseRule=UseColor_Specified, specifiedColor=<rgba 0 0 0 0>}
       cornerRadii=<0 0 0 0>
      }
      tintColor {colorUseRule=UseColor_Specified, specifiedColor=<rgba 1 1 1 1>}
      uVRegion {bIsValid=false, max=<0 0>, min=<0 0>}
     }
     uncheckedSlateSound {resourceObject=null}
     undeterminedForeground {colorUseRule=UseColor_Specified, specifiedColor=<rgba 1 1 1 1>}
     undeterminedHoveredImage {
      bIsDynamicallyLoaded=false
      drawAs=Image
      imageType=Vector
      mirroring=NoMirror
      resourceName=../../../Engine/Content/Slate/Starship/CoreWidgets/CheckBox/indeterminate.svg
      resourceObject=null
      tiling=NoTile
      imageSize=<16 16>
      margin=<ltrb 0 0 0 0>
      outlineSettings {
       bUseBrushTransparency=false
       roundingType=HalfHeightRadius
       width=0
       color {colorUseRule=UseColor_Specified, specifiedColor=<rgba 0 0 0 0>}
       cornerRadii=<0 0 0 0>
      }
      tintColor {colorUseRule=UseColor_Specified, specifiedColor=<rgba 0.0044 0.2384 1 1>}
      uVRegion {bIsValid=false, max=<0 0>, min=<0 0>}
     }
     undeterminedImage {
      bIsDynamicallyLoaded=false
      drawAs=Image
      imageType=Vector
      mirroring=NoMirror
      resourceName=../../../Engine/Content/Slate/Starship/CoreWidgets/CheckBox/indeterminate.svg
      resourceObject=null
      tiling=NoTile
      imageSize=<16 16>
      margin=<ltrb 0 0 0 0>
      outlineSettings {
       bUseBrushTransparency=false
       roundingType=HalfHeightRadius
       width=0
       color {colorUseRule=UseColor_Specified, specifiedColor=<rgba 0 0 0 0>}
       cornerRadii=<0 0 0 0>
      }
      tintColor {colorUseRule=UseColor_Specified, specifiedColor=<rgba 0 0.162 0.7454 1>}
      uVRegion {bIsValid=false, max=<0 0>, min=<0 0>}
     }
     undeterminedPressedImage {
      bIsDynamicallyLoaded=false
      drawAs=Image
      imageType=Vector
      mirroring=NoMirror
      resourceName=../../../Engine/Content/Slate/Starship/CoreWidgets/CheckBox/indeterminate.svg
      resourceObject=null
      tiling=NoTile
      imageSize=<16 16>
      margin=<ltrb 0 0 0 0>
      outlineSettings {
       bUseBrushTransparency=false
       roundingType=HalfHeightRadius
       width=0
       color {colorUseRule=UseColor_Specified, specifiedColor=<rgba 0 0 0 0>}
       cornerRadii=<0 0 0 0>
      }
      tintColor {colorUseRule=UseColor_Specified, specifiedColor=<rgba 0 0.0802 0.3515 1>}
      uVRegion {bIsValid=false, max=<0 0>, min=<0 0>}
     }
    }
   }
   slot {class=SizeBoxSlot}
  }
  T_ToggleGridSnapping {
   widgetClassPath=&TextBlock
   parent=CB_Toggle_GridSnapping
   widget=T_ToggleGridSnapping
   properties {
    justification=Center
    text=Grid Snapping
    font {
     bForceMonospaced=false
     bMaterialIsStencil=false
     fontMaterial=null
     fontObject=&Roboto
     letterSpacing=0
     monospacedWidth=1
     size=9
     skewAmount=0
     typefaceFontName=Regular
     outlineSettings {
      bApplyOutlineToDropShadows=false
      bMiteredCorners=false
      bSeparateFillAlpha=false
      outlineMaterial=null
      outlineSize=0
      outlineColor=<rgba 0 0 0 1>
     }
    }
   }
   slot {class=PanelSlot}
  }
  Border-Cleanup {
   widgetClassPath=&Border
   parent=VerticalBox_Master
   widget=Border-Cleanup
   properties {
    background {
     bIsDynamicallyLoaded=false
     drawAs=Box
     imageType=FullColor
     mirroring=NoMirror
     resourceName=None
     resourceObject=&M_UI_Border
     tiling=NoTile
     imageSize=<40 40>
     margin=<ltrb 0.5 0.5 0.5 0.5>
     outlineSettings {
      bUseBrushTransparency=false
      roundingType=HalfHeightRadius
      width=0
      color {colorUseRule=UseColor_Specified, specifiedColor=<rgba 0 0 0 0>}
      cornerRadii=<0 0 0 0>
     }
     tintColor {colorUseRule=UseColor_Specified, specifiedColor=<rgba 1 0.3875 0 0>}
     uVRegion {bIsValid=false, max=<0 0>, min=<0 0>}
    }
    padding=<ltrb 0 0 0 0>
   }
   slot {class=VerticalBoxSlot, properties {padding=<ltrb 0 10 10 0>}}
  }
  VerticalBox_146 {
   widgetClassPath=&VerticalBox
   parent=Border-Cleanup
   widget=VerticalBox_146
   slot {class=BorderSlot, properties {padding=<ltrb 0 0 0 0>}}
  }
  ToggleButton_Collapse_ActorCleanup {
   widgetClassPath=&EditorUtilityCheckBox
   bIsVariable=true
   parent=VerticalBox_146
   widget=ToggleButton_Collapse_ActorCleanup
   properties {
    toolTipText=Expands / Contracts the Actor Transform Utilities module
    widgetStyle {
     checkBoxType=ToggleButton
     backgroundHoveredImage {
      bIsDynamicallyLoaded=false
      drawAs=RoundedBox
      imageType=NoImage
      mirroring=NoMirror
      resourceName=None
      resourceObject=null
      tiling=NoTile
      imageSize=<18 18>
      margin=<ltrb 0 0 0 0>
      outlineSettings {
       bUseBrushTransparency=false
       roundingType=FixedRadius
       width=1
       color {colorUseRule=UseColor_ColorTable, specifiedColor=<rgba 0 0 0 0>}
       cornerRadii=<3 3 3 3>
      }
      tintColor {colorUseRule=UseColor_Specified, specifiedColor=<rgba 0.0048 0.0048 0.0048 1>}
      uVRegion {bIsValid=false, max=<0 0>, min=<0 0>}
     }
     backgroundImage {
      bIsDynamicallyLoaded=false
      drawAs=RoundedBox
      imageType=NoImage
      mirroring=NoMirror
      resourceName=None
      resourceObject=null
      tiling=NoTile
      imageSize=<18 18>
      margin=<ltrb 0 0 0 0>
      outlineSettings {
       bUseBrushTransparency=false
       roundingType=FixedRadius
       width=1
       color {colorUseRule=UseColor_ColorTable, specifiedColor=<rgba 0 0 0 0>}
       cornerRadii=<3 3 3 3>
      }
      tintColor {colorUseRule=UseColor_Specified, specifiedColor=<rgba 0.0048 0.0048 0.0048 1>}
      uVRegion {bIsValid=false, max=<0 0>, min=<0 0>}
     }
     backgroundPressedImage {
      bIsDynamicallyLoaded=false
      drawAs=RoundedBox
      imageType=NoImage
      mirroring=NoMirror
      resourceName=None
      resourceObject=null
      tiling=NoTile
      imageSize=<18 18>
      margin=<ltrb 0 0 0 0>
      outlineSettings {
       bUseBrushTransparency=false
       roundingType=FixedRadius
       width=1
       color {colorUseRule=UseColor_ColorTable, specifiedColor=<rgba 0 0 0 0>}
       cornerRadii=<3 3 3 3>
      }
      tintColor {colorUseRule=UseColor_Specified, specifiedColor=<rgba 0.0048 0.0048 0.0048 1>}
      uVRegion {bIsValid=false, max=<0 0>, min=<0 0>}
     }
     borderBackgroundColor {colorUseRule=UseColor_Specified, specifiedColor=<rgba 1 1 1 1>}
     checkedForeground {colorUseRule=UseColor_Specified, specifiedColor=<rgba 1 1 1 1>}
     checkedHoveredForeground {colorUseRule=UseColor_Specified, specifiedColor=<rgba 1 1 1 1>}
     checkedHoveredImage {
      bIsDynamicallyLoaded=false
      drawAs=Image
      imageType=NoImage
      mirroring=NoMirror
      resourceName=None
      resourceObject=null
      tiling=NoTile
      imageSize=<32 10.5981>
      margin=<ltrb 0 0 0 0>
      outlineSettings {
       bUseBrushTransparency=false
       roundingType=HalfHeightRadius
       width=0
       color {colorUseRule=UseColor_Specified, specifiedColor=<rgba 0 0 0 0>}
       cornerRadii=<0 0 0 0>
      }
      tintColor {colorUseRule=UseColor_Specified, specifiedColor=<rgba 0.0953 0.0953 0.0953 1>}
      uVRegion {bIsValid=false, max=<0 0>, min=<0 0>}
     }
     checkedImage {
      bIsDynamicallyLoaded=false
      drawAs=Image
      imageType=NoImage
      mirroring=NoMirror
      resourceName=None
      resourceObject=null
      tiling=NoTile
      imageSize=<32 10.5981>
      margin=<ltrb 0 0 0 0>
      outlineSettings {
       bUseBrushTransparency=false
       roundingType=HalfHeightRadius
       width=0
       color {colorUseRule=UseColor_Specified, specifiedColor=<rgba 0 0 0 0>}
       cornerRadii=<0 0 0 0>
      }
      tintColor {colorUseRule=UseColor_Specified, specifiedColor=<rgba 0.0284 0.0284 0.0284 1>}
      uVRegion {bIsValid=false, max=<0 0>, min=<0 0>}
     }
     checkedPressedForeground {colorUseRule=UseColor_Specified, specifiedColor=<rgba 1 1 1 1>}
     checkedPressedImage {
      bIsDynamicallyLoaded=false
      drawAs=Image
      imageType=NoImage
      mirroring=NoMirror
      resourceName=None
      resourceObject=null
      tiling=NoTile
      imageSize=<32 10.5981>
      margin=<ltrb 0 0 0 0>
      outlineSettings {
       bUseBrushTransparency=false
       roundingType=HalfHeightRadius
       width=0
       color {colorUseRule=UseColor_Specified, specifiedColor=<rgba 0 0 0 0>}
       cornerRadii=<0 0 0 0>
      }
      tintColor {colorUseRule=UseColor_Specified, specifiedColor=<rgba 0.0284 0.0284 0.0284 1>}
      uVRegion {bIsValid=false, max=<0 0>, min=<0 0>}
     }
     checkedSlateSound {resourceObject=null}
     foregroundColor {colorUseRule=UseColor_Specified, specifiedColor=<rgba 1 1 1 1>}
     hoveredForeground {colorUseRule=UseColor_Specified, specifiedColor=<rgba 1 1 1 1>}
     hoveredSlateSound {resourceObject=null}
     padding=<ltrb 0 0 0 0>
     pressedForeground {colorUseRule=UseColor_Specified, specifiedColor=<rgba 1 1 1 1>}
     uncheckedHoveredImage {
      bIsDynamicallyLoaded=false
      drawAs=Image
      imageType=NoImage
      mirroring=NoMirror
      resourceName=None
      resourceObject=null
      tiling=NoTile
      imageSize=<32 10.5981>
      margin=<ltrb 0 0 0 0>
      outlineSettings {
       bUseBrushTransparency=false
       roundingType=HalfHeightRadius
       width=0
       color {colorUseRule=UseColor_Specified, specifiedColor=<rgba 0 0 0 0>}
       cornerRadii=<0 0 0 0>
      }
      tintColor {colorUseRule=UseColor_Specified, specifiedColor=<rgba 0.0953 0.0953 0.0953 1>}
      uVRegion {bIsValid=false, max=<0 0>, min=<0 0>}
     }
     uncheckedImage {
      bIsDynamicallyLoaded=false
      drawAs=Image
      imageType=NoImage
      mirroring=NoMirror
      resourceName=None
      resourceObject=null
      tiling=NoTile
      imageSize=<32 10.5981>
      margin=<ltrb 0 0 0 0>
      outlineSettings {
       bUseBrushTransparency=false
       roundingType=HalfHeightRadius
       width=0
       color {colorUseRule=UseColor_Specified, specifiedColor=<rgba 0 0 0 0>}
       cornerRadii=<0 0 0 0>
      }
      tintColor {colorUseRule=UseColor_Specified, specifiedColor=<rgba 0.0284 0.0284 0.0284 1>}
      uVRegion {bIsValid=false, max=<0 0>, min=<0 0>}
     }
     uncheckedPressedImage {
      bIsDynamicallyLoaded=false
      drawAs=Image
      imageType=NoImage
      mirroring=NoMirror
      resourceName=None
      resourceObject=null
      tiling=NoTile
      imageSize=<32 10.5981>
      margin=<ltrb 0 0 0 0>
      outlineSettings {
       bUseBrushTransparency=false
       roundingType=HalfHeightRadius
       width=0
       color {colorUseRule=UseColor_Specified, specifiedColor=<rgba 0 0 0 0>}
       cornerRadii=<0 0 0 0>
      }
      tintColor {colorUseRule=UseColor_Specified, specifiedColor=<rgba 0.0284 0.0284 0.0284 1>}
      uVRegion {bIsValid=false, max=<0 0>, min=<0 0>}
     }
     uncheckedSlateSound {resourceObject=null}
     undeterminedForeground {colorUseRule=UseColor_Specified, specifiedColor=<rgba 1 1 1 1>}
     undeterminedHoveredImage {
      bIsDynamicallyLoaded=false
      drawAs=Image
      imageType=Vector
      mirroring=NoMirror
      resourceName=../../../Engine/Content/Slate/Starship/CoreWidgets/CheckBox/indeterminate.svg
      resourceObject=null
      tiling=NoTile
      imageSize=<16 16>
      margin=<ltrb 0 0 0 0>
      outlineSettings {
       bUseBrushTransparency=false
       roundingType=HalfHeightRadius
       width=0
       color {colorUseRule=UseColor_Specified, specifiedColor=<rgba 0 0 0 0>}
       cornerRadii=<0 0 0 0>
      }
      tintColor {colorUseRule=UseColor_Specified, specifiedColor=<rgba 0.0044 0.2384 1 1>}
      uVRegion {bIsValid=false, max=<0 0>, min=<0 0>}
     }
     undeterminedImage {
      bIsDynamicallyLoaded=false
      drawAs=Image
      imageType=Vector
      mirroring=NoMirror
      resourceName=../../../Engine/Content/Slate/Starship/CoreWidgets/CheckBox/indeterminate.svg
      resourceObject=null
      tiling=NoTile
      imageSize=<16 16>
      margin=<ltrb 0 0 0 0>
      outlineSettings {
       bUseBrushTransparency=false
       roundingType=HalfHeightRadius
       width=0
       color {colorUseRule=UseColor_Specified, specifiedColor=<rgba 0 0 0 0>}
       cornerRadii=<0 0 0 0>
      }
      tintColor {colorUseRule=UseColor_Specified, specifiedColor=<rgba 0 0.162 0.7454 1>}
      uVRegion {bIsValid=false, max=<0 0>, min=<0 0>}
     }
     undeterminedPressedImage {
      bIsDynamicallyLoaded=false
      drawAs=Image
      imageType=Vector
      mirroring=NoMirror
      resourceName=../../../Engine/Content/Slate/Starship/CoreWidgets/CheckBox/indeterminate.svg
      resourceObject=null
      tiling=NoTile
      imageSize=<16 16>
      margin=<ltrb 0 0 0 0>
      outlineSettings {
       bUseBrushTransparency=false
       roundingType=HalfHeightRadius
       width=0
       color {colorUseRule=UseColor_Specified, specifiedColor=<rgba 0 0 0 0>}
       cornerRadii=<0 0 0 0>
      }
      tintColor {colorUseRule=UseColor_Specified, specifiedColor=<rgba 0 0.0802 0.3515 1>}
      uVRegion {bIsValid=false, max=<0 0>, min=<0 0>}
     }
    }
   }
   slot {class=VerticalBoxSlot}
  }
  HorizontalBox_1 {widgetClassPath=&HorizontalBox, parent=ToggleButton_Collapse_ActorCleanup, widget=HorizontalBox_1, slot {class=PanelSlot}}
  Spacer_3 {
   widgetClassPath=&Spacer
   parent=HorizontalBox_1
   widget=Spacer_3
   properties {size=<12 1>}
   slot {class=HorizontalBoxSlot, properties {horizontalAlignment=HAlign_Left}}
  }
  T_ToggleButton_ActorCleanup {
   widgetClassPath=&TextBlock
   bIsVariable=true
   parent=HorizontalBox_1
   widget=T_ToggleButton_ActorCleanup
   properties {
    text="Actor Cleanup Utilities ▼"
    font {
     bForceMonospaced=false
     bMaterialIsStencil=false
     fontMaterial=null
     fontObject=&Roboto
     letterSpacing=0
     monospacedWidth=1
     size=12
     skewAmount=0
     typefaceFontName=Bold
     outlineSettings {
      bApplyOutlineToDropShadows=false
      bMiteredCorners=false
      bSeparateFillAlpha=false
      outlineMaterial=null
      outlineSize=0
      outlineColor=<rgba 0 0 0 1>
     }
    }
   }
   slot {class=HorizontalBoxSlot}
  }
  Socket_ActorCleanupTools {
   widgetClassPath=&VerticalBox
   bIsVariable=true
   parent=VerticalBox_146
   widget=Socket_ActorCleanupTools
   slot {class=VerticalBoxSlot, properties {padding=<ltrb 10 10 0 20>}}
  }
  EUW_AssetOrganizer_ActorCleanup {
   widgetClassPath=/Game/ArtTools/FNEC/AssetOrganizer/Widgets/EUW_AssetOrganizer_ActorCleanup
   bIsVariable=true
   parent=Socket_ActorCleanupTools
   widget=EUW_AssetOrganizer_ActorCleanup
   slot {class=VerticalBoxSlot}
  }
  Border-Arrange {
   widgetClassPath=&Border
   parent=VerticalBox_Master
   widget=Border-Arrange
   properties {
    background {
     bIsDynamicallyLoaded=false
     drawAs=Box
     imageType=FullColor
     mirroring=NoMirror
     resourceName=None
     resourceObject=&M_UI_Border
     tiling=NoTile
     imageSize=<40 40>
     margin=<ltrb 0.5 0.5 0.5 0.5>
     outlineSettings {
      bUseBrushTransparency=false
      roundingType=HalfHeightRadius
      width=0
      color {colorUseRule=UseColor_Specified, specifiedColor=<rgba 0 0 0 0>}
      cornerRadii=<0 0 0 0>
     }
     tintColor {colorUseRule=UseColor_Specified, specifiedColor=<rgba 1 0.3875 0 0>}
     uVRegion {bIsValid=false, max=<0 0>, min=<0 0>}
    }
    padding=<ltrb 0 0 0 0>
   }
   slot {class=VerticalBoxSlot, properties {padding=<ltrb 0 10 10 0>}}
  }
  VerticalBox_73 {
   widgetClassPath=&VerticalBox
   parent=Border-Arrange
   widget=VerticalBox_73
   slot {class=BorderSlot, properties {padding=<ltrb 0 0 0 0>}}
  }
  ToggleButton_Collapse_ArrangeInGrid {
   widgetClassPath=&EditorUtilityCheckBox
   bIsVariable=true
   parent=VerticalBox_73
   widget=ToggleButton_Collapse_ArrangeInGrid
   properties {
    toolTipText=Expands / Contracts the Arrange Actors in Grid module
    widgetStyle {
     checkBoxType=ToggleButton
     backgroundHoveredImage {
      bIsDynamicallyLoaded=false
      drawAs=RoundedBox
      imageType=NoImage
      mirroring=NoMirror
      resourceName=None
      resourceObject=null
      tiling=NoTile
      imageSize=<18 18>
      margin=<ltrb 0 0 0 0>
      outlineSettings {
       bUseBrushTransparency=false
       roundingType=FixedRadius
       width=1
       color {colorUseRule=UseColor_ColorTable, specifiedColor=<rgba 0 0 0 0>}
       cornerRadii=<3 3 3 3>
      }
      tintColor {colorUseRule=UseColor_Specified, specifiedColor=<rgba 0.0048 0.0048 0.0048 1>}
      uVRegion {bIsValid=false, max=<0 0>, min=<0 0>}
     }
     backgroundImage {
      bIsDynamicallyLoaded=false
      drawAs=RoundedBox
      imageType=NoImage
      mirroring=NoMirror
      resourceName=None
      resourceObject=null
      tiling=NoTile
      imageSize=<18 18>
      margin=<ltrb 0 0 0 0>
      outlineSettings {
       bUseBrushTransparency=false
       roundingType=FixedRadius
       width=1
       color {colorUseRule=UseColor_ColorTable, specifiedColor=<rgba 0 0 0 0>}
       cornerRadii=<3 3 3 3>
      }
      tintColor {colorUseRule=UseColor_Specified, specifiedColor=<rgba 0.0048 0.0048 0.0048 1>}
      uVRegion {bIsValid=false, max=<0 0>, min=<0 0>}
     }
     backgroundPressedImage {
      bIsDynamicallyLoaded=false
      drawAs=RoundedBox
      imageType=NoImage
      mirroring=NoMirror
      resourceName=None
      resourceObject=null
      tiling=NoTile
      imageSize=<18 18>
      margin=<ltrb 0 0 0 0>
      outlineSettings {
       bUseBrushTransparency=false
       roundingType=FixedRadius
       width=1
       color {colorUseRule=UseColor_ColorTable, specifiedColor=<rgba 0 0 0 0>}
       cornerRadii=<3 3 3 3>
      }
      tintColor {colorUseRule=UseColor_Specified, specifiedColor=<rgba 0.0048 0.0048 0.0048 1>}
      uVRegion {bIsValid=false, max=<0 0>, min=<0 0>}
     }
     borderBackgroundColor {colorUseRule=UseColor_Specified, specifiedColor=<rgba 1 1 1 1>}
     checkedForeground {colorUseRule=UseColor_Specified, specifiedColor=<rgba 1 1 1 1>}
     checkedHoveredForeground {colorUseRule=UseColor_Specified, specifiedColor=<rgba 1 1 1 1>}
     checkedHoveredImage {
      bIsDynamicallyLoaded=false
      drawAs=Image
      imageType=NoImage
      mirroring=NoMirror
      resourceName=None
      resourceObject=null
      tiling=NoTile
      imageSize=<32 10.5981>
      margin=<ltrb 0 0 0 0>
      outlineSettings {
       bUseBrushTransparency=false
       roundingType=HalfHeightRadius
       width=0
       color {colorUseRule=UseColor_Specified, specifiedColor=<rgba 0 0 0 0>}
       cornerRadii=<0 0 0 0>
      }
      tintColor {colorUseRule=UseColor_Specified, specifiedColor=<rgba 0.0953 0.0953 0.0953 1>}
      uVRegion {bIsValid=false, max=<0 0>, min=<0 0>}
     }
     checkedImage {
      bIsDynamicallyLoaded=false
      drawAs=Image
      imageType=NoImage
      mirroring=NoMirror
      resourceName=None
      resourceObject=null
      tiling=NoTile
      imageSize=<32 10.5981>
      margin=<ltrb 0 0 0 0>
      outlineSettings {
       bUseBrushTransparency=false
       roundingType=HalfHeightRadius
       width=0
       color {colorUseRule=UseColor_Specified, specifiedColor=<rgba 0 0 0 0>}
       cornerRadii=<0 0 0 0>
      }
      tintColor {colorUseRule=UseColor_Specified, specifiedColor=<rgba 0.0284 0.0284 0.0284 1>}
      uVRegion {bIsValid=false, max=<0 0>, min=<0 0>}
     }
     checkedPressedForeground {colorUseRule=UseColor_Specified, specifiedColor=<rgba 1 1 1 1>}
     checkedPressedImage {
      bIsDynamicallyLoaded=false
      drawAs=Image
      imageType=NoImage
      mirroring=NoMirror
      resourceName=None
      resourceObject=null
      tiling=NoTile
      imageSize=<32 10.5981>
      margin=<ltrb 0 0 0 0>
      outlineSettings {
       bUseBrushTransparency=false
       roundingType=HalfHeightRadius
       width=0
       color {colorUseRule=UseColor_Specified, specifiedColor=<rgba 0 0 0 0>}
       cornerRadii=<0 0 0 0>
      }
      tintColor {colorUseRule=UseColor_Specified, specifiedColor=<rgba 0.0284 0.0284 0.0284 1>}
      uVRegion {bIsValid=false, max=<0 0>, min=<0 0>}
     }
     checkedSlateSound {resourceObject=null}
     foregroundColor {colorUseRule=UseColor_Specified, specifiedColor=<rgba 1 1 1 1>}
     hoveredForeground {colorUseRule=UseColor_Specified, specifiedColor=<rgba 1 1 1 1>}
     hoveredSlateSound {resourceObject=null}
     padding=<ltrb 0 0 0 0>
     pressedForeground {colorUseRule=UseColor_Specified, specifiedColor=<rgba 1 1 1 1>}
     uncheckedHoveredImage {
      bIsDynamicallyLoaded=false
      drawAs=Image
      imageType=NoImage
      mirroring=NoMirror
      resourceName=None
      resourceObject=null
      tiling=NoTile
      imageSize=<32 10.5981>
      margin=<ltrb 0 0 0 0>
      outlineSettings {
       bUseBrushTransparency=false
       roundingType=HalfHeightRadius
       width=0
       color {colorUseRule=UseColor_Specified, specifiedColor=<rgba 0 0 0 0>}
       cornerRadii=<0 0 0 0>
      }
      tintColor {colorUseRule=UseColor_Specified, specifiedColor=<rgba 0.0953 0.0953 0.0953 1>}
      uVRegion {bIsValid=false, max=<0 0>, min=<0 0>}
     }
     uncheckedImage {
      bIsDynamicallyLoaded=false
      drawAs=Image
      imageType=NoImage
      mirroring=NoMirror
      resourceName=None
      resourceObject=null
      tiling=NoTile
      imageSize=<32 10.5981>
      margin=<ltrb 0 0 0 0>
      outlineSettings {
       bUseBrushTransparency=false
       roundingType=HalfHeightRadius
       width=0
       color {colorUseRule=UseColor_Specified, specifiedColor=<rgba 0 0 0 0>}
       cornerRadii=<0 0 0 0>
      }
      tintColor {colorUseRule=UseColor_Specified, specifiedColor=<rgba 0.0284 0.0284 0.0284 1>}
      uVRegion {bIsValid=false, max=<0 0>, min=<0 0>}
     }
     uncheckedPressedImage {
      bIsDynamicallyLoaded=false
      drawAs=Image
      imageType=NoImage
      mirroring=NoMirror
      resourceName=None
      resourceObject=null
      tiling=NoTile
      imageSize=<32 10.5981>
      margin=<ltrb 0 0 0 0>
      outlineSettings {
       bUseBrushTransparency=false
       roundingType=HalfHeightRadius
       width=0
       color {colorUseRule=UseColor_Specified, specifiedColor=<rgba 0 0 0 0>}
       cornerRadii=<0 0 0 0>
      }
      tintColor {colorUseRule=UseColor_Specified, specifiedColor=<rgba 0.0284 0.0284 0.0284 1>}
      uVRegion {bIsValid=false, max=<0 0>, min=<0 0>}
     }
     uncheckedSlateSound {resourceObject=null}
     undeterminedForeground {colorUseRule=UseColor_Specified, specifiedColor=<rgba 1 1 1 1>}
     undeterminedHoveredImage {
      bIsDynamicallyLoaded=false
      drawAs=Image
      imageType=Vector
      mirroring=NoMirror
      resourceName=../../../Engine/Content/Slate/Starship/CoreWidgets/CheckBox/indeterminate.svg
      resourceObject=null
      tiling=NoTile
      imageSize=<16 16>
      margin=<ltrb 0 0 0 0>
      outlineSettings {
       bUseBrushTransparency=false
       roundingType=HalfHeightRadius
       width=0
       color {colorUseRule=UseColor_Specified, specifiedColor=<rgba 0 0 0 0>}
       cornerRadii=<0 0 0 0>
      }
      tintColor {colorUseRule=UseColor_Specified, specifiedColor=<rgba 0.0044 0.2384 1 1>}
      uVRegion {bIsValid=false, max=<0 0>, min=<0 0>}
     }
     undeterminedImage {
      bIsDynamicallyLoaded=false
      drawAs=Image
      imageType=Vector
      mirroring=NoMirror
      resourceName=../../../Engine/Content/Slate/Starship/CoreWidgets/CheckBox/indeterminate.svg
      resourceObject=null
      tiling=NoTile
      imageSize=<16 16>
      margin=<ltrb 0 0 0 0>
      outlineSettings {
       bUseBrushTransparency=false
       roundingType=HalfHeightRadius
       width=0
       color {colorUseRule=UseColor_Specified, specifiedColor=<rgba 0 0 0 0>}
       cornerRadii=<0 0 0 0>
      }
      tintColor {colorUseRule=UseColor_Specified, specifiedColor=<rgba 0 0.162 0.7454 1>}
      uVRegion {bIsValid=false, max=<0 0>, min=<0 0>}
     }
     undeterminedPressedImage {
      bIsDynamicallyLoaded=false
      drawAs=Image
      imageType=Vector
      mirroring=NoMirror
      resourceName=../../../Engine/Content/Slate/Starship/CoreWidgets/CheckBox/indeterminate.svg
      resourceObject=null
      tiling=NoTile
      imageSize=<16 16>
      margin=<ltrb 0 0 0 0>
      outlineSettings {
       bUseBrushTransparency=false
       roundingType=HalfHeightRadius
       width=0
       color {colorUseRule=UseColor_Specified, specifiedColor=<rgba 0 0 0 0>}
       cornerRadii=<0 0 0 0>
      }
      tintColor {colorUseRule=UseColor_Specified, specifiedColor=<rgba 0 0.0802 0.3515 1>}
      uVRegion {bIsValid=false, max=<0 0>, min=<0 0>}
     }
    }
   }
   slot {class=VerticalBoxSlot}
  }
  HorizontalBox_4 {widgetClassPath=&HorizontalBox, parent=ToggleButton_Collapse_ArrangeInGrid, widget=HorizontalBox_4, slot {class=PanelSlot}}
  Spacer {
   widgetClassPath=&Spacer
   parent=HorizontalBox_4
   widget=Spacer
   properties {size=<12 1>}
   slot {class=HorizontalBoxSlot, properties {horizontalAlignment=HAlign_Left}}
  }
  T_ToggleButton_ArrageInGrid {
   widgetClassPath=&TextBlock
   bIsVariable=true
   parent=HorizontalBox_4
   widget=T_ToggleButton_ArrageInGrid
   properties {
    text="Grid Arrangement Utilities ▲"
    font {
     bForceMonospaced=false
     bMaterialIsStencil=false
     fontMaterial=null
     fontObject=&Roboto
     letterSpacing=0
     monospacedWidth=1
     size=12
     skewAmount=0
     typefaceFontName=Bold
     outlineSettings {
      bApplyOutlineToDropShadows=false
      bMiteredCorners=false
      bSeparateFillAlpha=false
      outlineMaterial=null
      outlineSize=0
      outlineColor=<rgba 0 0 0 1>
     }
    }
   }
   slot {class=HorizontalBoxSlot}
  }
  Socket_ArrangeInGrid {
   widgetClassPath=&VerticalBox
   bIsVariable=true
   parent=VerticalBox_73
   widget=Socket_ArrangeInGrid
   slot {class=VerticalBoxSlot, properties {padding=<ltrb 10 10 0 20>}}
  }
  VerticalBox_0 {widgetClassPath=&VerticalBox, parent=Socket_ArrangeInGrid, widget=VerticalBox_0, slot {class=VerticalBoxSlot}}
  HorizontalBox {widgetClassPath=&HorizontalBox, parent=VerticalBox_0, widget=HorizontalBox, slot {class=VerticalBoxSlot}}
  B_Arrange {
   widgetClassPath=&EditorUtilityButton
   parent=HorizontalBox
   widget=B_Arrange
   properties {
    toolTipText="Arranges selected actors into a grid. The position of the bottom left corner of the grid is defined by the 'Grid Origin' option. "
    widgetStyle {
     clickedSlateSound {resourceObject=null}
     disabled {
      bIsDynamicallyLoaded=false
      drawAs=RoundedBox
      imageType=NoImage
      mirroring=NoMirror
      resourceName=None
      resourceObject=null
      tiling=NoTile
      imageSize=<0 0>
      margin=<ltrb 0 0 0 0>
      outlineSettings {
       bUseBrushTransparency=false
       roundingType=FixedRadius
       width=1
       color {colorUseRule=UseColor_ColorTable, specifiedColor=<rgba 0 0 0 0>}
       cornerRadii=<4 4 4 4>
      }
      tintColor {colorUseRule=UseColor_Specified, specifiedColor=<rgba 0.0395 0.0395 0.0395 1>}
      uVRegion {bIsValid=false, max=<0 0>, min=<0 0>}
     }
     disabledForeground {colorUseRule=UseColor_ColorTable, specifiedColor=<rgba 0 0 0 0>}
     hovered {
      bIsDynamicallyLoaded=false
      drawAs=RoundedBox
      imageType=NoImage
      mirroring=NoMirror
      resourceName=None
      resourceObject=null
      tiling=NoTile
      imageSize=<32 32>
      margin=<ltrb 0 0 0 0>
      outlineSettings {
       bUseBrushTransparency=false
       roundingType=FixedRadius
       width=1
       color {colorUseRule=UseColor_ColorTable, specifiedColor=<rgba 0 0 0 0>}
       cornerRadii=<4 4 4 4>
      }
      tintColor {colorUseRule=UseColor_Specified, specifiedColor=<rgba 0.0953 0.0953 0.0953 1>}
      uVRegion {bIsValid=false, max=<0 0>, min=<0 0>}
     }
     hoveredForeground {colorUseRule=UseColor_ColorTable, specifiedColor=<rgba 0 0 0 0>}
     hoveredSlateSound {resourceObject=null}
     normal {
      bIsDynamicallyLoaded=false
      drawAs=RoundedBox
      imageType=NoImage
      mirroring=NoMirror
      resourceName=None
      resourceObject=null
      tiling=NoTile
      imageSize=<32 32>
      margin=<ltrb 0 0 0 0>
      outlineSettings {
       bUseBrushTransparency=false
       roundingType=FixedRadius
       width=1
       color {colorUseRule=UseColor_ColorTable, specifiedColor=<rgba 0 0 0 0>}
       cornerRadii=<4 4 4 4>
      }
      tintColor {colorUseRule=UseColor_Specified, specifiedColor=<rgba 0.0395 0.0395 0.0395 1>}
      uVRegion {bIsValid=false, max=<0 0>, min=<0 0>}
     }
     normalForeground {colorUseRule=UseColor_ColorTable, specifiedColor=<rgba 0 0 0 0>}
     normalPadding=<ltrb 12 1.5 12 1.5>
     pressed {
      bIsDynamicallyLoaded=false
      drawAs=RoundedBox
      imageType=NoImage
      mirroring=NoMirror
      resourceName=None
      resourceObject=null
      tiling=NoTile
      imageSize=<32 32>
      margin=<ltrb 0 0 0 0>
      outlineSettings {
       bUseBrushTransparency=false
       roundingType=FixedRadius
       width=1
       color {colorUseRule=UseColor_ColorTable, specifiedColor=<rgba 0 0 0 0>}
       cornerRadii=<4 4 4 4>
      }
      tintColor {colorUseRule=UseColor_Specified, specifiedColor=<rgba 0.0284 0.0284 0.0284 1>}
      uVRegion {bIsValid=false, max=<0 0>, min=<0 0>}
     }
     pressedForeground {colorUseRule=UseColor_ColorTable, specifiedColor=<rgba 0 0 0 0>}
     pressedPadding=<ltrb 12 2.5 12 0.5>
     pressedSlateSound {resourceObject=null}
    }
   }
   slot {class=HorizontalBoxSlot}
  }
  TextBlock {
   widgetClassPath=&TextBlock
   parent=B_Arrange
   widget=TextBlock
   properties {
    text=Arrange
    font {
     bForceMonospaced=false
     bMaterialIsStencil=false
     fontMaterial=null
     fontObject=&Roboto
     letterSpacing=0
     monospacedWidth=1
     size=16
     skewAmount=0
     typefaceFontName=Bold
     outlineSettings {
      bApplyOutlineToDropShadows=false
      bMiteredCorners=false
      bSeparateFillAlpha=false
      outlineMaterial=null
      outlineSize=0
      outlineColor=<rgba 0 0 0 1>
     }
    }
   }
   slot {class=ButtonSlot}
  }
  VerticalBox_210 {
   widgetClassPath=&VerticalBox
   parent=HorizontalBox
   widget=VerticalBox_210
   slot {class=HorizontalBoxSlot, properties {horizontalAlignment=HAlign_Left, padding=<ltrb 10 0 0 0>, size {sizeRule=Fill, value=1}}}
  }
  VerticalBox {
   widgetClassPath=&VerticalBox
   parent=VerticalBox_210
   widget=VerticalBox
   slot {class=VerticalBoxSlot, properties {size {sizeRule=Fill, value=1}}}
  }
  HorizontalBox-Rows {
   widgetClassPath=&HorizontalBox
   parent=VerticalBox
   widget=HorizontalBox-Rows
   slot {class=VerticalBoxSlot, properties {size {sizeRule=Fill, value=1}}}
  }
  HorizontalBox_3 {widgetClassPath=&HorizontalBox, parent=HorizontalBox-Rows, widget=HorizontalBox_3, slot {class=HorizontalBoxSlot}}
  TextBlock_3 {
   widgetClassPath=&TextBlock
   parent=HorizontalBox_3
   widget=TextBlock_3
   properties {
    text="Rows : "
    font {
     bForceMonospaced=false
     bMaterialIsStencil=false
     fontMaterial=null
     fontObject=&Roboto
     letterSpacing=0
     monospacedWidth=1
     size=9
     skewAmount=0
     typefaceFontName=Light
     outlineSettings {
      bApplyOutlineToDropShadows=false
      bMiteredCorners=false
      bSeparateFillAlpha=false
      outlineMaterial=null
      outlineSize=0
      outlineColor=<rgba 0 0 0 1>
     }
    }
   }
   slot {class=HorizontalBoxSlot, properties {verticalAlignment=VAlign_Center}}
  }
  Border-textfield {
   widgetClassPath=&Border
   parent=HorizontalBox_3
   widget=Border-textfield
   properties {
    verticalAlignment=VAlign_Center
    background {
     bIsDynamicallyLoaded=false
     drawAs=RoundedBox
     imageType=NoImage
     mirroring=NoMirror
     resourceName=None
     resourceObject=null
     tiling=NoTile
     imageSize=<32 32>
     margin=<ltrb 0 0 0 0>
     outlineSettings {
      bUseBrushTransparency=false
      roundingType=FixedRadius
      width=0
      color {colorUseRule=UseColor_Specified, specifiedColor=<rgba 0 0 0 0>}
      cornerRadii=<5 5 5 5>
     }
     tintColor {colorUseRule=UseColor_Specified, specifiedColor=<rgba 0.01 0.01 0.01 1>}
     uVRegion {bIsValid=false, max=<0 0>, min=<0 0>}
    }
   }
   slot {class=HorizontalBoxSlot, properties {verticalAlignment=VAlign_Center}}
  }
  T_Arrange_Num {
   widgetClassPath=/Script/Blutility.EditorUtilityEditableText
   bIsVariable=true
   parent=Border-textfield
   widget=T_Arrange_Num
   properties {
    text=5
    toolTipText=Number of rows to arrange selected assets on.
    widgetStyle {
     backgroundImageComposing {
      bIsDynamicallyLoaded=false
      drawAs=Border
      imageType=FullColor
      mirroring=NoMirror
      resourceName=../../../Engine/Content/Slate/Old/HyperlinkDotted.png
      resourceObject=null
      tiling=Both
      imageSize=<0 0>
      margin=<ltrb 0 0 0 0.1875>
      outlineSettings {
       bUseBrushTransparency=false
       roundingType=HalfHeightRadius
       width=0
       color {colorUseRule=UseColor_Specified, specifiedColor=<rgba 0 0 0 0>}
       cornerRadii=<0 0 0 0>
      }
      tintColor {colorUseRule=UseColor_Specified, specifiedColor=<rgba 1 1 1 1>}
      uVRegion {bIsValid=false, max=<0 0>, min=<0 0>}
     }
     backgroundImageSelected {
      bIsDynamicallyLoaded=false
      drawAs=Image
      imageType=NoImage
      mirroring=NoMirror
      resourceName=None
      resourceObject=null
      tiling=NoTile
      imageSize=<0 0>
      margin=<ltrb 0 0 0 0>
      outlineSettings {
       bUseBrushTransparency=false
       roundingType=HalfHeightRadius
       width=0
       color {colorUseRule=UseColor_Specified, specifiedColor=<rgba 0 0 0 0>}
       cornerRadii=<0 0 0 0>
      }
      tintColor {colorUseRule=UseColor_Specified, specifiedColor=<rgba 0 0.162 0.7454 1>}
      uVRegion {bIsValid=false, max=<0 0>, min=<0 0>}
     }
     caretImage {
      bIsDynamicallyLoaded=false
      drawAs=Image
      imageType=NoImage
      mirroring=NoMirror
      resourceName=None
      resourceObject=null
      tiling=NoTile
      imageSize=<0 0>
      margin=<ltrb 0 0 0 0>
      outlineSettings {
       bUseBrushTransparency=false
       roundingType=HalfHeightRadius
       width=0
       color {colorUseRule=UseColor_Specified, specifiedColor=<rgba 0 0 0 0>}
       cornerRadii=<0 0 0 0>
      }
      tintColor {colorUseRule=UseColor_Specified, specifiedColor=<rgba 1 1 1 1>}
      uVRegion {bIsValid=false, max=<0 0>, min=<0 0>}
     }
     colorAndOpacity {colorUseRule=UseColor_Foreground, specifiedColor=<rgba 1 0 1 1>}
     font {
      bForceMonospaced=false
      bMaterialIsStencil=false
      fontMaterial=null
      fontObject=&Roboto
      letterSpacing=0
      monospacedWidth=1
      size=12
      skewAmount=0
      typefaceFontName=Bold
      outlineSettings {
       bApplyOutlineToDropShadows=false
       bMiteredCorners=false
       bSeparateFillAlpha=false
       outlineMaterial=null
       outlineSize=0
       outlineColor=<rgba 0 0 0 1>
      }
     }
     hintColorAndOpacity {colorUseRule=UseColor_Foreground, specifiedColor=<rgba 1 0 1 1>}
     selectedBackgroundColor {colorUseRule=UseColor_Foreground, specifiedColor=<rgba 1 0 1 1>}
    }
   }
   slot {class=BorderSlot, properties {verticalAlignment=VAlign_Center}}
  }
  B_Arrange_Square {
   widgetClassPath=&EditorUtilityButton
   parent=HorizontalBox-Rows
   widget=B_Arrange_Square
   properties {
    toolTipText=Sets the number of rows so that the resulting grid is square
    widgetStyle {
     clickedSlateSound {resourceObject=null}
     disabled {
      bIsDynamicallyLoaded=false
      drawAs=RoundedBox
      imageType=NoImage
      mirroring=NoMirror
      resourceName=None
      resourceObject=null
      tiling=NoTile
      imageSize=<0 0>
      margin=<ltrb 0 0 0 0>
      outlineSettings {
       bUseBrushTransparency=false
       roundingType=FixedRadius
       width=1
       color {colorUseRule=UseColor_ColorTable, specifiedColor=<rgba 0 0 0 0>}
       cornerRadii=<4 4 4 4>
      }
      tintColor {colorUseRule=UseColor_Specified, specifiedColor=<rgba 0.0395 0.0395 0.0395 1>}
      uVRegion {bIsValid=false, max=<0 0>, min=<0 0>}
     }
     disabledForeground {colorUseRule=UseColor_ColorTable, specifiedColor=<rgba 0 0 0 0>}
     hovered {
      bIsDynamicallyLoaded=false
      drawAs=RoundedBox
      imageType=NoImage
      mirroring=NoMirror
      resourceName=None
      resourceObject=null
      tiling=NoTile
      imageSize=<32 32>
      margin=<ltrb 0 0 0 0>
      outlineSettings {
       bUseBrushTransparency=false
       roundingType=FixedRadius
       width=1
       color {colorUseRule=UseColor_ColorTable, specifiedColor=<rgba 0 0 0 0>}
       cornerRadii=<4 4 4 4>
      }
      tintColor {colorUseRule=UseColor_Specified, specifiedColor=<rgba 0.0953 0.0953 0.0953 1>}
      uVRegion {bIsValid=false, max=<0 0>, min=<0 0>}
     }
     hoveredForeground {colorUseRule=UseColor_ColorTable, specifiedColor=<rgba 0 0 0 0>}
     hoveredSlateSound {resourceObject=null}
     normal {
      bIsDynamicallyLoaded=false
      drawAs=RoundedBox
      imageType=NoImage
      mirroring=NoMirror
      resourceName=None
      resourceObject=null
      tiling=NoTile
      imageSize=<32 32>
      margin=<ltrb 0 0 0 0>
      outlineSettings {
       bUseBrushTransparency=false
       roundingType=FixedRadius
       width=1
       color {colorUseRule=UseColor_ColorTable, specifiedColor=<rgba 0 0 0 0>}
       cornerRadii=<4 4 4 4>
      }
      tintColor {colorUseRule=UseColor_Specified, specifiedColor=<rgba 0.0395 0.0395 0.0395 1>}
      uVRegion {bIsValid=false, max=<0 0>, min=<0 0>}
     }
     normalForeground {colorUseRule=UseColor_ColorTable, specifiedColor=<rgba 0 0 0 0>}
     normalPadding=<ltrb 0 0 0 0>
     pressed {
      bIsDynamicallyLoaded=false
      drawAs=RoundedBox
      imageType=NoImage
      mirroring=NoMirror
      resourceName=None
      resourceObject=null
      tiling=NoTile
      imageSize=<32 32>
      margin=<ltrb 0 0 0 0>
      outlineSettings {
       bUseBrushTransparency=false
       roundingType=FixedRadius
       width=1
       color {colorUseRule=UseColor_ColorTable, specifiedColor=<rgba 0 0 0 0>}
       cornerRadii=<4 4 4 4>
      }
      tintColor {colorUseRule=UseColor_Specified, specifiedColor=<rgba 0.0284 0.0284 0.0284 1>}
      uVRegion {bIsValid=false, max=<0 0>, min=<0 0>}
     }
     pressedForeground {colorUseRule=UseColor_ColorTable, specifiedColor=<rgba 0 0 0 0>}
     pressedPadding=<ltrb 0 0 0 0>
     pressedSlateSound {resourceObject=null}
    }
   }
   slot {class=HorizontalBoxSlot, properties {horizontalAlignment=HAlign_Center, verticalAlignment=VAlign_Center, padding=<ltrb 5 0 0 0>}}
  }
  TextBlock_4 {
   widgetClassPath=&TextBlock
   parent=B_Arrange_Square
   widget=TextBlock_4
   properties {
    justification=Center
    text=Square
    font {
     bForceMonospaced=false
     bMaterialIsStencil=false
     fontMaterial=null
     fontObject=&Roboto
     letterSpacing=0
     monospacedWidth=1
     size=9
     skewAmount=0
     typefaceFontName=Regular
     outlineSettings {
      bApplyOutlineToDropShadows=false
      bMiteredCorners=false
      bSeparateFillAlpha=false
      outlineMaterial=null
      outlineSize=0
      outlineColor=<rgba 0 0 0 1>
     }
    }
   }
   slot {class=ButtonSlot, properties {padding=<ltrb 6 4 6 4>}}
  }
  Combo_RowAlignment {
   widgetClassPath=/Script/Blutility.EditorUtilityComboBoxString
   bIsVariable=true
   parent=HorizontalBox-Rows
   widget=Combo_RowAlignment
   properties {
    selectedOption=Rows on X
    toolTipText="Determines how selected assets are arranged in rows.\r\n\r\nRows on X - The number of rows determines the amount of rows along the X axis.\r\n\r\nRows on Y - The number of rows determines the amount of rows along the Y axis."
    defaultOptions [Rows on X, Rows on Y]
    font {
     bForceMonospaced=false
     bMaterialIsStencil=false
     fontMaterial=null
     fontObject=&Roboto
     letterSpacing=0
     monospacedWidth=1
     size=9
     skewAmount=0
     typefaceFontName=Regular
     outlineSettings {
      bApplyOutlineToDropShadows=false
      bMiteredCorners=false
      bSeparateFillAlpha=false
      outlineMaterial=null
      outlineSize=0
      outlineColor=<rgba 0 0 0 1>
     }
    }
    foregroundColor {colorUseRule=UseColor_Specified, specifiedColor=<rgba 0.5271 0.5271 0.5271 1>}
   }
   slot {class=HorizontalBoxSlot, properties {verticalAlignment=VAlign_Center, padding=<ltrb 10 0 0 0>}}
  }
  HorizontalBox-XYDistance {
   widgetClassPath=&HorizontalBox
   parent=VerticalBox
   widget=HorizontalBox-XYDistance
   slot {class=VerticalBoxSlot, properties {size {sizeRule=Fill, value=1}}}
  }
  HorizontalBox-XDistance {widgetClassPath=&HorizontalBox, parent=HorizontalBox-XYDistance, widget=HorizontalBox-XDistance, slot {class=HorizontalBoxSlot}}
  TextBlock_1 {
   widgetClassPath=&TextBlock
   parent=HorizontalBox-XDistance
   widget=TextBlock_1
   properties {
    text="X Distance : "
    font {
     bForceMonospaced=false
     bMaterialIsStencil=false
     fontMaterial=null
     fontObject=&Roboto
     letterSpacing=0
     monospacedWidth=1
     size=9
     skewAmount=0
     typefaceFontName=Light
     outlineSettings {
      bApplyOutlineToDropShadows=false
      bMiteredCorners=false
      bSeparateFillAlpha=false
      outlineMaterial=null
      outlineSize=0
      outlineColor=<rgba 0 0 0 1>
     }
    }
   }
   slot {class=HorizontalBoxSlot, properties {verticalAlignment=VAlign_Center}}
  }
  Border_XDistance {
   widgetClassPath=&Border
   bIsVariable=true
   parent=HorizontalBox-XDistance
   widget=Border_XDistance
   properties {
    verticalAlignment=VAlign_Center
    background {
     bIsDynamicallyLoaded=false
     drawAs=RoundedBox
     imageType=NoImage
     mirroring=NoMirror
     resourceName=None
     resourceObject=null
     tiling=NoTile
     imageSize=<32 32>
     margin=<ltrb 0 0 0 0>
     outlineSettings {
      bUseBrushTransparency=false
      roundingType=FixedRadius
      width=0
      color {colorUseRule=UseColor_Specified, specifiedColor=<rgba 0 0 0 0>}
      cornerRadii=<5 5 5 5>
     }
     tintColor {colorUseRule=UseColor_Specified, specifiedColor=<rgba 1 1 1 1>}
     uVRegion {bIsValid=false, max=<0 0>, min=<0 0>}
    }
    brushColor=<rgba 0.0152 0.0152 0.0152 1>
   }
   slot {class=HorizontalBoxSlot, properties {verticalAlignment=VAlign_Center}}
  }
  T_Arrange_X {
   widgetClassPath=&EditorUtilityEditableText
   bIsVariable=true
   parent=Border_XDistance
   widget=T_Arrange_X
   properties {
    text=-640
    toolTipText=Distance between assets in the X Direction
    widgetStyle {
     backgroundImageComposing {
      bIsDynamicallyLoaded=false
      drawAs=Border
      imageType=FullColor
      mirroring=NoMirror
      resourceName=../../../Engine/Content/Slate/Old/HyperlinkDotted.png
      resourceObject=null
      tiling=Both
      imageSize=<0 0>
      margin=<ltrb 0 0 0 0.1875>
      outlineSettings {
       bUseBrushTransparency=false
       roundingType=HalfHeightRadius
       width=0
       color {colorUseRule=UseColor_Specified, specifiedColor=<rgba 0 0 0 0>}
       cornerRadii=<0 0 0 0>
      }
      tintColor {colorUseRule=UseColor_Specified, specifiedColor=<rgba 1 1 1 1>}
      uVRegion {bIsValid=false, max=<0 0>, min=<0 0>}
     }
     backgroundImageSelected {
      bIsDynamicallyLoaded=false
      drawAs=Image
      imageType=NoImage
      mirroring=NoMirror
      resourceName=None
      resourceObject=null
      tiling=NoTile
      imageSize=<0 0>
      margin=<ltrb 0 0 0 0>
      outlineSettings {
       bUseBrushTransparency=false
       roundingType=HalfHeightRadius
       width=0
       color {colorUseRule=UseColor_Specified, specifiedColor=<rgba 0 0 0 0>}
       cornerRadii=<0 0 0 0>
      }
      tintColor {colorUseRule=UseColor_Specified, specifiedColor=<rgba 0 0.162 0.7454 1>}
      uVRegion {bIsValid=false, max=<0 0>, min=<0 0>}
     }
     caretImage {
      bIsDynamicallyLoaded=false
      drawAs=Image
      imageType=NoImage
      mirroring=NoMirror
      resourceName=None
      resourceObject=null
      tiling=NoTile
      imageSize=<0 0>
      margin=<ltrb 0 0 0 0>
      outlineSettings {
       bUseBrushTransparency=false
       roundingType=HalfHeightRadius
       width=0
       color {colorUseRule=UseColor_Specified, specifiedColor=<rgba 0 0 0 0>}
       cornerRadii=<0 0 0 0>
      }
      tintColor {colorUseRule=UseColor_Specified, specifiedColor=<rgba 1 1 1 1>}
      uVRegion {bIsValid=false, max=<0 0>, min=<0 0>}
     }
     colorAndOpacity {colorUseRule=UseColor_Specified, specifiedColor=<rgba 1 1 1 1>}
     font {
      bForceMonospaced=false
      bMaterialIsStencil=false
      fontMaterial=null
      fontObject=&Roboto
      letterSpacing=0
      monospacedWidth=1
      size=10.5
      skewAmount=0
      typefaceFontName=Bold
      outlineSettings {
       bApplyOutlineToDropShadows=false
       bMiteredCorners=false
       bSeparateFillAlpha=false
       outlineMaterial=null
       outlineSize=0
       outlineColor=<rgba 0 0 0 1>
      }
     }
     hintColorAndOpacity {colorUseRule=UseColor_Foreground, specifiedColor=<rgba 1 0 1 1>}
     selectedBackgroundColor {colorUseRule=UseColor_Foreground, specifiedColor=<rgba 1 0 1 1>}
    }
   }
   slot {class=BorderSlot, properties {verticalAlignment=VAlign_Center}}
  }
  HorizontalBox-YDistance {
   widgetClassPath=&HorizontalBox
   parent=HorizontalBox-XYDistance
   widget=HorizontalBox-YDistance
   slot {class=HorizontalBoxSlot, properties {padding=<ltrb 10 0 0 0>}}
  }
  TextBlock_2 {
   widgetClassPath=&TextBlock
   parent=HorizontalBox-YDistance
   widget=TextBlock_2
   properties {
    text="Y Distance : "
    font {
     bForceMonospaced=false
     bMaterialIsStencil=false
     fontMaterial=null
     fontObject=&Roboto
     letterSpacing=0
     monospacedWidth=1
     size=9
     skewAmount=0
     typefaceFontName=Light
     outlineSettings {
      bApplyOutlineToDropShadows=false
      bMiteredCorners=false
      bSeparateFillAlpha=false
      outlineMaterial=null
      outlineSize=0
      outlineColor=<rgba 0 0 0 1>
     }
    }
   }
   slot {class=HorizontalBoxSlot, properties {verticalAlignment=VAlign_Center}}
  }
  Border_YDistance {
   widgetClassPath=&Border
   bIsVariable=true
   parent=HorizontalBox-YDistance
   widget=Border_YDistance
   properties {
    verticalAlignment=VAlign_Center
    background {
     bIsDynamicallyLoaded=false
     drawAs=RoundedBox
     imageType=NoImage
     mirroring=NoMirror
     resourceName=None
     resourceObject=null
     tiling=NoTile
     imageSize=<32 32>
     margin=<ltrb 0 0 0 0>
     outlineSettings {
      bUseBrushTransparency=false
      roundingType=FixedRadius
      width=0
      color {colorUseRule=UseColor_Specified, specifiedColor=<rgba 0 0 0 0>}
      cornerRadii=<5 5 5 5>
     }
     tintColor {colorUseRule=UseColor_Specified, specifiedColor=<rgba 1 1 1 1>}
     uVRegion {bIsValid=false, max=<0 0>, min=<0 0>}
    }
    brushColor=<rgba 0.0152 0.0152 0.0152 1>
   }
   slot {class=HorizontalBoxSlot, properties {verticalAlignment=VAlign_Center}}
  }
  T_Arrange_Y {
   widgetClassPath=&EditorUtilityEditableText
   bIsVariable=true
   parent=Border_YDistance
   widget=T_Arrange_Y
   properties {
    text=400
    toolTipText=Distance between assets in the Y Direction
    widgetStyle {
     backgroundImageComposing {
      bIsDynamicallyLoaded=false
      drawAs=Border
      imageType=FullColor
      mirroring=NoMirror
      resourceName=../../../Engine/Content/Slate/Old/HyperlinkDotted.png
      resourceObject=null
      tiling=Both
      imageSize=<0 0>
      margin=<ltrb 0 0 0 0.1875>
      outlineSettings {
       bUseBrushTransparency=false
       roundingType=HalfHeightRadius
       width=0
       color {colorUseRule=UseColor_Specified, specifiedColor=<rgba 0 0 0 0>}
       cornerRadii=<0 0 0 0>
      }
      tintColor {colorUseRule=UseColor_Specified, specifiedColor=<rgba 1 1 1 1>}
      uVRegion {bIsValid=false, max=<0 0>, min=<0 0>}
     }
     backgroundImageSelected {
      bIsDynamicallyLoaded=false
      drawAs=Image
      imageType=NoImage
      mirroring=NoMirror
      resourceName=None
      resourceObject=null
      tiling=NoTile
      imageSize=<0 0>
      margin=<ltrb 0 0 0 0>
      outlineSettings {
       bUseBrushTransparency=false
       roundingType=HalfHeightRadius
       width=0
       color {colorUseRule=UseColor_Specified, specifiedColor=<rgba 0 0 0 0>}
       cornerRadii=<0 0 0 0>
      }
      tintColor {colorUseRule=UseColor_Specified, specifiedColor=<rgba 0 0.162 0.7454 1>}
      uVRegion {bIsValid=false, max=<0 0>, min=<0 0>}
     }
     caretImage {
      bIsDynamicallyLoaded=false
      drawAs=Image
      imageType=NoImage
      mirroring=NoMirror
      resourceName=None
      resourceObject=null
      tiling=NoTile
      imageSize=<0 0>
      margin=<ltrb 0 0 0 0>
      outlineSettings {
       bUseBrushTransparency=false
       roundingType=HalfHeightRadius
       width=0
       color {colorUseRule=UseColor_Specified, specifiedColor=<rgba 0 0 0 0>}
       cornerRadii=<0 0 0 0>
      }
      tintColor {colorUseRule=UseColor_Specified, specifiedColor=<rgba 1 1 1 1>}
      uVRegion {bIsValid=false, max=<0 0>, min=<0 0>}
     }
     colorAndOpacity {colorUseRule=UseColor_Specified, specifiedColor=<rgba 1 1 1 1>}
     font {
      bForceMonospaced=false
      bMaterialIsStencil=false
      fontMaterial=null
      fontObject=&Roboto
      letterSpacing=0
      monospacedWidth=1
      size=10.5
      skewAmount=0
      typefaceFontName=Bold
      outlineSettings {
       bApplyOutlineToDropShadows=false
       bMiteredCorners=false
       bSeparateFillAlpha=false
       outlineMaterial=null
       outlineSize=0
       outlineColor=<rgba 0 0 0 1>
      }
     }
     hintColorAndOpacity {colorUseRule=UseColor_Foreground, specifiedColor=<rgba 1 0 1 1>}
     selectedBackgroundColor {colorUseRule=UseColor_Foreground, specifiedColor=<rgba 1 0 1 1>}
    }
   }
   slot {class=BorderSlot, properties {verticalAlignment=VAlign_Center}}
  }
  VerticalBox_3 {
   widgetClassPath=&VerticalBox
   parent=HorizontalBox
   widget=VerticalBox_3
   slot {class=HorizontalBoxSlot, properties {verticalAlignment=VAlign_Top, padding=<ltrb 1 1 1 1>}}
  }
  TextBlock_416 {
   widgetClassPath=&TextBlock
   parent=VerticalBox_3
   widget=TextBlock_416
   properties {
    text=Arrangement Presets
    colorAndOpacity {colorUseRule=UseColor_Specified, specifiedColor=<rgba 0.3385 0.3385 0.3385 1>}
    font {
     bForceMonospaced=false
     bMaterialIsStencil=false
     fontMaterial=null
     fontObject=&Roboto
     letterSpacing=0
     monospacedWidth=1
     size=10.5
     skewAmount=0
     typefaceFontName=Light
     outlineSettings {
      bApplyOutlineToDropShadows=false
      bMiteredCorners=false
      bSeparateFillAlpha=false
      outlineMaterial=null
      outlineSize=0
      outlineColor=<rgba 0 0 0 1>
     }
    }
   }
   slot {class=VerticalBoxSlot, properties {horizontalAlignment=HAlign_Center}}
  }
  B_Arrange_Distances_Custom {
   widgetClassPath=&EditorUtilityButton
   parent=VerticalBox_3
   widget=B_Arrange_Distances_Custom
   properties {widgetStyle {
     clickedSlateSound {resourceObject=null}
     disabled {
      bIsDynamicallyLoaded=false
      drawAs=RoundedBox
      imageType=NoImage
      mirroring=NoMirror
      resourceName=None
      resourceObject=null
      tiling=NoTile
      imageSize=<0 0>
      margin=<ltrb 0 0 0 0>
      outlineSettings {
       bUseBrushTransparency=false
       roundingType=FixedRadius
       width=1
       color {colorUseRule=UseColor_ColorTable, specifiedColor=<rgba 0 0 0 0>}
       cornerRadii=<4 4 4 4>
      }
      tintColor {colorUseRule=UseColor_Specified, specifiedColor=<rgba 0.0395 0.0395 0.0395 1>}
      uVRegion {bIsValid=false, max=<0 0>, min=<0 0>}
     }
     disabledForeground {colorUseRule=UseColor_ColorTable, specifiedColor=<rgba 0 0 0 0>}
     hovered {
      bIsDynamicallyLoaded=false
      drawAs=RoundedBox
      imageType=NoImage
      mirroring=NoMirror
      resourceName=None
      resourceObject=null
      tiling=NoTile
      imageSize=<32 32>
      margin=<ltrb 0 0 0 0>
      outlineSettings {
       bUseBrushTransparency=false
       roundingType=FixedRadius
       width=1
       color {colorUseRule=UseColor_ColorTable, specifiedColor=<rgba 0 0 0 0>}
       cornerRadii=<4 4 4 4>
      }
      tintColor {colorUseRule=UseColor_Specified, specifiedColor=<rgba 0.0953 0.0953 0.0953 1>}
      uVRegion {bIsValid=false, max=<0 0>, min=<0 0>}
     }
     hoveredForeground {colorUseRule=UseColor_ColorTable, specifiedColor=<rgba 0 0 0 0>}
     hoveredSlateSound {resourceObject=null}
     normal {
      bIsDynamicallyLoaded=false
      drawAs=RoundedBox
      imageType=NoImage
      mirroring=NoMirror
      resourceName=None
      resourceObject=null
      tiling=NoTile
      imageSize=<32 32>
      margin=<ltrb 0 0 0 0>
      outlineSettings {
       bUseBrushTransparency=false
       roundingType=FixedRadius
       width=1
       color {colorUseRule=UseColor_ColorTable, specifiedColor=<rgba 0 0 0 0>}
       cornerRadii=<4 4 4 4>
      }
      tintColor {colorUseRule=UseColor_Specified, specifiedColor=<rgba 0.0395 0.0395 0.0395 1>}
      uVRegion {bIsValid=false, max=<0 0>, min=<0 0>}
     }
     normalForeground {colorUseRule=UseColor_ColorTable, specifiedColor=<rgba 0 0 0 0>}
     normalPadding=<ltrb 12 1.5 12 1.5>
     pressed {
      bIsDynamicallyLoaded=false
      drawAs=RoundedBox
      imageType=NoImage
      mirroring=NoMirror
      resourceName=None
      resourceObject=null
      tiling=NoTile
      imageSize=<32 32>
      margin=<ltrb 0 0 0 0>
      outlineSettings {
       bUseBrushTransparency=false
       roundingType=FixedRadius
       width=1
       color {colorUseRule=UseColor_ColorTable, specifiedColor=<rgba 0 0 0 0>}
       cornerRadii=<4 4 4 4>
      }
      tintColor {colorUseRule=UseColor_Specified, specifiedColor=<rgba 0.0284 0.0284 0.0284 1>}
      uVRegion {bIsValid=false, max=<0 0>, min=<0 0>}
     }
     pressedForeground {colorUseRule=UseColor_ColorTable, specifiedColor=<rgba 0 0 0 0>}
     pressedPadding=<ltrb 12 2.5 12 0.5>
     pressedSlateSound {resourceObject=null}
    }}
   slot {class=VerticalBoxSlot}
  }
  TextBlock_28 {
   widgetClassPath=&TextBlock
   parent=B_Arrange_Distances_Custom
   widget=TextBlock_28
   properties {
    text=Custom
    font {
     bForceMonospaced=false
     bMaterialIsStencil=false
     fontMaterial=null
     fontObject=&Roboto
     letterSpacing=0
     monospacedWidth=1
     size=9
     skewAmount=0
     typefaceFontName=Regular
     outlineSettings {
      bApplyOutlineToDropShadows=false
      bMiteredCorners=false
      bSeparateFillAlpha=false
      outlineMaterial=null
      outlineSize=0
      outlineColor=<rgba 0 0 0 1>
     }
    }
   }
   slot {class=ButtonSlot}
  }
  HorizontalBox_57 {widgetClassPath=&HorizontalBox, parent=VerticalBox_3, widget=HorizontalBox_57, slot {class=VerticalBoxSlot}}
  VerticalBox_6 {
   widgetClassPath=&VerticalBox
   parent=HorizontalBox_57
   widget=VerticalBox_6
   slot {class=HorizontalBoxSlot, properties {size {sizeRule=Fill, value=1}}}
  }
  B_Arrange_Distances_Walls {
   widgetClassPath=&EditorUtilityButton
   parent=VerticalBox_6
   widget=B_Arrange_Distances_Walls
   properties {widgetStyle {
     clickedSlateSound {resourceObject=null}
     disabled {
      bIsDynamicallyLoaded=false
      drawAs=RoundedBox
      imageType=NoImage
      mirroring=NoMirror
      resourceName=None
      resourceObject=null
      tiling=NoTile
      imageSize=<0 0>
      margin=<ltrb 0 0 0 0>
      outlineSettings {
       bUseBrushTransparency=false
       roundingType=FixedRadius
       width=1
       color {colorUseRule=UseColor_ColorTable, specifiedColor=<rgba 0 0 0 0>}
       cornerRadii=<4 4 4 4>
      }
      tintColor {colorUseRule=UseColor_Specified, specifiedColor=<rgba 0.0395 0.0395 0.0395 1>}
      uVRegion {bIsValid=false, max=<0 0>, min=<0 0>}
     }
     disabledForeground {colorUseRule=UseColor_ColorTable, specifiedColor=<rgba 0 0 0 0>}
     hovered {
      bIsDynamicallyLoaded=false
      drawAs=RoundedBox
      imageType=NoImage
      mirroring=NoMirror
      resourceName=None
      resourceObject=null
      tiling=NoTile
      imageSize=<32 32>
      margin=<ltrb 0 0 0 0>
      outlineSettings {
       bUseBrushTransparency=false
       roundingType=FixedRadius
       width=1
       color {colorUseRule=UseColor_ColorTable, specifiedColor=<rgba 0 0 0 0>}
       cornerRadii=<4 4 4 4>
      }
      tintColor {colorUseRule=UseColor_Specified, specifiedColor=<rgba 0.0953 0.0953 0.0953 1>}
      uVRegion {bIsValid=false, max=<0 0>, min=<0 0>}
     }
     hoveredForeground {colorUseRule=UseColor_ColorTable, specifiedColor=<rgba 0 0 0 0>}
     hoveredSlateSound {resourceObject=null}
     normal {
      bIsDynamicallyLoaded=false
      drawAs=RoundedBox
      imageType=NoImage
      mirroring=NoMirror
      resourceName=None
      resourceObject=null
      tiling=NoTile
      imageSize=<32 32>
      margin=<ltrb 0 0 0 0>
      outlineSettings {
       bUseBrushTransparency=false
       roundingType=FixedRadius
       width=1
       color {colorUseRule=UseColor_ColorTable, specifiedColor=<rgba 0 0 0 0>}
       cornerRadii=<4 4 4 4>
      }
      tintColor {colorUseRule=UseColor_Specified, specifiedColor=<rgba 0.0395 0.0395 0.0395 1>}
      uVRegion {bIsValid=false, max=<0 0>, min=<0 0>}
     }
     normalForeground {colorUseRule=UseColor_ColorTable, specifiedColor=<rgba 0 0 0 0>}
     normalPadding=<ltrb 12 1.5 12 1.5>
     pressed {
      bIsDynamicallyLoaded=false
      drawAs=RoundedBox
      imageType=NoImage
      mirroring=NoMirror
      resourceName=None
      resourceObject=null
      tiling=NoTile
      imageSize=<32 32>
      margin=<ltrb 0 0 0 0>
      outlineSettings {
       bUseBrushTransparency=false
       roundingType=FixedRadius
       width=1
       color {colorUseRule=UseColor_ColorTable, specifiedColor=<rgba 0 0 0 0>}
       cornerRadii=<4 4 4 4>
      }
      tintColor {colorUseRule=UseColor_Specified, specifiedColor=<rgba 0.0284 0.0284 0.0284 1>}
      uVRegion {bIsValid=false, max=<0 0>, min=<0 0>}
     }
     pressedForeground {colorUseRule=UseColor_ColorTable, specifiedColor=<rgba 0 0 0 0>}
     pressedPadding=<ltrb 12 2.5 12 0.5>
     pressedSlateSound {resourceObject=null}
    }}
   slot {class=VerticalBoxSlot}
  }
  TextBlock_15 {
   widgetClassPath=&TextBlock
   parent=B_Arrange_Distances_Walls
   widget=TextBlock_15
   properties {
    justification=Center
    text=Walls
    font {
     bForceMonospaced=false
     bMaterialIsStencil=false
     fontMaterial=null
     fontObject=&Roboto
     letterSpacing=0
     monospacedWidth=1
     size=9
     skewAmount=0
     typefaceFontName=Regular
     outlineSettings {
      bApplyOutlineToDropShadows=false
      bMiteredCorners=false
      bSeparateFillAlpha=false
      outlineMaterial=null
      outlineSize=0
      outlineColor=<rgba 0 0 0 1>
     }
    }
   }
   slot {class=ButtonSlot}
  }
  B_Arrange_Distances_Props {
   widgetClassPath=&EditorUtilityButton
   parent=VerticalBox_6
   widget=B_Arrange_Distances_Props
   properties {widgetStyle {
     clickedSlateSound {resourceObject=null}
     disabled {
      bIsDynamicallyLoaded=false
      drawAs=RoundedBox
      imageType=NoImage
      mirroring=NoMirror
      resourceName=None
      resourceObject=null
      tiling=NoTile
      imageSize=<0 0>
      margin=<ltrb 0 0 0 0>
      outlineSettings {
       bUseBrushTransparency=false
       roundingType=FixedRadius
       width=1
       color {colorUseRule=UseColor_ColorTable, specifiedColor=<rgba 0 0 0 0>}
       cornerRadii=<4 4 4 4>
      }
      tintColor {colorUseRule=UseColor_Specified, specifiedColor=<rgba 0.0395 0.0395 0.0395 1>}
      uVRegion {bIsValid=false, max=<0 0>, min=<0 0>}
     }
     disabledForeground {colorUseRule=UseColor_ColorTable, specifiedColor=<rgba 0 0 0 0>}
     hovered {
      bIsDynamicallyLoaded=false
      drawAs=RoundedBox
      imageType=NoImage
      mirroring=NoMirror
      resourceName=None
      resourceObject=null
      tiling=NoTile
      imageSize=<32 32>
      margin=<ltrb 0 0 0 0>
      outlineSettings {
       bUseBrushTransparency=false
       roundingType=FixedRadius
       width=1
       color {colorUseRule=UseColor_ColorTable, specifiedColor=<rgba 0 0 0 0>}
       cornerRadii=<4 4 4 4>
      }
      tintColor {colorUseRule=UseColor_Specified, specifiedColor=<rgba 0.0953 0.0953 0.0953 1>}
      uVRegion {bIsValid=false, max=<0 0>, min=<0 0>}
     }
     hoveredForeground {colorUseRule=UseColor_ColorTable, specifiedColor=<rgba 0 0 0 0>}
     hoveredSlateSound {resourceObject=null}
     normal {
      bIsDynamicallyLoaded=false
      drawAs=RoundedBox
      imageType=NoImage
      mirroring=NoMirror
      resourceName=None
      resourceObject=null
      tiling=NoTile
      imageSize=<32 32>
      margin=<ltrb 0 0 0 0>
      outlineSettings {
       bUseBrushTransparency=false
       roundingType=FixedRadius
       width=1
       color {colorUseRule=UseColor_ColorTable, specifiedColor=<rgba 0 0 0 0>}
       cornerRadii=<4 4 4 4>
      }
      tintColor {colorUseRule=UseColor_Specified, specifiedColor=<rgba 0.0395 0.0395 0.0395 1>}
      uVRegion {bIsValid=false, max=<0 0>, min=<0 0>}
     }
     normalForeground {colorUseRule=UseColor_ColorTable, specifiedColor=<rgba 0 0 0 0>}
     normalPadding=<ltrb 12 1.5 12 1.5>
     pressed {
      bIsDynamicallyLoaded=false
      drawAs=RoundedBox
      imageType=NoImage
      mirroring=NoMirror
      resourceName=None
      resourceObject=null
      tiling=NoTile
      imageSize=<32 32>
      margin=<ltrb 0 0 0 0>
      outlineSettings {
       bUseBrushTransparency=false
       roundingType=FixedRadius
       width=1
       color {colorUseRule=UseColor_ColorTable, specifiedColor=<rgba 0 0 0 0>}
       cornerRadii=<4 4 4 4>
      }
      tintColor {colorUseRule=UseColor_Specified, specifiedColor=<rgba 0.0284 0.0284 0.0284 1>}
      uVRegion {bIsValid=false, max=<0 0>, min=<0 0>}
     }
     pressedForeground {colorUseRule=UseColor_ColorTable, specifiedColor=<rgba 0 0 0 0>}
     pressedPadding=<ltrb 12 2.5 12 0.5>
     pressedSlateSound {resourceObject=null}
    }}
   slot {class=VerticalBoxSlot}
  }
  TextBlock_16 {
   widgetClassPath=&TextBlock
   parent=B_Arrange_Distances_Props
   widget=TextBlock_16
   properties {
    justification=Center
    text=Props
    font {
     bForceMonospaced=false
     bMaterialIsStencil=false
     fontMaterial=null
     fontObject=&Roboto
     letterSpacing=0
     monospacedWidth=1
     size=9
     skewAmount=0
     typefaceFontName=Regular
     outlineSettings {
      bApplyOutlineToDropShadows=false
      bMiteredCorners=false
      bSeparateFillAlpha=false
      outlineMaterial=null
      outlineSize=0
      outlineColor=<rgba 0 0 0 1>
     }
    }
   }
   slot {class=ButtonSlot}
  }
  VerticalBox_5 {
   widgetClassPath=&VerticalBox
   parent=HorizontalBox_57
   widget=VerticalBox_5
   slot {class=HorizontalBoxSlot, properties {size {sizeRule=Fill, value=1}}}
  }
  B_Arrange_Distances_Floors {
   widgetClassPath=&EditorUtilityButton
   parent=VerticalBox_5
   widget=B_Arrange_Distances_Floors
   properties {widgetStyle {
     clickedSlateSound {resourceObject=null}
     disabled {
      bIsDynamicallyLoaded=false
      drawAs=RoundedBox
      imageType=NoImage
      mirroring=NoMirror
      resourceName=None
      resourceObject=null
      tiling=NoTile
      imageSize=<0 0>
      margin=<ltrb 0 0 0 0>
      outlineSettings {
       bUseBrushTransparency=false
       roundingType=FixedRadius
       width=1
       color {colorUseRule=UseColor_ColorTable, specifiedColor=<rgba 0 0 0 0>}
       cornerRadii=<4 4 4 4>
      }
      tintColor {colorUseRule=UseColor_Specified, specifiedColor=<rgba 0.0395 0.0395 0.0395 1>}
      uVRegion {bIsValid=false, max=<0 0>, min=<0 0>}
     }
     disabledForeground {colorUseRule=UseColor_ColorTable, specifiedColor=<rgba 0 0 0 0>}
     hovered {
      bIsDynamicallyLoaded=false
      drawAs=RoundedBox
      imageType=NoImage
      mirroring=NoMirror
      resourceName=None
      resourceObject=null
      tiling=NoTile
      imageSize=<32 32>
      margin=<ltrb 0 0 0 0>
      outlineSettings {
       bUseBrushTransparency=false
       roundingType=FixedRadius
       width=1
       color {colorUseRule=UseColor_ColorTable, specifiedColor=<rgba 0 0 0 0>}
       cornerRadii=<4 4 4 4>
      }
      tintColor {colorUseRule=UseColor_Specified, specifiedColor=<rgba 0.0953 0.0953 0.0953 1>}
      uVRegion {bIsValid=false, max=<0 0>, min=<0 0>}
     }
     hoveredForeground {colorUseRule=UseColor_ColorTable, specifiedColor=<rgba 0 0 0 0>}
     hoveredSlateSound {resourceObject=null}
     normal {
      bIsDynamicallyLoaded=false
      drawAs=RoundedBox
      imageType=NoImage
      mirroring=NoMirror
      resourceName=None
      resourceObject=null
      tiling=NoTile
      imageSize=<32 32>
      margin=<ltrb 0 0 0 0>
      outlineSettings {
       bUseBrushTransparency=false
       roundingType=FixedRadius
       width=1
       color {colorUseRule=UseColor_ColorTable, specifiedColor=<rgba 0 0 0 0>}
       cornerRadii=<4 4 4 4>
      }
      tintColor {colorUseRule=UseColor_Specified, specifiedColor=<rgba 0.0395 0.0395 0.0395 1>}
      uVRegion {bIsValid=false, max=<0 0>, min=<0 0>}
     }
     normalForeground {colorUseRule=UseColor_ColorTable, specifiedColor=<rgba 0 0 0 0>}
     normalPadding=<ltrb 12 1.5 12 1.5>
     pressed {
      bIsDynamicallyLoaded=false
      drawAs=RoundedBox
      imageType=NoImage
      mirroring=NoMirror
      resourceName=None
      resourceObject=null
      tiling=NoTile
      imageSize=<32 32>
      margin=<ltrb 0 0 0 0>
      outlineSettings {
       bUseBrushTransparency=false
       roundingType=FixedRadius
       width=1
       color {colorUseRule=UseColor_ColorTable, specifiedColor=<rgba 0 0 0 0>}
       cornerRadii=<4 4 4 4>
      }
      tintColor {colorUseRule=UseColor_Specified, specifiedColor=<rgba 0.0284 0.0284 0.0284 1>}
      uVRegion {bIsValid=false, max=<0 0>, min=<0 0>}
     }
     pressedForeground {colorUseRule=UseColor_ColorTable, specifiedColor=<rgba 0 0 0 0>}
     pressedPadding=<ltrb 12 2.5 12 0.5>
     pressedSlateSound {resourceObject=null}
    }}
   slot {class=VerticalBoxSlot}
  }
  TextBlock_13 {
   widgetClassPath=&TextBlock
   parent=B_Arrange_Distances_Floors
   widget=TextBlock_13
   properties {
    justification=Center
    text=Floors
    font {
     bForceMonospaced=false
     bMaterialIsStencil=false
     fontMaterial=null
     fontObject=&Roboto
     letterSpacing=0
     monospacedWidth=1
     size=9
     skewAmount=0
     typefaceFontName=Regular
     outlineSettings {
      bApplyOutlineToDropShadows=false
      bMiteredCorners=false
      bSeparateFillAlpha=false
      outlineMaterial=null
      outlineSize=0
      outlineColor=<rgba 0 0 0 1>
     }
    }
   }
   slot {class=ButtonSlot}
  }
  B_Arrange_Distances_SmallProps {
   widgetClassPath=&EditorUtilityButton
   parent=VerticalBox_5
   widget=B_Arrange_Distances_SmallProps
   properties {widgetStyle {
     clickedSlateSound {resourceObject=null}
     disabled {
      bIsDynamicallyLoaded=false
      drawAs=RoundedBox
      imageType=NoImage
      mirroring=NoMirror
      resourceName=None
      resourceObject=null
      tiling=NoTile
      imageSize=<0 0>
      margin=<ltrb 0 0 0 0>
      outlineSettings {
       bUseBrushTransparency=false
       roundingType=FixedRadius
       width=1
       color {colorUseRule=UseColor_ColorTable, specifiedColor=<rgba 0 0 0 0>}
       cornerRadii=<4 4 4 4>
      }
      tintColor {colorUseRule=UseColor_Specified, specifiedColor=<rgba 0.0395 0.0395 0.0395 1>}
      uVRegion {bIsValid=false, max=<0 0>, min=<0 0>}
     }
     disabledForeground {colorUseRule=UseColor_ColorTable, specifiedColor=<rgba 0 0 0 0>}
     hovered {
      bIsDynamicallyLoaded=false
      drawAs=RoundedBox
      imageType=NoImage
      mirroring=NoMirror
      resourceName=None
      resourceObject=null
      tiling=NoTile
      imageSize=<32 32>
      margin=<ltrb 0 0 0 0>
      outlineSettings {
       bUseBrushTransparency=false
       roundingType=FixedRadius
       width=1
       color {colorUseRule=UseColor_ColorTable, specifiedColor=<rgba 0 0 0 0>}
       cornerRadii=<4 4 4 4>
      }
      tintColor {colorUseRule=UseColor_Specified, specifiedColor=<rgba 0.0953 0.0953 0.0953 1>}
      uVRegion {bIsValid=false, max=<0 0>, min=<0 0>}
     }
     hoveredForeground {colorUseRule=UseColor_ColorTable, specifiedColor=<rgba 0 0 0 0>}
     hoveredSlateSound {resourceObject=null}
     normal {
      bIsDynamicallyLoaded=false
      drawAs=RoundedBox
      imageType=NoImage
      mirroring=NoMirror
      resourceName=None
      resourceObject=null
      tiling=NoTile
      imageSize=<32 32>
      margin=<ltrb 0 0 0 0>
      outlineSettings {
       bUseBrushTransparency=false
       roundingType=FixedRadius
       width=1
       color {colorUseRule=UseColor_ColorTable, specifiedColor=<rgba 0 0 0 0>}
       cornerRadii=<4 4 4 4>
      }
      tintColor {colorUseRule=UseColor_Specified, specifiedColor=<rgba 0.0395 0.0395 0.0395 1>}
      uVRegion {bIsValid=false, max=<0 0>, min=<0 0>}
     }
     normalForeground {colorUseRule=UseColor_ColorTable, specifiedColor=<rgba 0 0 0 0>}
     normalPadding=<ltrb 12 1.5 12 1.5>
     pressed {
      bIsDynamicallyLoaded=false
      drawAs=RoundedBox
      imageType=NoImage
      mirroring=NoMirror
      resourceName=None
      resourceObject=null
      tiling=NoTile
      imageSize=<32 32>
      margin=<ltrb 0 0 0 0>
      outlineSettings {
       bUseBrushTransparency=false
       roundingType=FixedRadius
       width=1
       color {colorUseRule=UseColor_ColorTable, specifiedColor=<rgba 0 0 0 0>}
       cornerRadii=<4 4 4 4>
      }
      tintColor {colorUseRule=UseColor_Specified, specifiedColor=<rgba 0.0284 0.0284 0.0284 1>}
      uVRegion {bIsValid=false, max=<0 0>, min=<0 0>}
     }
     pressedForeground {colorUseRule=UseColor_ColorTable, specifiedColor=<rgba 0 0 0 0>}
     pressedPadding=<ltrb 12 2.5 12 0.5>
     pressedSlateSound {resourceObject=null}
    }}
   slot {class=VerticalBoxSlot}
  }
  TextBlock_14 {
   widgetClassPath=&TextBlock
   parent=B_Arrange_Distances_SmallProps
   widget=TextBlock_14
   properties {
    justification=Center
    text=Small Props
    font {
     bForceMonospaced=false
     bMaterialIsStencil=false
     fontMaterial=null
     fontObject=&Roboto
     letterSpacing=0
     monospacedWidth=1
     size=9
     skewAmount=0
     typefaceFontName=Regular
     outlineSettings {
      bApplyOutlineToDropShadows=false
      bMiteredCorners=false
      bSeparateFillAlpha=false
      outlineMaterial=null
      outlineSize=0
      outlineColor=<rgba 0 0 0 1>
     }
    }
   }
   slot {class=ButtonSlot, properties {padding=<ltrb 0 2 0 2>}}
  }
  HorizontalBox_178 {
   widgetClassPath=&HorizontalBox
   parent=VerticalBox_0
   widget=HorizontalBox_178
   slot {class=VerticalBoxSlot, properties {horizontalAlignment=HAlign_Center, padding=<ltrb 0 10 0 0>}}
  }
  HorizontalBox_11 {
   widgetClassPath=&HorizontalBox
   parent=HorizontalBox_178
   widget=HorizontalBox_11
   slot {class=HorizontalBoxSlot, properties {padding=<ltrb 0 0 10 0>}}
  }
  TextBlock_12 {
   widgetClassPath=&TextBlock
   parent=HorizontalBox_11
   widget=TextBlock_12
   properties {
    text="Order by Size : "
    font {
     bForceMonospaced=false
     bMaterialIsStencil=false
     fontMaterial=null
     fontObject=&Roboto
     letterSpacing=0
     monospacedWidth=1
     size=9
     skewAmount=0
     typefaceFontName=Light
     outlineSettings {
      bApplyOutlineToDropShadows=false
      bMiteredCorners=false
      bSeparateFillAlpha=false
      outlineMaterial=null
      outlineSize=0
      outlineColor=<rgba 0 0 0 1>
     }
    }
   }
   slot {class=HorizontalBoxSlot, properties {verticalAlignment=VAlign_Center}}
  }
  CB_Arrange_BySize {
   widgetClassPath=&EditorUtilityCheckBox
   bIsVariable=true
   parent=HorizontalBox_11
   widget=CB_Arrange_BySize
   properties {
    checkedState=Checked
    toolTipText="Arrange selected assets by size (bounds)"
    widgetStyle {
     checkBoxType=CheckBox
     backgroundHoveredImage {
      bIsDynamicallyLoaded=false
      drawAs=RoundedBox
      imageType=NoImage
      mirroring=NoMirror
      resourceName=None
      resourceObject=null
      tiling=NoTile
      imageSize=<18 18>
      margin=<ltrb 0 0 0 0>
      outlineSettings {
       bUseBrushTransparency=false
       roundingType=FixedRadius
       width=1
       color {colorUseRule=UseColor_ColorTable, specifiedColor=<rgba 0 0 0 0>}
       cornerRadii=<3 3 3 3>
      }
      tintColor {colorUseRule=UseColor_Specified, specifiedColor=<rgba 0.0048 0.0048 0.0048 1>}
      uVRegion {bIsValid=false, max=<0 0>, min=<0 0>}
     }
     backgroundImage {
      bIsDynamicallyLoaded=false
      drawAs=RoundedBox
      imageType=NoImage
      mirroring=NoMirror
      resourceName=None
      resourceObject=null
      tiling=NoTile
      imageSize=<18 18>
      margin=<ltrb 0 0 0 0>
      outlineSettings {
       bUseBrushTransparency=false
       roundingType=FixedRadius
       width=1
       color {colorUseRule=UseColor_ColorTable, specifiedColor=<rgba 0 0 0 0>}
       cornerRadii=<3 3 3 3>
      }
      tintColor {colorUseRule=UseColor_Specified, specifiedColor=<rgba 0.0048 0.0048 0.0048 1>}
      uVRegion {bIsValid=false, max=<0 0>, min=<0 0>}
     }
     backgroundPressedImage {
      bIsDynamicallyLoaded=false
      drawAs=RoundedBox
      imageType=NoImage
      mirroring=NoMirror
      resourceName=None
      resourceObject=null
      tiling=NoTile
      imageSize=<18 18>
      margin=<ltrb 0 0 0 0>
      outlineSettings {
       bUseBrushTransparency=false
       roundingType=FixedRadius
       width=1
       color {colorUseRule=UseColor_ColorTable, specifiedColor=<rgba 0 0 0 0>}
       cornerRadii=<3 3 3 3>
      }
      tintColor {colorUseRule=UseColor_Specified, specifiedColor=<rgba 0.0048 0.0048 0.0048 1>}
      uVRegion {bIsValid=false, max=<0 0>, min=<0 0>}
     }
     borderBackgroundColor {colorUseRule=UseColor_Specified, specifiedColor=<rgba 1 1 1 1>}
     checkedForeground {colorUseRule=UseColor_Specified, specifiedColor=<rgba 1 1 1 1>}
     checkedHoveredForeground {colorUseRule=UseColor_Specified, specifiedColor=<rgba 1 1 1 1>}
     checkedHoveredImage {
      bIsDynamicallyLoaded=false
      drawAs=Image
      imageType=Vector
      mirroring=NoMirror
      resourceName=../../../Engine/Content/Slate/Starship/CoreWidgets/CheckBox/check.svg
      resourceObject=null
      tiling=NoTile
      imageSize=<24 24>
      margin=<ltrb 0 0 0 0>
      outlineSettings {
       bUseBrushTransparency=false
       roundingType=HalfHeightRadius
       width=0
       color {colorUseRule=UseColor_Specified, specifiedColor=<rgba 0 0 0 0>}
       cornerRadii=<0 0 0 0>
      }
      tintColor {colorUseRule=UseColor_Specified, specifiedColor=<rgba 0.0044 0.2384 1 1>}
      uVRegion {bIsValid=false, max=<0 0>, min=<0 0>}
     }
     checkedImage {
      bIsDynamicallyLoaded=false
      drawAs=Image
      imageType=Vector
      mirroring=NoMirror
      resourceName=../../../Engine/Content/Slate/Starship/CoreWidgets/CheckBox/check.svg
      resourceObject=null
      tiling=NoTile
      imageSize=<24 24>
      margin=<ltrb 0 0 0 0>
      outlineSettings {
       bUseBrushTransparency=false
       roundingType=HalfHeightRadius
       width=0
       color {colorUseRule=UseColor_Specified, specifiedColor=<rgba 0 0 0 0>}
       cornerRadii=<0 0 0 0>
      }
      tintColor {colorUseRule=UseColor_Specified, specifiedColor=<rgba 0 0.162 0.7454 1>}
      uVRegion {bIsValid=false, max=<0 0>, min=<0 0>}
     }
     checkedPressedForeground {colorUseRule=UseColor_Specified, specifiedColor=<rgba 1 1 1 1>}
     checkedPressedImage {
      bIsDynamicallyLoaded=false
      drawAs=Image
      imageType=Vector
      mirroring=NoMirror
      resourceName=../../../Engine/Content/Slate/Starship/CoreWidgets/CheckBox/check.svg
      resourceObject=null
      tiling=NoTile
      imageSize=<24 24>
      margin=<ltrb 0 0 0 0>
      outlineSettings {
       bUseBrushTransparency=false
       roundingType=HalfHeightRadius
       width=0
       color {colorUseRule=UseColor_Specified, specifiedColor=<rgba 0 0 0 0>}
       cornerRadii=<0 0 0 0>
      }
      tintColor {colorUseRule=UseColor_Specified, specifiedColor=<rgba 0 0.0802 0.3515 1>}
      uVRegion {bIsValid=false, max=<0 0>, min=<0 0>}
     }
     checkedSlateSound {resourceObject=null}
     foregroundColor {colorUseRule=UseColor_Specified, specifiedColor=<rgba 1 1 1 1>}
     hoveredForeground {colorUseRule=UseColor_Specified, specifiedColor=<rgba 1 1 1 1>}
     hoveredSlateSound {resourceObject=null}
     padding=<ltrb 0 0 0 0>
     pressedForeground {colorUseRule=UseColor_Specified, specifiedColor=<rgba 1 1 1 1>}
     uncheckedHoveredImage {
      bIsDynamicallyLoaded=false
      drawAs=NoDrawType
      imageType=NoImage
      mirroring=NoMirror
      resourceName=None
      resourceObject=null
      tiling=NoTile
      imageSize=<24 24>
      margin=<ltrb 0 0 0 0>
      outlineSettings {
       bUseBrushTransparency=false
       roundingType=HalfHeightRadius
       width=0
       color {colorUseRule=UseColor_Specified, specifiedColor=<rgba 0 0 0 0>}
       cornerRadii=<0 0 0 0>
      }
      tintColor {colorUseRule=UseColor_Specified, specifiedColor=<rgba 1 1 1 1>}
      uVRegion {bIsValid=false, max=<0 0>, min=<0 0>}
     }
     uncheckedImage {
      bIsDynamicallyLoaded=false
      drawAs=NoDrawType
      imageType=NoImage
      mirroring=NoMirror
      resourceName=None
      resourceObject=null
      tiling=NoTile
      imageSize=<24 24>
      margin=<ltrb 0 0 0 0>
      outlineSettings {
       bUseBrushTransparency=false
       roundingType=HalfHeightRadius
       width=0
       color {colorUseRule=UseColor_Specified, specifiedColor=<rgba 0 0 0 0>}
       cornerRadii=<0 0 0 0>
      }
      tintColor {colorUseRule=UseColor_Specified, specifiedColor=<rgba 1 1 1 1>}
      uVRegion {bIsValid=false, max=<0 0>, min=<0 0>}
     }
     uncheckedPressedImage {
      bIsDynamicallyLoaded=false
      drawAs=NoDrawType
      imageType=NoImage
      mirroring=NoMirror
      resourceName=None
      resourceObject=null
      tiling=NoTile
      imageSize=<24 24>
      margin=<ltrb 0 0 0 0>
      outlineSettings {
       bUseBrushTransparency=false
       roundingType=HalfHeightRadius
       width=0
       color {colorUseRule=UseColor_Specified, specifiedColor=<rgba 0 0 0 0>}
       cornerRadii=<0 0 0 0>
      }
      tintColor {colorUseRule=UseColor_Specified, specifiedColor=<rgba 1 1 1 1>}
      uVRegion {bIsValid=false, max=<0 0>, min=<0 0>}
     }
     uncheckedSlateSound {resourceObject=null}
     undeterminedForeground {colorUseRule=UseColor_Specified, specifiedColor=<rgba 1 1 1 1>}
     undeterminedHoveredImage {
      bIsDynamicallyLoaded=false
      drawAs=Image
      imageType=Vector
      mirroring=NoMirror
      resourceName=../../../Engine/Content/Slate/Starship/CoreWidgets/CheckBox/indeterminate.svg
      resourceObject=null
      tiling=NoTile
      imageSize=<16 16>
      margin=<ltrb 0 0 0 0>
      outlineSettings {
       bUseBrushTransparency=false
       roundingType=HalfHeightRadius
       width=0
       color {colorUseRule=UseColor_Specified, specifiedColor=<rgba 0 0 0 0>}
       cornerRadii=<0 0 0 0>
      }
      tintColor {colorUseRule=UseColor_Specified, specifiedColor=<rgba 0.0044 0.2384 1 1>}
      uVRegion {bIsValid=false, max=<0 0>, min=<0 0>}
     }
     undeterminedImage {
      bIsDynamicallyLoaded=false
      drawAs=Image
      imageType=Vector
      mirroring=NoMirror
      resourceName=../../../Engine/Content/Slate/Starship/CoreWidgets/CheckBox/indeterminate.svg
      resourceObject=null
      tiling=NoTile
      imageSize=<16 16>
      margin=<ltrb 0 0 0 0>
      outlineSettings {
       bUseBrushTransparency=false
       roundingType=HalfHeightRadius
       width=0
       color {colorUseRule=UseColor_Specified, specifiedColor=<rgba 0 0 0 0>}
       cornerRadii=<0 0 0 0>
      }
      tintColor {colorUseRule=UseColor_Specified, specifiedColor=<rgba 0 0.162 0.7454 1>}
      uVRegion {bIsValid=false, max=<0 0>, min=<0 0>}
     }
     undeterminedPressedImage {
      bIsDynamicallyLoaded=false
      drawAs=Image
      imageType=Vector
      mirroring=NoMirror
      resourceName=../../../Engine/Content/Slate/Starship/CoreWidgets/CheckBox/indeterminate.svg
      resourceObject=null
      tiling=NoTile
      imageSize=<16 16>
      margin=<ltrb 0 0 0 0>
      outlineSettings {
       bUseBrushTransparency=false
       roundingType=HalfHeightRadius
       width=0
       color {colorUseRule=UseColor_Specified, specifiedColor=<rgba 0 0 0 0>}
       cornerRadii=<0 0 0 0>
      }
      tintColor {colorUseRule=UseColor_Specified, specifiedColor=<rgba 0 0.0802 0.3515 1>}
      uVRegion {bIsValid=false, max=<0 0>, min=<0 0>}
     }
    }
   }
   slot {class=HorizontalBoxSlot}
  }
  HorizontalBox_9 {
   widgetClassPath=&HorizontalBox
   parent=HorizontalBox_178
   widget=HorizontalBox_9
   slot {class=HorizontalBoxSlot, properties {horizontalAlignment=HAlign_Center, padding=<ltrb 30 0 0 0>}}
  }
  TextBlock_10 {
   widgetClassPath=&TextBlock
   parent=HorizontalBox_9
   widget=TextBlock_10
   properties {
    text="Grid Origin : "
    font {
     bForceMonospaced=false
     bMaterialIsStencil=false
     fontMaterial=null
     fontObject=&Roboto
     letterSpacing=0
     monospacedWidth=1
     size=9
     skewAmount=0
     typefaceFontName=Light
     outlineSettings {
      bApplyOutlineToDropShadows=false
      bMiteredCorners=false
      bSeparateFillAlpha=false
      outlineMaterial=null
      outlineSize=0
      outlineColor=<rgba 0 0 0 1>
     }
    }
   }
   slot {class=HorizontalBoxSlot, properties {verticalAlignment=VAlign_Center}}
  }
  Combo_Arrange_OriginMode {
   widgetClassPath=&EditorUtilityComboBoxString
   bIsVariable=true
   parent=HorizontalBox_9
   widget=Combo_Arrange_OriginMode
   properties {
    selectedOption=Bottom left Corner
    toolTipText="Determine the final position of the grid. The origin of the grid is the element at its bottom left corner.\r\n\r\nBottom left Corner - Align the grid origin with the bottom left corner of the selection's bounds.\r\n\r\nFrom First Selected - Align the grid origin with the first selected actor.\r\n\r\nFrom Last Selected - Align the grid origin with the last selected actor.\r\n\r\nAbove Last Selected (Y offset) - Align the grid origin with the last selected actor with an offset in Y. Does NOT include the last selected actor into the grid.\r\n\r\nBesides Last Selected (X offset) - Align the grid origin with the last selected actor with an offset in X. Does NOT include the last selected actor into the grid. \r\n\r\nStored Position - Align the grid origin with the Stored Position.\r\n\r\nMarker Object Position - Align the grid origin with the Marker Object position.\r\n\r\nWorld Origin - Align the grid origin to 0/0/0.\r\n\r\nGallery Corner - Align the grid origin to the default gallery corner."
    defaultOptions [
     Bottom left Corner
     From First Selected
     From Last Selected
     "Above Last Selected (Y offset)"
     "Besides Last Selected (X offset)"
     Stored Position
     Marker Object Position
     World Origin
     Gallery Corner
    ]
    font {
     bForceMonospaced=false
     bMaterialIsStencil=false
     fontMaterial=null
     fontObject=&Roboto
     letterSpacing=0
     monospacedWidth=1
     size=9
     skewAmount=0
     typefaceFontName=Regular
     outlineSettings {
      bApplyOutlineToDropShadows=false
      bMiteredCorners=false
      bSeparateFillAlpha=false
      outlineMaterial=null
      outlineSize=0
      outlineColor=<rgba 0 0 0 1>
     }
    }
    foregroundColor {colorUseRule=UseColor_Specified, specifiedColor=<rgba 0.5271 0.5271 0.5271 1>}
   }
   slot {class=HorizontalBoxSlot}
  }
  Border-ActorTransforms {
   widgetClassPath=&Border
   parent=VerticalBox_Master
   widget=Border-ActorTransforms
   properties {
    background {
     bIsDynamicallyLoaded=false
     drawAs=Box
     imageType=FullColor
     mirroring=NoMirror
     resourceName=None
     resourceObject=&M_UI_Border
     tiling=NoTile
     imageSize=<40 40>
     margin=<ltrb 0.5 0.5 0.5 0.5>
     outlineSettings {
      bUseBrushTransparency=false
      roundingType=HalfHeightRadius
      width=0
      color {colorUseRule=UseColor_Specified, specifiedColor=<rgba 0 0 0 0>}
      cornerRadii=<0 0 0 0>
     }
     tintColor {colorUseRule=UseColor_Specified, specifiedColor=<rgba 1 0.3875 0 0>}
     uVRegion {bIsValid=false, max=<0 0>, min=<0 0>}
    }
    padding=<ltrb 0 0 0 0>
   }
   slot {class=VerticalBoxSlot, properties {padding=<ltrb 0 10 10 0>}}
  }
  VerticalBox_2 {
   widgetClassPath=&VerticalBox
   parent=Border-ActorTransforms
   widget=VerticalBox_2
   slot {class=BorderSlot, properties {padding=<ltrb 0 0 0 0>}}
  }
  ToggleButton_Collapse_ActorTransforms {
   widgetClassPath=&EditorUtilityCheckBox
   bIsVariable=true
   parent=VerticalBox_2
   widget=ToggleButton_Collapse_ActorTransforms
   properties {
    toolTipText=Expands / Contracts the Actor Transform Utilities module
    widgetStyle {
     checkBoxType=ToggleButton
     backgroundHoveredImage {
      bIsDynamicallyLoaded=false
      drawAs=RoundedBox
      imageType=NoImage
      mirroring=NoMirror
      resourceName=None
      resourceObject=null
      tiling=NoTile
      imageSize=<18 18>
      margin=<ltrb 0 0 0 0>
      outlineSettings {
       bUseBrushTransparency=false
       roundingType=FixedRadius
       width=1
       color {colorUseRule=UseColor_ColorTable, specifiedColor=<rgba 0 0 0 0>}
       cornerRadii=<3 3 3 3>
      }
      tintColor {colorUseRule=UseColor_Specified, specifiedColor=<rgba 0.0048 0.0048 0.0048 1>}
      uVRegion {bIsValid=false, max=<0 0>, min=<0 0>}
     }
     backgroundImage {
      bIsDynamicallyLoaded=false
      drawAs=RoundedBox
      imageType=NoImage
      mirroring=NoMirror
      resourceName=None
      resourceObject=null
      tiling=NoTile
      imageSize=<18 18>
      margin=<ltrb 0 0 0 0>
      outlineSettings {
       bUseBrushTransparency=false
       roundingType=FixedRadius
       width=1
       color {colorUseRule=UseColor_ColorTable, specifiedColor=<rgba 0 0 0 0>}
       cornerRadii=<3 3 3 3>
      }
      tintColor {colorUseRule=UseColor_Specified, specifiedColor=<rgba 0.0048 0.0048 0.0048 1>}
      uVRegion {bIsValid=false, max=<0 0>, min=<0 0>}
     }
     backgroundPressedImage {
      bIsDynamicallyLoaded=false
      drawAs=RoundedBox
      imageType=NoImage
      mirroring=NoMirror
      resourceName=None
      resourceObject=null
      tiling=NoTile
      imageSize=<18 18>
      margin=<ltrb 0 0 0 0>
      outlineSettings {
       bUseBrushTransparency=false
       roundingType=FixedRadius
       width=1
       color {colorUseRule=UseColor_ColorTable, specifiedColor=<rgba 0 0 0 0>}
       cornerRadii=<3 3 3 3>
      }
      tintColor {colorUseRule=UseColor_Specified, specifiedColor=<rgba 0.0048 0.0048 0.0048 1>}
      uVRegion {bIsValid=false, max=<0 0>, min=<0 0>}
     }
     borderBackgroundColor {colorUseRule=UseColor_Specified, specifiedColor=<rgba 1 1 1 1>}
     checkedForeground {colorUseRule=UseColor_Specified, specifiedColor=<rgba 1 1 1 1>}
     checkedHoveredForeground {colorUseRule=UseColor_Specified, specifiedColor=<rgba 1 1 1 1>}
     checkedHoveredImage {
      bIsDynamicallyLoaded=false
      drawAs=Image
      imageType=NoImage
      mirroring=NoMirror
      resourceName=None
      resourceObject=null
      tiling=NoTile
      imageSize=<32 10.5981>
      margin=<ltrb 0 0 0 0>
      outlineSettings {
       bUseBrushTransparency=false
       roundingType=HalfHeightRadius
       width=0
       color {colorUseRule=UseColor_Specified, specifiedColor=<rgba 0 0 0 0>}
       cornerRadii=<0 0 0 0>
      }
      tintColor {colorUseRule=UseColor_Specified, specifiedColor=<rgba 0.0953 0.0953 0.0953 1>}
      uVRegion {bIsValid=false, max=<0 0>, min=<0 0>}
     }
     checkedImage {
      bIsDynamicallyLoaded=false
      drawAs=Image
      imageType=NoImage
      mirroring=NoMirror
      resourceName=None
      resourceObject=null
      tiling=NoTile
      imageSize=<32 10.5981>
      margin=<ltrb 0 0 0 0>
      outlineSettings {
       bUseBrushTransparency=false
       roundingType=HalfHeightRadius
       width=0
       color {colorUseRule=UseColor_Specified, specifiedColor=<rgba 0 0 0 0>}
       cornerRadii=<0 0 0 0>
      }
      tintColor {colorUseRule=UseColor_Specified, specifiedColor=<rgba 0.0284 0.0284 0.0284 1>}
      uVRegion {bIsValid=false, max=<0 0>, min=<0 0>}
     }
     checkedPressedForeground {colorUseRule=UseColor_Specified, specifiedColor=<rgba 1 1 1 1>}
     checkedPressedImage {
      bIsDynamicallyLoaded=false
      drawAs=Image
      imageType=NoImage
      mirroring=NoMirror
      resourceName=None
      resourceObject=null
      tiling=NoTile
      imageSize=<32 10.5981>
      margin=<ltrb 0 0 0 0>
      outlineSettings {
       bUseBrushTransparency=false
       roundingType=HalfHeightRadius
       width=0
       color {colorUseRule=UseColor_Specified, specifiedColor=<rgba 0 0 0 0>}
       cornerRadii=<0 0 0 0>
      }
      tintColor {colorUseRule=UseColor_Specified, specifiedColor=<rgba 0.0284 0.0284 0.0284 1>}
      uVRegion {bIsValid=false, max=<0 0>, min=<0 0>}
     }
     checkedSlateSound {resourceObject=null}
     foregroundColor {colorUseRule=UseColor_Specified, specifiedColor=<rgba 1 1 1 1>}
     hoveredForeground {colorUseRule=UseColor_Specified, specifiedColor=<rgba 1 1 1 1>}
     hoveredSlateSound {resourceObject=null}
     padding=<ltrb 0 0 0 0>
     pressedForeground {colorUseRule=UseColor_Specified, specifiedColor=<rgba 1 1 1 1>}
     uncheckedHoveredImage {
      bIsDynamicallyLoaded=false
      drawAs=Image
      imageType=NoImage
      mirroring=NoMirror
      resourceName=None
      resourceObject=null
      tiling=NoTile
      imageSize=<32 10.5981>
      margin=<ltrb 0 0 0 0>
      outlineSettings {
       bUseBrushTransparency=false
       roundingType=HalfHeightRadius
       width=0
       color {colorUseRule=UseColor_Specified, specifiedColor=<rgba 0 0 0 0>}
       cornerRadii=<0 0 0 0>
      }
      tintColor {colorUseRule=UseColor_Specified, specifiedColor=<rgba 0.0953 0.0953 0.0953 1>}
      uVRegion {bIsValid=false, max=<0 0>, min=<0 0>}
     }
     uncheckedImage {
      bIsDynamicallyLoaded=false
      drawAs=Image
      imageType=NoImage
      mirroring=NoMirror
      resourceName=None
      resourceObject=null
      tiling=NoTile
      imageSize=<32 10.5981>
      margin=<ltrb 0 0 0 0>
      outlineSettings {
       bUseBrushTransparency=false
       roundingType=HalfHeightRadius
       width=0
       color {colorUseRule=UseColor_Specified, specifiedColor=<rgba 0 0 0 0>}
       cornerRadii=<0 0 0 0>
      }
      tintColor {colorUseRule=UseColor_Specified, specifiedColor=<rgba 0.0284 0.0284 0.0284 1>}
      uVRegion {bIsValid=false, max=<0 0>, min=<0 0>}
     }
     uncheckedPressedImage {
      bIsDynamicallyLoaded=false
      drawAs=Image
      imageType=NoImage
      mirroring=NoMirror
      resourceName=None
      resourceObject=null
      tiling=NoTile
      imageSize=<32 10.5981>
      margin=<ltrb 0 0 0 0>
      outlineSettings {
       bUseBrushTransparency=false
       roundingType=HalfHeightRadius
       width=0
       color {colorUseRule=UseColor_Specified, specifiedColor=<rgba 0 0 0 0>}
       cornerRadii=<0 0 0 0>
      }
      tintColor {colorUseRule=UseColor_Specified, specifiedColor=<rgba 0.0284 0.0284 0.0284 1>}
      uVRegion {bIsValid=false, max=<0 0>, min=<0 0>}
     }
     uncheckedSlateSound {resourceObject=null}
     undeterminedForeground {colorUseRule=UseColor_Specified, specifiedColor=<rgba 1 1 1 1>}
     undeterminedHoveredImage {
      bIsDynamicallyLoaded=false
      drawAs=Image
      imageType=Vector
      mirroring=NoMirror
      resourceName=../../../Engine/Content/Slate/Starship/CoreWidgets/CheckBox/indeterminate.svg
      resourceObject=null
      tiling=NoTile
      imageSize=<16 16>
      margin=<ltrb 0 0 0 0>
      outlineSettings {
       bUseBrushTransparency=false
       roundingType=HalfHeightRadius
       width=0
       color {colorUseRule=UseColor_Specified, specifiedColor=<rgba 0 0 0 0>}
       cornerRadii=<0 0 0 0>
      }
      tintColor {colorUseRule=UseColor_Specified, specifiedColor=<rgba 0.0044 0.2384 1 1>}
      uVRegion {bIsValid=false, max=<0 0>, min=<0 0>}
     }
     undeterminedImage {
      bIsDynamicallyLoaded=false
      drawAs=Image
      imageType=Vector
      mirroring=NoMirror
      resourceName=../../../Engine/Content/Slate/Starship/CoreWidgets/CheckBox/indeterminate.svg
      resourceObject=null
      tiling=NoTile
      imageSize=<16 16>
      margin=<ltrb 0 0 0 0>
      outlineSettings {
       bUseBrushTransparency=false
       roundingType=HalfHeightRadius
       width=0
       color {colorUseRule=UseColor_Specified, specifiedColor=<rgba 0 0 0 0>}
       cornerRadii=<0 0 0 0>
      }
      tintColor {colorUseRule=UseColor_Specified, specifiedColor=<rgba 0 0.162 0.7454 1>}
      uVRegion {bIsValid=false, max=<0 0>, min=<0 0>}
     }
     undeterminedPressedImage {
      bIsDynamicallyLoaded=false
      drawAs=Image
      imageType=Vector
      mirroring=NoMirror
      resourceName=../../../Engine/Content/Slate/Starship/CoreWidgets/CheckBox/indeterminate.svg
      resourceObject=null
      tiling=NoTile
      imageSize=<16 16>
      margin=<ltrb 0 0 0 0>
      outlineSettings {
       bUseBrushTransparency=false
       roundingType=HalfHeightRadius
       width=0
       color {colorUseRule=UseColor_Specified, specifiedColor=<rgba 0 0 0 0>}
       cornerRadii=<0 0 0 0>
      }
      tintColor {colorUseRule=UseColor_Specified, specifiedColor=<rgba 0 0.0802 0.3515 1>}
      uVRegion {bIsValid=false, max=<0 0>, min=<0 0>}
     }
    }
   }
   slot {class=VerticalBoxSlot}
  }
  HorizontalBox_6 {widgetClassPath=&HorizontalBox, parent=ToggleButton_Collapse_ActorTransforms, widget=HorizontalBox_6, slot {class=PanelSlot}}
  Spacer_2 {
   widgetClassPath=&Spacer
   parent=HorizontalBox_6
   widget=Spacer_2
   properties {size=<12 1>}
   slot {class=HorizontalBoxSlot, properties {horizontalAlignment=HAlign_Left}}
  }
  T_ToggleButton_ActorTransforms {
   widgetClassPath=&TextBlock
   bIsVariable=true
   parent=HorizontalBox_6
   widget=T_ToggleButton_ActorTransforms
   properties {
    text="Actor Transform Utilities ▲"
    font {
     bForceMonospaced=false
     bMaterialIsStencil=false
     fontMaterial=null
     fontObject=&Roboto
     letterSpacing=0
     monospacedWidth=1
     size=12
     skewAmount=0
     typefaceFontName=Bold
     outlineSettings {
      bApplyOutlineToDropShadows=false
      bMiteredCorners=false
      bSeparateFillAlpha=false
      outlineMaterial=null
      outlineSize=0
      outlineColor=<rgba 0 0 0 1>
     }
    }
   }
   slot {class=HorizontalBoxSlot}
  }
  Socket_ActorTransforms {
   widgetClassPath=&VerticalBox
   bIsVariable=true
   parent=VerticalBox_2
   widget=Socket_ActorTransforms
   slot {class=VerticalBoxSlot, properties {padding=<ltrb 10 10 0 20>}}
  }
  VerticalBox-MoveToPosition {
   widgetClassPath=&VerticalBox
   parent=Socket_ActorTransforms
   widget=VerticalBox-MoveToPosition
   slot {class=VerticalBoxSlot, properties {padding=<ltrb 0 10 0 10>}}
  }
  VerticalBox-MoveModes {
   widgetClassPath=&VerticalBox
   parent=VerticalBox-MoveToPosition
   widget=VerticalBox-MoveModes
   slot {class=VerticalBoxSlot, properties {padding=<ltrb 10 0 0 0>}}
  }
  EUW_AssetOrganizer_MoveSelected_ToXYZ {
   widgetClassPath=/Game/ArtTools/FNEC/AssetOrganizer/Widgets/EUW_AssetOrganizer_MoveSelected_ToXYZ
   bIsVariable=true
   parent=VerticalBox-MoveModes
   widget=EUW_AssetOrganizer_MoveSelected_ToXYZ
   slot {class=VerticalBoxSlot}
  }
  Border-Debug {
   widgetClassPath=&Border
   parent=VerticalBox_Master
   widget=Border-Debug
   properties {
    bIsEnabled=false
    visibility=Collapsed
    background {
     bIsDynamicallyLoaded=false
     drawAs=Box
     imageType=FullColor
     mirroring=NoMirror
     resourceName=None
     resourceObject=&M_UI_Border
     tiling=NoTile
     imageSize=<40 40>
     margin=<ltrb 0.5 0.5 0.5 0.5>
     outlineSettings {
      bUseBrushTransparency=false
      roundingType=HalfHeightRadius
      width=0
      color {colorUseRule=UseColor_Specified, specifiedColor=<rgba 0 0 0 0>}
      cornerRadii=<0 0 0 0>
     }
     tintColor {colorUseRule=UseColor_Specified, specifiedColor=<rgba 1 0 0.8506 0>}
     uVRegion {bIsValid=false, max=<0 0>, min=<0 0>}
    }
    padding=<ltrb 15 15 15 15>
   }
   slot {class=VerticalBoxSlot, properties {padding=<ltrb 0 10 10 0>}}
  }
  VerticalBox_75 {
   widgetClassPath=&VerticalBox
   parent=Border-Debug
   widget=VerticalBox_75
   slot {class=BorderSlot, properties {padding=<ltrb 15 15 15 15>}}
  }
  TextBlock_19 {
   widgetClassPath=&TextBlock
   parent=VerticalBox_75
   widget=TextBlock_19
   properties {
    text=TEST
    colorAndOpacity {colorUseRule=UseColor_Specified, specifiedColor=<rgba 1 0 0.868 1>}
    font {
     bForceMonospaced=false
     bMaterialIsStencil=false
     fontMaterial=null
     fontObject=&Roboto
     letterSpacing=0
     monospacedWidth=1
     size=10.5
     skewAmount=0
     typefaceFontName=Bold
     outlineSettings {
      bApplyOutlineToDropShadows=false
      bMiteredCorners=false
      bSeparateFillAlpha=false
      outlineMaterial=null
      outlineSize=0
      outlineColor=<rgba 0 0 0 1>
     }
    }
   }
   slot {class=VerticalBoxSlot, properties {padding=<ltrb 0 0 0 10>}}
  }
  HorizontalBox_0 {
   widgetClassPath=&HorizontalBox
   parent=VerticalBox_75
   widget=HorizontalBox_0
   slot {class=VerticalBoxSlot, properties {size {sizeRule=Fill, value=1}}}
  }
  B_Test {
   widgetClassPath=&EditorUtilityButton
   parent=HorizontalBox_0
   widget=B_Test
   properties {widgetStyle {
     clickedSlateSound {resourceObject=null}
     disabled {
      bIsDynamicallyLoaded=false
      drawAs=RoundedBox
      imageType=NoImage
      mirroring=NoMirror
      resourceName=None
      resourceObject=null
      tiling=NoTile
      imageSize=<0 0>
      margin=<ltrb 0 0 0 0>
      outlineSettings {
       bUseBrushTransparency=false
       roundingType=FixedRadius
       width=1
       color {colorUseRule=UseColor_ColorTable, specifiedColor=<rgba 0 0 0 0>}
       cornerRadii=<10 10 10 10>
      }
      tintColor {colorUseRule=UseColor_Specified, specifiedColor=<rgba 0.0395 0.0395 0.0395 1>}
      uVRegion {bIsValid=false, max=<0 0>, min=<0 0>}
     }
     disabledForeground {colorUseRule=UseColor_ColorTable, specifiedColor=<rgba 0 0 0 0>}
     hovered {
      bIsDynamicallyLoaded=false
      drawAs=RoundedBox
      imageType=NoImage
      mirroring=NoMirror
      resourceName=None
      resourceObject=null
      tiling=NoTile
      imageSize=<32 32>
      margin=<ltrb 0 0 0 0>
      outlineSettings {
       bUseBrushTransparency=false
       roundingType=FixedRadius
       width=1
       color {colorUseRule=UseColor_ColorTable, specifiedColor=<rgba 0 0 0 0>}
       cornerRadii=<10 10 10 10>
      }
      tintColor {colorUseRule=UseColor_Specified, specifiedColor=<rgba 0.0412 0.2523 0.4167 1>}
      uVRegion {bIsValid=false, max=<0 0>, min=<0 0>}
     }
     hoveredForeground {colorUseRule=UseColor_ColorTable, specifiedColor=<rgba 0 0 0 0>}
     hoveredSlateSound {resourceObject=null}
     normal {
      bIsDynamicallyLoaded=false
      drawAs=RoundedBox
      imageType=FullColor
      mirroring=NoMirror
      resourceName=None
      resourceObject=null
      tiling=NoTile
      imageSize=<32 32>
      margin=<ltrb 0 0 0 0>
      outlineSettings {
       bUseBrushTransparency=false
       roundingType=FixedRadius
       width=1
       color {colorUseRule=UseColor_ColorTable, specifiedColor=<rgba 0 0 0 0>}
       cornerRadii=<10 10 10 10>
      }
      tintColor {colorUseRule=UseColor_Specified, specifiedColor=<rgba 0 0.1347 0.2396 1>}
      uVRegion {bIsValid=false, max=<0 0>, min=<0 0>}
     }
     normalForeground {colorUseRule=UseColor_ColorTable, specifiedColor=<rgba 0 0 0 0>}
     normalPadding=<ltrb 12 1.5 12 1.5>
     pressed {
      bIsDynamicallyLoaded=false
      drawAs=RoundedBox
      imageType=NoImage
      mirroring=NoMirror
      resourceName=None
      resourceObject=null
      tiling=NoTile
      imageSize=<32 32>
      margin=<ltrb 0 0 0 0>
      outlineSettings {
       bUseBrushTransparency=false
       roundingType=FixedRadius
       width=1
       color {colorUseRule=UseColor_ColorTable, specifiedColor=<rgba 0 0 0 0>}
       cornerRadii=<10 10 10 10>
      }
      tintColor {colorUseRule=UseColor_Specified, specifiedColor=<rgba 0.0137 0.0587 0.0938 1>}
      uVRegion {bIsValid=false, max=<0 0>, min=<0 0>}
     }
     pressedForeground {colorUseRule=UseColor_ColorTable, specifiedColor=<rgba 0 0 0 0>}
     pressedPadding=<ltrb 12 2.5 12 0.5>
     pressedSlateSound {resourceObject=null}
    }}
   slot {class=HorizontalBoxSlot, properties {size {sizeRule=Fill, value=1}}}
  }
  TextBlock_127 {
   widgetClassPath=&TextBlock
   parent=B_Test
   widget=TextBlock_127
   properties {
    text=TEST BUTTON
    font {
     bForceMonospaced=false
     bMaterialIsStencil=false
     fontMaterial=null
     fontObject=&Roboto
     letterSpacing=0
     monospacedWidth=1
     size=12
     skewAmount=0
     typefaceFontName=Bold
     outlineSettings {
      bApplyOutlineToDropShadows=false
      bMiteredCorners=false
      bSeparateFillAlpha=false
      outlineMaterial=null
      outlineSize=0
      outlineColor=<rgba 0 0 0 1>
     }
    }
   }
   slot {class=ButtonSlot}
  }
 ]
}
