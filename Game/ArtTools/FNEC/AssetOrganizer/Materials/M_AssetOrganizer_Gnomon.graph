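# Unlit material for the gnomon in the Asset Organizer art tool.
# The graph holds a single VertexColor node and no textures or parameters, so any color variation comes from colors stored on the mesh vertices.
# Automatic setting of usage flags in the editor is turned off.
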Material /Game/ArtTools/FNEC/AssetOrganizer/Materials/M_AssetOrganizer_Gnomon
graph {properties {bAutomaticallySetUsageInEditor=false, shadingModel=MSM_Unlit, usedShadingModels=Unlit}, nodes [VertexColor()]}
